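# Translucent UI material for the Season Pass that procedurally draws a check mark icon without textures.
# Two offset UV sets, rotated 45 degrees in opposite directions, feed rounded-mask SDFs for the short and long strokes, each with one rounded corner.
# A Min merges both distance fields, and an anti-aliasing function turns the result into a smooth-edged shape.
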Material /SeasonPassHub/Widgets/Reboot/Materials/M_SeasonPass_Icon_CheckMark
graph {
 properties {
  bUsedWithStaticMesh=false
  blendMode=BLEND_Translucent
  materialDomain=MD_UI
  pixelDepthOffsetMode=PDOM_Legacy
  usedShadingModels=Default Lit
 }
 nodes [
  n1 = MaterialFunctionCall(UVs=Add(A=TextureCoordinate(), B=Constant2Vector() {g=-0.055, r=0.25}), "Rotation Angle (0-1)"=Constant() {r=-0.125}) {materialFunction=&CustomRotator}
  n2 = Constant2Vector() {g=0.25, r=0.25}
  n3 = MaterialFunctionCall(X=Constant() {r=0.25}, Y=Constant(), Z=Constant(), A=Constant()) {materialFunction=&MakeFloat4}
  n4 = MaterialFunctionCall(UVs=Add(A=TextureCoordinate(), B=Constant2Vector() {r=-0.05}), "Rotation Angle (0-1)"=Constant() {r=0.125}) {materialFunction=&CustomRotator}
  n5 = Constant2Vector() {g=0.25, r=0.75}
  n6 = MaterialFunctionCall(X=Constant(), Y=Constant(), Z=Constant() {r=0.25}, A=Constant()) {materialFunction=&MakeFloat4}
  MaterialFunctionCall(
   DistanceField=Constant()
   Threshold=Min(
    A=MaterialFunctionCall(UV=@n1.Rotated Values, Size=@n2, Corner Radius=@n3.Result) {materialFunction=&MF_UI_CoreUI_RoundedMask}.SDF
    B=MaterialFunctionCall(UV=@n4.Rotated Values, Size=@n5, Corner Radius=@n6.Result) {materialFunction=&MF_UI_CoreUI_RoundedMask}.SDF
   )
  ) {materialFunction=&MF_UI_AntiAliasing}
 ]
}
