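# UI material function for Season Pass that adjusts a colour's hue, saturation and value in perceptual OkHSV space.
# Input RGB goes through the OkLab conversion function, a Custom HLSL node, then back to RGB.
# The node uses Björn Ottosson's analytical sRGB gamut cusp math, with Halley refinement and toe compensation, for both original and new hue.
# Non-negative overrides replace components before shifts apply; hue uses degrees, saturation and value clamp to 0–1.
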MaterialFunction /SeasonPass/MaterialFunctions/MF_UI_SeasonPass_OkHSV_ColorModify
graph {
 properties {previewBlendMode=BLEND_Translucent, previewMaterialDomain=MD_UI}
 nodes [
  MaterialExpressionMaterialFunctionCall_0 = MaterialFunctionCall!(In=FunctionInput() {bUsePreviewValueAsDefault=true, inputName=Color, previewValue=<0.2587 0.25 0.5 1>}) {materialFunction=&/MF_UI_SeasonPass_RBG_to_OkLab}
  MaterialExpressionFunctionInput_2 = FunctionInput() {bUsePreviewValueAsDefault=true, inputName=HueShift, inputType=FunctionInput_Scalar}
  MaterialExpressionFunctionInput_3 = FunctionInput() {bUsePreviewValueAsDefault=true, inputName=SaturationShift, inputType=FunctionInput_Scalar}
  MaterialExpressionFunctionInput_4 = FunctionInput() {bUsePreviewValueAsDefault=true, inputName=ValueShift, inputType=FunctionInput_Scalar}
  MaterialExpressionFunctionInput_1 = FunctionInput() {bUsePreviewValueAsDefault=true, inputName=HueOverride, inputType=FunctionInput_Scalar, previewValue=<-1 0 0 1>}
  MaterialExpressionFunctionInput_5 = FunctionInput() {bUsePreviewValueAsDefault=true, inputName=SaturationOverride, inputType=FunctionInput_Scalar, previewValue=<-1 0 0 1>}
  MaterialExpressionFunctionInput_6 = FunctionInput() {bUsePreviewValueAsDefault=true, inputName=ValueOverride, inputType=FunctionInput_Scalar, previewValue=<-1 0 0 1>}
  n1 = Custom!(
   InLab=@MaterialExpressionMaterialFunctionCall_0.Result
   HueShift=@MaterialExpressionFunctionInput_2
   SaturationShift=@MaterialExpressionFunctionInput_3
   ValueShift=@MaterialExpressionFunctionInput_4
   HueOverride=@MaterialExpressionFunctionInput_1
   SaturationOverride=@MaterialExpressionFunctionInput_5
   ValueOverride=@MaterialExpressionFunctionInput_6
  ) {
   code="// Uses Björn Ottoson's methods for sRGB gamut clipping and conversions\r\n\r\nfloat PI = 3.1415926535897932;\r\n\r\n\r\n// Toe function constants\r\nfloat toe_k1 = 0.206;\r\nfloat toe_k2 = 0.03;\r\nfloat toe_k3 = (1.0 + toe_k1) / (1.0 + toe_k2);\r\n\r\nfloat toe_4k2k3 = 4.0 * toe_k2 * toe_k3;\r\n\r\n//--- Step 1: Decompose input OkLab to OkLCH ---/\r\nfloat L = InLab.x;\r\nfloat C = length(InLab.yz);\r\n\r\n// Guard achromatic inputs (atan2(0,0) is undefined)\r\nfloat hasChroma = step(1e-6, C);\r\nfloat H_rad = atan2(InLab.z, InLab.y + (1.0 - hasChroma) * 1e-10);\r\n\r\n// Wrap to [0, 2PI]\r\nH_rad += (2.0 * PI) * step(H_rad, 0.0);\r\n\r\n// Preserve original hue direction for first gamut query\r\nfloat cosH_orig = cos(H_rad);\r\nfloat sinH_orig = sin(H_rad);\r\n\r\n//--- Step 2: Analytical gamut boundary (original Hue) --//\r\nfloat a_orig = cosH_orig;\r\nfloat b_orig = sinH_orig;\r\n\r\n//-- Compute max saturation of the original hue --//\r\nfloat selR_o = step(1.0, -1.88170328 * a_orig - 0.80936493 * b_orig);\r\nfloat selG_o = step(1.0, 1.81444104 * a_orig - 1.19445276 * b_orig) * (1.0 - selR_o);\r\nfloat selB_o = 1.0 - selR_o - selG_o;\r\n\r\nfloat k0_o = selR_o * 1.19086277 + selG_o * 0.73956515 + selB_o * 1.35733652;\r\nfloat k1_o = selR_o * 1.76576728 + selG_o * -0.45954404 + selB_o * -0.00915799;\r\nfloat k2_o = selR_o * 0.59662641 + selG_o * 0.08285427 + selB_o * -1.15130210;\r\nfloat k3_o = selR_o * 0.75515197 + selG_o * 0.12541070 + selB_o * -0.50559606;\r\nfloat k4_o = selR_o * 0.56771245 + selG_o * 0.14503204 + selB_o * 0.00692167;\r\nfloat wl_o = selR_o * 4.0767416621 + selG_o * -1.2684380046 + selB_o * -0.0041960863;\r\nfloat wm_o = selR_o * -3.3077115913 + selG_o * 2.6097574011 + selB_o * -0.7034186147;\r\nfloat ws_o = selR_o * 0.2309699292 + selG_o * -0.3413193965 + selB_o * 1.7076147010;\r\n\r\nfloat S_max_o = k0_o + k1_o * a_orig + k2_o * b_orig + k3_o * a_orig * a_orig + k4_o * a_orig * b_orig;\r\n\r\n// Halley refinement (find exact S)\r\nfloat kl_o = 0.3963377774 * a_orig + 0.2158037573 * b_orig;\r\nfloat km_o = -0.1055613458 * a_orig - 0.0638541728 * b_orig;\r\nfloat ks_o = -0.0894841775 * a_orig - 1.2914855480 * b_orig;\r\n\r\n// LMS values at current S estimate\r\nfloat lo_ = 1.0 + S_max_o * kl_o;\r\nfloat mo_ = 1.0 + S_max_o * km_o;\r\nfloat so_ = 1.0 + S_max_o * ks_o;\r\n\r\n// LMS\r\nfloat lo3 = lo_ * lo_ * lo_;\r\nfloat mo3 = mo_ * mo_ * mo_;\r\nfloat so3 = so_ * so_ * so_;\r\n\r\n// First derivative of LMS\r\nfloat lo_dS = 3.0 * kl_o * lo_ * lo_;\r\nfloat mo_dS = 3.0 * km_o * mo_ * mo_;\r\nfloat so_dS = 3.0 * ks_o * so_ * so_;\r\n\r\n// Second derivative of LMS\r\nfloat lo_dS2 = 6.0 * kl_o * kl_o * lo_;\r\nfloat mo_dS2 = 6.0 * km_o * km_o * mo_;\r\nfloat so_dS2 = 6.0 * ks_o * ks_o * so_;\r\n\r\n// Apply RGB matrix\r\nfloat fo = wl_o * lo3 + wm_o * mo3 + ws_o * so3;\r\nfloat fo1 = wl_o * lo_dS + wm_o * mo_dS + ws_o * so_dS;\r\nfloat fo2 = wl_o * lo_dS2 + wm_o * mo_dS2 + ws_o * so_dS2;\r\n\r\nS_max_o = S_max_o - fo * fo1 / (fo1 * fo1 - 0.5 * fo * fo2 + 1e-10);\r\n\r\n// Find cusp of original hue\r\nfloat3 lms_cusp_o = float3(\r\n  1.0 + S_max_o * kl_o,\r\n  1.0 + S_max_o * km_o,\r\n  1.0 + S_max_o * ks_o\r\n);\r\n\r\nlms_cusp_o = lms_cusp_o * lms_cusp_o * lms_cusp_o;\r\n\r\nfloat3 rgb_cusp_o = float3(\r\n  4.0767416621 * lms_cusp_o.x - 3.3077115913 * lms_cusp_o.y + 0.2309699292 * lms_cusp_o.z,\r\n  -1.2684380046 * lms_cusp_o.x + 2.6097574011 * lms_cusp_o.y - 0.3413193965 * lms_cusp_o.z,\r\n  -0.0041960863 * lms_cusp_o.x - 0.7034186147 * lms_cusp_o.y + 1.7076147010 * lms_cusp_o.z\r\n);\r\n\r\nfloat L_cusp_o = pow(1.0 / max(max(rgb_cusp_o.x, max(rgb_cusp_o.y, rgb_cusp_o.z)), 1e-10), 1.0 / 3.0);\r\nfloat C_cusp_o = L_cusp_o * S_max_o;\r\n\r\n//--- Step 3: Convert OkLab L, C to OkHSV S,V using original cusp ---//\r\n\r\n// ST representation of cusp\r\nfloat ST_S_max_o = C_cusp_o / max(L_cusp_o, 1e-10);\r\nfloat T_max_o = C_cusp_o / max(1.0 - L_cusp_o, 1e-10);\r\nfloat s_0 = 0.5;\r\nfloat k_hsv_o = 1.0 - s_0 / max(ST_S_max_o, 1e-10);\r\n\r\n// Find L_v, C_v (intersection with v=1 edge of triangle)\r\nfloat t_v = T_max_o / max(C + L * T_max_o, 1e-10);\r\nfloat L_v = t_v * L;\r\nfloat C_v = t_v * C;\r\n\r\n// Toe compensation for curved gamut top\r\nfloat L_vt = 0.5 * (toe_k3 * L_v - toe_k1 + sqrt((toe_k3 * L_v - toe_k1) * (toe_k3 * L_v - toe_k1) + toe_4k2k3 * L_v));\r\nfloat C_vt = C_v * L_vt / max(L_v, 1e-10);\r\n\r\n// Scale factor from gamut boundary\r\nfloat3 lms_vt = float3(\r\n  L_vt + C_vt * a_orig * 0.3963377774 + C_vt * b_orig * 0.2158037573,\r\n  L_vt - C_vt * a_orig * 0.1055613458 - C_vt * b_orig * 0.0638541728,\r\n  L_vt - C_vt * a_orig * 0.0894841775 - C_vt * b_orig * 1.2914855480\r\n);\r\n\r\nlms_vt = lms_vt * lms_vt * lms_vt;\r\n\r\nfloat3 rgb_vt = float3(\r\n  4.0767416621 * lms_vt.x - 3.3077115913 * lms_vt.y + 0.2309699292 * lms_vt.z,\r\n  -1.2684380046 * lms_vt.x + 2.6097574011 * lms_vt.y - 0.3413193965 * lms_vt.z,\r\n  -0.0041960863 * lms_vt.x - 0.7034186147 * lms_vt.y + 1.7076147010 * lms_vt.z\r\n);\r\n\r\nfloat scale_L_o = pow(1.0 / max(max(rgb_vt.x, max(rgb_vt.y, rgb_vt.z)), 1e-10), 1.0 / 3.0);\r\n\r\n// Scale and apply toe\r\nfloat L_scaled = L / max(scale_L_o, 1e-10);\r\nfloat C_scaled = C / max(scale_L_o, 1e-10);\r\n\r\nfloat L_toed = 0.5 * (toe_k3 * L_scaled - toe_k1 + sqrt((toe_k3 * L_scaled - toe_k1) * (toe_k3 * L_scaled - toe_k1) + toe_4k2k3 * L_scaled));\r\nfloat C_toed = C_scaled * L_toed / max(L_scaled, 1e-10);\r\n\r\n// Extract OkHSV S and V\r\nfloat V = L_toed / max(L_v, 1e-10);\r\nfloat S_hsv = (s_0 + T_max_o) * C_v / max((T_max_o * s_0) + T_max_o * k_hsv_o * C_v, 1e-10);\r\n\r\n\r\n//--- Step 4: Convert OkLab to L,C to OkHSV overrides and shifts ---//\r\nfloat useValueOverride = step(0.0, ValueOverride);\r\nfloat useSaturationOverride = step(0.0, SaturationOverride);\r\nfloat useHueOverride = step(0.0, HueOverride);\r\n\r\nV = lerp(V, max(ValueOverride, 0.0), useValueOverride);\r\nV = clamp(V + ValueShift, 0.0, 1.0);\r\n\r\nS_hsv = lerp(S_hsv, max(SaturationOverride, 0.0), useSaturationOverride);\r\nS_hsv = clamp(S_hsv + SaturationShift, 0.0, 1.0);\r\n\r\nfloat H_rad_set = max(HueOverride, 0.0) * (PI / 180.0);\r\nH_rad = lerp(H_rad, H_rad_set, useHueOverride);\r\nH_rad += HueShift * (PI / 180.0);\r\n\r\nfloat cosH = cos(H_rad);\r\nfloat sinH = sin(H_rad);\r\n\r\n//--- Step 5: Analytical gamut boundary (new hue) --//\r\nfloat a_new = cosH;\r\nfloat b_new = sinH;\r\n\r\n// Compute max saturation (new hue)\r\nfloat selR_n = step(1.0, -1.88170328 * a_new - 0.80936493 * b_new);\r\nfloat selG_n = step(1.0, 1.81444104 * a_new - 1.19445276 * b_new) * (1.0 - selR_n);\r\nfloat selB_n = 1.0 - selR_n - selG_n;\r\n\r\nfloat k0_n = selR_n * 1.19086277 + selG_n * 0.73956515 + selB_n * 1.35733652;\r\nfloat k1_n = selR_n * 1.76576728 + selG_n * -0.45954404 + selB_n * -0.00915799;\r\nfloat k2_n = selR_n * 0.59662641 + selG_n * 0.08285427 + selB_n * -1.15130210;\r\nfloat k3_n = selR_n * 0.75515197 + selG_n * 0.12541070 + selB_n * -0.50559606;\r\nfloat k4_n = selR_n * 0.56771245 + selG_n * 0.14503204 + selB_n * 0.00692167;\r\nfloat wl_n = selR_n * 4.0767416621 + selG_n * -1.2684380046 + selB_n * -0.0041960863;\r\nfloat wm_n = selR_n * -3.3077115913 + selG_n * 2.6097574011 + selB_n * -0.7034186147;\r\nfloat ws_n = selR_n * 0.2309699292 + selG_n * -0.3413193965 + selB_n * 1.7076147010;\r\n\r\nfloat S_max_n = k0_n + k1_n * a_new + k2_n * b_new + k3_n * a_new * a_new + k4_n * a_new * b_new;\r\n\r\n// Halley refinement (find exact S)\r\nfloat kl_n = 0.3963377774 * a_new + 0.2158037573 * b_new;\r\nfloat km_n = -0.1055613458 * a_new - 0.0638541728 * b_new;\r\nfloat ks_n = -0.0894841775 * a_new - 1.2914855480 * b_new;\r\n\r\n// LMS values at current S estimate\r\nfloat ln_ = 1.0 + S_max_n * kl_n;\r\nfloat mn_ = 1.0 + S_max_n * km_n;\r\nfloat sn_ = 1.0 + S_max_n * ks_n;\r\n\r\n// LMS\r\nfloat ln3 = ln_ * ln_ * ln_;\r\nfloat mn3 = mn_ * mn_ * mn_;\r\nfloat sn3 = sn_ * sn_ * sn_;\r\n\r\n// First derivative of LMS\r\nfloat ln_dS = 3.0 * kl_n * ln_ * ln_;\r\nfloat mn_dS = 3.0 * km_n * mn_ * mn_;\r\nfloat sn_dS = 3.0 * ks_n * sn_ * sn_;\r\n\r\n// Second derivative of LMS\r\nfloat ln_dS2 = 6.0 * kl_n * kl_n * ln_;\r\nfloat mn_dS2 = 6.0 * km_n * km_n * mn_;\r\nfloat sn_dS2 = 6.0 * ks_n * ks_n * sn_;\r\n\r\n// Apply RGB matrix\r\nfloat fn = wl_n * ln3 + wm_n * mn3 + ws_n * sn3;\r\nfloat fn1 = wl_n * ln_dS + wm_n * mn_dS + ws_n * sn_dS;\r\nfloat fn2 = wl_n * ln_dS2 + wm_n * mn_dS2 + ws_n * sn_dS2;\r\n\r\nS_max_n = S_max_n - fn * fn1 / (fn1 * fn1 - 0.5 * fn * fn2 + 1e-10);\r\n\r\n// Find cusp of new hue\r\nfloat3 lms_cusp_n = float3(\r\n  1.0 + S_max_n * kl_n,\r\n  1.0 + S_max_n * km_n,\r\n  1.0 + S_max_n * ks_n\r\n);\r\n\r\nlms_cusp_n = lms_cusp_n * lms_cusp_n * lms_cusp_n;\r\n\r\nfloat3 rgb_cusp_n = float3(\r\n  4.0767416621 * lms_cusp_n.x - 3.3077115913 * lms_cusp_n.y + 0.2309699292 * lms_cusp_n.z,\r\n  -1.2684380046 * lms_cusp_n.x + 2.6097574011 * lms_cusp_n.y - 0.3413193965 * lms_cusp_n.z,\r\n  -0.0041960863 * lms_cusp_n.x - 0.7034186147 * lms_cusp_n.y + 1.7076147010 * lms_cusp_n.z\r\n);\r\n\r\nfloat L_cusp_n = pow(1.0 / max(max(rgb_cusp_n.x, max(rgb_cusp_n.y, rgb_cusp_n.z)), 1e-10), 1.0 / 3.0);\r\nfloat C_cusp_n = L_cusp_n * S_max_n;\r\n\r\n//--- Step 6: Toe compensation for scale\r\nfloat ST_S_max_n = C_cusp_n / max(L_cusp_n, 1e-10);\r\nfloat T_max_n = C_cusp_n / max(1.0 - L_cusp_n, 1e-10);\r\nfloat k_hsv_n = 1.0 - s_0 / max(ST_S_max_n, 1e-10);\r\n\r\n// S,V to L,C on the ideal triangle\r\nfloat L_v_n = 1.0 - S_hsv * s_0 / (s_0 + T_max_n - T_max_n * k_hsv_n * S_hsv);\r\nfloat C_v_n = S_hsv * T_max_n * s_0 / (s_0 + T_max_n -T_max_n * k_hsv_n * S_hsv);\r\n\r\nfloat L_out = V * L_v_n;\r\nfloat C_out = V * C_v_n;\r\n\r\n// Inverse toe on L_out: toe_inv(x) = x * (x + k1) / (k3 * (x + k2))\r\nfloat L_out_inv = L_out * (L_out + toe_k1) / max(toe_k3 * (L_out + toe_k2), 1e-10);\r\nfloat C_out_adj = C_out * L_out_inv / max(L_out, 1e-10);\r\nL_out = L_out_inv;\r\n\r\n// Toe compensation for scale\r\nfloat L_vt_n = 0.5 * (toe_k3 * L_v_n - toe_k1 + sqrt((toe_k3 * L_v_n - toe_k1) * (toe_k3 * L_v_n - toe_k1) + toe_4k2k3 * L_v_n));\r\nfloat C_vt_n = C_v_n * L_vt_n / max(L_v_n, 1e-10);\r\n\r\n// Scale factor from gamut boundary\r\nfloat3 lms_vt_n = float3(\r\n  L_vt_n + C_vt_n * a_new * 0.3963377774 + C_vt_n * b_new * 0.2158037573,\r\n  L_vt_n - C_vt_n * a_new * 0.1055613458 - C_vt_n * b_new * 0.0638541728,\r\n  L_vt_n - C_vt_n * a_new * 0.0894841775 - C_vt_n * b_new * 1.2914855480\r\n);\r\n\r\nlms_vt_n = lms_vt_n * lms_vt_n * lms_vt_n;\r\n\r\nfloat3 rgb_vt_n = float3(\r\n  4.0767416621 * lms_vt_n.x - 3.3077115913 * lms_vt_n.y + 0.2309699292 * lms_vt_n.z,\r\n  -1.2684380046 * lms_vt_n.x + 2.6097574011 * lms_vt_n.y - 0.3413193965 * lms_vt_n.z,\r\n  -0.0041960863 * lms_vt_n.x - 0.7034186147 * lms_vt_n.y + 1.7076147010 * lms_vt_n.z\r\n);\r\n\r\nfloat scale_L_n = pow(1.0 / max(max(rgb_vt_n.x, max(rgb_vt_n.y, rgb_vt_n.z)), 1e-10), 1.0 / 3.0);\r\n\r\nL_out = L_out * scale_L_n;\r\nC_out_adj = C_out_adj * scale_L_n;\r\n\r\n//--- Step 7: Reconstruct OkLab --//\r\nfloat3 FinalLab = float3(L_out, C_out_adj * cosH, C_out_adj * sinH);\r\n\r\nreturn FinalLab;"
   description="Takes an OkLab input, shifts it more intuitively in OkHSV, and allows setting overrides in OkHSV, and then outputs to OkLab"
   showCode=true
   includeFilePaths [/Engine/Private/ColorSpace.ush]
   inputs [
    {
     inputName=InLab
     input {expression=&self:MaterialExpressionMaterialFunctionCall_0}
    }
    {inputName=HueShift, input {expression=&self:MaterialExpressionFunctionInput_2}}
    {
     inputName=SaturationShift
     input {expression=&self:MaterialExpressionFunctionInput_3}
    }
    {inputName=ValueShift, input {expression=&self:MaterialExpressionFunctionInput_4}}
    {inputName=HueOverride, input {expression=&self:MaterialExpressionFunctionInput_1}}
    {
     inputName=SaturationOverride
     input {expression=&self:MaterialExpressionFunctionInput_5}
    }
    {
     inputName=ValueOverride
     input {expression=&self:MaterialExpressionFunctionInput_6}
    }
   ]
  }
  FunctionOutput(None=MaterialFunctionCall!(In=@n1) {materialFunction=&/MF_UI_SeasonPass_OkLab_to_RGB}.Result)
 ]
}
