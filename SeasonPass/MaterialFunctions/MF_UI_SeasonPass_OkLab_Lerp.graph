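# Season Pass UI material function that blends two colors in OkLab color space rather than RGB.
# ColorA and ColorB each pass through an RGB-to-OkLab function, and a Custom HLSL node linearly interpolates them by a scalar Alpha.
# The blended result is converted back to RGB through a companion OkLab-to-RGB function for output.
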MaterialFunction /SeasonPass/MaterialFunctions/MF_UI_SeasonPass_OkLab_Lerp
graph {
 properties {previewBlendMode=BLEND_Translucent, previewMaterialDomain=MD_UI}
 nodes [
  MaterialExpressionMaterialFunctionCall_1 = MaterialFunctionCall!(In=FunctionInput() {bUsePreviewValueAsDefault=true, inputName=ColorA, previewValue=<0.7 0.8 0 1>}) {materialFunction=&/MF_UI_SeasonPass_RBG_to_OkLab}
  MaterialExpressionMaterialFunctionCall_2 = MaterialFunctionCall(In=FunctionInput() {bUsePreviewValueAsDefault=true, inputName=ColorB, sortPriority=1, previewValue=<0.2 0.2 8 1>}) {materialFunction=&MF_UI_SeasonPass_RBG_to_OkLab}
  MaterialExpressionFunctionInput_2 = FunctionInput() {bUsePreviewValueAsDefault=true, inputName=Alpha, inputType=FunctionInput_Scalar, sortPriority=2, previewValue=<0.5 0 0 1>}
  n1 = Custom!(
   Lab_A=@MaterialExpressionMaterialFunctionCall_1.Result
   Lab_B=@MaterialExpressionMaterialFunctionCall_2.Result
   Alpha=@MaterialExpressionFunctionInput_2
  ) {
   code="// Lerp in OkLab\r\nfloat3 LabMix = lerp(Lab_A, Lab_B, Alpha);\r\n\r\nreturn LabMix;"
   description=Takes two OkLab inputs and blends them and outputs OkLab
   showCode=true
   includeFilePaths [/Engine/Private/ColorSpace.ush]
   inputs [
    {inputName=Lab_A, input {expression=&self:MaterialExpressionMaterialFunctionCall_1}}
    {inputName=Lab_B, input {expression=&self:MaterialExpressionMaterialFunctionCall_2}}
    {inputName=Alpha, input {expression=&self:MaterialExpressionFunctionInput_2}}
   ]
  }
  FunctionOutput(None=MaterialFunctionCall!(In=@n1) {materialFunction=&/MF_UI_SeasonPass_OkLab_to_RGB}.Result)
 ]
}
