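# Translucent UI material for the CH7S2 Season Pass rivalry billboard, showing a character texture over a colored arrow-shaped plate with glows.
# Polar-coordinate radial SDFs build the arrow, while a skewed box SDF masks the offset-animatable character.
# Base Color drives darker and lighter variants through OkHSV shifts, blended with OkLab lerps.
# A Mirror parameter flips the layout and vertex offset, and Reveal parameters drive a horizontal wipe mask.
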Material /SeasonPassHub/Widgets/Reboot/Materials/M_UI_CH7S2_RivalryBillboard
graph {
 properties {
  bAutomaticallySetUsageInEditor=false
  bUsedWithStaticMesh=false
  blendMode=BLEND_Translucent
  materialDomain=MD_UI
  usedShadingModels=Default Lit
 }
 nodes [
  n1 = ScalarParameter() {defaultValue=1, parameterName=TextureOpacity}
  n2 = NamedRerouteUsage()
  n3 = NamedRerouteUsage()
  n4 = Power(Base=LinearInterpolate(A=NamedRerouteUsage(), Alpha=NamedRerouteUsage())) {constExponent=0.5}
  LinearInterpolate(A=LinearInterpolate(A=Multiply(A=@n1, B=@n2), Alpha=LinearInterpolate(A=@n3, Alpha=@n4)), Alpha=NamedRerouteUsage())
  n5 = ScalarParameter() {parameterName=Mirror}
  MaterialFunctionCall!(
   Anchor
   Translation=AppendVector(A=LinearInterpolate(Alpha=@n5) {constA=30, constB=-25}, B=Constant())
   Padding
   Scale=ScalarParameter() {defaultValue=1.15}
  ) {materialFunction=/FortUILibrary/MaterialFunctions/Basics/MF_UI_VertexTransform}
  n6 = NamedRerouteUsage()
  MaterialFunctionCall!(
   ColorA=NamedRerouteUsage()
   ColorB=NamedRerouteUsage()
   Alpha=Multiply(A=Multiply(A=@n6, B=@n1), B=ScalarParameter() {defaultValue=1, parameterName=TextureColorMask})
  ) {materialFunction=/SeasonPass/MaterialFunctions/MF_UI_SeasonPass_OkLab_Lerp}
  n7 = NamedRerouteUsage()
  n8 = Multiply(A=ComponentMask(None=NamedRerouteUsage()) {r=true}) {constB=0.2}
  n9 = MaterialFunctionCall(Base=@n7, Blend=@n8) {materialFunction=&Blend_Overlay}
  MaterialFunctionCall(ColorA=NamedRerouteUsage(), ColorB=@n7, Alpha=@n9.Result) {materialFunction=&MF_UI_SeasonPass_OkLab_Lerp}
  Multiply(A=@n8, B=@n7)
  NamedRerouteDeclaration(None=VectorParameter() {parameterName=Base Color, defaultValue=<rgba 1 0.3812 0.021 1>}.RGB) {name=BaseColor, nodeColor=<rgba 0 0 0 1>}
  n10 = Add(
   A=MaterialFunctionCall(
    UVs=MaterialFunctionCall() {materialFunction=&GetUserInterfaceUV}.Normalized UV
    Texture Scale=ScalarParameter() {defaultValue=0.8, parameterName=ArrowScale}
   ) {materialFunction=&ScaleUVsByCenter}.UVs
   B=Constant2Vector() {g=-0.3}
  )
  n11 = ComponentMask(None=NamedRerouteDeclaration(None=@n10) {name=UV, nodeColor=<rgba 0 1 0.1797 1>}) {r=true}
  n12 = OneMinus(None=@n11)
  n13 = ScalarParameter() {parameterName=Mirror}
  NamedRerouteDeclaration(None=AppendVector(A=LinearInterpolate(A=@n11, B=@n12, Alpha=@n13), B=ComponentMask(None=@n10) {g=true})) {name=UV Arrow, nodeColor=<rgba 0 0 0 1>}
  n14 = Add(
   A=NamedRerouteUsage()
   B=AppendVector(A=ScalarParameter() {parameterName=TexturePosX_Anim}, B=ScalarParameter() {parameterName=TexturePosY_Anim})
  )
  n15 = AppendVector(A=ScalarParameter() {defaultValue=0.5, parameterName=TexturePosX}, B=ScalarParameter() {defaultValue=0.2, parameterName=TexturePosY})
  n16 = TextureSampleParameter2D(UVs=MaterialFunctionCall(UVs=Add(A=@n14, B=@n15), Texture Scale=ScalarParameter() {defaultValue=1.25, parameterName=TextureScale}) {materialFunction=&ScaleUVsByCenter}.UVs) {
   parameterName=Character Texture
   samplerSource=SSM_Clamp_WorldGroupSettings
   texture=/FortUILibrary/Systems/Items/Textures/T_UI_SoldierCommando_F
  }
  NamedRerouteDeclaration(None=Multiply(A=@n16.A, B=NamedRerouteUsage())) {name=Character Alpha, nodeColor=<rgba 0 0 0 1>}
  n17 = ScalarParameter() {parameterName=TextureMask_Size}
  n18 = MaterialFunctionCall!(UV=Add(A=NamedRerouteUsage(), B=Constant2Vector() {r=0.691}), Skew Angle=Constant2Vector() {r=33}) {materialFunction=/FortUILibrary/MaterialFunctions/Basics/MF_UI_Skew}
  n19 = SmoothStep(
   Min=@n17
   Max=Add(A=@n17) {constB=0.02}
   Value=MaterialFunctionCall!(
    UVs=NamedRerouteDeclaration(None=@n18.Skewed UVs) {name=Skewed UV, nodeColor=<rgba 0 0 0 1>}
    Size=Constant2Vector() {g=2, r=1}
    CornerRoundness
    KeepAspectRatio
    PixelSize
   ) {materialFunction=/FortUILibrary/MaterialFunctions/Shapes/SDFs/MF_SDF_Box}.SDF
  ) {constMax=0.02}
  n20 = OneMinus(None=SmoothStep(Value=NamedRerouteUsage()) {constMax=0.8, constMin=0.9})
  NamedRerouteDeclaration(None=Multiply(A=OneMinus(None=@n19), B=@n20)) {name=Character Mask, nodeColor=<rgba 0 0 0 1>}
  n21 = NamedRerouteUsage()
  n22 = Add(A=@n21, B=Constant2Vector() {r=-0.5})
  n23 = MaterialFunctionCall(UV=@n21, ScalePivotPoint=Constant2Vector() {g=0.5}) {materialFunction=&MF_UI_UVScaling}
  n24 = Constant2Vector() {r=-0.5}
  NamedRerouteDeclaration(None=LinearInterpolate(
    A=MaterialFunctionCall(UV=@n22) {materialFunction=&MF_UI_UVScaling}.Scaled to Fit
    B=Add(A=@n23.Scaled to Fit, B=@n24)
    Alpha=ScalarParameter() {parameterName=Mirror}
   )) {name=Right Anchored UV, nodeColor=<rgba 0 0 0 1>}
  NamedRerouteDeclaration(None=@n20) {name=TopMask, nodeColor=<rgba 0 0 0 1>}
  n25 = NamedRerouteUsage()
  n26 = ComponentMask(None=@n25) {r=true}
  n27 = Step(Y=OneMinus(None=@n26), X=ScalarParameter() {defaultValue=0.85, parameterName=SkewGlow_1_Size}) {constX=2}
  n28 = Multiply(A=Step(Y=@n26), B=NamedRerouteUsage())
  n29 = MaterialFunctionCall(UVs=@n25, "Rotation Angle (0-1)"=Constant() {r=-0.04}) {materialFunction=&CustomRotator}
  n30 = SmoothStep(Value=Multiply(A=ComponentMask(None=@n29.Rotated Values) {g=true}, B=ScalarParameter() {defaultValue=1, parameterName=SkewedGlow_1_Mask})) {constMax=0.8, constMin=0.15}
  NamedRerouteDeclaration(None=Multiply(
    A=Multiply(A=@n27, B=@n28)
    B=MaterialFunctionCall!("Value (0-1)"=@n30) {materialFunction=/FortUILibrary/MaterialFunctions/Basics/MF_UI_EaseCurves_01_Sine}.In-Out
   )) {name=SkewedGlow 1, nodeColor=<rgba 0 0 0 1>}
  n31 = NamedRerouteUsage()
  n32 = ComponentMask(None=@n31) {r=true}
  n33 = Step(Y=OneMinus(None=@n32), X=ScalarParameter() {defaultValue=1, parameterName=SkewGlow_2_Size}) {constX=2}
  n34 = Multiply(A=Step(Y=@n32), B=NamedRerouteUsage())
  n35 = MaterialFunctionCall(UVs=@n31, "Rotation Angle (0-1)"=Constant() {r=-0.04}) {materialFunction=&CustomRotator}
  n36 = SmoothStep(Value=Multiply(A=ComponentMask(None=@n35.Rotated Values) {g=true}, B=ScalarParameter() {defaultValue=1, parameterName=SkewedGlow_2_Mask})) {constMin=0.25}
  NamedRerouteDeclaration(None=Multiply(
    A=Multiply(A=@n33, B=@n34)
    B=MaterialFunctionCall("Value (0-1)"=@n36) {materialFunction=&MF_UI_EaseCurves_01_Sine}.In-Out
   )) {name=SkewedGlow 2, nodeColor=<rgba 0 0 0 1>}
  n37 = Distance(A=Add(A=NamedRerouteUsage(), B=Constant2Vector() {r=0.35}), B=Constant() {r=0.25})
  n38 = OneMinus(None=SmoothStep(Value=Multiply(A=@n37) {constB=2}) {constMax=1.35})
  n39 = MaterialFunctionCall() {materialFunction=&GetUserInterfaceUV}
  n40 = Constant() {r=0.5}
  n41 = OneMinus(None=Multiply(A=Distance(A=@n39.Normalized UV, B=@n40)) {constB=2})
  n42 = ComponentMask(None=@n39.Normalized UV) {r=true}
  n43 = LinearInterpolate(A=@n42, B=OneMinus(None=@n42), Alpha=ScalarParameter() {parameterName=Mirror})
  n44 = LinearInterpolate(A=Multiply(A=Power(Base=@n38) {constExponent=1.5}) {constB=0.6}, Alpha=Multiply(A=Saturate(None=@n41), B=Power(Base=@n43))) {constB=0.7}
  NamedRerouteDeclaration(None=Multiply(A=Power(Base=@n44) {constExponent=1.5}, B=ScalarParameter() {defaultValue=1, parameterName=BackgroundGlow_Opacity})) {name=BackgroundGlow, nodeColor=<rgba 0 0 0 1>}
  NamedRerouteDeclaration(None=MaterialFunctionCall!(
    HueShift=Constant() {r=10}
    SaturationShift
    ValueShift=Constant() {r=-0.1}
    HueOverride
    SaturationOverride=Constant() {r=1}
    ValueOverride
    Color=NamedRerouteUsage()
   ) {materialFunction=/SeasonPass/MaterialFunctions/MF_UI_SeasonPass_OkHSV_ColorModify}.Result) {name=ColorDarker, nodeColor=<rgba 0 0 0 1>}
  n45 = Add(A=NamedRerouteUsage(), B=Constant2Vector() {r=0.65})
  n46 = MaterialFunctionCall!(
   min=Constant()
   max=ScalarParameter() {defaultValue=-0.025, parameterName=Angle_Top}
   value=ComponentMask(None=MaterialFunctionCall!(UV Coord=@n45, LenghtScale, RadialScale) {materialFunction=/Game/Global/MaterialFunctions/TechArt/MF_PolarCoordinates}.Result) {g=true}
  ) {materialFunction=/Game/Effects/Fort_Effects/Materials/_Functions/InverseLerp_MF}
  NamedRerouteDeclaration(None=@n46.Clamped) {name=Top_Radial, nodeColor=<rgba 0 0 0 1>}
  n47 = NamedRerouteUsage()
  n48 = NamedRerouteUsage()
  n49 = NamedRerouteUsage()
  NamedRerouteDeclaration(None=MaterialFunctionCall(
    ColorA=MaterialFunctionCall(ColorA=@n47, ColorB=@n48, Alpha=@n49) {materialFunction=&MF_UI_SeasonPass_OkLab_Lerp}.Result
    ColorB=@n9.Result
    Alpha=NamedRerouteUsage()
   ) {materialFunction=&MF_UI_SeasonPass_OkLab_Lerp}.Result) {name=PlateColor, nodeColor=<rgba 0 0 0 1>}
  n50 = ScalarParameter() {defaultValue=1, group=9 - Reveal Mask, parameterName=Reveal Fade}
  n51 = Subtract(A=Constant(), B=@n50)
  n52 = NamedRerouteDeclaration(None=OneMinus(None=ScalarParameter() {defaultValue=1, group=9 - Reveal Mask, parameterName=Reveal Amount})) {name=Reveal Amount, nodeColor=<rgba 0 0 0 1>}
  n53 = Add(A=Constant() {r=1}, B=@n50)
  n54 = OneMinus(None=MaterialFunctionCall() {materialFunction=&GetUserInterfaceUV}.Normalized UV)
  NamedRerouteDeclaration(None=SmoothStep(Min=LinearInterpolate(A=@n51, Alpha=@n52), Max=LinearInterpolate(B=@n53, Alpha=@n52), Value=ComponentMask(None=@n54) {r=true})) {name=Reveal Mask, nodeColor=<rgba 0 0 0 1>}
  NamedRerouteDeclaration(None=MaterialFunctionCall(HueShift=Constant() {r=20}, SaturationOverride=Constant() {r=0.3}, ValueOverride=Constant() {r=1}, Color=NamedRerouteUsage()) {materialFunction=&MF_UI_SeasonPass_OkHSV_ColorModify}.Result) {name=ColorLighter, nodeColor=<rgba 0 0 0 1>}
  n55 = ComponentMask(None=MaterialFunctionCall() {materialFunction=&GetUserInterfaceUV}.Normalized UV) {r=true}
  n56 = LinearInterpolate(A=OneMinus(None=@n55), B=@n55, Alpha=ScalarParameter() {parameterName=Mirror})
  n57 = Multiply(A=@n56, B=@n56)
  n58 = Max(A=NamedRerouteUsage(), B=Min(A=NamedRerouteUsage(), B=NamedRerouteUsage()))
  n59 = NamedRerouteUsage()
  n60 = Multiply(A=OneMinus(None=Max(A=@n58, B=@n59)), B=NamedRerouteUsage())
  NamedRerouteDeclaration(None=SmoothStep(Max=LinearInterpolate(Alpha=@n57) {constA=0.05}, Value=NamedRerouteDeclaration(None=@n60) {name=Arrow SDF, nodeColor=<rgba 0 0 0 1>})) {name=Arrow Shape, nodeColor=<rgba 0 0 0 1>}
  n61 = Constant() {r=0.5}
  NamedRerouteDeclaration(None=Multiply(A=Distance(A=@n46.Clamped, B=@n61)) {constB=2}) {name=Top Part SDF, nodeColor=<rgba 0 0 0 1>}
  n62 = Add(A=NamedRerouteUsage(), B=Constant2Vector() {r=-0.5})
  n63 = LinearInterpolate(A=Constant() {r=0.3}, B=Constant(), Alpha=ScalarParameter() {parameterName=Mirror})
  n64 = Constant() {r=-0.15}
  n65 = MaterialFunctionCall(UV Coord=Add(
    A=MaterialFunctionCall(UV=@n62) {materialFunction=&MF_UI_UVScaling}.Scaled to Fit
    B=AppendVector(A=@n63, B=@n64)
   )) {materialFunction=&MF_PolarCoordinates}
  n66 = NamedRerouteDeclaration(None=MaterialFunctionCall(min=Constant() {r=-0.4}, max=Constant() {r=-0.25}, value=ComponentMask(None=@n65.Result) {g=true}) {materialFunction=&InverseLerp_MF}.Clamped) {name=Side_Radial, nodeColor=<rgba 0 0 0 1>}
  n67 = Constant() {r=0.5}
  NamedRerouteDeclaration(None=Multiply(A=Distance(A=@n66, B=@n67)) {constB=2}) {name=Side Part SDF, nodeColor=<rgba 0 0 0 1>}
  NamedRerouteDeclaration(None=@n16.RGB) {name=Character Color, nodeColor=<rgba 0 0 0 1>}
 ]
}
