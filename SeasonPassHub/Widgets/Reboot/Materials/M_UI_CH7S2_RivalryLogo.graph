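# Translucent UI material for the Rivalry logo in the Season Pass Hub, built around the Showdown logo texture.
# MF_UI_UVScaling supplies scaled-to-fit UVs, which ScaleUVsByCenter rescales around the center before a clamped texture sample.
# One constant drives both scale axes, doubled for X to give a 2:1 ratio.
# Two constant vector nodes sit unconnected.
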Material /SeasonPassHub/Widgets/Reboot/Materials/M_UI_CH7S2_RivalryLogo
graph {
 properties {
  bAutomaticallySetUsageInEditor=false
  bUsedWithStaticMesh=false
  blendMode=BLEND_Translucent
  materialDomain=MD_UI
  usedShadingModels=Default Lit
 }
 nodes [
  Constant2Vector() {g=1, r=2}
  Constant3Vector() {constant=<rgba 1 1 1 0>}
  n1 = Constant() {r=0.9}
  n2 = Multiply(A=@n1) {constB=2}
  TextureSample(UVs=MaterialFunctionCall(
    UVs=MaterialFunctionCall() {materialFunction=&MF_UI_UVScaling}.Scaled to Fit
    Texture Scale=AppendVector(A=@n2, B=@n1)
   ) {materialFunction=&ScaleUVsByCenter}.UVs) {samplerSource=SSM_Clamp_WorldGroupSettings, texture=/SeasonPassHub/Widgets/Reboot/Showdown/T_UI_Showdown_Logo}
 ]
}
