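# Translucent UI material for the landing page overlay of the Season Pass hub, shaped by the Shape Mask from MF_UI_BPReboot_LandingPage_Overlay.
# Inside the mask, an eased, offset and scaled radial falloff blends light blue into a dark blue vertical gradient.
# A randomly hue-shifted copy of that color, capped low, feeds a Color Dodge blend over it.
# A separate eased vertical "Base Gradient" reroute is also defined.
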Material /SeasonPassHub/Widgets/Reboot/Materials/M_UI_LandingPage_Overlay
graph {
 properties {
  bAutomaticallySetUsageInEditor=false
  bUsedWithStaticMesh=false
  blendMode=BLEND_Translucent
  materialDomain=MD_UI
  usedShadingModels=Default Lit
 }
 nodes [
  n1 = MaterialFunctionCall() {materialFunction=/SeasonPass/MaterialFunctions/MF_UI_BPReboot_LandingPage_Overlay}
  LinearInterpolate(A=NamedRerouteUsage(), Alpha=@n1.Shape Mask)
  n2 = MaterialFunctionCall!(UV, Seed) {materialFunction=&MF_UI_Random}
  n3 = Constant3Vector() {constant=<rgba 0.0097 0.0648 0.2122 1>}
  n4 = Multiply(A=@n3) {constB=0.5}
  n5 = TextureCoordinate()
  n6 = MaterialFunctionCall("Value (0-1)"=SmoothStep(Value=ComponentMask(None=@n5) {g=true}) {constMin=0.5}) {materialFunction=&MF_UI_EaseCurves}
  n7 = MaterialFunctionCall(UVs=Add(A=TextureCoordinate(), B=Constant2Vector() {g=0.2}), Texture Scale=Constant2Vector() {g=1.5, r=1.25}) {materialFunction=&ScaleUVsByCenter}
  n8 = Constant() {r=0.5}
  n9 = Saturate(None=Multiply(A=Distance(A=@n7.UVs, B=@n8)) {constB=2})
  n10 = LinearInterpolate(
   B=LinearInterpolate(
    A=Constant3Vector() {constant=<rgba 0.4452 0.7011 1 1>}
    B=LinearInterpolate(A=@n3, B=@n4, Alpha=@n6.EaseInOutSine)
    Alpha=MaterialFunctionCall("Value (0-1)"=@n9) {materialFunction=&MF_UI_EaseCurves}.EaseInOutSine
   )
   Alpha=@n1.Shape Mask
  )
  MaterialFunctionCall!(
   Blend=Min(A=MaterialFunctionCall(Hue Shift Percentage=@n2.Result, Texture=@n10) {materialFunction=&HueShift}.Result) {constB=0.15}
   Base=@n10
  ) {materialFunction=/Engine/Functions/Engine_MaterialFunctions03/Blends/Blend_ColorDodge}
  n11 = TextureCoordinate()
  NamedRerouteDeclaration(None=MaterialFunctionCall("Value (0-1)"=ComponentMask(None=@n11) {g=true}) {materialFunction=&MF_UI_EaseCurves}.EaseInOutSine) {name=Base Gradient, nodeColor=<rgba 0 0 0 1>}
 ]
}
