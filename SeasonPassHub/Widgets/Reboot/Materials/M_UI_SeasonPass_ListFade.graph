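# Translucent UI material for the Season Pass hub, named as a list fade, that builds a horizontal gradient with a switchable direction.
# It takes normalized UI UVs and blends U with its inverse through the FlipMask scalar.
# The result is recombined with V and masked back to its red channel.
# The graph also includes a GetSlatePost1 function for the Slate post buffer, though the final output wiring is not shown.
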Material /SeasonPassHub/Widgets/Reboot/Materials/M_UI_SeasonPass_ListFade
graph {
 properties {
  bAutomaticallySetUsageInEditor=false
  bUsedWithStaticMesh=false
  blendMode=BLEND_Translucent
  materialDomain=MD_UI
  usedShadingModels=Default Lit
 }
 nodes [
  n1 = MaterialFunctionCall() {materialFunction=&GetUserInterfaceUV}
  n2 = ComponentMask(None=@n1.Normalized UV) {r=true}
  n3 = OneMinus(None=@n2)
  n4 = ScalarParameter() {parameterName=FlipMask, sliderMax=1}
  ComponentMask(None=AppendVector(A=LinearInterpolate(A=@n2, B=@n3, Alpha=@n4), B=ComponentMask(None=@n1.Normalized UV) {g=true})) {r=true}
  MaterialFunctionCall!("UVs (Optional)") {materialFunction=/Engine/Functions/UserInterface/GetSlatePost1}
 ]
}
