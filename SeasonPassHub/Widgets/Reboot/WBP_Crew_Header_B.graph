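# Crew header widget for the Season Pass Hub, showing the Fortnite Crew badge, benefit text and a join button.
# HandleSubState sets benefit text and visibility by subscription state, showing the button only when Inactive, while SetPrice fills its price label.
# SetCabinMode flips two switchers to parental-controls text and button for restricted accounts.
# Join clicks broadcast HeaderEnteredSubscreen and start subscription; Opening plays Intro, a staggered slide-and-fade, at double speed.
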WidgetBlueprint /SeasonPassHub/Widgets/Reboot/WBP_Crew_Header_B
animations [Intro {
  binding_count=4
  length_seconds=0.9
  object=&self:Intro
  sequence {bindings [
    BenefitDescription {
     id=#1
     possessed_class=CommonRichTextBlock
     tracks [
      {
       type=MovieSceneFloatTrack
       display_name=Render Opacity
       sections [{type=MovieSceneFloatSection, end_frame=18, start_frame=0, channels [None {type=MovieSceneScriptingFloatChannel, keys [[2,0],[12,1]]}]}]
      }
      {
       type=MovieScene2DTransformTrack
       sections [{
         type=MovieScene2DTransformSection
         end_frame=18
         start_frame=0
         channels [
          Translation.X {type=MovieSceneScriptingFloatChannel, const [0,2]}
          Translation.Y {type=MovieSceneScriptingFloatChannel, keys [[2,80],[12,0]]}
         ]
        }]
      }
     ]
    }
    BenefitName {
     id=#2
     possessed_class=CommonTextBlock
     tracks [
      {
       type=MovieSceneFloatTrack
       display_name=Render Opacity
       sections [{type=MovieSceneFloatSection, end_frame=15, start_frame=0, channels [None {type=MovieSceneScriptingFloatChannel, keys [[3,0],[10,1]]}]}]
      }
      {
       type=MovieScene2DTransformTrack
       sections [{
         type=MovieScene2DTransformSection
         end_frame=15
         start_frame=0
         channels [
          Translation.X {type=MovieSceneScriptingFloatChannel, const [0,2]}
          Translation.Y {type=MovieSceneScriptingFloatChannel, keys [[3,80],[10,0]]}
         ]
        }]
      }
     ]
    }
    ViewCrew_Button {
     id=#3
     possessed_class=WBP_UIKit_Button_Loud_C
     tracks [{
       type=MovieSceneFloatTrack
       display_name=Render Opacity
       sections [{type=MovieSceneFloatSection, end_frame=15, start_frame=0, channels [None {type=MovieSceneScriptingFloatChannel, keys [[0,0],[10,0],[15,1]]}]}]
      }]
    }
    Pass_Badge {
     id=#4
     possessed_class=FortLazyImage
     tracks [
      {
       type=MovieSceneFloatTrack
       display_name=Render Opacity
       sections [{type=MovieSceneFloatSection, end_frame=10, start_frame=0, channels [None {type=MovieSceneScriptingFloatChannel, keys [[0,0],[8,1]]}]}]
      }
      {
       type=MovieScene2DTransformTrack
       sections [{
         type=MovieScene2DTransformSection
         end_frame=13
         start_frame=0
         channels [
          Translation.X {type=MovieSceneScriptingFloatChannel, const [0,2]}
          Translation.Y {type=MovieSceneScriptingFloatChannel, keys [[0,80],[8.0003,0]]}
         ]
        }]
      }
     ]
    }
   ]}
 }]
graphs [
 BP_GetDesiredFocusTarget {nodes [
   Get Desired Focus Target()
   > Return Node!(execute, ReturnValue:Widget Object Reference=Get ViewCrew_Button() -> (ViewCrew_Button:WBP UIKit Button Loud Object Reference))
  ]}
 EventGraph {nodes [
   "On Clicked (ViewCrew_Button)"!() -> (OutputDelegate, then, Button:Common Button Base Object Reference) {is=title K2Node_ComponentBoundEvent}
   > Call Header Entered Subscreen!(execute:Exec, self:WBP Crew Header B Object Reference) -> (then:Exec) {is=title K2Node_CallDelegate}
   > HandleSubscribeEngaged(self:Fort Progressive Table Of Contents Screen Object Reference)
   "On Clicked (Button_OpenParentalControls)"!() -> (OutputDelegate, then, Button:Common Button Base Object Reference) {is=title K2Node_ComponentBoundEvent}
   > RequestToOpenParentalControls!(execute) -> (then)
   Opening() {is=title K2Node_CustomEvent}
   > PlayAnimation(InAnimation=Get Intro(), PlaybackSpeed=2)
   Event Construct()
   > SetMobileStyle() {is=title K2Node_CallFunction}
   Event PreConstruct()
   Event Tick()
  ]}
 HandleSubState {nodes [
   HandleSubState!() -> (then, SubState:EMcpSubscriptionState Enum=Inactive) {is=title K2Node_FunctionEntry}
   > n1 = Sequence()
   > SetVisibility(self=Get ViewCrew_Button(), InVisibility=Collapsed)
   > SetVisibility(
    self=[
     Get Pass_Badge_2() -> (Pass_Badge_2:Fort Lazy Image Object Reference)
     Get BenefitDescription() -> (BenefitDescription:Common Rich Text Block Object Reference)
    ]
    InVisibility=Collapsed
   )
   n2 = Get CrewBenefitVM() -> (CrewBenefitVM:Crew Benefit VM Object Reference)
   n3 = Get BenefitDescription()
   > n4 = Set Text(
    Text=Select!(
     Inactive:Text=GetDescription!(self:Crew Benefit VM Object Reference=@n2) -> (ReturnValue:Text)
     Active:Text=Get ActiveRegularCrew() -> (ActiveRegularCrew:Text)
     ActiveWithExternalBilling:Text
     NotAutoRenewable:Text
     Canceled:Text=Get Cancelled() -> (Cancelled:Text)
     PaymentProcessError:Text
     PlatformAuthError:Text
     RewardsDelayed:Text
     Unavailable:Text
     Unknown:Text
     Index:EMcpSubscriptionState Enum=Get SubState() -> (SubState:EMcpSubscriptionState Enum=Inactive)
    ) -> (ReturnValue:Text)
    self:Rich Text Block Object Reference=@n3
   )
   > SetVisibility(
    self=@n3
    InVisibility=Select!(Option 0:ESlateVisibility Enum=SelfHitTestInvisible, Option 1:ESlateVisibility Enum=Collapsed, Index:bool=TextIsEmpty(InText=@n4)) -> (ReturnValue:ESlateVisibility Enum)
   )
   n5 = Switch on EMcpSubscriptionState!(execute=@n1.then_1, Selection:EMcpSubscriptionState Enum=Get SubState() -> (SubState=Inactive)) -> (
    Inactive:Exec
    Active:Exec
    ActiveWithExternalBilling:Exec
    NotAutoRenewable:Exec
    Canceled:Exec
    PaymentProcessError:Exec
    PlatformAuthError:Exec
    RewardsDelayed:Exec
    Unavailable:Exec
    Unknown:Exec
   ) {is=title K2Node_SwitchEnum}
   SetVisibility(execute=[@n5.Active, @n5.Canceled], self=Get BenefitDescription(), InVisibility=SelfHitTestInvisible)
   > SetVisibility(self=Get Pass_Badge_2(), InVisibility=SelfHitTestInvisible)
   SetVisibility(execute=@n5.Inactive, self=Get ViewCrew_Button(), InVisibility=Visible)
  ]}
 HeaderEnteredSubscreen {nodes [HeaderEnteredSubscreen() {is=title K2Node_FunctionEntry}]}
 SetCabinMode {nodes [
   n1 = SetCabinMode!() -> (then, IsCabinMode:bool=~) {is=title K2Node_FunctionEntry}
   > SetActiveWidget(
    self=Get ParentalSwitch() -> (ParentalSwitch:Common Visibility Switcher Object Reference)
    Widget=Select(
     Option 0:Widget Object Reference=Get ViewCrew_Button()
     Option 1:Widget Object Reference=Get Button_OpenParentalControls() -> (Button_OpenParentalControls:WBP UIKit Button Loud Object Reference)
     Index=@n1
    ) -> (ReturnValue:Widget Object Reference)
   )
   > SetActiveWidget(
    self=Get TextCabinSwitch() -> (TextCabinSwitch:Common Visibility Switcher Object Reference)
    Widget=Select(
     Option 0=Get CrewContent() -> (CrewContent:Horizontal Box Object Reference)
     Option 1=Get CabinText() -> (CabinText:Common Rich Text Block Object Reference)
     Index=@n1
    )
   )
  ]}
 SetMobileStyle {nodes [
   SetAutoWrapText!(execute, self:Text Object Reference, InAutoTextWrap:bool=false) -> (then)
   > SetWrapTextWidth!(execute, self:Common Text Object Reference, InWrapTextAt:int=1024) -> (then)
   SetMobileStyle() {is=title K2Node_FunctionEntry}
   > Branch(Condition=IsMobileGame())
   n1 = Get BenefitDescription()
   > SetAutoWrapText(self:Rich Text Block Object Reference=@n1, InAutoTextWrap=false)
   > Set WrapTextAt!(execute, WrapTextAt:float=1024, self:Text Layout Widget Object Reference=@n1) -> (then, Output_Get:float)
  ]}
 SetPrice {nodes [
   n1 = SetPrice!() -> (then, Crew Price:Text) {is=title K2Node_FunctionEntry}
   > Set Text(Text="LOCTEXT(\"JOIN FORTNITE CREW\")", self:Fort CTAButton Object Reference=Get ViewCrew_Button())
   n2 = Get SubscriptionState!(self:Crew VM Object Reference=Get CrewVM() -> (CrewVM:Crew VM Object Reference)) -> (SubscriptionState:EMcpSubscriptionState Enum=Inactive)
   n3 = TextToUpper!(InText:"Text (by ref)"=Get PriceUnownedRegular() -> (PriceUnownedRegular:Text)) -> (ReturnValue:Text)
   > SetSecondaryText!(
    execute
    self:Fort CTAButton Object Reference=Get ViewCrew_Button()
    ButtonSecondaryText:"Text (by ref)"=SelectText(
     B=Format Text!(Format=@n3, PRICE:Text=@n1) -> (Result)
     bPickA="Equal (Enum)"(A:EMcpSubscriptionState Enum=@n2, B:EMcpSubscriptionState Enum=Canceled)
    )
   ) -> (then)
   > n4 = Get!(execute:Exec, self:WBP UIKit Button CTA Base Object Reference=Get ViewCrew_Button()) -> (then:Exec, else:Exec, Block_SecondaryText:WBP UIKit Block Text Object Reference)
   > SetTransformPolicy!(execute, self:WBP UIKit Block Text Object Reference=@n4, Transform Policy:ETextTransformPolicy Enum=None) -> (then)
  ]}
]
variables [
 ActiveRegularCrew {type=string, default=Benefit unlocked through FN Crew}
 Cancelled {type=string, default=Benefit unlocked through FN Crew until JAN 26}
 HeaderEnteredSubscreen {}
]
widgets {
 info {
  inheritedWidgetCount=0
  parentClass=/Script/CrewUI.FortProgressiveTableOfContentsScreen
  rootWidgetClass=&CommonVisibilitySwitcher
  widgetCount=12
 }
 widgets [
  HeaderSwitcher {widgetClassPath=&CommonVisibilitySwitcher, bIsVariable=true, widget=HeaderSwitcher, properties {bOverride_Cursor=true}}
  VerticalBox_55 {widgetClassPath=&VerticalBox, parent=HeaderSwitcher, widget=VerticalBox_55, slot {class=CommonVisibilitySwitcherSlot}}
  Pass_Badge {
   widgetClassPath=&FortLazyImage
   bIsVariable=true
   parent=VerticalBox_55
   widget=Pass_Badge
   properties {brush {
     bIsDynamicallyLoaded=false
     drawAs=Image
     imageType=FullColor
     mirroring=NoMirror
     resourceName=None
     resourceObject=/Game/Athena/UI/Logos/Textures/M_UI_CrewLogo_Large
     tiling=NoTile
     imageSize=<189 56>
     margin=<ltrb 0 0 0 0>
     outlineSettings {
      bUseBrushTransparency=false
      roundingType=HalfHeightRadius
      width=0
      color {colorUseRule=UseColor_Specified, specifiedColor=<rgba 0 0 0 0>}
      cornerRadii=<0 0 0 0>
     }
     tintColor {colorUseRule=UseColor_Specified, specifiedColor=<rgba 1 1 1 1>}
     uVRegion {bIsValid=false, max=<0 0>, min=<0 0>}
    }}
   slot {class=VerticalBoxSlot, properties {horizontalAlignment=HAlign_Center, verticalAlignment=VAlign_Top, padding=<ltrb 0 0 0 8>}}
  }
  BenefitName {
   widgetClassPath=&CommonTextBlock
   bIsVariable=true
   parent=VerticalBox_55
   widget=BenefitName
   properties {
    justification=Center
    text=Text Block
    textTransformPolicy=ToUpper
    wrapTextAt=700
    font {
     bForceMonospaced=false
     bMaterialIsStencil=false
     fontMaterial=null
     fontObject=&HeadingNow
     letterSpacing=0
     monospacedWidth=1
     size=41.25
     skewAmount=0
     typefaceFontName=76 Bold Italic
     outlineSettings {
      bApplyOutlineToDropShadows=false
      bMiteredCorners=false
      bSeparateFillAlpha=false
      outlineMaterial=null
      outlineSize=0
      outlineColor=<rgba 0 0 0 1>
     }
    }
    shadowOffset=<0 0>
   }
   slot {class=VerticalBoxSlot, properties {horizontalAlignment=HAlign_Center, padding=<ltrb 0 0 0 20>}}
  }
  TextCabinSwitch {
   widgetClassPath=&CommonVisibilitySwitcher
   bIsVariable=true
   parent=VerticalBox_55
   widget=TextCabinSwitch
   slot {class=VerticalBoxSlot, properties {horizontalAlignment=HAlign_Center, padding=<ltrb 0 20 0 20>}}
  }
  CrewContent {
   widgetClassPath=&HorizontalBox
   bIsVariable=true
   parent=TextCabinSwitch
   widget=CrewContent
   slot {class=CommonVisibilitySwitcherSlot}
  }
  Pass_Badge_2 {
   widgetClassPath=&FortLazyImage
   bIsVariable=true
   parent=CrewContent
   widget=Pass_Badge_2
   properties {
    visibility=Collapsed
    brush {
     bIsDynamicallyLoaded=false
     drawAs=Image
     imageType=FullColor
     mirroring=NoMirror
     resourceName=None
     resourceObject=&M_UI_CrewLogo_Large
     tiling=NoTile
     imageSize=<99 32>
     margin=<ltrb 0 0 0 0>
     outlineSettings {
      bUseBrushTransparency=false
      roundingType=HalfHeightRadius
      width=0
      color {colorUseRule=UseColor_Specified, specifiedColor=<rgba 0 0 0 0>}
      cornerRadii=<0 0 0 0>
     }
     tintColor {colorUseRule=UseColor_Specified, specifiedColor=<rgba 1 1 1 1>}
     uVRegion {bIsValid=false, max=<0 0>, min=<0 0>}
    }
   }
   slot {class=HorizontalBoxSlot, properties {horizontalAlignment=HAlign_Center, verticalAlignment=VAlign_Top, padding=<ltrb 15 0 15 8>}}
  }
  BenefitDescription {
   widgetClassPath=&CommonRichTextBlock
   bIsVariable=true
   parent=CrewContent
   widget=BenefitDescription
   properties {
    bOverrideDefaultStyle=true
    justification=Center
    lineHeightPercentage=1.25
    text=Access this thing while subscribed to Crew
    wrapTextAt=550
    decoratorClasses [/Game/Athena/UI/Frontend/CosmeticGating/Data/RichTextDecorator_CosmeticGate.RichTextDecorator_CosmeticGate_C]
    defaultTextStyleOverride {
     overflowPolicy=Clip
     transformPolicy=None
     colorAndOpacity {colorUseRule=UseColor_Specified, specifiedColor=<rgba 1 1 1 1>}
     font {
      bForceMonospaced=false
      bMaterialIsStencil=false
      fontMaterial=null
      fontObject=&HeadingNow
      letterSpacing=10
      monospacedWidth=1
      size=18
      skewAmount=0
      typefaceFontName=74 Regular
      outlineSettings {
       bApplyOutlineToDropShadows=false
       bMiteredCorners=false
       bSeparateFillAlpha=false
       outlineMaterial=null
       outlineSize=0
       outlineColor=<rgba 0 0 0 1>
      }
     }
     highlightColor {colorUseRule=UseColor_Specified, specifiedColor=<rgba 0 0 0 1>}
     highlightShape {
      bIsDynamicallyLoaded=false
      drawAs=Image
      imageType=NoImage
      mirroring=NoMirror
      resourceName=None
      resourceObject=null
      tiling=NoTile
      imageSize=<32 32>
      margin=<ltrb 0 0 0 0>
      outlineSettings {
       bUseBrushTransparency=false
       roundingType=HalfHeightRadius
       width=0
       color {colorUseRule=UseColor_Specified, specifiedColor=<rgba 0 0 0 0>}
       cornerRadii=<0 0 0 0>
      }
      tintColor {colorUseRule=UseColor_Specified, specifiedColor=<rgba 1 1 1 1>}
      uVRegion {bIsValid=false, max=<0 0>, min=<0 0>}
     }
     selectedBackgroundColor {colorUseRule=UseColor_Foreground, specifiedColor=<rgba 1 0 1 1>}
     shadowColorAndOpacity=<rgba 0 0 0 1>
     shadowOffset=<0 0>
     strikeBrush {
      bIsDynamicallyLoaded=false
      drawAs=Image
      imageType=NoImage
      mirroring=NoMirror
      resourceName=None
      resourceObject=null
      tiling=NoTile
      imageSize=<32 32>
      margin=<ltrb 0 0 0 0>
      outlineSettings {
       bUseBrushTransparency=false
       roundingType=HalfHeightRadius
       width=0
       color {colorUseRule=UseColor_Specified, specifiedColor=<rgba 0 0 0 0>}
       cornerRadii=<0 0 0 0>
      }
      tintColor {colorUseRule=UseColor_Specified, specifiedColor=<rgba 1 1 1 1>}
      uVRegion {bIsValid=false, max=<0 0>, min=<0 0>}
     }
     underlineBrush {
      bIsDynamicallyLoaded=false
      drawAs=Image
      imageType=NoImage
      mirroring=NoMirror
      resourceName=None
      resourceObject=null
      tiling=NoTile
      imageSize=<32 32>
      margin=<ltrb 0 0 0 0>
      outlineSettings {
       bUseBrushTransparency=false
       roundingType=HalfHeightRadius
       width=0
       color {colorUseRule=UseColor_Specified, specifiedColor=<rgba 0 0 0 0>}
       cornerRadii=<0 0 0 0>
      }
      tintColor {colorUseRule=UseColor_Specified, specifiedColor=<rgba 1 1 1 1>}
      uVRegion {bIsValid=false, max=<0 0>, min=<0 0>}
     }
    }
   }
   slot {class=HorizontalBoxSlot, properties {horizontalAlignment=HAlign_Center, verticalAlignment=VAlign_Bottom, padding=<ltrb 15 0 15 40>}}
  }
  CabinText {
   widgetClassPath=&CommonRichTextBlock
   bIsVariable=true
   parent=TextCabinSwitch
   widget=CabinText
   properties {
    bOverrideDefaultStyle=true
    justification=Center
    lineHeightPercentage=1.25
    text="<text color=\"FFFFFFFF\" fontface=\"75 Medium\">Account setup required:</> to purchase this item or bundle ask your parent or guardian to set up your account from Parental Controls."
    wrapTextAt=550
    decoratorClasses [&RichTextDecorator_CosmeticGate_C]
    defaultTextStyleOverride {
     overflowPolicy=Clip
     transformPolicy=None
     colorAndOpacity {colorUseRule=UseColor_Specified, specifiedColor=<rgba 1 1 1 1>}
     font {
      bForceMonospaced=false
      bMaterialIsStencil=false
      fontMaterial=null
      fontObject=&HeadingNow
      letterSpacing=10
      monospacedWidth=1
      size=18
      skewAmount=0
      typefaceFontName=74 Regular
      outlineSettings {
       bApplyOutlineToDropShadows=false
       bMiteredCorners=false
       bSeparateFillAlpha=false
       outlineMaterial=null
       outlineSize=0
       outlineColor=<rgba 0 0 0 1>
      }
     }
     highlightColor {colorUseRule=UseColor_Specified, specifiedColor=<rgba 0 0 0 1>}
     highlightShape {
      bIsDynamicallyLoaded=false
      drawAs=Image
      imageType=NoImage
      mirroring=NoMirror
      resourceName=None
      resourceObject=null
      tiling=NoTile
      imageSize=<32 32>
      margin=<ltrb 0 0 0 0>
      outlineSettings {
       bUseBrushTransparency=false
       roundingType=HalfHeightRadius
       width=0
       color {colorUseRule=UseColor_Specified, specifiedColor=<rgba 0 0 0 0>}
       cornerRadii=<0 0 0 0>
      }
      tintColor {colorUseRule=UseColor_Specified, specifiedColor=<rgba 1 1 1 1>}
      uVRegion {bIsValid=false, max=<0 0>, min=<0 0>}
     }
     selectedBackgroundColor {colorUseRule=UseColor_Foreground, specifiedColor=<rgba 1 0 1 1>}
     shadowColorAndOpacity=<rgba 0 0 0 1>
     shadowOffset=<0 0>
     strikeBrush {
      bIsDynamicallyLoaded=false
      drawAs=Image
      imageType=NoImage
      mirroring=NoMirror
      resourceName=None
      resourceObject=null
      tiling=NoTile
      imageSize=<32 32>
      margin=<ltrb 0 0 0 0>
      outlineSettings {
       bUseBrushTransparency=false
       roundingType=HalfHeightRadius
       width=0
       color {colorUseRule=UseColor_Specified, specifiedColor=<rgba 0 0 0 0>}
       cornerRadii=<0 0 0 0>
      }
      tintColor {colorUseRule=UseColor_Specified, specifiedColor=<rgba 1 1 1 1>}
      uVRegion {bIsValid=false, max=<0 0>, min=<0 0>}
     }
     underlineBrush {
      bIsDynamicallyLoaded=false
      drawAs=Image
      imageType=NoImage
      mirroring=NoMirror
      resourceName=None
      resourceObject=null
      tiling=NoTile
      imageSize=<32 32>
      margin=<ltrb 0 0 0 0>
      outlineSettings {
       bUseBrushTransparency=false
       roundingType=HalfHeightRadius
       width=0
       color {colorUseRule=UseColor_Specified, specifiedColor=<rgba 0 0 0 0>}
       cornerRadii=<0 0 0 0>
      }
      tintColor {colorUseRule=UseColor_Specified, specifiedColor=<rgba 1 1 1 1>}
      uVRegion {bIsValid=false, max=<0 0>, min=<0 0>}
     }
    }
   }
   slot {class=CommonVisibilitySwitcherSlot, properties {horizontalAlignment=HAlign_Center, verticalAlignment=VAlign_Bottom, padding=<ltrb 0 0 0 40>}}
  }
  ParentalSwitch {
   widgetClassPath=&CommonVisibilitySwitcher
   bIsVariable=true
   parent=VerticalBox_55
   widget=ParentalSwitch
   slot {class=VerticalBoxSlot, properties {horizontalAlignment=HAlign_Center}}
  }
  ViewCrew_Button {
   widgetClassPath=/FortUILibrary/UIKit/Buttons/WBP_UIKit_Button_Loud
   bIsVariable=true
   parent=ParentalSwitch
   widget=ViewCrew_Button
   properties {bHideInputAction=true, buttonSize=Large, minWidth=458, secondaryText=" $X.XX / Month", text="Join fortnite Crew "}
   slot {class=CommonVisibilitySwitcherSlot}
  }
  Button_OpenParentalControls {
   widgetClassPath=&WBP_UIKit_Button_Loud
   bIsVariable=true
   parent=ParentalSwitch
   widget=Button_OpenParentalControls
   properties {bHideInputAction=true, text=Open Parental Controls}
   slot {class=CommonVisibilitySwitcherSlot, properties {horizontalAlignment=HAlign_Center}}
  }
 ]
}
