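# WBP_SeasonPass_CrewPassAccess is a Season Pass hub banner telling the player that the pass is unlocked through Fortnite Crew.
# An Overlay wraps a horizontal row holding the Crew logo image, a thin semi-transparent separator and an auto-wrapping CommonTextBlock that fills the remaining width.
# The widget is purely presentational, with static default text, an empty event graph and no variables of its own beyond the exposed child widgets.
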WidgetBlueprint /SeasonPassHub/Widgets/Reboot/WBP_SeasonPass_CrewPassAccess
graphs [EventGraph {links [], nodes []}]
variables []
widgets {
 info {inheritedWidgetCount=0, parentClass=&UserWidget, rootWidgetClass=&Overlay, widgetCount=5}
 widgets [
  Overlay_23 {widgetClassPath=&Overlay, widget=Overlay_23}
  HorizontalBox_26 {
   widgetClassPath=&HorizontalBox
   parent=Overlay_23
   widget=HorizontalBox_26
   slot {class=OverlaySlot, properties {horizontalAlignment=HAlign_Fill, verticalAlignment=VAlign_Fill}}
  }
  Crew_Icon {
   widgetClassPath=&Image
   bIsVariable=true
   parent=HorizontalBox_26
   widget=Crew_Icon
   properties {brush {
     bIsDynamicallyLoaded=false
     drawAs=Image
     imageType=FullColor
     mirroring=NoMirror
     resourceName=None
     resourceObject=/Game/Athena/UI/Logos/Textures/M_UI_CrewLogo_Large
     tiling=NoTile
     imageSize=<100 32>
     margin=<ltrb 0 0 0 0>
     outlineSettings {
      bUseBrushTransparency=false
      roundingType=HalfHeightRadius
      width=0
      color {colorUseRule=UseColor_Specified, specifiedColor=<rgba 0 0 0 0>}
      cornerRadii=<0 0 0 0>
     }
     tintColor {colorUseRule=UseColor_Specified, specifiedColor=<rgba 1 1 1 1>}
     uVRegion {bIsValid=false, max=<0 0>, min=<0 0>}
    }}
   slot {class=HorizontalBoxSlot, properties {horizontalAlignment=HAlign_Left, verticalAlignment=VAlign_Center}}
  }
  Separator {
   widgetClassPath=&Image
   bIsVariable=true
   parent=HorizontalBox_26
   widget=Separator
   properties {
    renderOpacity=0.25
    visibility=SelfHitTestInvisible
    brush {
     bIsDynamicallyLoaded=false
     drawAs=Image
     imageType=NoImage
     mirroring=NoMirror
     resourceName=None
     resourceObject=null
     tiling=NoTile
     imageSize=<2 1>
     margin=<ltrb 0 0 0 0>
     outlineSettings {
      bUseBrushTransparency=false
      roundingType=HalfHeightRadius
      width=0
      color {colorUseRule=UseColor_Specified, specifiedColor=<rgba 0 0 0 0>}
      cornerRadii=<0 0 0 0>
     }
     tintColor {colorUseRule=UseColor_Specified, specifiedColor=<rgba 1 1 1 1>}
     uVRegion {bIsValid=false, max=<0 0>, min=<0 0>}
    }
   }
   slot {class=HorizontalBoxSlot, properties {horizontalAlignment=HAlign_Center, padding=<ltrb 16 12 8 12>}}
  }
  Text_Description {
   widgetClassPath=&CommonTextBlock
   bIsVariable=true
   parent=HorizontalBox_26
   widget=Text_Description
   properties {
    autoWrapText=true
    scrollStyle=/FortUILibrary/Styles/Typography/Foundation/TextScrollStyle-NoFade.TextScrollStyle-NoFade_C
    text=Pass unlocked through FN Crew
    wrapTextAt=668
    font {
     bForceMonospaced=false
     bMaterialIsStencil=false
     fontMaterial=null
     fontObject=&HeadingNow
     letterSpacing=1
     monospacedWidth=1
     size=15.75
     skewAmount=0
     typefaceFontName=74 Regular
     outlineSettings {
      bApplyOutlineToDropShadows=false
      bMiteredCorners=false
      bSeparateFillAlpha=false
      outlineMaterial=null
      outlineSize=0
      outlineColor=<rgba 0 0 0 1>
     }
    }
    shadowOffset=<0 0>
   }
   slot {
    class=HorizontalBoxSlot
    properties {horizontalAlignment=HAlign_Left, verticalAlignment=VAlign_Center, padding=<ltrb 16 8 16 8>, size {sizeRule=Fill, value=1}}
   }
  }
 ]
}
