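# Translucent UI material for a Crew category button, with a rounded-corner mask, themed background and foreground image.
# Four theme color parameters supply the palette, with gradient, spotlight color-dodge and bottom-fade layers adding shading.
# Rotated Crew bento texture tiles pan over time, using random UV noise for time offsets.
# Content parameters pan and scale the clamped, edge-clipped image, IsLocked lowers its opacity, and Hover Overlay blends in a brightened state.
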Material /Crew/UI/Materials/M_UI_CategoryButton_Crew
graph {
 properties {
  bAutomaticallySetUsageInEditor=false
  bUsedWithStaticMesh=false
  blendMode=BLEND_Translucent
  materialDomain=MD_UI
  usedShadingModels=Default Lit
 }
 nodes [
  n1 = MaterialFunctionCall(Corner Radius=Constant() {r=20}, Radius in Pixels=Constant() {r=1}) {materialFunction=&MF_UI_CoreUI_RoundedMask}
  NamedRerouteDeclaration(None=@n1.Shape Mask) {name=Shape Mask, nodeColor=<rgba 0 0 0 1>}
  n2 = SmoothStep(Value=NamedRerouteUsage()) {constMax=0.5}
  n3 = Saturate(None=OneMinus(None=SquareRoot(None=@n2)))
  NamedRerouteDeclaration(None=LinearInterpolate(A=NamedRerouteUsage(), B=NamedRerouteUsage(), Alpha=Multiply(A=@n3) {constB=0.4}) {constAlpha=0}) {name=BG color, nodeColor=<rgba 0 0 0 1>}
  n4 = VectorParameter() {group=ThemeColors, parameterName=Color Theme 1, sortPriority=1, defaultValue=<rgba 0.2664 0.4397 0.7529 1>}
  NamedRerouteDeclaration(None=@n4.RGB) {name=Color Theme 1, nodeColor=<rgba 0 0 0 1>}
  n5 = VectorParameter() {group=ThemeColors, parameterName=Color Theme 2, sortPriority=2, defaultValue=<rgba 0.0999 0.107 0.3864 1>}
  NamedRerouteDeclaration(None=@n5.RGB) {name=Color Theme 2, nodeColor=<rgba 0 0 0 1>}
  n6 = VectorParameter() {group=ThemeColors, parameterName=Color Theme 3, sortPriority=3, defaultValue=<rgba 0.0356 0.0168 0.1779 1>}
  NamedRerouteDeclaration(None=@n6.RGB) {name=Color Theme 3, nodeColor=<rgba 0 0 0 1>}
  n7 = VectorParameter() {group=ThemeColors, parameterName=Color Theme 4, sortPriority=4, defaultValue=<rgba 0.0232 0.0056 0.0865 1>}
  NamedRerouteDeclaration(None=@n7.RGB) {name=Color Theme 4, nodeColor=<rgba 0 0 0 1>}
  n8 = TextureObjectParameter() {group=Content Controller, parameterName=ForegroundImage}
  n9 = TextureSample(UVs=NamedRerouteUsage(), Tex=@n8) {samplerSource=SSM_Clamp_WorldGroupSettings}
  n10 = MaterialFunctionCall!(UVs=NamedRerouteUsage(), Alpha=@n9.A) {materialFunction=/FortUILibrary/MaterialFunctions/Basics/MF_UI_ClipEdgePixel}
  n11 = ScalarParameter() {group=Content Controller, parameterName=IsLocked}
  n12 = LinearInterpolate(B=Multiply(A=@n11) {constB=0.4}, Alpha=@n11) {constA=1}
  NamedRerouteDeclaration(None=LinearInterpolate(A=NamedRerouteUsage(), B=@n9.RGB, Alpha=Multiply(A=@n10.Result, B=@n12))) {name=Content + BG, nodeColor=<rgba 0 0 0 1>}
  n13 = Add(A=TextureCoordinate(), B=Constant2Vector() {r=-0.5})
  n14 = Constant() {r=0.8}
  n15 = Multiply(
   A=Multiply(A=NamedRerouteUsage(), B=Constant3Vector() {constant=<rgba 0.5521 0.5154 1 1>}) {constB=0.2}
   B=Power(Base=MaterialFunctionCall!(UVs=@n13, Radius=@n14) {materialFunction=/FortUILibrary/MaterialFunctions/Shapes/SDFs/MF_SDF_Circle}.SDF)
  )
  n16 = NamedRerouteUsage()
  n17 = AppendVector(A=ComponentMask(None=@n16) {r=true}, B=Add(A=ComponentMask(None=@n16) {g=true}) {constB=-0.2})
  n18 = Constant() {r=4}
  n19 = Multiply(A=MaterialFunctionCall(UV=MaterialFunctionCall(UVs=@n17, Texture Scale=@n18) {materialFunction=&ScaleUVsByCenter}.UVs) {materialFunction=&MF_UI_RadialGradient}.Smooth) {constB=0.4}
  n20 = Divide(A=NamedRerouteUsage(), B=Max(A=OneMinus(None=@n19)) {constB=0})
  NamedRerouteDeclaration(None=MaterialFunctionCall!(Blend=Min(A=@n15) {constB=0.3}, Base=Saturate(None=@n20)) {materialFunction=/Engine/Functions/Engine_MaterialFunctions03/Blends/Blend_ColorDodge}.Result) {name=Background + Spotlight, nodeColor=<rgba 0 0 0 1>}
  n21 = NamedRerouteUsage()
  n22 = NamedRerouteUsage()
  n23 = Power(Base=@n22) {constExponent=0.2}
  n24 = Max(A=OneMinus(None=Multiply(A=@n23) {constB=0.2})) {constB=0}
  n25 = Add(A=Saturate(None=Divide(A=@n21, B=@n24)), B=Multiply(A=Power(Base=@n22) {constExponent=0.85}) {constB=0.3})
  NamedRerouteDeclaration(None=LinearInterpolate(A=@n21, B=Saturate(None=@n25), Alpha=ScalarParameter() {group=Hover, parameterName=Hover Overlay})) {name=Hover Overlay, nodeColor=<rgba 0 0 0 1>}
  NamedRerouteDeclaration(None=TextureObjectParameter() {parameterName=CMS_Crew_Bento, texture=/Game/Athena/UI/Subscriptions/Art/Textures/T_CrewUpsell_Bento}) {name=Texture Object, nodeColor=<rgba 0 0 0 1>}
  NamedRerouteDeclaration(None=MaterialFunctionCall() {materialFunction=&MF_UI_UVScaling}.ScaleY) {name=UV-Scale-Y, nodeColor=<rgba 0 0 0 1>}
  n26 = MaterialFunctionCall(
   UVs=TextureCoordinate()
   GradientStart=Constant() {r=0.15}
   RotationAngle=Constant() {r=0.71}
   GradientOffset=ScalarParameter() {group=Content Controller, parameterName=GradientOffset}
  ) {materialFunction=&MF_UI_LinearGradient}
  n27 = Sine(None=Multiply(A=Subtract(A=@n26.Result) {constB=0.5}) {constB=0.5})
  n28 = Saturate(None=Multiply(A=Add(A=@n27)) {constB=0.5})
  n29 = ScalarParameter() {defaultValue=0.95, group=Content Controller, parameterName=BottomFade}
  NamedRerouteDeclaration(None=LinearInterpolate(A=NamedRerouteUsage(), B=NamedRerouteUsage(), Alpha=Multiply(A=@n28, B=@n29))) {name=Texture + Color, nodeColor=<rgba 0 0 0 1>}
  n30 = Multiply(A=NamedRerouteUsage(), B=AppendVector(A=Constant() {r=0.5}, B=NamedRerouteUsage()))
  n31 = AppendVector(
   A=ScalarParameter() {defaultValue=0.25, group=Crew, parameterName=FG_Time_Offset_X-Tiling}
   B=ScalarParameter() {defaultValue=1, group=Crew, parameterName=FG_Time_Offset_Y-Tiling}
  )
  n32 = ScalarParameter() {defaultValue=-0.5, group=Crew, parameterName=FG_Time_Offset_X-Offsets}
  n33 = Round(None=ScalarParameter() {defaultValue=4, group=Crew, parameterName=FG_Time_Offset_Y-Offsets})
  n34 = MaterialFunctionCall(Float2=Add(A=ScalarParameter() {defaultValue=0.0625, group=Crew, parameterName=FG_Time_Mask_Offset}, B=@n30)) {materialFunction=&BreakOutFloat2Components}
  n35 = ConstantBiasScale(None=Add(A=MaterialFunctionCall(Value=@n34.G) {materialFunction=&LinearSine}.Linear Sine) {constB=-0.5}) {bias=-0.125, scale=12}
  n36 = MaterialFunctionCall!(
   PrimaryTextureCoordinates=Add(A=Multiply(A=@n30, B=@n31), B=AppendVector(A=@n32, B=@n33))
   2ndOrderTextureCoordinates
   RandomSeed
   NoiseDensityBias=LinearInterpolate(
    A=ScalarParameter() {defaultValue=1, group=Crew, parameterName=FG_Time_Mask_Upper_Range}
    B=ScalarParameter() {defaultValue=0.125, group=Crew, parameterName=FG_Time_Mask_Lower_Range}
    Alpha=Saturate(None=@n35)
   ) {constA=1, constB=0.5}
   Bypass Noise Influence Crosstalk
  ) {materialFunction=/Game/Effects/Fort_Effects/Materials/_Functions/MF_RandomUVSample}
  NamedRerouteDeclaration(None=Multiply(
    A=ConstantBiasScale(None=@n36.NoiseWithRange) {bias=0.5, scale=-2}
    B=ScalarParameter() {defaultValue=0.05, group=Crew, parameterName=Tile_FG_Time_Factor}
   )) {name=FG_Time_Offset, nodeColor=<rgba 0 0 0 1>}
  n37 = Multiply(A=Constant2Vector() {g=1, r=0.25}, B=Multiply(A=NamedRerouteUsage(), B=NamedRerouteUsage()))
  n38 = Multiply(A=Time(), B=NamedRerouteUsage())
  NamedRerouteDeclaration(None=TextureSample(UVs=Panner(Coordinate=@n37, Time=@n38) {speedX=0.125}, Tex=NamedRerouteUsage()).A) {name=Tiles, nodeColor=<rgba 0 0 0 1>}
  NamedRerouteDeclaration(None=@n1.Outline SDF) {name=ShapeMask_SDF, nodeColor=<rgba 0 0 0 1>}
  NamedRerouteDeclaration(None=TextureProperty(Tex=@n8)) {name=Texture Size, nodeColor=<rgba 0 0 0 1>}
  n39 = Saturate(None=MaterialFunctionCall(UVs=TextureCoordinate(), GradientStart=Constant() {r=-1}, RotationAngle=Constant() {r=-0.05}, GradientOffset=Constant() {r=-0.5}) {materialFunction=&MF_UI_LinearGradient}.Result)
  NamedRerouteDeclaration(None=LinearInterpolate(A=NamedRerouteUsage(), B=NamedRerouteUsage(), Alpha=OneMinus(None=@n39))) {name=BaseGradient, nodeColor=<rgba 0 0 0 1>}
  n40 = MaterialFunctionCall(
   UVs=TextureCoordinate()
   GradientStart=Constant() {r=0.25}
   GradientEnd=Constant() {r=0.8}
   RotationAngle=Constant() {r=-0.05}
   GradientOffset=Constant() {r=-0.1}
  ) {materialFunction=&MF_UI_LinearGradient}
  NamedRerouteDeclaration(None=LinearInterpolate(A=NamedRerouteUsage(), B=NamedRerouteUsage(), Alpha=Multiply(A=@n40.Result, B=@n40.Result))) {name=Gradient Highlight, nodeColor=<rgba 0 0 0 1>}
  NamedRerouteDeclaration(None=Rotator(Coordinate=NamedRerouteUsage(), Time=ScalarParameter() {defaultValue=0.08, group=Crew, parameterName=Tile_Angle}) {speed=6.2832}) {name=Screen_Angle, nodeColor=<rgba 0 0 0 1>}
  NamedRerouteDeclaration(None=ScalarParameter() {defaultValue=1.675, group=Crew, parameterName=Tile-Scale-Foreground}) {name=FG_Tile_Scale, nodeColor=<rgba 0 0 0 1>}
  n41 = MaterialFunctionCall(
   UV=MaterialFunctionCall() {materialFunction=&GetUserInterfaceUV}.Normalized UV
   TextureSize=NamedRerouteUsage()
   ScalePivotPoint=Constant2Vector() {g=1, r=1}
  ) {materialFunction=&MF_UI_UVScaling}
  n42 = AppendVector(A=ScalarParameter() {group=Content Controller, parameterName=PanX}, B=ScalarParameter() {group=Content Controller, parameterName=PanY})
  NamedRerouteDeclaration(None=MaterialFunctionCall(UVs=Add(A=@n41.ScaleY, B=@n42), Texture Scale=ScalarParameter() {defaultValue=1, group=Content Controller, parameterName=Scale}) {materialFunction=&ScaleUVsByCenter}.UVs) {name=Content Controller, nodeColor=<rgba 0 0 0 1>}
  NamedRerouteUsage()
  NamedRerouteUsage()
 ]
}
