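# Translucent UI material for character name text on the Battle Pass Reboot screen, with animated color and an adjustable reveal.
# The font signed distance field produces antialiased fill and outline alphas.
# A rotatable, flippable UV gradient and smoothstep thresholds form reveal masks, with separate amounts for fill and outline.
# Color builds on a purple-to-pale-lavender blend and a time-driven hue shift, alongside panning screen-space noise; named reroutes link the parts.
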Material /SeasonPass/Widgets/Reboot/Materials/M_UI_BPReboot_CharacterName_Font
graph {
 properties {bAutomaticallySetUsageInEditor=false, blendMode=BLEND_Translucent, materialDomain=MD_UI, usedShadingModels=Default Lit}
 nodes [
  LinearInterpolate(A=Multiply(A=NamedRerouteUsage(), B=NamedRerouteUsage()), Alpha=Multiply(A=NamedRerouteUsage(), B=NamedRerouteUsage()))
  LinearInterpolate(A=NamedRerouteUsage(), Alpha=Multiply(A=NamedRerouteUsage(), B=NamedRerouteUsage()))
  MaterialFunctionCall() {materialFunction=&DebugTimeSine}
  n1 = NamedRerouteUsage()
  Multiply(A=@n1) {constB=2}
  n2 = TextureCoordinate()
  n3 = ScalarParameter() {defaultValue=-0.225, parameterName=Angle}
  n4 = ComponentMask(None=MaterialFunctionCall(
    UVs=MaterialFunctionCall(UVs=@n2, "Rotation Angle (0-1)"=@n3) {materialFunction=&CustomRotator}.Rotated Values
    Texture Scale=ScalarParameter() {defaultValue=1.5, parameterName=Mask Scale}
   ) {materialFunction=&ScaleUVsByCenter}.UVs) {g=true}
  NamedRerouteDeclaration(None=LinearInterpolate(A=@n4, B=OneMinus(None=@n4), Alpha=ScalarParameter() {parameterName=Flip Mask})) {name=Mask Gradient, nodeColor=<rgba 0 0 0 1>}
  n5 = Panner(
   Coordinate=MaterialFunctionCall(UVs=ScreenPosition().ViewportUV, "Rotation Angle (0-1)"=Constant() {r=0.125}) {materialFunction=&CustomRotator}.Rotated Values
   Time=Time()
  ) {speedX=-0.2, speedY=0.1}
  n6 = Constant2Vector() {g=2, r=1}
  NamedRerouteDeclaration(None=TextureSample(UVs=MaterialFunctionCall(UVs=@n5, Texture Scale=@n6) {materialFunction=&ScaleUVsByCenter}.UVs) {samplerType=SAMPLERTYPE_Grayscale, texture=&LowResBlurredNoise}.R) {name=Noise, nodeColor=<rgba 0 0 0 1>}
  n7 = ScalarParameter() {defaultValue=0.5, group=9 - Reveal Mask, parameterName=Reveal Fade}
  n8 = Subtract(A=Constant(), B=@n7)
  n9 = OneMinus(None=ScalarParameter() {defaultValue=1, group=9 - Reveal Mask, parameterName=Outline Reveal Amount})
  n10 = Add(A=Constant() {r=1}, B=@n7)
  NamedRerouteDeclaration(None=SmoothStep(Min=LinearInterpolate(A=@n8, Alpha=@n9), Max=LinearInterpolate(B=@n10, Alpha=@n9), Value=NamedRerouteUsage())) {name=Outline Mask, nodeColor=<rgba 0 0 0 1>}
  n11 = FontSignedDistance()
  n12 = Multiply(A=@n11.Signed Distance, B=@n11.Signed Distance)
  n13 = Constant() {r=0.02}
  NamedRerouteDeclaration(None=MaterialFunctionCall!(DistanceField=@n12, Threshold, Thickness=Multiply(A=@n13, B=@n11.Pixel Distance Factor), Alignment=Constant() {r=1}, Smoothing) {materialFunction=/FortUILibrary/MaterialFunctions/Shapes/Modifiers/MF_UI_ShapeOutline}.Outline) {name=Outline Alpha, nodeColor=<rgba 0 0 0 1>}
  NamedRerouteDeclaration(None=MaterialFunctionCall!(Threshold=Constant(), Gradient=@n12, Filter Width Override, Filter Width Fudge Factor) {materialFunction=/Game/UI/Materials/AAThreshold}.Result) {name=Fill Alpha, nodeColor=<rgba 0 0 0 1>}
  n14 = Time()
  n15 = LinearInterpolate(A=Constant3Vector() {constant=<rgba 0.3356 0 1 1>}, B=Constant3Vector() {constant=<rgba 0.9013 0.7865 1 1>}, Alpha=NamedRerouteUsage())
  n16 = MaterialFunctionCall(Hue Shift Percentage=Add(A=@n1, B=Multiply(A=@n14) {constB=0.5}), Texture=@n15) {materialFunction=&HueShift}
  NamedRerouteDeclaration(None=LinearInterpolate(A=Multiply(A=@n16.Result), B=@n15)) {name=Color, nodeColor=<rgba 0 0.6016 1 1>}
  n17 = ScalarParameter() {defaultValue=0.5, group=9 - Reveal Mask, parameterName=Reveal Fade}
  n18 = Subtract(A=Constant(), B=@n17)
  n19 = OneMinus(None=ScalarParameter() {defaultValue=1, group=9 - Reveal Mask, parameterName=Reveal Amount})
  n20 = Add(A=Constant() {r=1}, B=@n17)
  NamedRerouteDeclaration(None=SmoothStep(Min=LinearInterpolate(A=@n18, Alpha=@n19), Max=LinearInterpolate(B=@n20, Alpha=@n19), Value=NamedRerouteUsage())) {name=Mask, nodeColor=<rgba 0 0 0 1>}
 ]
}
