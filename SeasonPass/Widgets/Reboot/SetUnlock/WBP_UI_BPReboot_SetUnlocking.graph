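# Full-screen Season Pass modal that celebrates unlocking a new set, as a CommonActivatableWidget.
# On activation it plays the one-shot Celebration animation and an unlock sound, then deactivates itself when the animation finishes.
# Celebration fades the blur, animates the background material's reveal parameters, and slides the set image and name texts across the screen.
# SetCharacterImage loads the theme data's LargePreview image into the lazy image; the set name is placeholder text no graph updates.
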WidgetBlueprint /SeasonPass/Widgets/Reboot/SetUnlock/WBP_UI_BPReboot_SetUnlocking
animations [Celebration {
  binding_count=6
  length_seconds=4.25
  object=&self:Celebration
  sequence {bindings [
    SetName_Background {
     id=#1
     possessed_class=UEFN_TextBlock_C
     tracks [
      {
       type=MovieScene2DTransformTrack
       sections [{
         type=MovieScene2DTransformSection
         end_frame=78
         start_frame=0
         channels [
          Translation.X {type=MovieSceneScriptingFloatChannel, keys [[5,1200],[78,-1200]]}
          Translation.Y {type=MovieSceneScriptingFloatChannel, const [0,2]}
          Angle {type=MovieSceneScriptingFloatChannel, const [0,1]}
          Scale.X {type=MovieSceneScriptingFloatChannel, const [3,1]}
          Scale.Y {type=MovieSceneScriptingFloatChannel, const [3,1]}
          Shear.X {type=MovieSceneScriptingFloatChannel, const [0,1]}
          Shear.Y {type=MovieSceneScriptingFloatChannel, const [0,1]}
         ]
        }]
      }
      {
       type=MovieSceneColorTrack
       display_name=Color and Opacity
       sections [{
         type=MovieSceneColorSection
         end_frame=78
         start_frame=0
         channels [
          Color.R {type=MovieSceneScriptingFloatChannel, const [1,1]}
          Color.G {type=MovieSceneScriptingFloatChannel, const [1,1]}
          Color.B {type=MovieSceneScriptingFloatChannel, const [1,1]}
          Color.A {type=MovieSceneScriptingFloatChannel, keys [[5,0],[12,1],[71,1],[78,0]]}
         ]
        }]
      }
     ]
    }
    BackgroundBlur_51 {
     id=#2
     possessed_class=BackgroundBlur
     tracks [{
       type=MovieSceneFloatTrack
       display_name=Blur Strength
       sections [{
         type=MovieSceneFloatSection
         end_frame=85
         start_frame=0
         channels [None {type=MovieSceneScriptingFloatChannel, keys [[0,0],[8,30],[78,30],[85,0]]}]
        }]
      }]
    }
    SetName {
     id=#3
     possessed_class=UEFN_TextBlock_C
     tracks [
      {
       type=MovieScene2DTransformTrack
       sections [{
         type=MovieScene2DTransformSection
         end_frame=78
         start_frame=0
         channels [
          Translation.X {type=MovieSceneScriptingFloatChannel, keys [[5,-320],[78,324.589]]}
          Translation.Y {type=MovieSceneScriptingFloatChannel, const [0,2]}
          Angle {type=MovieSceneScriptingFloatChannel, const [0,1]}
          Scale.X {type=MovieSceneScriptingFloatChannel, const [1,1]}
          Scale.Y {type=MovieSceneScriptingFloatChannel, const [1,1]}
          Shear.X {type=MovieSceneScriptingFloatChannel, const [0,1]}
          Shear.Y {type=MovieSceneScriptingFloatChannel, const [0,1]}
         ]
        }]
      }
      {
       type=MovieSceneColorTrack
       display_name=Color and Opacity
       sections [{
         type=MovieSceneColorSection
         end_frame=78
         start_frame=0
         channels [
          Color.R {type=MovieSceneScriptingFloatChannel, const [1,2]}
          Color.G {type=MovieSceneScriptingFloatChannel, const [1,2]}
          Color.B {type=MovieSceneScriptingFloatChannel, const [1,2]}
          Color.A {type=MovieSceneScriptingFloatChannel, keys [[5,0],[12,1],[71,1],[78,0]]}
         ]
        }]
      }
     ]
    }
    SetImage {
     id=#4
     possessed_class=FortLazyImage
     tracks [
      {
       type=MovieSceneFloatTrack
       display_name=Render Opacity
       sections [{
         type=MovieSceneFloatSection
         end_frame=78
         start_frame=0
         channels [None {type=MovieSceneScriptingFloatChannel, keys [[6,0],[21,1],[75,1],[78,0]]}]
        }]
      }
      {
       type=MovieScene2DTransformTrack
       sections [{
         type=MovieScene2DTransformSection
         end_frame=78
         start_frame=0
         channels [
          Translation.X {type=MovieSceneScriptingFloatChannel, keys [[6,240],[78,-170]]}
          Translation.Y {type=MovieSceneScriptingFloatChannel, const [0,2]}
          Angle {type=MovieSceneScriptingFloatChannel, const [0,2]}
          Scale.X {type=MovieSceneScriptingFloatChannel, const [1,2]}
          Scale.Y {type=MovieSceneScriptingFloatChannel, const [1,2]}
          Shear.X {type=MovieSceneScriptingFloatChannel, const [0,2]}
          Shear.Y {type=MovieSceneScriptingFloatChannel, const [0,2]}
         ]
        }]
      }
     ]
    }
    Background {
     id=#5
     possessed_class=Image
     tracks [{
       type=MovieSceneWidgetMaterialTrack
       sections [{
         type=MovieSceneParameterSection
         end_frame=85
         start_frame=0
         channels [
          Modal Reveal Amount {type=MovieSceneScriptingFloatChannel, keys [[6,0],[15,1],[73,1],[81,0]]}
          OverlayOpacity {type=MovieSceneScriptingFloatChannel, keys [[0,0],[7,1],[73,1],[84,0]]}
          Blobs Opacity {type=MovieSceneScriptingFloatChannel, keys [[0,0],[10,0.35]]}
          Line Reveal {type=MovieSceneScriptingFloatChannel, keys [[0,0],[6,1],[12,1],[18,0]]}
          Edge Intensity {type=MovieSceneScriptingFloatChannel, keys [[0,1],[7,0.65]]}
          Underlight_Opacity {type=MovieSceneScriptingFloatChannel, keys [[0,0],[5,1],[78,1],[84,0]]}
          Color 1R {type=MovieSceneScriptingFloatChannel, keys [[0,0.5885],[8,0]]}
          Color 1G {type=MovieSceneScriptingFloatChannel, keys [[0,0.7541],[8,0.4025]]}
          Color 1B {type=MovieSceneScriptingFloatChannel, const [1,2]}
          Color 1A {type=MovieSceneScriptingFloatChannel, const [1,2]}
          Color 2R {type=MovieSceneScriptingFloatChannel, keys [[0,0.8603],[8,0.6984]]}
          Color 2G {type=MovieSceneScriptingFloatChannel, keys [[0,0.5729],[8,0.0779]]}
          Color 2B {type=MovieSceneScriptingFloatChannel, const [1,2]}
          Color 2A {type=MovieSceneScriptingFloatChannel, const [1,2]}
         ]
        }]
      }]
    }
    SetUnlockedTag {
     id=#6
     possessed_class=UEFN_TextBlock_C
     tracks [
      {
       type=MovieScene2DTransformTrack
       sections [{
         type=MovieScene2DTransformSection
         end_frame=78
         start_frame=0
         channels [
          Translation.X {type=MovieSceneScriptingFloatChannel, keys [[5,160],[78,-160]]}
          Translation.Y {type=MovieSceneScriptingFloatChannel, const [0,2]}
          Angle {type=MovieSceneScriptingFloatChannel, default=0}
          Scale.X {type=MovieSceneScriptingFloatChannel, default=1}
          Scale.Y {type=MovieSceneScriptingFloatChannel, default=1}
          Shear.X {type=MovieSceneScriptingFloatChannel, default=0}
          Shear.Y {type=MovieSceneScriptingFloatChannel, default=0}
         ]
        }]
      }
      {
       type=MovieSceneColorTrack
       display_name=Color and Opacity
       sections [{
         type=MovieSceneColorSection
         end_frame=78
         start_frame=0
         channels [
          Color.R {type=MovieSceneScriptingFloatChannel, const [1,4]}
          Color.G {type=MovieSceneScriptingFloatChannel, const [1,4]}
          Color.B {type=MovieSceneScriptingFloatChannel, const [1,4]}
          Color.A {type=MovieSceneScriptingFloatChannel, keys [[5,0],[12.8973,1],[70.1027,1],[78,0]]}
         ]
        }]
      }
     ]
    }
   ]}
 }]
graphs [
 AnimFinished {nodes [AnimFinished() {is=title K2Node_FunctionEntry}]}
 EventGraph {nodes [
   Event On Activated()
   > n1 = Play Animation with Finished event!(
    execute:Exec
    Widget:User Widget Object Reference=Self-Reference()
    InAnimation:Widget Animation Object Reference=Get Celebration() -> (Celebration:Widget Animation Object Reference)
    StartAtTime:float=0
    NumLoopsToPlay:int=1
    PlayMode:EUMGSequencePlayMode Enum
    PlaybackSpeed:float=1
   ) -> (then:Exec, Result:Widget Animation Handle Structure, Finished:Exec) {is=title K2Node_PlayAnimation2}
   > PlaySound2D(Sound=/SeasonPass/audio/MSS_UI_BP_Char_Unlocked)
   DeactivateWidget(execute=@n1.Finished)
  ]}
 SetCharacterImage {nodes [
   n1 = SetCharacterImage!() -> (then, ThemeData:Season Pass Category Theme Data Structure) {is=title K2Node_FunctionEntry}
   > SetBrushFromImageReference!(
    execute
    self:Fort Lazy Image Object Reference=Get SetImage() -> (SetImage:Fort Lazy Image Object Reference)
    ImageReference:"Fort Image Reference Structure (by ref)"=LookupImage!(
     Table:"Fort Image Reference Table Structure (by ref)"=Break Season Pass Category Theme Data!(SeasonPassCategoryThemeData:"Season Pass Category Theme Data Structure (by ref)"=@n1) -> (
      ImageReferenceTable:Fort Image Reference Table Structure
      ColorReferenceTable:Fort Color Reference Table Structure
      TextReferenceTable:Fort Text Reference Table Structure
     ) {is=title K2Node_BreakStruct}
     Name:String=LargePreview
    ) -> (ReturnValue:"Fort Image Reference Structure (by ref)")
    bMatchSize:bool=false
    DownloadPriority:ECMSDownloadPriority Enum=Normal
   ) -> (then)
  ]}
]
variables [AnimFinished {}]
widgets {
 info {inheritedWidgetCount=0, parentClass=&CommonActivatableWidget, rootWidgetClass=&Overlay, widgetCount=9}
 widgets [
  Overlay_0 {widgetClassPath=&Overlay, widget=Overlay_0}
  BackgroundBlur_51 {
   widgetClassPath=&BackgroundBlur
   parent=Overlay_0
   widget=BackgroundBlur_51
   properties {blurStrength=30}
   slot {class=OverlaySlot, properties {horizontalAlignment=HAlign_Fill, verticalAlignment=VAlign_Fill}}
  }
  Overlay_1 {
   widgetClassPath=&Overlay
   parent=Overlay_0
   widget=Overlay_1
   slot {class=OverlaySlot, properties {horizontalAlignment=HAlign_Fill, verticalAlignment=VAlign_Fill}}
  }
  Background {
   widgetClassPath=&Image
   bIsVariable=true
   parent=Overlay_1
   widget=Background
   properties {brush {
     bIsDynamicallyLoaded=false
     drawAs=Image
     imageType=FullColor
     mirroring=NoMirror
     resourceName=None
     resourceObject=&/M_UI_BPReboot_SetUnlocking_Modal
     tiling=NoTile
     imageSize=<32 32>
     margin=<ltrb 0 0 0 0>
     outlineSettings {
      bUseBrushTransparency=false
      roundingType=HalfHeightRadius
      width=0
      color {colorUseRule=UseColor_Specified, specifiedColor=<rgba 0 0 0 0>}
      cornerRadii=<0 0 0 0>
     }
     tintColor {colorUseRule=UseColor_Specified, specifiedColor=<rgba 1 1 1 1>}
     uVRegion {bIsValid=false, max=<0 0>, min=<0 0>}
    }}
   slot {class=OverlaySlot, properties {horizontalAlignment=HAlign_Fill, verticalAlignment=VAlign_Fill}}
  }
  SetName_Background {
   widgetClassPath=&UEFN_TextBlock
   parent=Overlay_1
   widget=SetName_Background
   properties {
    justification=Center
    text=Voyager's Cube
    font {
     bForceMonospaced=false
     bMaterialIsStencil=true
     fontMaterial=&/M_UI_BPReboot_BPUnlocking_Font_Outline
     fontObject=&HeadingNow
     letterSpacing=-40
     monospacedWidth=1
     size=72.75
     skewAmount=0
     typefaceFontName=86 Bold Italic
     outlineSettings {
      bApplyOutlineToDropShadows=false
      bMiteredCorners=false
      bSeparateFillAlpha=false
      outlineMaterial=null
      outlineSize=0
      outlineColor=<rgba 0 0 0 1>
     }
    }
    renderTransform {angle=0, scale=<3 3>, shear=<0 0>, translation=<0 0>}
   }
   slot {class=OverlaySlot, properties {horizontalAlignment=HAlign_Center, verticalAlignment=VAlign_Center}}
  }
  SetImage {
   widgetClassPath=&FortLazyImage
   bIsVariable=true
   parent=Overlay_1
   widget=SetImage
   properties {
    materialTextureParamName=SetImage
    brush {
     bIsDynamicallyLoaded=false
     drawAs=Image
     imageType=FullColor
     mirroring=NoMirror
     resourceName=None
     resourceObject=&/M_UI_BPReboot_SetUnlocking_Character
     tiling=NoTile
     imageSize=<720 720>
     margin=<ltrb 0 0 0 0>
     outlineSettings {
      bUseBrushTransparency=false
      roundingType=HalfHeightRadius
      width=0
      color {colorUseRule=UseColor_Specified, specifiedColor=<rgba 0 0 0 0>}
      cornerRadii=<0 0 0 0>
     }
     tintColor {colorUseRule=UseColor_Specified, specifiedColor=<rgba 1 1 1 1>}
     uVRegion {bIsValid=false, max=<0 0>, min=<0 0>}
    }
   }
   slot {class=OverlaySlot, properties {horizontalAlignment=HAlign_Center, verticalAlignment=VAlign_Center, padding=<ltrb 0 0 320 64>}}
  }
  VerticalBox_0 {
   widgetClassPath=&VerticalBox
   parent=Overlay_1
   widget=VerticalBox_0
   slot {class=OverlaySlot, properties {horizontalAlignment=HAlign_Center, verticalAlignment=VAlign_Center, padding=<ltrb 180 0 0 0>}}
  }
  SetName {
   widgetClassPath=&UEFN_TextBlock
   parent=VerticalBox_0
   widget=SetName
   properties {justification=Center, text=Voyager's Cube}
   slot {class=VerticalBoxSlot}
  }
  SetUnlockedTag {
   widgetClassPath=&UEFN_TextBlock
   parent=VerticalBox_0
   widget=SetUnlockedTag
   properties {
    justification=Center
    text="New set unlocked!"
    font {
     bForceMonospaced=false
     bMaterialIsStencil=false
     fontMaterial=&/M_UI_BPReboot_BPUnlocking_Font
     fontObject=&HeadingNow
     letterSpacing=10
     monospacedWidth=1
     size=20.25
     skewAmount=0
     typefaceFontName=86 Bold
     outlineSettings {
      bApplyOutlineToDropShadows=false
      bMiteredCorners=false
      bSeparateFillAlpha=false
      outlineMaterial=null
      outlineSize=0
      outlineColor=<rgba 0 0 0 1>
     }
    }
   }
   slot {class=VerticalBoxSlot}
  }
 ]
}
