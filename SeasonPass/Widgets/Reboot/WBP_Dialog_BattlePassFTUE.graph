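# Activatable dialog widget for the Battle Pass FTUE, built as a preset wrapper around the shared UI Kit dialog.
# Its only widget, WBP_UIKit_Dialog_Internal, is configured with a fixed width, capped desired height, image area and bottom-aligned buttons.
# ApplyDialogVM passes a dialog viewmodel to that inner widget, first determining which sub-widgets are used, then setting the data.
# The event graph is empty, so content comes through the viewmodel.
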WidgetBlueprint /SeasonPass/Widgets/Reboot/WBP_Dialog_BattlePassFTUE
graphs [
 ApplyDialogVM {nodes [
   n1 = ApplyDialogVM!() -> (then, Dialog VM:UI Kit Dialog Viewmodel Object Reference) {is=title K2Node_FunctionEntry}
   n2 = Get WBP_UIKit_Dialog_Internal() -> (WBP_UIKit_Dialog_Internal:WBP UIKit Dialog Internal Object Reference)
   > DetermineWidgetsUsedBasedOnVM!(execute, self:WBP UIKit Dialog Internal Object Reference=@n2, DialogVM:UI Kit Dialog Viewmodel Object Reference=@n1) -> (then)
   > SetDialogData!(execute, self:WBP UIKit Dialog Internal Object Reference=@n2, Data:UI Kit Dialog Viewmodel Object Reference=@n1) -> (then)
  ]}
 EventGraph {links [], nodes []}
]
variables []
widgets {
 info {
  inheritedWidgetCount=0
  parentClass=&CommonActivatableWidget
  rootWidgetClass=/FortUILibrary/Systems/Dialog/WBP_UIKit_Dialog_Internal.WBP_UIKit_Dialog_Internal_C
  widgetCount=1
 }
 widgets [WBP_UIKit_Dialog_Internal {
   widgetClassPath=/FortUILibrary/Systems/Dialog/WBP_UIKit_Dialog_Internal
   bIsVariable=true
   widget=WBP_UIKit_Dialog_Internal
   properties {
    useImage=true
    buttons {
     each Button Horizontal Fill=false
     each Button Min Width=0
     each Button Size=Medium
     focuse Button=0
     space Between Buttons=30
     cTA Buttons Types [Regular]
     group Placement {
      horizontal Alignment=HAlign_Fill
      vertical Alignment=VAlign_Bottom
      grid {column=1, column Span=0, layer=0, row=2, row Span=0, nudge=<0 0>}
      padding=<ltrb 80 64 80 48>
     }
    }
    dialogSIze {
     height Override {enable=false, value=800}
     max Aspect Ratio {enable=false, value=0}
     max Desired Height {enable=true, value=800}
     max Desired Width {enable=false, value=0}
     min Aspect Ratio {enable=false, value=0}
     min Desired Height {enable=false, value=0}
     min Desired Width {enable=false, value=0}
     width Override {enable=true, value=1424}
    }
    imageProperties {
     material=/FortUILibrary/Systems/Dialog/MaterialInstances/MI_UI_DialogImage
     image Size=<512 0>
     placement {
      horizontal Alignment=HAlign_Fill
      vertical Alignment=VAlign_Fill
      grid {column=1, column Span=0, layer=20, row=0, row Span=1, nudge=<0 0>}
      padding=<ltrb 0 0 0 0>
     }
    }
   }
  }]
}
