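# Season pass styles viewer screen for previewing and equipping variant styles of a customizable item reward.
# A variant selector fed by the item view model sits beside an item preview and a side info panel that follows the equipped variant on filter changes.
# Equip All shows only for an owned, equippable, unequipped selection and equips it asynchronously with the chosen overrides.
# Activation sets the preview background, fires a pass analytics event and adjusts layout.
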WidgetBlueprint /SeasonPass/Widgets/Reboot/WBP_Passes_StylesViewer
graphs [
 BP_GetDesiredFocusTarget {nodes [
   Get Desired Focus Target()
   n1 = GetBoxEntries!(self:WBP UIKit Variant Selector Base Object Reference=Get WBP_CustomizeMenu_VariantSelector() -> (WBP_CustomizeMenu_VariantSelector:WBP Customize Menu Variant Selector Object Reference)) -> (ReturnValue:Array of User Widget Object References)
   > Branch(Condition=Is Valid Index!(TargetArray:Array of User Widget Object References=@n1, IndexToTest:int) -> (ReturnValue:bool))
   > n2 = Cast To WBP_UIKit_VariantRow_Wrapper!(execute, Object="Get (a copy)"(Array:Array of User Widget Object References=@n1) -> (Output:User Widget Object Reference)) -> (then, CastFailed, AsWBP UIKit Variant Row Wrapper:WBP UIKit Variant Row Wrapper Object Reference) {is=title K2Node_DynamicCast}
   > Return Node!(
    execute
    ReturnValue:Widget Object Reference=GetDesiredFocusTarget!(self:WBP UIKit Variant Row Wrapper Object Reference=@n2) -> (FocusTarget:Widget Object Reference)
   )
  ]}
 BP_GetDesiredInputConfig {nodes [
   Get Desired Input Config() {is=title K2Node_FunctionEntry}
   > Return Node(ReturnValue:UIInput Config Structure=Make UIInput Config!(
     bIgnoreMoveInput:bool=False
     bIgnoreLookInput:bool=False
     InputMode:ECommonInputMode Enum=All
     MouseCaptureMode:EMouseCaptureMode Enum=CaptureDuringMouseDown
     MouseLockMode:EMouseLockMode Enum=DoNotLock
     bHideCursorDuringViewportCapture:bool=True
    ) -> (UIInputConfig:UIInput Config Structure) {is=title K2Node_MakeStruct})
  ]}
 CF_GetFirstFortItemVM {nodes [
   n1 = CF_GetFirstFortItemVM!() -> (then, SeasonPassRewardVM:Season Pass Reward VM Object Reference) {is=title K2Node_FunctionEntry}
   > n2 = Cast To SeasonPassCustomizableItemRewardVM!(
    execute
    Object=GetFirstExtensionVMByClass!(
     self:Season Pass Element Base VM Object Reference=@n1
     ExtensionClass:Season Pass Element Base VM Class Reference=/Script/SeasonPass.SeasonPassCustomizableItemRewardVM
    ) -> (ReturnValue:Season Pass Element Base VM Object Reference)
   ) -> (then, CastFailed, AsSeason Pass Customizable Item Reward VM:Season Pass Customizable Item Reward VM Object Reference) {is=title K2Node_DynamicCast}
   > Return Node!(
    execute
    FortItemVM:Fort Item VM Object Reference=Get ItemVM!(self:Season Pass Customizable Item Reward VM Object Reference=@n2) -> (ItemVM:Fort Item VM Object Reference)
   )
  ]}
 CF_GetTitle {nodes [
   n1 = CF_GetTitle!() -> (then, SeasonPassRewardVM:Season Pass Element Base VM Object Reference) {is=title K2Node_FunctionEntry}
   > n2 = Cast To SeasonPassCustomizableItemRewardVM(Object=GetFirstExtensionVMByClass(self=@n1, ExtensionClass=&SeasonPassCustomizableItemRewardVM))
   > Return Node!(execute, Label:Text=Get Label!(self:Season Pass Customizable Item Reward VM Object Reference=@n2) -> (Label:Text))
   Return Node(execute=@n2.CastFailed)
  ]}
 EventGraph {nodes [
   n1 = "On Clicked (WBP_BottomBarDecoy_EquipAll)"!() -> (OutputDelegate, then, Button:Common Button Base Object Reference) {is=title K2Node_ComponentBoundEvent}
   > SetIsEnabled(self=@n1.Button, bInIsEnabled=false)
   > n2 = EquipItem!(
    execute:Exec
    Item:Fort Item VM Object Reference=CF_GetFirstFortItemVM!(
     self
     SeasonPassRewardVM:Season Pass Reward VM Object Reference=Get SeasonPassRewardVM() -> (SeasonPassRewardVM:Season Pass Reward VM Object Reference)
    ) -> (FortItemVM:Fort Item VM Object Reference) {is=title K2Node_CallFunction}
    SlotIndex:int=0
    Variants:Array of Mcp Variant Channel Info Structures=Get CurrentVariantOverrides!(self:Fort Variant Selector VM Object Reference=Get FortVariantSelectorVM() -> (FortVariantSelectorVM:Fort Variant Selector VM Object Reference)) -> (CurrentVariantOverrides:Array of Mcp Variant Channel Info Structures)
    bApplyToAllSlots:bool=false
    bAllowMocking:bool=true
   ) -> (then:Exec, OnOperationCompleted:Exec, Success:bool=~) {is=title K2Node_AsyncAction}
   PrintString!(
    execute=@n2.OnOperationCompleted
    InString=OperationComplete
    bPrintToScreen
    bPrintToLog
    TextColor=SelectColor(A=(R=0.0626,G=1,B=0,A=1), B=(R=1,G=0,B=0.0486,A=1), bPickA=@n2)
    Duration
    Key
   ) -> (then)
   > SetIsEnabled(self=@n1.Button, bInIsEnabled=true)
   > SetBottomButtonVisibility!(
    execute
    self:WBP Bottom Bar Decoy Button Object Reference=Get WBP_BottomBarDecoy_EquipAll() -> (WBP_BottomBarDecoy_EquipAll:WBP Bottom Bar Decoy Button Object Reference)
    ShouldButtonBeVisible:bool=NOT Boolean(A=@n2)
   ) -> (then)
   "On Filter Changed (WBP_CustomizeMenu_VariantSelector)"() {is=title K2Node_ComponentBoundEvent}
   n3 = Get FortVariantSelectorVM!(self:WBP UIKit Variant Selector Base Object Reference=Get WBP_CustomizeMenu_VariantSelector()) -> (FortVariantSelectorVM)
   n4 = Get VariantChannels!(self:Fort Variant Set VM Object Reference="Get (a copy)"(Array:Array of Fort Variant Set VM Object References=Get VariantSets!(self:Fort Variant Selector VM Object Reference=@n3) -> (VariantSets:Array of Fort Variant Set VM Object References)) -> (Output:Fort Variant Set VM Object Reference)) -> (VariantChannels:Array of Fort Variant Channel VM Object References)
   > SetFortVariantItemVM!(
    execute
    self:WBP Passes Styles Viewer Side Info Object Reference=Get WBP_Passes_StylesViewer_SideInfo() -> (WBP_Passes_StylesViewer_SideInfo:WBP Passes Styles Viewer Side Info Object Reference)
    Viewmodel:Fort Variant Item VM Object Reference=Get EquippedVariantItem!(self:Fort Variant Channel VM Object Reference="Get (a copy)"(Array:Array of Fort Variant Channel VM Object References=@n4) -> (Output:Fort Variant Channel VM Object Reference)) -> (EquippedVariantItem:Fort Variant Item VM Object Reference)
   ) -> (then)
   "On Variant Item Engaged (WBP_CustomizeMenu_VariantSelector)"!() -> (
    OutputDelegate
    then
    VariantItem:Fort Variant Item VM Object Reference
    VariantChannel:Fort Variant Channel VM Object Reference
    AdditionalInfo:String
   ) {is=title K2Node_ComponentBoundEvent}
   > SetBottomButtonVisibility(self=Get WBP_BottomBarDecoy_EquipAll(), ShouldButtonBeVisible=GetEquipButtonVisibility!(self) -> (ShouldBeVisible:bool))
   "On Clicked (WBP_BottomBarDecoy_Back)"!() -> (OutputDelegate, then, Button:Common Button Base Object Reference) {is=title K2Node_ComponentBoundEvent}
   > DeactivateWidget()
   Event On Activated()
   n5 = Get WBP_ItemPreview() -> (WBP_ItemPreview:WBP Item Preview Object Reference)
   > ActivateWidget(self=@n5)
   > UpdateBackground!(
    execute
    self:Fort Item Preview Background Object Reference=Get FortItemPreviewBackground() -> (FortItemPreviewBackground:Fort Item Preview Background Object Reference)
    LoadedBackground:Object Reference=/SeasonPass/Widgets/Reboot/MaterialInstances/MI_UI_CustomCharacter_Background
    TransitionDirection:EVaultWorldTransitionDirection Enum=Forward
    PrimaryColor:Linear Color Structure=(R=1,G=1,B=1,A=0)
    SecondaryColor:Linear Color Structure=(R=0,G=0,B=0,A=0)
   ) -> (then)
   n6 = GetDisplayName!(self:Season Pass VM Object Reference=Get SeasonPassVM() -> (SeasonPassVM:Season Pass VM Object Reference)) -> (ReturnValue:Text)
   > Fire Event Sub View Opened!(
    execute
    View:Widget Object Reference=Self-Reference()
    AdditionalInfo:Flow Analytics Additional Info Structure="Add String (Non-Localized)"!(
     Info:"Flow Analytics Additional Info Structure (by ref)"=Make Flow Analytics Additional Info!() -> (FlowAnalyticsAdditionalInfo:Flow Analytics Additional Info Structure) {is=title K2Node_MakeStruct}
     Key:String=PassName
     Value:String=Get Source String from Text!(Text:Text=@n6) -> (ReturnValue:String) {is=title K2Node_CallFunction}
    ) -> (ReturnValue:Flow Analytics Additional Info Structure)
    bIsOverlay:bool=false
   ) -> (then)
   > SetCompactUIStyle() {is=title K2Node_CallFunction}
   n7 = Get WBP_CustomizeMenu_VariantSelector()
   > Set!(
    execute
    Padding_Left:float=0
    Padding_Top:float=0
    Padding_Right:float=0
    Padding_Bottom:float=0
    self:Stack Box Slot Object Reference=SlotAsStackBoxSlot!(Widget:Widget Object Reference=Get Switcher_Content!(self:WBP Customize Menu Variant Selector Object Reference=@n7) -> (Switcher_Content:Common Visibility Switcher Object Reference)) -> (ReturnValue:Stack Box Slot Object Reference)
   ) -> (then, Output_Get:Margin Structure)
   > SetBottomButtonVisibility(self=Get WBP_BottomBarDecoy_EquipAll(), ShouldButtonBeVisible=GetEquipButtonVisibility())
   Event On Deactivated()
   > DeactivateWidget(self=@n5)
  ]}
 GetEquipButtonVisibility {nodes [
   GetEquipButtonVisibility() {is=title K2Node_FunctionEntry}
   n1 = Get FortVariantSelectorVM!() -> (FortVariantSelectorVM)
   n2 = Get bIsCurrentSelectionEquipped!(self:Fort Variant Selector VM Object Reference=@n1) -> (bIsCurrentSelectionEquipped:bool)
   n3 = Get SeasonPassRewardVM()
   > Return Node!(
    execute
    ShouldBeVisible:bool=AND Boolean!(
     A=Get bIsRewardOwned!(self:Season Pass Reward VM Object Reference=@n3) -> (bIsRewardOwned:bool)
     B=Get bCanEquipCurrentSelection!(self:Fort Variant Selector VM Object Reference=@n1) -> (bCanEquipCurrentSelection:bool)
     C:bool=NOT Boolean(A=@n2)
    ) -> (ReturnValue)
   )
  ]}
 InitializeWidget {nodes [
   InitializeWidget!() -> (then, FortItemVM:Fort Item VM Object Reference, PreviewData:"Fort Cosmetic Preview Data Structure (by ref)") {is=title K2Node_FunctionEntry}
   > Set FortItemVM!(
    execute
    FortItemVM:Fort Item VM Object Reference=Get FortItemVM() -> (FortItemVM:Fort Item VM Object Reference)
    self:WBP UIKit Variant Selector Base Object Reference=Get WBP_CustomizeMenu_VariantSelector()
   ) -> (then, Output_Get:Fort Item VM Object Reference)
   > ActivateWidget(self=Get WBP_ItemPreview())
   > SetPreviewData!(
    execute
    self:Fort Item Preview Widget Object Reference=Get WBP_ItemPreview()
    Data:"Fort Cosmetic Preview Data Structure (by ref)"=Get PreviewData() -> (PreviewData:"Fort Cosmetic Preview Data Structure (by ref)")
   ) -> (then)
  ]}
 SetCompactUIStyle {nodes [
   SetCompactUIStyle() {is=title K2Node_FunctionEntry}
   n1 = Get WBP_CustomizeMenu_VariantSelector()
   > Set Padding!(
    execute
    Padding:Margin Structure=Select!(
     Option 0_Left:float=0
     Option 0_Top:float=0
     Option 0_Right:float=40
     Option 0_Bottom:float=16
     Option 1_Left:float=0
     Option 1_Top:float=0
     Option 1_Right:float=46
     Option 1_Bottom:float=16
     Index=IsUsingCompactUI()
    ) -> (ReturnValue:Margin Structure)
    self:Stack Box Slot Object Reference=SlotAsStackBoxSlot(Widget=Get Tabs!(self:WBP Customize Menu Variant Selector Object Reference=@n1) -> (Tabs:WBP UIKit Tabs Object Reference))
   ) -> (then, Output_Get:Margin Structure)
   > SetWidthOverride(
    self=Get SB_SideInfo() -> (SB_SideInfo:Size Box Object Reference)
    InWidthOverride=SelectFloat(
     A=440
     B=400
     bPickA=CompareToScreenAspectRatio!(Comparison:EAspectRatioComparison Enum=GreaterThan, AspectRatio:EAspectRatio Enum=Ratio_16_9) -> (ReturnValue:bool)
    )
   )
  ]}
 SetTitle {nodes [
   n1 = SetTitle!() -> (then, ItemName:Text) {is=title K2Node_FunctionEntry}
   > "SetText (Text)"(
    self=Get Text_Title() -> (Text_Title:Common Text Object Reference)
    InText=Format Text!(Format="LOCTEXT(\"Assemble {ItemName} parts\")", ItemName:Text=@n1) -> (Result)
   )
  ]}
]
variables [
 Current Variant Override Index {type=integer}
 Current Variant Overrides {type="array<McpVariantChannelInfo>", default []}
 Primary Variant Channel Tag {type=GameplayTag, default {tagName=None}}
]
widgets {
 info {inheritedWidgetCount=0, parentClass=&CommonActivatableWidget, rootWidgetClass=&Overlay, widgetCount=14}
 widgets [
  Overlay_1 {widgetClassPath=&Overlay, widget=Overlay_1}
  PostBufferUpdate {
   widgetClassPath=/Script/UMG.PostBufferUpdate
   bIsVariable=true
   parent=Overlay_1
   widget=PostBufferUpdate
   properties {
    visibility=HitTestInvisible
    updateBufferInfos [
     {
      bufferToUpdate=ESlatePostRT_0
      postParamUpdater=&self:WidgetTree.PostBufferUpdate.PostBufferBlurUpdater_0
     }
     {bufferToUpdate=ESlatePostRT_1, postParamUpdater=null}
    ]
   }
   slot {class=OverlaySlot, properties {horizontalAlignment=HAlign_Fill, verticalAlignment=VAlign_Fill}}
  }
  SafeZone {
   widgetClassPath=&SafeZone
   bIsVariable=true
   parent=Overlay_1
   widget=SafeZone
   slot {class=OverlaySlot, properties {horizontalAlignment=HAlign_Fill, verticalAlignment=VAlign_Fill}}
  }
  Overlay_2 {widgetClassPath=&Overlay, parent=SafeZone, widget=Overlay_2, slot {class=SafeZoneSlot}}
  GridPanel_33 {
   widgetClassPath=&GridPanel
   parent=Overlay_2
   widget=GridPanel_33
   properties {columnFill [0,1,0], rowFill [0,1,0]}
   slot {class=OverlaySlot, properties {horizontalAlignment=HAlign_Fill, verticalAlignment=VAlign_Fill, padding=<ltrb 80 40 80 40>}}
  }
  FortItemPreviewBackground {
   widgetClassPath=/Script/FortniteUI.FortItemPreviewBackground
   bIsVariable=true
   parent=GridPanel_33
   widget=FortItemPreviewBackground
   properties {previewSceneChangerTag {tagName=UI.ItemPreview.SceneChanger.MPItemShop}}
   slot {class=GridSlot}
  }
  Text_Title {
   widgetClassPath=&CommonTextBlock
   bIsVariable=true
   parent=GridPanel_33
   widget=Text_Title
   properties {
    autoWrapText=true
    bAutoCollapseWithEmptyText=true
    bIsScrollingEnabled=false
    lineHeightPercentage=0.9
    scrollOrientation=Orient_Vertical
    scrollStyle=/FortUILibrary/Styles/Typography/Foundation/TextScrollStyle-NoFade.TextScrollStyle-NoFade_C
    text="[PH] Title"
    textOverflowPolicy=Ellipsis
    textTransformPolicy=ToUpper
    wrapTextAt=700
    wrappingPolicy=AllowPerCharacterWrapping
    font {
     bForceMonospaced=false
     bMaterialIsStencil=false
     fontMaterial=null
     fontObject=&HeadingNow
     letterSpacing=0
     monospacedWidth=1
     size=32.25
     skewAmount=0
     typefaceFontName=HeadingNow-86BoldItalic
     outlineSettings {
      bApplyOutlineToDropShadows=false
      bMiteredCorners=false
      bSeparateFillAlpha=false
      outlineMaterial=null
      outlineSize=0
      outlineColor=<rgba 0 0 0 1>
     }
    }
    shadowOffset=<0 0>
   }
   slot {class=GridSlot}
  }
  SB_SideInfo {
   widgetClassPath=&SizeBox
   bIsVariable=true
   parent=GridPanel_33
   widget=SB_SideInfo
   properties {bOverride_WidthOverride=true, widthOverride=350}
   slot {class=GridSlot, properties {column=2, row=1, verticalAlignment=VAlign_Top, padding=<ltrb 0 64 0 0>}}
  }
  WBP_Passes_StylesViewer_SideInfo {
   widgetClassPath=&/WBP_Passes_StylesViewer_SideInfo
   bIsVariable=true
   parent=SB_SideInfo
   widget=WBP_Passes_StylesViewer_SideInfo
   properties {priority=1}
   slot {class=SizeBoxSlot}
  }
  WBP_ItemPreview {
   widgetClassPath=/FortniteCoreUI/Frontend/ItemPreview/WBP_ItemPreview
   bIsVariable=true
   parent=GridPanel_33
   widget=WBP_ItemPreview
   properties {
    bCanDisplayUserAccessories=false
    bRestrictZoomToWidget=true
    bShow2dPreviewWhileDownloading=false
    bSupportsActivationFocus=false
    infoAlignment_H=HAlign_Center
    infoAlignment_V=VAlign_Center
    itemViewObject=&self:WidgetTree.WBP_ItemPreview.ItemViewObject
    visibility=Visible
    infoPadding=<ltrb 0 0 0 0>
    previewActionButtons [null,null,null]
    previewSceneChangerTag {tagName=UI.ItemPreview.SceneChanger.Customization}
    screenTag {tagName=UI.ItemPreview.Screen.Locker.VariantSelector}
   }
   slot {class=GridSlot, properties {column=1, row=1, rowSpan=3}}
  }
  SB_CustomizeMenu_Holder {
   widgetClassPath=&SizeBox
   bIsVariable=true
   parent=GridPanel_33
   widget=SB_CustomizeMenu_Holder
   properties {bOverride_WidthOverride=true, widthOverride=712}
   slot {class=GridSlot, properties {row=1, rowSpan=1, padding=<ltrb 0 0 0 48>}}
  }
  WBP_CustomizeMenu_VariantSelector {
   widgetClassPath=/FortUILibrary/UIKit/VariantSelector/WBP_CustomizeMenu_VariantSelector
   bIsVariable=true
   parent=SB_CustomizeMenu_Holder
   widget=WBP_CustomizeMenu_VariantSelector
   properties {
    allowSelectionOfLockedVariants=true
    displayUnownedVariants=true
    forceApplyVariantOverrides=true
    priority=3
    additional Arguments {useCustomizeCharacterStyle=true}
    mobileTabIconSize=<72 72>
    variantChannelStyleTag {tagName=UI.Passes.Customizable.Variants}
   }
   slot {class=SizeBoxSlot, properties {padding=<ltrb 0 32 0 0>}}
  }
  WBP_BottomBarDecoy_Back {
   widgetClassPath=/Game/UI/Frontend/BottomBar/WBP_BottomBarDecoyButton
   bIsVariable=true
   parent=Overlay_2
   widget=WBP_BottomBarDecoy_Back
   properties {triggeringInputAction {dataTable=&InputActionDataTable, rowName=Back}}
   slot {class=OverlaySlot, properties {horizontalAlignment=HAlign_Right, verticalAlignment=VAlign_Bottom, padding=<ltrb 0 0 80 48>}}
  }
  WBP_BottomBarDecoy_EquipAll {
   widgetClassPath=&WBP_BottomBarDecoyButton
   bIsVariable=true
   parent=Overlay_2
   widget=WBP_BottomBarDecoy_EquipAll
   properties {triggeringInputAction {dataTable=/SeasonPass/Data/SeasonPassInputActionDataTable, rowName=EquipAll}}
   slot {class=OverlaySlot, properties {horizontalAlignment=HAlign_Right, verticalAlignment=VAlign_Bottom, padding=<ltrb 0 0 80 48>}}
  }
 ]
}
