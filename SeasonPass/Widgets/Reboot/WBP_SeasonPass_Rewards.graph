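# Season pass summary widget showing a "rewards to claim" counter beside a circular level badge with an XP progress image.
# Condensed mode collapses the rewards section and shifts padding, driven by a forced flag OR'd with the caller's state.
# SetForcedCondensed derives that state from the SeasonPassVM purchaseable reward count.
# XP setters read the pass level view model and set the progress material's Progress to current XP over current plus remaining.
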WidgetBlueprint /SeasonPass/Widgets/Reboot/WBP_SeasonPass_Rewards
graphs [
 EventGraph {nodes [
   Event PreConstruct()
   > SetBrushFromMaterial(
    self=Get Background!(self:WBP UIKit Highlight Base Object Reference=Get RewardCounter() -> (RewardCounter:WBP UIKit Highlight Object Reference)) -> (Background:Image Object Reference)
    Material=/SeasonPass/Widgets/Reboot/Materials/M_SeasonPass_Background_Round
   )
  ]}
 SetCondensed {nodes [
   n1 = SetCondensed!() -> (then, IsCondensed:bool=~) {is=title K2Node_FunctionEntry}
   > n2 = Set ShouldCondense(ShouldCondense:bool=OR Boolean(A=@n1, B=Get IsForcedCondensed() -> (IsForcedCondensed:bool))) -> (Output_Get:bool)
   > SetVisibility(self=Get RewardsInfo() -> (RewardsInfo:Size Box Object Reference), InVisibility=Select(Option 0=Visible, Option 1=Collapsed, Index=@n2))
   > SetGridSlotPadding!(
    execute
    Widget:Widget Object Reference=Get GridPanel_LevelBox() -> (GridPanel_LevelBox:Grid Panel Object Reference)
    Padding:Margin Structure=Select(
     Option 0:Margin Structure=Make Margin(Left=-24, Top=0, Right=0, Bottom=0)
     Option 1:Margin Structure=Make Margin(Left=0, Top=0, Right=0, Bottom=0)
     Index=Get ShouldCondense() -> (ShouldCondense:bool)
    ) -> (ReturnValue:Margin Structure)
   ) -> (then)
   > SetPadding!(
    execute
    self:Overlay Slot Object Reference=SlotAsOverlaySlot(Widget=Get HorizontalBox_31() -> (HorizontalBox_31:Horizontal Box Object Reference))
    InPadding:Margin Structure=Select(
     Option 0=Make Margin(Left=0, Top=-16, Right=-16, Bottom=-16)
     Option 1=Make Margin(Left=8, Top=-16, Right=-16, Bottom=-16)
     Index=Get ShouldCondense()
    )
   ) -> (then)
  ]}
 SetCurrentXP {nodes [
   n1 = SetCurrentXP!() -> (then, CurrentXP:int) {is=title K2Node_FunctionEntry}
   > SetLevelProgressPercent!(
    execute
    self
    CurrentXP:int=@n1
    XPToNextLevel:int=Get XpToNextLevel!(self:Pass Level View Model Object Reference=Get FortPassLevelViewModel() -> (FortPassLevelViewModel:Pass Level View Model Object Reference)) -> (XpToNextLevel:int=0)
   ) -> (then)
  ]}
 SetForcedCondensed {nodes [
   n1 = SetForcedCondensed!() -> (then, IsForced:bool=~) {is=title K2Node_FunctionEntry}
   > Set IsForcedCondensed(IsForcedCondensed:bool=@n1) -> (Output_Get:bool)
   n2 = Get SeasonPassVM() -> (SeasonPassVM:Season Pass VM Object Reference)
   > SetCondensed!(
    execute
    self
    IsCondensed:bool="integer <= integer"(A=GetNumPurchaseableRewards!(self:Season Pass Element Base VM Object Reference=@n2) -> (ReturnValue:int=0))
   ) -> (then) {is=title K2Node_CallFunction}
  ]}
 SetLevelProgressPercent {nodes [
   n1 = SetLevelProgressPercent!() -> (then, CurrentXP:int, XPToNextLevel:int) {is=title K2Node_FunctionEntry}
   > n2 = GetDynamicMaterial(self=Get LevelProgress() -> (LevelProgress:Image Object Reference))
   > SetScalarParameterValue(self=@n2, ParameterName=Progress, Value=float / float(A:int=@n1, B:int=int + int(A=@n1.XPToNextLevel, B=@n1)))
  ]}
 SetXPToNextLevel {nodes [
   n1 = SetXPToNextLevel!() -> (then, XPToNextLevel:int) {is=title K2Node_FunctionEntry}
   > SetLevelProgressPercent(CurrentXP=Get CurrentXP!(self:Pass Level View Model Object Reference=Get FortPassLevelViewModel()) -> (CurrentXP:int=0), XPToNextLevel=@n1) {is=title K2Node_CallFunction}
  ]}
]
variables [
 IsForcedCondensed {type=boolean}
 ShowRewardsInsteadOfXp {type=boolean}
]
widgets {
 info {inheritedWidgetCount=0, parentClass=&UserWidget, rootWidgetClass=&Overlay, widgetCount=14}
 widgets [
  Overlay_21 {widgetClassPath=&Overlay, widget=Overlay_21}
  WBP_UIKit_Backplate {
   widgetClassPath=&WBP_UIKit_Backplate
   bIsVariable=true
   parent=Overlay_21
   widget=WBP_UIKit_Backplate
   properties {brightness=Dark, cornerRadius=Interactive Component}
   slot {class=OverlaySlot, properties {horizontalAlignment=HAlign_Fill, verticalAlignment=VAlign_Fill}}
  }
  HorizontalBox_31 {
   widgetClassPath=&HorizontalBox
   bIsVariable=true
   parent=Overlay_21
   widget=HorizontalBox_31
   slot {class=OverlaySlot, properties {horizontalAlignment=HAlign_Right, verticalAlignment=VAlign_Fill, padding=<ltrb 0 -16 -16 -16>}}
  }
  RewardsInfo {
   widgetClassPath=&SizeBox
   bIsVariable=true
   parent=HorizontalBox_31
   widget=RewardsInfo
   properties {bOverride_MinDesiredWidth=true, minDesiredWidth=145}
   slot {class=HorizontalBoxSlot, properties {padding=<ltrb 15 0 10 0>}}
  }
  VerticalBox_RewardInfo {widgetClassPath=&VerticalBox, bIsVariable=true, parent=RewardsInfo, widget=VerticalBox_RewardInfo, slot {class=SizeBoxSlot}}
  HorizontalBox {
   widgetClassPath=&HorizontalBox
   parent=VerticalBox_RewardInfo
   widget=HorizontalBox
   slot {class=VerticalBoxSlot, properties {verticalAlignment=VAlign_Center, size {sizeRule=Fill, value=1}}}
  }
  RewardCounter {
   widgetClassPath=/FortUILibrary/UIKit/Highlight/WBP_UIKit_Highlight
   bIsVariable=true
   parent=HorizontalBox
   widget=RewardCounter
   properties {highlightText=0, highlightType=Attention}
   slot {class=HorizontalBoxSlot, properties {verticalAlignment=VAlign_Center, padding=<ltrb 8 6 16 6>}}
  }
  Text_RewardsLabel {
   widgetClassPath=&CommonTextBlock
   parent=HorizontalBox
   widget=Text_RewardsLabel
   properties {
    bIsScrollingEnabled=false
    mobileFontSizeMultiplier=1.25
    text="REWARDS\r\nTO CLAIM"
    textTransformPolicy=ToUpper
    font {
     bForceMonospaced=false
     bMaterialIsStencil=false
     fontMaterial=null
     fontObject=&HeadingNow
     letterSpacing=0
     monospacedWidth=1
     size=16.5
     skewAmount=0
     typefaceFontName=76 Bold
     outlineSettings {
      bApplyOutlineToDropShadows=false
      bMiteredCorners=false
      bSeparateFillAlpha=false
      outlineMaterial=null
      outlineSize=0
      outlineColor=<rgba 0 0 0 1>
     }
    }
    shadowOffset=<0 0>
    shapedTextOptions {bOverride_TextFlowDirection=true, bOverride_TextShapingMethod=false, textFlowDirection=Culture, textShapingMethod=Auto}
   }
   slot {class=HorizontalBoxSlot, properties {padding=<ltrb 0 0 16 0>}}
  }
  Separator {
   widgetClassPath=&Image
   bIsVariable=true
   parent=HorizontalBox
   widget=Separator
   properties {
    renderOpacity=0.25
    visibility=SelfHitTestInvisible
    brush {
     bIsDynamicallyLoaded=false
     drawAs=Image
     imageType=NoImage
     mirroring=NoMirror
     resourceName=None
     resourceObject=null
     tiling=NoTile
     imageSize=<2 1>
     margin=<ltrb 0 0 0 0>
     outlineSettings {
      bUseBrushTransparency=false
      roundingType=HalfHeightRadius
      width=0
      color {colorUseRule=UseColor_Specified, specifiedColor=<rgba 0 0 0 0>}
      cornerRadii=<0 0 0 0>
     }
     tintColor {colorUseRule=UseColor_Specified, specifiedColor=<rgba 1 1 1 1>}
     uVRegion {bIsValid=false, max=<0 0>, min=<0 0>}
    }
   }
   slot {class=HorizontalBoxSlot, properties {horizontalAlignment=HAlign_Center}}
  }
  GridPanel_LevelBox {
   widgetClassPath=&GridPanel
   bIsVariable=true
   parent=HorizontalBox_31
   widget=GridPanel_LevelBox
   properties {columnFill [1], rowFill [1]}
   slot {class=HorizontalBoxSlot, properties {horizontalAlignment=HAlign_Right, padding=<ltrb -24 0 0 0>}}
  }
  LevelProgress {
   widgetClassPath=&Image
   bIsVariable=true
   parent=GridPanel_LevelBox
   widget=LevelProgress
   properties {
    visibility=SelfHitTestInvisible
    brush {
     bIsDynamicallyLoaded=false
     drawAs=Image
     imageType=FullColor
     mirroring=NoMirror
     resourceName=None
     resourceObject=/Game/UI/Frontend/Lobby/Habanero/Habanero_Materials/MaterialInstances/ProgressBars/MI_UI_Habanero_ProgressBar_PassLevel
     tiling=NoTile
     imageSize=<144 144>
     margin=<ltrb 0 0 0 0>
     outlineSettings {
      bUseBrushTransparency=false
      roundingType=HalfHeightRadius
      width=0
      color {colorUseRule=UseColor_Specified, specifiedColor=<rgba 0 0 0 0>}
      cornerRadii=<0 0 0 0>
     }
     tintColor {colorUseRule=UseColor_Specified, specifiedColor=<rgba 1 1 1 1>}
     uVRegion {bIsValid=false, max=<0 0>, min=<0 0>}
    }
   }
   slot {class=GridSlot}
  }
  VerticalBox_Level {
   widgetClassPath=&VerticalBox
   parent=GridPanel_LevelBox
   widget=VerticalBox_Level
   slot {class=GridSlot, properties {horizontalAlignment=HAlign_Center, layer=1, verticalAlignment=VAlign_Center}}
  }
  Text_2 {
   widgetClassPath=&CommonTextBlock
   parent=VerticalBox_Level
   widget=Text_2
   properties {
    bIsScrollingEnabled=false
    mobileFontSizeMultiplier=1.15
    text=LVL
    textTransformPolicy=ToUpper
    colorAndOpacity {colorUseRule=UseColor_Specified, specifiedColor=<rgba 0.5776 0.5776 0.5776 0.8>}
    font {
     bForceMonospaced=false
     bMaterialIsStencil=false
     fontMaterial=null
     fontObject=&HeadingNow
     letterSpacing=0
     monospacedWidth=1
     size=15.75
     skewAmount=0
     typefaceFontName=75 Medium
     outlineSettings {
      bApplyOutlineToDropShadows=false
      bMiteredCorners=false
      bSeparateFillAlpha=false
      outlineMaterial=null
      outlineSize=0
      outlineColor=<rgba 0 0 0 1>
     }
    }
    shadowOffset=<0 0>
    shapedTextOptions {bOverride_TextFlowDirection=true, bOverride_TextShapingMethod=false, textFlowDirection=Culture, textShapingMethod=Auto}
   }
   slot {class=VerticalBoxSlot, properties {horizontalAlignment=HAlign_Center, verticalAlignment=VAlign_Top}}
  }
  Text_Level {
   widgetClassPath=&CommonTextBlock
   bIsVariable=true
   parent=VerticalBox_Level
   widget=Text_Level
   properties {
    bIsScrollingEnabled=false
    lineHeightPercentage=0.9
    mobileFontSizeMultiplier=1.25
    text=999
    font {
     bForceMonospaced=false
     bMaterialIsStencil=false
     fontMaterial=null
     fontObject=&HeadingNow
     letterSpacing=10
     monospacedWidth=1
     size=18
     skewAmount=0
     typefaceFontName=86 Bold
     outlineSettings {
      bApplyOutlineToDropShadows=false
      bMiteredCorners=false
      bSeparateFillAlpha=false
      outlineMaterial=null
      outlineSize=0
      outlineColor=<rgba 0 0 0 1>
     }
    }
    shadowOffset=<0 0>
    shapedTextOptions {bOverride_TextFlowDirection=true, bOverride_TextShapingMethod=false, textFlowDirection=Culture, textShapingMethod=Auto}
   }
   slot {
    class=VerticalBoxSlot
    properties {horizontalAlignment=HAlign_Center, verticalAlignment=VAlign_Bottom, padding=<ltrb 0 -4 0 0>, size {sizeRule=Fill, value=1}}
   }
  }
 ]
}
