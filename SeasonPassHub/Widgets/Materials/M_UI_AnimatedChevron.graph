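# Translucent UI material for the Season Pass hub that draws a white chevron icon nudging back and forth horizontally in a continuous loop.
# Time feeds a triangle wave, squared and inverted into an eased bounce that offsets the normalized UI UVs along X.
# The offset UVs sample an MSDF chevron texture, which is converted to a signed distance field and thresholded through anti-aliasing for crisp edges.
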Material /SeasonPassHub/Widgets/Materials/M_UI_AnimatedChevron
graph {
 properties {
  bAutomaticallySetUsageInEditor=false
  bUsedWithStaticMesh=false
  blendMode=BLEND_Translucent
  materialDomain=MD_UI
  usedShadingModels=Default Lit
 }
 nodes [
  Constant3Vector() {constant=<rgba 1 1 1 0>}
  n1 = Frac(None=Multiply(A=Time()))
  n2 = Multiply(A=Min(A=@n1, B=OneMinus(None=@n1))) {constB=2}
  n3 = OneMinus(None=Multiply(A=@n2, B=@n2))
  n4 = Add(
   A=MaterialFunctionCall() {materialFunction=&GetUserInterfaceUV}.Normalized UV
   B=AppendVector(A=LinearInterpolate(Alpha=@n3) {constA=-0.1, constB=0.1}, B=Constant())
  )
  MaterialFunctionCall(
   DistanceField=MaterialFunctionCall!(MSDF RGB=TextureSample(UVs=@n4) {samplerType=SAMPLERTYPE_LinearColor, texture=/FortUILibrary/Textures/Icons/T_UI_Icon_UIKit_Chevron_MSDF}.RGB) {materialFunction=/FortUILibrary/MaterialFunctions/Shapes/Modifiers/MF_UI_MSDFtoSDF}.SDF
   Threshold=Constant() {r=0.5}
  ) {materialFunction=&MF_UI_AntiAliasing}
 ]
}
