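# Translucent UI material for a Season Pass unlock background, drawing an animated, glowing circular graphic.
# A circle SDF supplies the anti-aliased shape mask, soft falloff and outline distance, while rotating eased linear gradients drive shading and a two-color rift tint.
# Time-driven expanding rings form a pulse with a fainter echo, both fading as they grow.
# Named reroutes share these layers, and a Color Dodge blend adds brightness over the mixed base colors.
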Material /SeasonPassHub/Widgets/Reboot/Materials/M_SeasonPass_Background_Unlock
graph {
 properties {
  bAutomaticallySetUsageInEditor=false
  bUsedWithStaticMesh=false
  blendMode=BLEND_Translucent
  materialDomain=MD_UI
  usedShadingModels=Default Lit
 }
 nodes [
  n1 = MaterialFunctionCall() {materialFunction=&GetUserInterfaceUV}
  n2 = Multiply(A=@n1.Pixel Size) {constB=0.5}
  n3 = ConstantBiasScale(None=@n1.Normalized UV) {bias=-0.5, scale=2}
  n4 = ScalarParameter() {defaultValue=2, parameterName=ScaleFactor}
  Add(A=WorldPosition().XY, B=Multiply(A=Multiply(A=@n2, B=@n3) {constB=0.5}, B=Subtract(A=@n4)) {constB=3})
  n5 = Multiply(A=Time()) {constB=0.2}
  n6 = Power(
   Base=OneMinus(None=NamedRerouteUsage())
   Exp=LinearInterpolate(Alpha=MaterialFunctionCall!(Value=@n5, Scale, Height) {materialFunction=/FortUILibrary/MaterialFunctions/Basics/MF_UI_PingPong}.Result) {constA=3.5, constB=5}
  ) {constExponent=3}
  LinearInterpolate(A=LinearInterpolate(A=NamedRerouteUsage(), Alpha=Multiply(A=@n6) {constB=0.6}), Alpha=NamedRerouteUsage())
  n7 = Constant() {r=1}
  MaterialFunctionCall(Corner Radius=@n7) {materialFunction=&MF_UI_CoreUI_RoundedMask}
  n8 = Power(Base=NamedRerouteUsage())
  n9 = NamedRerouteUsage()
  n10 = OneMinus(None=@n9)
  n11 = NamedRerouteUsage()
  n12 = LinearInterpolate(A=Power(Base=@n11) {constExponent=0.1}, B=@n11, Alpha=Power(Base=NamedRerouteUsage()) {constExponent=0.1})
  n13 = NamedRerouteUsage()
  n14 = Constant3Vector() {constant=<rgba 0.4896 0.8139 1 1>}
  n15 = NamedRerouteUsage()
  n16 = LinearInterpolate(
   A=NamedRerouteDeclaration(None=@n14) {name=BaseColor, nodeColor=<rgba 0 0 0 1>}
   B=Power(Base=@n14) {constExponent=0.5}
   Alpha=Power(Base=@n15) {constExponent=3}
  )
  MaterialFunctionCall!(
   Blend=Min(A=Multiply(A=@n8, B=@n10)) {constB=0.99}
   Base=LinearInterpolate(A=Power(Base=@n12), B=LinearInterpolate(A=@n13, B=@n16, Alpha=@n15) {constA=1}, Alpha=@n9)
  ) {materialFunction=/Engine/Functions/Engine_MaterialFunctions03/Blends/Blend_ColorDodge}
  n17 = MaterialFunctionCall(Size=@n7) {materialFunction=&MF_UI_Circle}
  NamedRerouteDeclaration(None=MaterialFunctionCall(DistanceField=@n17.SDF, Threshold=Constant() {r=-0.5}) {materialFunction=&MF_UI_AntiAliasing}.Inner) {name=Shape Mask, nodeColor=<rgba 0 0 0 1>}
  NamedRerouteDeclaration(None=SmoothStep(Value=@n17.SDF) {constMax=0.5, constMin=-0.75}) {name=SDF, nodeColor=<rgba 0 0 0 1>}
  n18 = Rotator() {speed=3}
  NamedRerouteDeclaration(None=MaterialFunctionCall("Value (0-1)"=MaterialFunctionCall(UV Channel=@n18) {materialFunction=&LinearGradient}.UGradient) {materialFunction=&MF_UI_EaseCurves}.EaseInOutSine) {name=Gradient, nodeColor=<rgba 0 0 0 1>}
  n19 = SmoothStep(Value=OneMinus(None=NamedRerouteUsage())) {constMax=0.5, constMin=-0.5}
  n20 = Power(Base=Multiply(A=NamedRerouteUsage(), B=Power(Base=@n19) {constExponent=4}))
  n21 = NamedRerouteUsage()
  n22 = NamedRerouteUsage()
  NamedRerouteDeclaration(None=LinearInterpolate(A=Multiply(A=@n20) {constB=0.85}, Alpha=Multiply(A=@n21, B=@n22))) {name=Highlights, nodeColor=<rgba 0 0 0 1>}
  n23 = Time()
  n24 = ScalarParameter() {defaultValue=0.5, parameterName=Pulse -Speed}
  n25 = SmoothStep(Value=Frac(None=Multiply(A=@n23, B=@n24) {constB=0.6})) {constMax=0.9, constMin=0.6}
  n26 = ScalarParameter() {defaultValue=0.85, parameterName=Pulse - Smoothness}
  n27 = ScalarParameter() {defaultValue=0.2, parameterName=Pulse -Anim Curve}
  n28 = Constant() {r=0.01}
  n29 = MaterialFunctionCall(Size=Power(Base=@n25, Exp=@n27) {constExponent=0.5}, OutlineThickness=@n28) {materialFunction=&MF_UI_Circle}
  n30 = Multiply(A=OneMinus(None=@n25), B=SmoothStep(Min=@n26, Value=OneMinus(None=@n29.SDFOutline)) {constMin=0.99})
  n31 = Time()
  n32 = SmoothStep(Value=Frac(None=Multiply(A=@n31, B=@n24) {constB=0.6})) {constMin=0.65}
  n33 = Power(Base=@n32, Exp=@n27) {constExponent=0.5}
  n34 = SmoothStep(
   Min=@n26
   Value=OneMinus(None=MaterialFunctionCall(Size=@n33, OutlineThickness=@n28) {materialFunction=&MF_UI_Circle}.SDFOutline)
  ) {constMin=0.99}
  n35 = Multiply(
   A=Multiply(A=Power(Base=@n34) {constExponent=4}, B=OneMinus(None=@n32))
   B=ScalarParameter() {defaultValue=0.75, parameterName=Pulse -Echo Opacity}
  ) {constB=0.35}
  n36 = SmoothStep(Value=NamedRerouteUsage()) {constMax=0.8}
  NamedRerouteDeclaration(None=Multiply(A=LinearInterpolate(A=@n30, Alpha=@n35), B=OneMinus(None=@n36))) {name=Pulse, nodeColor=<rgba 0 0 0 1>}
  NamedRerouteDeclaration(None=Distance(A=@n17.SDF, B=Constant())) {name=Outline SDF, nodeColor=<rgba 0 0 0 1>}
  n37 = MaterialFunctionCall(UV Channel=Rotator() {speed=2.5}) {materialFunction=&LinearGradient}
  NamedRerouteDeclaration(None=LinearInterpolate(
    A=VectorParameter() {group=0 - Coloring, parameterName=Color 1, defaultValue=<rgba 0.1615 0.4154 1 1>}.RGB
    B=VectorParameter() {group=0 - Coloring, parameterName=Color 2, defaultValue=<rgba 0.6883 0.3193 1 1>}.RGB
    Alpha=MaterialFunctionCall("Value (0-1)"=@n37.UGradient) {materialFunction=&MF_UI_EaseCurves}.EaseInOutCubic
   )) {name=Rift Gradient, nodeColor=<rgba 0 0 0 1>}
 ]
}
