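# UI material for the Season Pass hub's CH7S2 rivalry billboard, compositing a character texture over an angled, bolt-shaped color plate with glows.
# Box-angle and step masks cut the bolt body and tail, while rotated gradients add skewed glows and clip the character.
# One Base Color derives lighter and darker tones through OkHSV shifts, blended in OkLab.
# A Mirror parameter flips UVs, character scale and offsets, with further parameters for animated texture position and layer opacities.
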Material /SeasonPassHub/Widgets/Reboot/Materials/M_UI_CH7S2_RivalryBillboard_Revamp
graph {
 properties {bAutomaticallySetUsageInEditor=false, blendMode=BLEND_Translucent, materialDomain=MD_UI, usedShadingModels=Default Lit}
 nodes [
  n1 = ScalarParameter() {defaultValue=50, parameterName=HorizontalOffset}
  n2 = Multiply(A=@n1) {constB=-1}
  n3 = ScalarParameter() {parameterName=Mirror}
  n4 = ComponentMask(None=TextureCoordinate() {coordinateIndex=3}) {g=true}
  AppendVector(A=Multiply(A=LinearInterpolate(A=@n1, B=@n2, Alpha=@n3) {constA=50, constB=-50}, B=Divide(A=@n4) {constB=144}), B=Constant())
  n5 = MaterialFunctionCall() {materialFunction=&GetUserInterfaceUV}
  n6 = ComponentMask(None=@n5.Pixel Size) {g=true}
  Divide(A=ComponentMask(None=@n5.Pixel Size) {r=true}, B=@n6)
  n7 = ScalarParameter() {defaultValue=45, group=Bolt, parameterName=TailAngleOut}
  Divide(A=Add(A=@n7, B=ScalarParameter() {defaultValue=20, group=Bolt, parameterName=TailAngleIn}))
  n8 = LinearInterpolate(Alpha=@n3) {constA=50, constB=-50}
  MaterialFunctionCall!(
   Anchor
   Translation=AppendVector(A=Multiply(A=@n8), B=Constant())
   Padding
   Scale=ScalarParameter() {defaultValue=1, parameterName=TextureScreenScale}
  ) {materialFunction=/FortUILibrary/MaterialFunctions/Basics/MF_UI_VertexTransform}
  n9 = NamedRerouteUsage()
  n10 = Multiply(A=ComponentMask(None=NamedRerouteUsage()) {r=true}) {constB=0.2}
  n11 = MaterialFunctionCall(Base=@n9, Blend=@n10) {materialFunction=&Blend_Overlay}
  MaterialFunctionCall!(ColorA=NamedRerouteUsage(), ColorB=@n9, Alpha=@n11.Result) {materialFunction=/SeasonPass/MaterialFunctions/MF_UI_SeasonPass_OkLab_Lerp}
  MaterialFunctionCall(Number=@n6) {materialFunction=&DebugScalarValues}
  Multiply(A=@n10, B=@n9)
  n12 = Add(
   A=NamedRerouteUsage()
   B=AppendVector(A=ScalarParameter() {parameterName=TexturePosX_Anim}, B=ScalarParameter() {parameterName=TexturePosY_Anim})
  )
  n13 = AppendVector(A=ScalarParameter() {parameterName=TexturePosX}, B=ScalarParameter() {defaultValue=-0.14, parameterName=TexturePosY})
  n14 = TextureSampleParameter2D(UVs=MaterialFunctionCall(UVs=Add(A=@n12, B=@n13), Texture Scale=NamedRerouteUsage()) {materialFunction=&ScaleUVsByCenter}.UVs) {
   parameterName=Character Texture
   samplerSource=SSM_Clamp_WorldGroupSettings
   texture=/FortUILibrary/Systems/Items/Textures/T_UI_SoldierCommando_F
  }
  NamedRerouteDeclaration(None=@n14.A) {name=CaracterAlphaRaw, nodeColor=<rgba 1 0.1641 0 1>}
  NamedRerouteDeclaration(None=Multiply(A=@n14.A, B=NamedRerouteUsage())) {name=CharacterAlpha, nodeColor=<rgba 0 0 0 1>}
  n15 = Distance(A=Add(A=NamedRerouteUsage(), B=Constant2Vector() {g=-0.6, r=-0.48}), B=Constant())
  n16 = OneMinus(None=SmoothStep(Value=Multiply(A=@n15) {constB=2}) {constMax=1.35})
  n17 = MaterialFunctionCall() {materialFunction=&GetUserInterfaceUV}
  n18 = Constant()
  n19 = OneMinus(None=Multiply(A=Distance(A=@n17.Normalized UV, B=@n18)) {constB=2})
  n20 = ComponentMask(None=@n17.Normalized UV) {r=true}
  n21 = LinearInterpolate(A=Multiply(A=Power(Base=@n16) {constExponent=1.5}) {constB=0.6}, Alpha=Multiply(A=Saturate(None=@n19), B=Power(Base=@n20))) {constB=0.7}
  NamedRerouteDeclaration(None=Multiply(A=Power(Base=@n21) {constExponent=1.5}, B=ScalarParameter() {defaultValue=1, parameterName=BackgroundGlow_Opacity})) {name=BackgroundGlow, nodeColor=<rgba 0 0 0 1>}
  n22 = NamedRerouteUsage()
  n23 = ComponentMask(None=@n22) {r=true}
  n24 = OneMinus(None=@n23)
  n25 = ScalarParameter() {parameterName=Mirror}
  NamedRerouteDeclaration(None=AppendVector(A=LinearInterpolate(A=@n23, B=@n24, Alpha=@n25), B=ComponentMask(None=@n22) {g=true})) {name=MirrorUV, nodeColor=<rgba 1 0.4688 0 1>}
  n26 = Add(A=NamedRerouteDeclaration(None=NamedRerouteUsage()) {name=UV, nodeColor=<rgba 0.4297 0 1 1>}, B=Constant2Vector() {r=-0.5})
  NamedRerouteDeclaration(None=Add(
    A=MaterialFunctionCall(UV=@n26) {materialFunction=&MF_UI_UVScaling}.Scaled to Fit
    B=Constant2Vector() {r=0.5}
   )) {name=UVRightAnchor, nodeColor=<rgba 0 1 0.4375 1>}
  n27 = MaterialFunctionCall!(Input UVs=NamedRerouteUsage(), Translate=AppendVector(A=Constant() {r=-0.45}, B=Constant())) {materialFunction=/FortUILibrary/MaterialFunctions/Basics/MF_UI_Translate}
  n28 = ScalarParameter() {defaultValue=3, parameterName=BottomLeftCornerIndex}
  n29 = NamedRerouteUsage()
  n30 = Multiply(
   A=Multiply(
    A=MaterialFunctionCall!(UV=@n27.UVs raw, Corner=@n28, Angle=@n29, Flip, Falloff) {materialFunction=/FortUILibrary/MaterialFunctions/Shapes/MF_UI_BoxAngle}.Result
    B=NamedRerouteUsage()
   )
   B=OneMinus(None=NamedRerouteUsage())
  )
  n31 = NamedRerouteUsage()
  n32 = Multiply(
   A=MaterialFunctionCall!("Value (0-1)"=SmoothStep(Value=@n31) {constMin=0.15}) {materialFunction=/FortUILibrary/MaterialFunctions/Basics/MF_UI_EaseCurves_01_Sine}.In-Out
   B=ScalarParameter() {defaultValue=1, parameterName=SkewedGlow_2_Mask}
  )
  NamedRerouteDeclaration(None=Multiply(A=Multiply(A=@n30, B=@n32)) {constB=0.5}) {name=SkewedGlow02, nodeColor=<rgba 0 0 0 1>}
  n33 = NamedRerouteUsage()
  n34 = ComponentMask(None=@n33) {g=true}
  n35 = ScalarParameter() {defaultValue=0.82, desc=Distance from top %, group=Bolt, parameterName=BodyHeight}
  n36 = AppendVector(
   A=NamedRerouteDeclaration(None=ComponentMask(None=@n33) {r=true}) {name=HorizontalGradient, nodeColor=<rgba 0 0.9766 1 1>}
   B=Add(A=OneMinus(None=@n34), B=Subtract(A=@n35)) {constB=-0.2}
  )
  n37 = ScalarParameter() {defaultValue=4.5, group=Bolt, parameterName=BodyAngleTop}
  n38 = NamedRerouteDeclaration(None=OneMinus(None=MaterialFunctionCall(UV=@n36, Angle=@n37) {materialFunction=&MF_UI_BoxAngle}.Result)) {name=Body Top Angle Mask, nodeColor=<rgba 0 0.0156 1 1>}
  n39 = Step(Y=@n34, X=@n35) {constX=0.8}
  n40 = ScalarParameter() {defaultValue=0.14, group=Bolt, parameterName=TailLenght}
  n41 = Multiply(A=Add(A=Add(A=@n40, B=@n35) {constB=0.01}) {constB=-1}) {constB=-1}
  n42 = Add(
   A=NamedRerouteUsage()
   B=AppendVector(A=ScalarParameter() {defaultValue=-0.65, group=Bolt, parameterName=TailPositionX, sliderMax=-1, sliderMin=-0.5}, B=Saturate(None=@n41))
  ) {constB=0}
  n43 = ScalarParameter() {defaultValue=3, parameterName=BottomLeftCornerIndex}
  n44 = MaterialFunctionCall(UV=@n42, Corner=@n43, Angle=ScalarParameter() {defaultValue=20, group=Bolt, parameterName=TailAngleIn}) {materialFunction=&MF_UI_BoxAngle}
  n45 = MaterialFunctionCall(UV=@n42, Corner=@n43, Angle=ScalarParameter() {defaultValue=45, group=Bolt, parameterName=TailAngleOut}) {materialFunction=&MF_UI_BoxAngle}
  n46 = Min(
   A=Max(
    A=NamedRerouteDeclaration(None=@n39) {name=Body Bottom Horizontal Mask, nodeColor=<rgba 1 0.9844 0 1>}
    B=NamedRerouteDeclaration(None=@n44.Result) {name=Tail Angle In Mask, nodeColor=<rgba 1 0.2344 0 1>}
   )
   B=NamedRerouteDeclaration(None=OneMinus(None=@n45.Result)) {name=Tail Angle Out Mask, nodeColor=<rgba 0 0.0859 1 1>}
  )
  NamedRerouteDeclaration(None=Saturate(None=Multiply(A=@n38, B=@n46))) {name=Shape, nodeColor=<rgba 0 1 0.4609 1>}
  n47 = NamedRerouteUsage()
  n48 = AppendVector(A=Constant() {r=-0.3}, B=Constant())
  n49 = Multiply(
   A=MaterialFunctionCall(
    UV=MaterialFunctionCall(Input UVs=@n47, Translate=@n48) {materialFunction=&MF_UI_Translate}.UVs raw
    Corner=ScalarParameter() {defaultValue=3, parameterName=BottomLeftCornerIndex}
    Angle=NamedRerouteUsage()
   ) {materialFunction=&MF_UI_BoxAngle}.Result
   B=NamedRerouteUsage()
  )
  n50 = OneMinus(None=NamedRerouteUsage())
  n51 = MaterialFunctionCall(Input UVs=NamedRerouteUsage(), Translate=AppendVector(A=Constant() {r=0.1}, B=Constant() {r=-0.4})) {materialFunction=&MF_UI_Translate}
  n52 = Constant() {r=-0.038}
  n53 = Multiply(A=ComponentMask(None=MaterialFunctionCall(UVs=@n51.UVs raw, "Rotation Angle (0-1)"=@n52) {materialFunction=&CustomRotator}.Rotated Values) {g=true})
  n54 = MaterialFunctionCall("Value (0-1)"=SmoothStep(Value=NamedRerouteDeclaration(None=@n53) {name=Glow Angle, nodeColor=<rgba 1 0 0.0703 1>}) {constMax=0.8, constMin=0.15}) {materialFunction=&MF_UI_EaseCurves_01_Sine}
  n55 = ScalarParameter() {defaultValue=1, parameterName=SkewedGlow_1_Mask}
  NamedRerouteDeclaration(None=Multiply(A=Multiply(A=@n49, B=@n50), B=Multiply(A=@n54.In-Out, B=@n55))) {name=SkewedGlow01, nodeColor=<rgba 0 0 0 1>}
  n56 = NamedRerouteUsage()
  n57 = NamedRerouteUsage()
  n58 = NamedRerouteUsage()
  n59 = TextureCoordinate()
  n60 = Add(
   A=OneMinus(None=ComponentMask(None=@n59) {r=true})
   B=LinearInterpolate(Alpha=ScalarParameter() {defaultValue=1, parameterName=ShapeOpacity, sliderMax=1}) {constA=-1}
  ) {constB=-1}
  n61 = Add(A=Power(Base=LinearInterpolate(A=@n56, Alpha=@n57)) {constExponent=0.5}, B=Saturate(None=Multiply(A=@n58, B=@n60)))
  n62 = Multiply(A=NamedRerouteUsage()) {constB=0.9}
  n63 = Multiply(A=NamedRerouteUsage(), B=ScalarParameter() {defaultValue=1, parameterName=TextureOpacity})
  NamedRerouteDeclaration(None=Add(A=Add(A=@n61, B=@n62), B=NamedRerouteDeclaration(None=@n63) {name=CharacterOpacity, nodeColor=<rgba 0 1 0.4375 1>})) {name=CompositionAlpha, nodeColor=<rgba 0.7578 0 1 1>}
  NamedRerouteDeclaration(None=MaterialFunctionCall() {materialFunction=&GetUserInterfaceUV}.Normalized UV) {name=UVRaw, nodeColor=<rgba 0.1484 0 1 1>}
  n64 = NamedRerouteUsage()
  n65 = NamedRerouteUsage()
  n66 = Multiply(A=Multiply(A=NamedRerouteUsage()), B=ScalarParameter() {defaultValue=1, parameterName=TextureColorMask})
  NamedRerouteDeclaration(None=Multiply(A=MaterialFunctionCall(ColorA=@n64, ColorB=@n65, Alpha=@n66) {materialFunction=&MF_UI_SeasonPass_OkLab_Lerp}.Result)) {name=CompositionColor, nodeColor=<rgba 1 0.1172 0 1>}
  NamedRerouteDeclaration(None=@n7) {name=TailAngleMiddle, nodeColor=<rgba 0.1719 0 1 1>}
  NamedRerouteDeclaration(None=VectorParameter() {parameterName=Base Color, defaultValue=<rgba 1 0.3812 0.021 1>}.RGB) {name=BaseColor, nodeColor=<rgba 0 0 0 1>}
  NamedRerouteDeclaration(None=MaterialFunctionCall!(
    HueShift=Constant() {r=20}
    SaturationShift
    ValueShift
    HueOverride
    SaturationOverride=Constant() {r=0.3}
    ValueOverride=Constant() {r=1}
    Color=NamedRerouteUsage()
   ) {materialFunction=/SeasonPass/MaterialFunctions/MF_UI_SeasonPass_OkHSV_ColorModify}.Result) {name=ColorLighter, nodeColor=<rgba 0 0 0 1>}
  NamedRerouteDeclaration(None=MaterialFunctionCall(HueShift=Constant() {r=10}, ValueShift=Constant() {r=-0.1}, SaturationOverride=Constant() {r=1}, Color=NamedRerouteUsage()) {materialFunction=&MF_UI_SeasonPass_OkHSV_ColorModify}.Result) {name=ColorDarker, nodeColor=<rgba 0 0 0 1>}
  n67 = NamedRerouteUsage()
  n68 = NamedRerouteUsage()
  n69 = NamedRerouteUsage()
  NamedRerouteDeclaration(None=MaterialFunctionCall(
    ColorA=MaterialFunctionCall(ColorA=@n67, ColorB=@n68, Alpha=@n69) {materialFunction=&MF_UI_SeasonPass_OkLab_Lerp}.Result
    ColorB=@n11.Result
    Alpha=NamedRerouteUsage()
   ) {materialFunction=&MF_UI_SeasonPass_OkLab_Lerp}.Result) {name=PlateColor, nodeColor=<rgba 0 0 0 1>}
  n70 = MaterialFunctionCall(Input UVs=NamedRerouteUsage(), Translate=AppendVector(A=Constant() {r=-0.2}, B=Constant())) {materialFunction=&MF_UI_Translate}
  n71 = Constant() {r=-0.1}
  n72 = SmoothStep(Value=ComponentMask(None=MaterialFunctionCall(UVs=@n70.UVs raw, "Rotation Angle (0-1)"=@n71) {materialFunction=&CustomRotator}.Rotated Values) {r=true}) {constMax=0.25}
  n73 = NamedRerouteUsage()
  n74 = NamedRerouteUsage()
  NamedRerouteDeclaration(None=Multiply(A=Multiply(A=@n72, B=@n73), B=OneMinus(None=@n74))) {name=CharacterMask, nodeColor=<rgba 1 0 0.1406 1>}
  n75 = NamedRerouteDeclaration(None=NamedRerouteUsage()) {name=CharacterScale, nodeColor=<rgba 0.1719 1 0 1>}
  n76 = Multiply(A=@n75) {constB=-1}
  n77 = ScalarParameter() {parameterName=Mirror}
  NamedRerouteDeclaration(None=AppendVector(A=LinearInterpolate(A=@n75, B=@n76, Alpha=@n77), B=@n75)) {name=CharacterScaleLogic, nodeColor=<rgba 1 0 0.4922 1>}
  NamedRerouteDeclaration(None=ScalarParameter() {defaultValue=1.15, parameterName=TextureScale}) {name=Character scale, nodeColor=<rgba 0 0.8125 1 1>}
  NamedRerouteDeclaration(None=@n14.RGB) {name=CharacterColor, nodeColor=<rgba 0 0 0 1>}
  NamedRerouteUsage()
  NamedRerouteUsage()
  NamedRerouteUsage()
  OneMinus(None=@n44.Result)
 ]
}
