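# M_UI_DebugLogo is a translucent UI-domain material for a debug logo in the Season Pass Hub widgets.
# A LogoTexture parameter, defaulting to an empty 1x1 placeholder, allows material instances to supply the logo image.
# Its UVs come from the "Scaled to Fit" output of the shared MF_UI_UVScaling function, fitting the image to the drawn area.
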Material /SeasonPassHub/Widgets/Reboot/Materials/M_UI_DebugLogo
graph {
 properties {
  bAutomaticallySetUsageInEditor=false
  bUsedWithStaticMesh=false
  blendMode=BLEND_Translucent
  materialDomain=MD_UI
  usedShadingModels=Default Lit
 }
 nodes [TextureSampleParameter2D(UVs=MaterialFunctionCall() {materialFunction=&MF_UI_UVScaling}.Scaled to Fit) {parameterName=LogoTexture, texture=/FortUILibrary/Input/PeripheralGuitar/Source/Empty1x1}]
}
