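# UI material for progress text on the Season Pass Hub landing page, deriving a pale tint from a character color parameter.
# Converts CharacterColor to HSV, keeps the hue, reduces saturation and raises value with a clamp, then converts back to RGB.
# The result is a lighter, softer version of whatever color is supplied.
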Material /SeasonPassHub/Widgets/Reboot/Materials/M_UI_LandingPage_Progress_Text
graph {
 properties {
  bAutomaticallySetUsageInEditor=false
  bUsedWithStaticMesh=false
  blendMode=BLEND_Translucent
  materialDomain=MD_UI
  usedShadingModels=Default Lit
 }
 nodes [
  n1 = RgbToHsv!(None=VectorParameter() {parameterName=CharacterColor, defaultValue=<rgba 0.1714 0.0561 0.521 1>}.RGB)
  n2 = ComponentMask(None=@n1) {g=true}
  n3 = Add(A=ComponentMask(None=@n1) {b=true}) {constB=0.85}
  HsvToRgb!(None=MaterialFunctionCall(R=ComponentMask(None=@n1) {r=true}, G=Multiply(A=@n2) {constB=0.65}, B=Saturate(None=@n3)) {materialFunction=&AppendMany}.RGB)
 ]
}
