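# Translucent UI material for a Season Pass landing page tile, providing a themed gradient and a rounded-rectangle shape.
# The gradient blends ThemeColor into a brightened, slightly hue-shifted variant along vertical UV, using sine ease-in-out.
# An SDF rounded mask takes per-corner radii, in pixels by default, and exposes named reroutes for an antialiased shape mask, SDF and outline SDF.
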Material /SeasonPassHub/Widgets/Reboot/Materials/M_UI_LandingPage_Tile
graph {
 properties {
  bAutomaticallySetUsageInEditor=false
  bUsedWithStaticMesh=false
  blendMode=BLEND_Translucent
  materialDomain=MD_UI
  usedShadingModels=Default Lit
 }
 nodes [
  Constant() {r=0.02}
  n1 = VectorParameter() {parameterName=ThemeColor, defaultValue=<rgba 0.3467 0.9131 1 1>}
  n2 = MaterialFunctionCall() {materialFunction=&GetUserInterfaceUV}
  LinearInterpolate(
   A=@n1.RGB
   B=MaterialFunctionCall(Hue Shift Percentage=Constant() {r=0.02}, Texture=Power(Base=@n1.RGB) {constExponent=0.3}) {materialFunction=&HueShift}.Result
   Alpha=MaterialFunctionCall("Value (0-1)"=ComponentMask(None=@n2.Normalized UV) {g=true}) {materialFunction=&MF_UI_EaseCurves}.EaseInOutSine
  )
  n3 = ScalarParameter() {
   defaultValue=1
   desc="If enabled reads corner radius values in pixels.\r\nbool"
   group=Layout
   parameterName=UseValueInPixels
   sliderMax=1
   sortPriority=1
  }
  n4 = MaterialFunctionCall(
   Corner Radius=MaterialFunctionCall(
    R=ScalarParameter() {defaultValue=10, group=Layout, parameterName=TopLeftCornerRadius, sortPriority=4}
    G=ScalarParameter() {defaultValue=10, group=Layout, parameterName=BottomLeftCornerRadius, sortPriority=7}
    B=ScalarParameter() {defaultValue=10, group=Layout, parameterName=TopRightCornerRadius, sortPriority=5}
    A=ScalarParameter() {defaultValue=10, group=Layout, parameterName=BottomRightCornerRadius, sortPriority=6}
   ) {materialFunction=&AppendMany}.RGBA
   Radius in Pixels=@n3
  ) {materialFunction=&MF_UI_CoreUI_RoundedMask}
  NamedRerouteDeclaration(None=MaterialFunctionCall(DistanceField=@n4.SDF, Threshold=Constant() {r=-0.01}) {materialFunction=&MF_UI_AntiAliasing}.Inner) {name=Shape Mask, nodeColor=<rgba 0 0 0 1>}
  NamedRerouteDeclaration(None=@n4.SDF) {name=SDF, nodeColor=<rgba 0 0 0 1>}
  NamedRerouteDeclaration(None=@n4.Outline SDF) {name=Outline SDF, nodeColor=<rgba 0 0 0 1>}
  NamedRerouteUsage()
 ]
}
