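# UI material for a Season Pass landing page tile, showing a reward character texture on a darkened color backdrop inside rounded corners.
# The backdrop blends from ThemeColor toward CharacterColor on Hover, with brightness cut through an HSV round trip.
# The character uses fill-scaled, offset UVs that shift vertically on Hover, plus an eased vertical gradient overlay for fading.
# A Whiteout parameter pushes color to over-bright white, and a CoreUI rounded mask supplies the shape.
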Material /SeasonPassHub/Widgets/Reboot/Materials/M_UI_LandingPage_Tile_Content
graph {
 properties {
  bAutomaticallySetUsageInEditor=false
  bUsedWithStaticMesh=false
  blendMode=BLEND_Translucent
  materialDomain=MD_UI
  usedShadingModels=Default Lit
 }
 nodes [
  n1 = VectorParameter() {parameterName=ThemeColor, defaultValue=<rgba 0.3467 0.9131 1 1>}
  n2 = RgbToHsv!(None=LinearInterpolate(
    A=Multiply(A=@n1.RGB) {constB=0.15}
    B=VectorParameter() {parameterName=CharacterColor, defaultValue=<rgba 0.1714 0.0561 0.521 1>}.RGB
    Alpha=ScalarParameter() {parameterName=Hover}
   ) {constAlpha=1})
  n3 = MaterialFunctionCall(R=ComponentMask(None=@n2) {r=true}, G=ComponentMask(None=@n2) {g=true}, B=Multiply(A=ComponentMask(None=@n2) {b=true}) {constB=0.2}) {materialFunction=&AppendMany}
  n4 = Power(Base=NamedRerouteUsage()) {constExponent=0.2}
  n5 = OneMinus(None=NamedRerouteUsage())
  LinearInterpolate(
   A=LinearInterpolate(A=HsvToRgb!(None=@n3.RGB), B=NamedRerouteUsage(), Alpha=Multiply(A=@n4, B=@n5))
   Alpha=ScalarParameter() {parameterName=Whiteout}
  ) {constB=2}
  Multiply(A=NamedRerouteUsage(), B=LinearInterpolate(A=NamedRerouteUsage(), Alpha=NamedRerouteUsage()))
  n6 = MaterialFunctionCall(
   R=ScalarParameter() {defaultValue=10, group=Layout, parameterName=TopLeftCornerRadius, sortPriority=4}
   G=ScalarParameter() {defaultValue=10, group=Layout, parameterName=BottomLeftCornerRadius, sortPriority=7}
   B=ScalarParameter() {defaultValue=10, group=Layout, parameterName=TopRightCornerRadius, sortPriority=5}
   A=ScalarParameter() {defaultValue=10, group=Layout, parameterName=BottomRightCornerRadius, sortPriority=6}
  ) {materialFunction=&AppendMany}
  n7 = ScalarParameter() {
   defaultValue=1
   desc="If enabled reads corner radius values in pixels.\r\nbool"
   group=Layout
   parameterName=UseValueInPixels
   sliderMax=1
   sortPriority=1
  }
  NamedRerouteDeclaration(None=Power(Base=MaterialFunctionCall(Corner Radius=@n6.RGBA, Radius in Pixels=@n7) {materialFunction=&MF_UI_CoreUI_RoundedMask}.Shape Mask) {constExponent=0.2}) {name=Shape Mask, nodeColor=<rgba 0 0 0 1>}
  n8 = Add(A=TextureCoordinate(), B=Constant2Vector() {g=-0.5})
  n9 = Saturate(None=SmoothStep(Value=ComponentMask(None=@n8) {g=true}) {constMax=0.5})
  NamedRerouteDeclaration(None=Multiply(
    A=MaterialFunctionCall("Value (0-1)"=@n9) {materialFunction=&MF_UI_EaseCurves}.EaseInOutSine
    B=ScalarParameter() {defaultValue=0.95, parameterName=Overlay Opacity}
   )) {name=Overlay, nodeColor=<rgba 0 0 0 1>}
  n10 = Add(
   A=NamedRerouteUsage()
   B=AppendVector(A=ScalarParameter() {parameterName=CharacterPositionX}, B=ScalarParameter() {defaultValue=-0.5, parameterName=CharacterPositionY})
  )
  n11 = AppendVector(A=Constant(), B=LinearInterpolate(Alpha=ScalarParameter() {parameterName=Hover}) {constA=-0.2, constB=0})
  n12 = TextureSampleParameter2D(UVs=MaterialFunctionCall(UVs=Add(A=@n10, B=@n11), Texture Scale=Constant() {r=1}) {materialFunction=&ScaleUVsByCenter}.UVs) {
   parameterName=RewardTexture
   samplerSource=SSM_Clamp_WorldGroupSettings
   texture=/FortUILibrary/Systems/Items/Textures/T_UI_SoldierCommando_F
  }
  NamedRerouteDeclaration(None=@n12.A) {name=Character Alpha, nodeColor=<rgba 0 0 0 1>}
  n13 = TextureCoordinate()
  n14 = Constant2Vector() {g=0.5}
  NamedRerouteDeclaration(None=MaterialFunctionCall(UV=Add(A=@n13, B=@n14)) {materialFunction=&MF_UI_UVScaling}.Scaled To Fill) {name=UV, nodeColor=<rgba 0 0 0 1>}
  NamedRerouteDeclaration(None=@n12.RGB) {name=Character, nodeColor=<rgba 0 0 0 1>}
 ]
}
