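# Translucent UI material for a widget backing in the Season Pass hub's upsell screens, going by its name and path.
# It declares three named reroutes: screen-space random Noise, a smooth radial-gradient Shape Mask, and a Color built from the BackingColor parameter.
# Color darkens the tint and multiplies it by a clamped reroute value.
# The export omits which reroute each usage reads and how the material outputs are wired.
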Material /SeasonPassHub/Widgets/Reboot/Materials/M_UI_Upsell_WidgetBacking
graph {
 properties {bAutomaticallySetUsageInEditor=false, blendMode=BLEND_Translucent, materialDomain=MD_UI, usedShadingModels=Default Lit}
 nodes [
  Divide(A=NamedRerouteUsage()) {constB=1.1}
  NamedRerouteDeclaration(None=MaterialFunctionCall!(UV=ScreenPosition().ViewportUV, Seed) {materialFunction=&MF_UI_Random}.Result) {name=Noise, nodeColor=<rgba 1 0.3516 0 1>}
  NamedRerouteDeclaration(None=MaterialFunctionCall() {materialFunction=&MF_UI_RadialGradient}.Smooth) {name=Shape Mask, nodeColor=<rgba 0 0 0 1>}
  n1 = VectorParameter() {group=0 - Coloring, parameterName=BackingColor, defaultValue=<rgba 1 1 1 1>}
  n2 = Clamp(None=NamedRerouteUsage()) {minDefault=0.7}
  NamedRerouteDeclaration(None=Multiply(A=Multiply(A=@n1.RGB, B=@n2)) {constB=0.2}) {name=Color, nodeColor=<rgba 0 0 0 1>}
  NamedRerouteUsage()
 ]
}
