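# Tile button for a Crew, Battle Pass or other season pass, showing name, dates, reward progress and claimable count.
# SetBenefit resolves the pass view model from a Crew benefit, toggling coming-soon state.
# Hover or focus highlights it and, except on touch, selects the pass and updates preview data.
# Clicking pushes the Crew, reward track or category screen with analytics; on touch, the first tap on an unselected pass only selects it.
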WidgetBlueprint /SeasonPassHub/Widgets/Reboot/WBP_CrewPassButton
graphs [
 CrewPassButtonClicked {nodes [CrewPassButtonClicked!() -> (then, Button:WBP Crew Pass Button Object Reference, Index:int, SeasonPassVM:Season Pass VM Object Reference) {is=title K2Node_FunctionEntry}]}
 CrewPassButtonSelected {nodes [CrewPassButtonSelected!() -> (then, SeasonPassVM:Season Pass VM Object Reference, ButtonSelected:WBP Crew Pass Button Object Reference) {is=title K2Node_FunctionEntry}]}
 EventGraph {nodes [
   n1 = "On Unhovered (Button_Hitbox)"!() -> (OutputDelegate, then, Button:Common Button Base Object Reference) {is=title K2Node_ComponentBoundEvent}
   RequestTransition!(
    execute=[@n1.then, @n10.then]
    self:UIKit Block Object Reference=Get Block_Outline() -> (Block_Outline:WBP UIKit Block Outline Object Reference)
    State:EUIKitBlockVisualState Enum=Unfocused
   ) -> (then)
   n2 = "On Hovered (Button_Hitbox)"!() -> (OutputDelegate, then, Button:Common Button Base Object Reference) {is=title K2Node_ComponentBoundEvent}
   RequestTransition(execute=[@n2.then, @n9.then], self=Get Block_Outline(), State=Focused)
   > PlaySound2D(Sound=Get HoveredSound() -> (HoveredSound:Sound Base Object Reference))
   > Set IsFocused(IsFocused:bool=true) -> (Output_Get:bool)
   n3 = GetCurrentInputType(self=Get CommonInputSubsystem())
   n4 = Get SeasonPassVM() -> (SeasonPassVM:Season Pass VM Object Reference)
   n5 = Get SelectedPass!(self:Season Pass Selection VM Object Reference=Get SeasonPassSelectionVM() -> (SeasonPassSelectionVM:Season Pass Selection VM Object Reference)) -> (SelectedPass:Season Pass VM Object Reference)
   > Branch(Condition=AND Boolean(
     A="Not Equal (Object)"(A:Object Reference=@n4, B:Object Reference=@n5) -> (ReturnValue:bool=~) {is=title K2Node_PromotableOperator}
     B="Not Equal (Enum)"!(A:ECommonInputType Enum=@n3, B:ECommonInputType Enum=Touch) -> (ReturnValue:bool)
    ))
   > Call Crew Pass Button Selected!(
    execute:Exec
    self:WBP Crew Pass Button Object Reference
    SeasonPassVM:Season Pass VM Object Reference=@n4
    ButtonSelected:WBP Crew Pass Button Object Reference=Self-Reference()
   ) -> (then:Exec) {is=title K2Node_CallDelegate}
   n6 = Get SeasonPassVM()
   > n7 = Is Valid(InputObject=@n6)
   n8 = Get SeasonPassSelectionVM()
   SetPreviewData!(
    execute=@n7.Is Not Valid
    self:Season Pass Selection VM Object Reference=@n8
    InPreviewData:Array of Fort Cosmetic Preview Data Structures=Make Array("[0]":Fort Cosmetic Preview Data Structure) -> (Array:Array of Fort Cosmetic Preview Data Structures)
   ) -> (then)
   SetPreviewData(
    execute=@n7.Is Valid
    self=@n8
    InPreviewData=GetPreviewData!(self:Season Pass VM Object Reference=@n6) -> (ReturnValue:Array of Fort Cosmetic Preview Data Structures)
   )
   n9 = "On Focused (Button_Hitbox)"!() -> (OutputDelegate, then, Button:Common Button Base Object Reference) {is=title K2Node_ComponentBoundEvent}
   "On Unfocused (Button_Hitbox)"!() -> (OutputDelegate, then, Button:Common Button Base Object Reference) {is=title K2Node_ComponentBoundEvent}
   > n10 = Set IsFocused(IsFocused=false)
   "On Clicked (Button_Hitbox)"!() -> (OutputDelegate, then, Button:Common Button Base Object Reference) {is=title K2Node_ComponentBoundEvent}
   > PlaySound2D(Sound=Get PressedSound() -> (PressedSound:Sound Base Object Reference))
   n11 = Get CommonInputSubsystem()
   > n12 = Branch(Condition="Equal (Enum)"(A=GetCurrentInputType(self=@n11), B=Touch))
   n13 = Get SeasonPassSelectionVM()
   > n14 = Branch(Condition="Not Equal (Object)"(A=Get SeasonPassVM(), B=Get SelectedPass(self=@n13)) -> (ReturnValue=~))
   > Call Crew Pass Button Selected(SeasonPassVM=Get SeasonPassVM(), ButtonSelected=Self-Reference())
   Call Crew Pass Button Clicked!(
    execute:Exec=[@n14.else, @n12.else]
    self:WBP Crew Pass Button Object Reference
    Button:WBP Crew Pass Button Object Reference=Self-Reference()
    Index:int=Get Index() -> (Index:int=0)
    SeasonPassVM:Season Pass VM Object Reference=Get SeasonPassVM()
   ) -> (then:Exec) {is=title K2Node_CallDelegate}
   > n15 = Branch(Condition=Get IsCrewTile() -> (IsCrewTile:bool))
   > n16 = Branch(Condition=GetConfigDataBoolWithDefault!(EntryName:String=CrewRework, bDefaultValue:bool=true) -> (ReturnValue:bool))
   > n17 = PushPrimaryContentClassAsync!(execute:Exec, WidgetClass:Common Activatable Widget Class Reference=/Crew/UI/CrewInfoContainer_B.CrewInfoContainer_B_C) -> (then:Exec, OnPrimaryContentPushed:Exec, ActivatableWidget:Common Activatable Widget Object Reference) {is=title K2Node_AsyncAction}
   n18 = Cast To CrewInfoContainer_B!(execute=@n17.OnPrimaryContentPushed, Object=@n17) -> (then, CastFailed, AsCrew Info Container B:Crew Info Container B Object Reference) {is=title K2Node_DynamicCast}
   > n19 = PageSelected!(execute, self:Crew Info Container B Object Reference=@n18, PageToGo:E_CrewContainer_Page Enum=NewEnumerator0, IsBenefitsRoot:bool=false) -> (then)
   Fire Event Element Interacted(
    execute=[@n20.OnPrimaryContentPushed, @n19.then]
    AdditionalInfo="Add String (Non-Localized)"!(
     Info:"Flow Analytics Additional Info Structure (by ref)"=Make Flow Analytics Additional Info!() -> (FlowAnalyticsAdditionalInfo:Flow Analytics Additional Info Structure) {is=title K2Node_MakeStruct}
     Key:String=CategoryTitle
     Value:String=Crew
    ) -> (ReturnValue:Flow Analytics Additional Info Structure)
   )
   n20 = PushPrimaryContentClassAsync(execute=@n16.else, WidgetClass=/Crew/UI/CrewInfoContainer.CrewInfoContainer_C)
   n21 = PushPrimaryContentClassAsync(
    execute=@n15.else
    WidgetClass=Select(
     Option 0:Common Activatable Widget Class Reference=/SeasonPass/Widgets/SeasonPassCategory.SeasonPassCategory_C
     Option 1:Common Activatable Widget Class Reference=/SeasonPass/Widgets/Reboot/WBP_RewardTrack_Screen.WBP_RewardTrack_Screen_C
     Index=Get IsBattlePass() -> (IsBattlePass:bool)
    ) -> (ReturnValue:Common Activatable Widget Class Reference)
   )
   n22 = Cast To SeasonPassCategory!(execute=@n21.OnPrimaryContentPushed, Object=@n21) -> (then, CastFailed, AsSeason Pass Category:Season Pass Category Object Reference) {is=title K2Node_DynamicCast}
   > SetIsInitializingFromLandingPage!(execute, self:Season Pass Category Object Reference=@n22, IsInitializingFromLandingPage:bool=true) -> (then)
   n23 = GetCategoryPageTitle!(self:Season Pass Category VM Object Reference=Get SeasonPassCategoryVM() -> (SeasonPassCategoryVM:Season Pass Category VM Object Reference)) -> (ReturnValue:Text)
   > Fire Event Element Interacted(
    Element=@n21
    AdditionalInfo="Add String (Non-Localized)"(
     Info=Make Flow Analytics Additional Info()
     Key=CategoryTitle
     Value=Get Source String from Text!(Text:Text=@n23) -> (ReturnValue:String) {is=title K2Node_CallFunction}
    )
   )
   n24 = Set Double Width!() -> (OutputDelegate, then, Is Double Width:bool=~) {is=title K2Node_CustomEvent}
   > Set Is Double Width(Is Double Width:bool=@n24.Is Double Width) -> (Output_Get:bool)
   > SetWidthOverride(self=Get Content() -> (Content:Size Box Object Reference), InWidthOverride=SelectFloat(A=480, B=312, bPickA=@n24.Is Double Width))
   n25 = Get SeasonPassVM()
   > Branch(Condition=IsValid(Object=@n25))
   > SetHighlightText!(execute, self, RewardsToClaim:int=GetNumPurchaseableRewards!(self:Season Pass Element Base VM Object Reference=@n25) -> (ReturnValue:int=0)) -> (then)
   Event Construct()
   > Branch(Condition=IsCultureRightToLeft!(Culture:String=GetCurrentCulture!() -> (ReturnValue:String)) -> (ReturnValue:bool))
   > HandleRTL()
  ]}
 HandleRTL {nodes [
   HandleRTL() {is=title K2Node_FunctionEntry}
   > Set FlowDirectionPreference!(execute, FlowDirectionPreference:EFlowDirectionPreference Enum=Culture, self:Widget Object Reference=Get Content()) -> (then, Output_Get:EFlowDirectionPreference Enum=Inherit)
  ]}
 HandleSeasonEndDate {nodes [
   n1 = HandleSeasonEndDate!() -> (then, End:Date Time Structure) {is=title K2Node_FunctionEntry}
   > Set Text(
    Text=Format Text!(
     Format="LOCTEXT(\"Ends {date}\")"
     date:Text=As Date!(InDateTime:"Date Time Structure (by ref)"=@n1, InDateStyle:EDateTimeStyle Enum=Medium) -> (ReturnValue:Text)
    ) -> (Result)
    self=Get Text_EndDate() -> (Text_EndDate:Common Text Object Reference)
   )
  ]}
 SetBenefit {nodes [
   n1 = SetBenefit!() -> (then, Benefit:Crew Benefit VM Object Reference) {is=title K2Node_FunctionEntry}
   > n2 = Set CrewBenefitVM(CrewBenefitVM:Crew Benefit VM Object Reference=@n1) -> (Output_Get:Crew Benefit VM Object Reference)
   > n3 = Set SeasonPassVM(SeasonPassVM:Season Pass VM Object Reference=FindSeasonPass!(
     self:Season Pass System VM Object Reference=Get SeasonPassSystemVM() -> (SeasonPassSystemVM:Season Pass System VM Object Reference)
     TemplateId:String=GetTemplateId!(self:Crew Benefit VM Object Reference=@n2) -> (ReturnValue:String)
    ) -> (ReturnValue:Season Pass VM Object Reference)) -> (Output_Get:Season Pass VM Object Reference)
   n4 = IsValid(Object=@n3)
   > n5 = Set FoundPassVM(FoundPassVM:bool=@n4) -> (Output_Get:bool)
   > ShowComingSoonState!(execute, self, ShouldShowComingSoon:bool=NOT Boolean(A=@n5)) -> (then)
   > SetVisibility(
    self=Get EpicCMSImage() -> (EpicCMSImage:Epic CMSImage Object Reference)
    InVisibility=Select(Option 0:ESlateVisibility Enum=SelfHitTestInvisible, Option 1:ESlateVisibility Enum=Collapsed, Index=@n4) -> (ReturnValue:ESlateVisibility Enum)
   )
   > SetVisibility(
    self=Get Background() -> (Background:Fort Lazy Image Object Reference)
    InVisibility=Select(Option 0=Collapsed, Option 1=SelfHitTestInvisible, Index=@n4)
   )
  ]}
 SetComingSoon {nodes [
   n1 = SetComingSoon!() -> (then, CSText:Text) {is=title K2Node_FunctionEntry}
   > Set Text(Text=@n1, self=Get Text_ComingSoonDate() -> (Text_ComingSoonDate:Common Text Object Reference))
  ]}
 SetHighlightText {nodes [
   n1 = SetHighlightText!() -> (then, RewardsToClaim:int) {is=title K2Node_FunctionEntry}
   n2 = "To Text (Integer)"(Value=@n1)
   n3 = Format Text!(Format="LOCTEXT(\"{num} to claim\")", num:int=@n1) -> (Result)
   n4 = Get Is Double Width() -> (Is Double Width:bool)
   n5 = Get RewardsToClaim() -> (RewardsToClaim:int=0)
   > SetText!(
    execute
    self:WBP UIKit Highlight Base Object Reference=Get Rewards() -> (Rewards:WBP UIKit Highlight Object Reference)
    Text:Text=SelectText(A=SelectText(A=@n3, B=@n2, bPickA=@n4), B=GetEmptyText!() -> (ReturnValue:Text), bPickA="integer > integer"(A=@n5))
   ) -> (then)
  ]}
 SetLockedState {nodes [
   SetLockedState!() -> (then, ForceHideProgress:bool=~) {is=title K2Node_FunctionEntry}
   n1 = Get ForceHideProgress() -> (ForceHideProgress:bool)
   > SetVisibility(
    self=Get ProgressHorizontalBox() -> (ProgressHorizontalBox:Horizontal Box Object Reference)
    InVisibility=Select(Option 0=Select(Option 0=Visible, Option 1=Collapsed, Index=@n1), Option 1=Collapsed, Index=false)
   )
  ]}
 SetOwnedRewardsCount {nodes [
   n1 = SetOwnedRewardsCount!() -> (then, Owned:int) {is=title K2Node_FunctionEntry}
   > Set OwnedRewardsCount(OwnedRewardsCount:int=@n1) -> (Output_Get:int=0)
   > UpdateProgressVisual()
  ]}
 SetProgressionColor {nodes [
   n1 = SetProgressionColor!() -> (then, Color:Linear Color Structure) {is=title K2Node_FunctionEntry}
   > Set ProgressionColor(ProgressionColor:Linear Color Structure=@n1) -> (Output_Get:Linear Color Structure)
   > UpdateProgressVisual()
  ]}
 SetSeasonPass {nodes [SetSeasonPass!() -> (then, Passes:Array of Season Pass VM Object References) {is=title K2Node_FunctionEntry}]}
 SetThemeData {nodes [
   n1 = SetThemeData!() -> (then, ThemeData:Season Pass Theme Data Structure) {is=title K2Node_FunctionEntry}
   n2 = Break Season Pass Theme Data!(SeasonPassThemeData:"Season Pass Theme Data Structure (by ref)"=@n1) -> (
    ProgressionColor:Linear Color Structure
    GiftboxVaultBackgroundMaterial:Soft Object Reference
    CrewCategoryBackgroundMaterial:Soft Object Reference
    CustomLogo:Soft Object Reference
    CustomSeasonTitle:Text
    CustomSeasonTitleColor:Linear Color Structure
    Upsell2BackgrounMaterial:Soft Object Reference
    ImageReferenceTable:Fort Image Reference Table Structure
    ColorReferenceTable:Fort Color Reference Table Structure
    TextReferenceTable:Fort Text Reference Table Structure
   ) {is=title K2Node_BreakStruct}
   > Set ProgressionColor(ProgressionColor=@n2)
   > UpdateProgressVisual()
   > SetBrushFromLazyDisplayAsset!(
    execute
    self:Common Lazy Image Object Reference=Get Background()
    LazyObject:"Soft Object Reference (by ref)"=@n2.CrewCategoryBackgroundMaterial
    bMatchTextureSize:bool=false
   ) -> (then)
  ]}
 SetTotalRewardsCount {nodes [
   n1 = SetTotalRewardsCount!() -> (then, Total:int) {is=title K2Node_FunctionEntry}
   > Set TotalRewardsCount(TotalRewardsCount:int=@n1) -> (Output_Get:int=0)
   > UpdateProgressVisual()
  ]}
 ShowComingSoonState {nodes [
   n1 = ShowComingSoonState!() -> (then, ShouldShowComingSoon:bool=~) {is=title K2Node_FunctionEntry}
   > SetVisibility(
    self=[
     Get ProgressHorizontalBox()
     Get SeasonPassBadgeIcon() -> (SeasonPassBadgeIcon:Season Pass Badge Icon Object Reference)
     Get Date_Spacer_Collapses() -> (Date_Spacer_Collapses:Spacer Object Reference)
    ]
    InVisibility=Select(Option 0=SelfHitTestInvisible, Option 1=Collapsed, Index=@n1)
   )
   n2 = Get Text_EndDate()
   n3 = Get ShouldShowComingSoon() -> (ShouldShowComingSoon:bool)
   n4 = Get Text_ComingSoonDate()
   > SetActiveWidget(
    self=Get DateSwitcher() -> (DateSwitcher:Common Visibility Switcher Object Reference)
    Widget=Select(
     Option 0:Widget Object Reference=Select(Option 0=@n2, Option 1=@n4, Index=@n3)
     Option 1:Widget Object Reference=Get Text_Subscription() -> (Text_Subscription:Common Text Object Reference)
     Index=Get IsCrewTile()
    ) -> (ReturnValue:Widget Object Reference)
   )
  ]}
 UpdateProgressVisual {nodes [
   UpdateProgressVisual() {is=title K2Node_FunctionEntry}
   > n1 = Sequence!(execute) -> (then_0, then_1, then_2:Exec)
   Set Text(
    execute=@n1.then_1
    Text="To Text (Integer)"(Value=Get OwnedRewardsCount() -> (OwnedRewardsCount:int=0))
    self=Get ProgressTextX() -> (ProgressTextX:Common Text Object Reference)
   )
   n2 = "To String (Integer)"(InInt=Get TotalRewardsCount() -> (TotalRewardsCount:int=0))
   > Set Text(Text="To Text (String)"(InString=@n2), self=Get ProgressTextOf() -> (ProgressTextOf:Common Text Object Reference))
   > SetText(
    self=Get Highlight_ProgressComplete() -> (Highlight_ProgressComplete:WBP UIKit Highlight Object Reference)
    Text=Format Text!(Format="LOCTEXT(\"{X}/{X}\")", X:String=@n2) -> (Result)
   )
   n3 = "Equal (Integer)"(A=Get OwnedRewardsCount() -> (OwnedRewardsCount=0), B=Get TotalRewardsCount() -> (TotalRewardsCount=0))
   SetColorAndOpacity!(
    execute=@n1.then_2
    self=Get ProgressTextX()
    InColorAndOpacity_SpecifiedColor:Linear Color Structure=SelectColor(
     A=Get CompletionColor() -> (CompletionColor:Linear Color Structure)
     B=Get ProgressionColor() -> (ProgressionColor:Linear Color Structure)
     bPickA=@n3
    )
    InColorAndOpacity_ColorUseRule:ESlateColorStylingMode Enum
   ) -> (then)
   > SetColorAndOpacity(
    self=[Get ProgressTextOf(), Get Divider() -> (Divider:Common Text Object Reference)]
    InColorAndOpacity_SpecifiedColor=SelectColor(A=Get CompletionColor(), B=(R=1,G=1,B=1,A=1), bPickA=@n3)
   )
   > SetActiveWidgetIndex(
    self:Common Visibility Switcher Object Reference=Get ProgressVisibilitySwitcher() -> (ProgressVisibilitySwitcher:Common Visibility Switcher Object Reference)
    Index=SelectInt(A=1, B=0, bPickA=@n3)
   )
   n4 = Get TotalRewardsCount() -> (TotalRewardsCount=0)
   n5 = Get OwnedRewardsCount() -> (OwnedRewardsCount=0)
   SetProgress!(
    execute=@n1.then_0
    self:WBP UIKit Progress Bar Base Object Reference=Get ProgressBar() -> (ProgressBar:WBP UIKit Progress Bar Object Reference)
    Progress:double=float / float(A="To Float (Integer)"(InInt=@n5), B="To Float (Integer)"(InInt=@n4))
   ) -> (then)
   > SetFillColors!(
    execute
    self:WBP UIKit Progress Bar Base Object Reference=Get ProgressBar()
    ColorA:Linear Color Structure=Get ProgressionColor()
    ColorB:Linear Color Structure=(R=1,G=1,B=1,A=1)
    ColorC:Linear Color Structure=(R=1,G=1,B=1,A=1)
   ) -> (then)
  ]}
]
variables [
 CompletionColor {type=LinearColor, default=<rgba 0.0395 1 0.2307 1>}
 CrewPassButtonClicked {}
 CrewPassButtonSelected {}
 FoundPassVM {type=boolean}
 HoveredSound {type=SoundBase, default=null}
 Index {type=integer, default=-1}
 Is Double Width {type=boolean}
 IsBattlePass {type=boolean}
 IsCrewTile {type=boolean}
 IsFocused {type=boolean}
 OwnedRewardsCount {type=integer}
 PressedSound {type=SoundBase, default=null}
 ProgressionColor {type=LinearColor, default=<rgba 0 0 0 0>}
 TotalRewardsCount {type=integer}
]
widgets {
 info {inheritedWidgetCount=0, parentClass=&CommonActivatableWidget, rootWidgetClass=&SizeBox, widgetCount=26}
 widgets [
  Content {
   widgetClassPath=&SizeBox
   bIsVariable=true
   widget=Content
   properties {
    bOverride_HeightOverride=true
    bOverride_WidthOverride=true
    heightOverride=192
    maxDesiredWidth=312
    minDesiredHeight=192
    minDesiredWidth=312
    widthOverride=312
   }
  }
  Overlay_1 {widgetClassPath=&Overlay, parent=Content, widget=Overlay_1, slot {class=SizeBoxSlot}}
  Background {
   widgetClassPath=&FortLazyImage
   bIsVariable=true
   parent=Overlay_1
   widget=Background
   properties {
    brush {
     bIsDynamicallyLoaded=false
     drawAs=Image
     imageType=FullColor
     mirroring=NoMirror
     resourceName=None
     resourceObject=/FortUILibrary/Systems/Items/MaterialInstances/MI_UIKit_Item_Background
     tiling=NoTile
     imageSize=<32 32>
     margin=<ltrb 0 0 0 0>
     outlineSettings {
      bUseBrushTransparency=false
      roundingType=HalfHeightRadius
      width=0
      color {colorUseRule=UseColor_Specified, specifiedColor=<rgba 0 0 0 0>}
      cornerRadii=<0 0 0 0>
     }
     tintColor {colorUseRule=UseColor_Specified, specifiedColor=<rgba 1 1 1 1>}
     uVRegion {bIsValid=false, max=<0 0>, min=<0 0>}
    }
    loadingBackgroundBrush {
     bIsDynamicallyLoaded=false
     drawAs=Image
     imageType=FullColor
     mirroring=NoMirror
     resourceName=None
     resourceObject=null
     tiling=NoTile
     imageSize=<32 32>
     margin=<ltrb 0 0 0 0>
     outlineSettings {
      bUseBrushTransparency=false
      roundingType=HalfHeightRadius
      width=0
      color {colorUseRule=UseColor_Specified, specifiedColor=<rgba 0 0 0 0>}
      cornerRadii=<0 0 0 0>
     }
     tintColor {colorUseRule=UseColor_Specified, specifiedColor=<rgba 1 1 1 1>}
     uVRegion {bIsValid=false, max=<0 0>, min=<0 0>}
    }
   }
   slot {class=OverlaySlot, properties {horizontalAlignment=HAlign_Fill, verticalAlignment=VAlign_Fill}}
  }
  ScaleBox_0 {
   widgetClassPath=&ScaleBox
   parent=Overlay_1
   widget=ScaleBox_0
   slot {class=OverlaySlot, properties {horizontalAlignment=HAlign_Fill, verticalAlignment=VAlign_Fill}}
  }
  EpicCMSImage {
   widgetClassPath=/Script/EpicCMSUIFramework.EpicCMSImage
   bIsVariable=true
   parent=ScaleBox_0
   widget=EpicCMSImage
   properties {
    materialTextureParamName=ForegroundImage
    brush {
     bIsDynamicallyLoaded=false
     drawAs=Image
     imageType=FullColor
     mirroring=NoMirror
     resourceName=None
     resourceObject=/SeasonPass/Materials/M_UI_Nexus_CategoryButton
     tiling=NoTile
     imageSize=<32 32>
     margin=<ltrb 0 0 0 0>
     outlineSettings {
      bUseBrushTransparency=false
      roundingType=HalfHeightRadius
      width=0
      color {colorUseRule=UseColor_Specified, specifiedColor=<rgba 0 0 0 0>}
      cornerRadii=<0 0 0 0>
     }
     tintColor {colorUseRule=UseColor_Specified, specifiedColor=<rgba 1 1 1 1>}
     uVRegion {bIsValid=false, max=<0 0>, min=<0 0>}
    }
    loadingFailFallback {
     bIsDynamicallyLoaded=false
     drawAs=Image
     imageType=FullColor
     mirroring=NoMirror
     resourceName=None
     resourceObject=/FortUILibrary/Textures/Icons/T_UI_Icon_UIKit_ImageError
     tiling=NoTile
     imageSize=<64 64>
     margin=<ltrb 0 0 0 0>
     outlineSettings {
      bUseBrushTransparency=false
      roundingType=HalfHeightRadius
      width=0
      color {colorUseRule=UseColor_Specified, specifiedColor=<rgba 0 0 0 0>}
      cornerRadii=<0 0 0 0>
     }
     tintColor {colorUseRule=UseColor_Specified, specifiedColor=<rgba 1 1 1 1>}
     uVRegion {bIsValid=false, max=<0 0>, min=<0 0>}
    }
   }
   slot {class=ScaleBoxSlot}
  }
  Block_Outline {
   widgetClassPath=/FortUILibrary/UIKit/Buttons/Blocks/WBP_UIKit_Block_Outline
   bIsVariable=true
   parent=Overlay_1
   widget=Block_Outline
   properties {bottomLeftCornerRadius=30, bottomRightCornerRadius=30, radiusInPixels=true, topLeftCornerRadius=30, topRightCornerRadius=30}
   slot {class=OverlaySlot, properties {horizontalAlignment=HAlign_Fill, verticalAlignment=VAlign_Fill}}
  }
  Overlay_2 {
   widgetClassPath=&Overlay
   parent=Overlay_1
   widget=Overlay_2
   slot {class=OverlaySlot, properties {horizontalAlignment=HAlign_Fill, verticalAlignment=VAlign_Fill, padding=<ltrb 16 16 16 16>}}
  }
  SeasonPassBadgeIcon {
   widgetClassPath=/SeasonPass/Widgets/SeasonPassBadgeIcon
   bIsVariable=true
   parent=Overlay_2
   widget=SeasonPassBadgeIcon
   properties {size=<56 56>}
   slot {class=OverlaySlot}
  }
  VerticalBox_45 {
   widgetClassPath=&VerticalBox
   parent=Overlay_2
   widget=VerticalBox_45
   properties {visibility=HitTestInvisible}
   slot {class=OverlaySlot, properties {horizontalAlignment=HAlign_Fill, verticalAlignment=VAlign_Bottom}}
  }
  PassName {
   widgetClassPath=&CommonTextBlock
   bIsVariable=true
   parent=VerticalBox_45
   widget=PassName
   properties {
    scrollStyle=&TextScrollStyle-Base_C
    text=Pass Name
    font {
     bForceMonospaced=false
     bMaterialIsStencil=false
     fontMaterial=null
     fontObject=&HeadingNow
     letterSpacing=0
     monospacedWidth=1
     size=24
     skewAmount=0
     typefaceFontName=86 Bold
     outlineSettings {
      bApplyOutlineToDropShadows=false
      bMiteredCorners=false
      bSeparateFillAlpha=false
      outlineMaterial=null
      outlineSize=0
      outlineColor=<rgba 0 0 0 1>
     }
    }
    shadowOffset=<0 0>
   }
   slot {class=VerticalBoxSlot, properties {horizontalAlignment=HAlign_Left, verticalAlignment=VAlign_Bottom}}
  }
  DateSwitcher {
   widgetClassPath=&CommonVisibilitySwitcher
   bIsVariable=true
   parent=VerticalBox_45
   widget=DateSwitcher
   slot {class=VerticalBoxSlot}
  }
  Text_EndDate {
   widgetClassPath=&CommonTextBlock
   bIsVariable=true
   parent=DateSwitcher
   widget=Text_EndDate
   properties {
    text=End Date
    visibility=Collapsed
    colorAndOpacity {colorUseRule=UseColor_Specified, specifiedColor=<rgba 1 1 1 0.65>}
    font {
     bForceMonospaced=false
     bMaterialIsStencil=false
     fontMaterial=null
     fontObject=&HeadingNow
     letterSpacing=0
     monospacedWidth=1
     size=24
     skewAmount=0
     typefaceFontName=86 Bold
     outlineSettings {
      bApplyOutlineToDropShadows=false
      bMiteredCorners=false
      bSeparateFillAlpha=false
      outlineMaterial=null
      outlineSize=0
      outlineColor=<rgba 0 0 0 1>
     }
    }
    shadowOffset=<0 0>
   }
   slot {class=CommonVisibilitySwitcherSlot, properties {horizontalAlignment=HAlign_Left}}
  }
  Text_ComingSoonDate {
   widgetClassPath=&CommonTextBlock
   bIsVariable=true
   parent=DateSwitcher
   widget=Text_ComingSoonDate
   properties {
    text=Coming Soon
    visibility=Collapsed
    colorAndOpacity {colorUseRule=UseColor_Specified, specifiedColor=<rgba 1 1 1 0.65>}
    font {
     bForceMonospaced=false
     bMaterialIsStencil=false
     fontMaterial=null
     fontObject=&HeadingNow
     letterSpacing=0
     monospacedWidth=1
     size=24
     skewAmount=0
     typefaceFontName=86 Bold
     outlineSettings {
      bApplyOutlineToDropShadows=false
      bMiteredCorners=false
      bSeparateFillAlpha=false
      outlineMaterial=null
      outlineSize=0
      outlineColor=<rgba 0 0 0 1>
     }
    }
    shadowOffset=<0 0>
   }
   slot {class=CommonVisibilitySwitcherSlot, properties {horizontalAlignment=HAlign_Left, verticalAlignment=VAlign_Bottom}}
  }
  Text_Subscription {
   widgetClassPath=&CommonTextBlock
   bIsVariable=true
   parent=DateSwitcher
   widget=Text_Subscription
   properties {
    text=Subscription
    visibility=Collapsed
    colorAndOpacity {colorUseRule=UseColor_Specified, specifiedColor=<rgba 1 1 1 0.65>}
    font {
     bForceMonospaced=false
     bMaterialIsStencil=false
     fontMaterial=null
     fontObject=&HeadingNow
     letterSpacing=0
     monospacedWidth=1
     size=24
     skewAmount=0
     typefaceFontName=86 Bold
     outlineSettings {
      bApplyOutlineToDropShadows=false
      bMiteredCorners=false
      bSeparateFillAlpha=false
      outlineMaterial=null
      outlineSize=0
      outlineColor=<rgba 0 0 0 1>
     }
    }
    shadowOffset=<0 0>
   }
   slot {class=CommonVisibilitySwitcherSlot, properties {horizontalAlignment=HAlign_Left, verticalAlignment=VAlign_Bottom}}
  }
  Date_Spacer_Collapses {
   widgetClassPath=&Spacer
   bIsVariable=true
   parent=VerticalBox_45
   widget=Date_Spacer_Collapses
   properties {size=<1 2>}
   slot {class=VerticalBoxSlot}
  }
  Date_Spacer_Remains {widgetClassPath=&Spacer, parent=VerticalBox_45, widget=Date_Spacer_Remains, properties {size=<1 4>}, slot {class=VerticalBoxSlot}}
  ProgressVisibilitySwitcher {
   widgetClassPath=&CommonVisibilitySwitcher
   bIsVariable=true
   parent=VerticalBox_45
   widget=ProgressVisibilitySwitcher
   slot {class=VerticalBoxSlot, properties {verticalAlignment=VAlign_Bottom, padding=<ltrb 0 0 0 3>}}
  }
  ProgressHorizontalBox {
   widgetClassPath=&HorizontalBox
   bIsVariable=true
   parent=ProgressVisibilitySwitcher
   widget=ProgressHorizontalBox
   slot {class=CommonVisibilitySwitcherSlot, properties {horizontalAlignment=HAlign_Left}}
  }
  ProgressBar {
   widgetClassPath=/FortUILibrary/UIKit/ProgressBar/WBP_UIKit_ProgressBar
   bIsVariable=true
   parent=ProgressHorizontalBox
   widget=ProgressBar
   properties {
    height=16
    linearMaterial=/SeasonPass/Materials/ProgressBar/MI_UI_NexusProgerssBar
    outline_PX=6
    progressValue=0.5
    useCircular=false
    useOutline=true
    width=128
    colorA=<rgba 1 0.6849 0 1>
    colorB=<rgba 1 0.9857 0.9484 1>
    emptyColor=<rgba 0.3573 0.3573 0.3573 0.305>
   }
   slot {class=HorizontalBoxSlot, properties {horizontalAlignment=HAlign_Left, verticalAlignment=VAlign_Center, padding=<ltrb 0 0 16 0>}}
  }
  HorizontalBox {
   widgetClassPath=&HorizontalBox
   bIsVariable=true
   parent=ProgressHorizontalBox
   widget=HorizontalBox
   slot {class=HorizontalBoxSlot, properties {padding=<ltrb 0 0 12 0>}}
  }
  ProgressTextX {
   widgetClassPath=&CommonTextBlock
   bIsVariable=true
   parent=HorizontalBox
   widget=ProgressTextX
   properties {
    text=20
    font {
     bForceMonospaced=false
     bMaterialIsStencil=false
     fontMaterial=null
     fontObject=&NotoSans
     letterSpacing=0
     monospacedWidth=1
     size=24
     skewAmount=0
     typefaceFontName=NotoSans-Regular
     outlineSettings {
      bApplyOutlineToDropShadows=false
      bMiteredCorners=false
      bSeparateFillAlpha=false
      outlineMaterial=null
      outlineSize=0
      outlineColor=<rgba 0 0 0 1>
     }
    }
    shadowOffset=<0 0>
   }
   slot {class=HorizontalBoxSlot, properties {horizontalAlignment=HAlign_Left}}
  }
  Divider {
   widgetClassPath=&CommonTextBlock
   bIsVariable=true
   parent=HorizontalBox
   widget=Divider
   properties {
    text=/
    font {
     bForceMonospaced=false
     bMaterialIsStencil=false
     fontMaterial=null
     fontObject=&HeadingNow
     letterSpacing=10
     monospacedWidth=1
     size=15.75
     skewAmount=0
     typefaceFontName=76 Bold
     outlineSettings {
      bApplyOutlineToDropShadows=false
      bMiteredCorners=false
      bSeparateFillAlpha=false
      outlineMaterial=null
      outlineSize=0
      outlineColor=<rgba 0 0 0 1>
     }
    }
    shadowOffset=<0 0>
   }
   slot {class=HorizontalBoxSlot, properties {horizontalAlignment=HAlign_Left}}
  }
  ProgressTextOf {
   widgetClassPath=&CommonTextBlock
   bIsVariable=true
   parent=HorizontalBox
   widget=ProgressTextOf
   properties {
    text=100
    font {
     bForceMonospaced=false
     bMaterialIsStencil=false
     fontMaterial=null
     fontObject=&NotoSans
     letterSpacing=0
     monospacedWidth=1
     size=24
     skewAmount=0
     typefaceFontName=NotoSans-Regular
     outlineSettings {
      bApplyOutlineToDropShadows=false
      bMiteredCorners=false
      bSeparateFillAlpha=false
      outlineMaterial=null
      outlineSize=0
      outlineColor=<rgba 0 0 0 1>
     }
    }
    shadowOffset=<0 0>
   }
   slot {class=HorizontalBoxSlot, properties {horizontalAlignment=HAlign_Left}}
  }
  Rewards {
   widgetClassPath=/FortUILibrary/UIKit/Highlight/WBP_UIKit_Highlight
   bIsVariable=true
   parent=ProgressHorizontalBox
   widget=Rewards
   properties {highlightSize=Compact, highlightText=0, highlightType=Attention}
   slot {class=HorizontalBoxSlot, properties {horizontalAlignment=HAlign_Right, size {sizeRule=Fill, value=1}}}
  }
  Highlight_ProgressComplete {
   widgetClassPath=&WBP_UIKit_Highlight
   bIsVariable=true
   parent=ProgressVisibilitySwitcher
   widget=Highlight_ProgressComplete
   properties {highlightSize=Compact, highlightText=100/100, highlightType=Social}
   slot {class=CommonVisibilitySwitcherSlot, properties {horizontalAlignment=HAlign_Left, verticalAlignment=VAlign_Center}}
  }
  Button_Hitbox {
   widgetClassPath=&WBP_UIKit_Button_Generic
   bIsVariable=true
   parent=Overlay_1
   widget=Button_Hitbox
   properties {
    bHideInputAction=true
    bSimulateHoverOnTouchInput=false
    contentHorizontalAlignment=HAlign_Fill
    contentVerticalAlignment=VAlign_Fill
    touchMethod=PreciseTap
    useBackgroundBlock=false
    useIconBlock=false
    useTextBlock=false
    visibility=Visible
   }
   slot {class=OverlaySlot, properties {horizontalAlignment=HAlign_Fill, verticalAlignment=VAlign_Fill}}
  }
 ]
}
