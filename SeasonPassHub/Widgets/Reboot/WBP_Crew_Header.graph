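# Fortnite Crew header widget for the season pass hub, switching between an unsubscribed offer layout and a subscribed layout.
# HandleSubState maps subscription state to View Offer, Manage Subscription or Resolve Issue buttons and billing text, and desired focus follows the visible button.
# View and Manage broadcast HeaderEnteredSubscreen and push CrewInfoContainer, while Resolve queues a modal fed a CrewVM sub-viewmodel.
# Opening plays the fade-and-slide Intro at double speed; mobile adjusts description wrapping.
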WidgetBlueprint /SeasonPassHub/Widgets/Reboot/WBP_Crew_Header
animations [Intro {
  binding_count=8
  length_seconds=0.9
  object=&self:Intro
  sequence {bindings [
    HorizontalBox_0 {
     id=#1
     possessed_class=HorizontalBox
     tracks [{
       type=MovieSceneFloatTrack
       display_name=Render Opacity
       sections [{type=MovieSceneFloatSection, end_frame=15, start_frame=0, channels [None {type=MovieSceneScriptingFloatChannel, keys [[0,0],[10,0],[15,1]]}]}]
      }]
    }
    SubTag {
     id=#2
     possessed_class=WBP_UIKit_Tag_C
     tracks [
      {
       type=MovieSceneFloatTrack
       display_name=Render Opacity
       sections [{type=MovieSceneFloatSection, end_frame=15, start_frame=0, channels [None {type=MovieSceneScriptingFloatChannel, keys [[3,0],[10,1]]}]}]
      }
      {
       type=MovieScene2DTransformTrack
       sections [{
         type=MovieScene2DTransformSection
         end_frame=15
         start_frame=0
         channels [
          Translation.X {type=MovieSceneScriptingFloatChannel, const [0,2]}
          Translation.Y {type=MovieSceneScriptingFloatChannel, keys [[3,80],[10,0]]}
         ]
        }]
      }
     ]
    }
    ViewCrew_Button {
     id=#3
     possessed_class=WBP_UIKit_Button_Loud_C
     tracks [{
       type=MovieSceneFloatTrack
       display_name=Render Opacity
       sections [{type=MovieSceneFloatSection, end_frame=15, start_frame=0, channels [None {type=MovieSceneScriptingFloatChannel, keys [[0,0],[10,0],[15,1]]}]}]
      }]
    }
    WBP_UIKit_Highlight {
     id=#4
     possessed_class=WBP_UIKit_Highlight_C
     tracks [
      {
       type=MovieSceneFloatTrack
       display_name=Render Opacity
       sections [{type=MovieSceneFloatSection, end_frame=15, start_frame=0, channels [None {type=MovieSceneScriptingFloatChannel, keys [[3,0],[10,1]]}]}]
      }
      {
       type=MovieScene2DTransformTrack
       sections [{
         type=MovieScene2DTransformSection
         end_frame=15
         start_frame=0
         channels [
          Translation.X {type=MovieSceneScriptingFloatChannel, const [0,2]}
          Translation.Y {type=MovieSceneScriptingFloatChannel, keys [[3,80],[10,0]]}
         ]
        }]
      }
     ]
    }
    PriceNextBill {
     id=#5
     possessed_class=CommonTextBlock
     tracks [{
       type=MovieSceneFloatTrack
       display_name=Render Opacity
       sections [{type=MovieSceneFloatSection, end_frame=15, start_frame=0, channels [None {type=MovieSceneScriptingFloatChannel, keys [[0,0],[10,0],[15,1]]}]}]
      }]
    }
    Text_Description {
     id=#6
     possessed_class=CommonTextBlock
     tracks [
      {
       type=MovieSceneFloatTrack
       display_name=Render Opacity
       sections [{type=MovieSceneFloatSection, end_frame=18, start_frame=0, channels [None {type=MovieSceneScriptingFloatChannel, keys [[2,0],[12,1]]}]}]
      }
      {
       type=MovieScene2DTransformTrack
       sections [{
         type=MovieScene2DTransformSection
         end_frame=18
         start_frame=0
         channels [
          Translation.X {type=MovieSceneScriptingFloatChannel, const [0,2]}
          Translation.Y {type=MovieSceneScriptingFloatChannel, keys [[2,80],[12,0]]}
         ]
        }]
      }
     ]
    }
    Pass_Badge {
     id=#7
     possessed_class=FortLazyImage
     tracks [
      {
       type=MovieSceneFloatTrack
       display_name=Render Opacity
       sections [{type=MovieSceneFloatSection, end_frame=10, start_frame=0, channels [None {type=MovieSceneScriptingFloatChannel, keys [[0,0],[8,1]]}]}]
      }
      {
       type=MovieScene2DTransformTrack
       sections [{
         type=MovieScene2DTransformSection
         end_frame=13
         start_frame=0
         channels [
          Translation.X {type=MovieSceneScriptingFloatChannel, const [0,2]}
          Translation.Y {type=MovieSceneScriptingFloatChannel, keys [[0,80],[8.0003,0]]}
         ]
        }]
      }
     ]
    }
    Pass_Badge_1 {
     id=#8
     possessed_class=FortLazyImage
     tracks [
      {
       type=MovieSceneFloatTrack
       display_name=Render Opacity
       sections [{type=MovieSceneFloatSection, end_frame=10, start_frame=0, channels [None {type=MovieSceneScriptingFloatChannel, keys [[0,0],[8,1]]}]}]
      }
      {
       type=MovieScene2DTransformTrack
       sections [{
         type=MovieScene2DTransformSection
         end_frame=13
         start_frame=0
         channels [
          Translation.X {type=MovieSceneScriptingFloatChannel, const [0,2]}
          Translation.Y {type=MovieSceneScriptingFloatChannel, keys [[0,80],[8.0003,0]]}
         ]
        }]
      }
     ]
    }
   ]}
 }]
graphs [
 BP_GetDesiredFocusTarget {nodes [
   Get Desired Focus Target()
   n1 = Get HeaderSwitcher() -> (HeaderSwitcher:Common Visibility Switcher Object Reference)
   n2 = Get ViewCrew_Button_1() -> (ViewCrew_Button_1:WBP UIKit Button Loud Object Reference)
   n3 = IsVisible(self=@n2)
   n4 = Get ResolveIssue_Button() -> (ResolveIssue_Button:WBP UIKit Button Loud Object Reference)
   n5 = Select(
    Option 0:Widget Object Reference=Get ManageSub_Button() -> (ManageSub_Button:WBP UIKit Button Regular Object Reference)
    Option 1:Widget Object Reference=@n4
    Index=IsVisible(self=@n4)
   ) -> (ReturnValue:Widget Object Reference)
   > Return Node!(
    execute
    ReturnValue:Widget Object Reference=Select(
     Option 0=Get ViewCrew_Button() -> (ViewCrew_Button:WBP UIKit Button Loud Object Reference)
     Option 1=Select(Option 0=@n5, Option 1=@n2, Index:bool=@n3)
     Index:int=GetActiveWidgetIndex!(self:Common Visibility Switcher Object Reference=@n1) -> (ReturnValue:int=0)
    )
   )
  ]}
 EventGraph {nodes [
   n1 = "On Clicked (ViewCrew_Button)"!() -> (OutputDelegate, then, Button:Common Button Base Object Reference) {is=title K2Node_ComponentBoundEvent}
   Call Header Entered Subscreen!(execute:Exec=[@n1.then, @n2.then, @n3.then], self:WBP Crew Header Object Reference) -> (then:Exec) {is=title K2Node_CallDelegate}
   > PushPrimaryContentClassAsync!(execute:Exec, WidgetClass:Common Activatable Widget Class Reference=/Crew/UI/CrewInfoContainer.CrewInfoContainer_C) -> (then:Exec, OnPrimaryContentPushed:Exec, ActivatableWidget:Common Activatable Widget Object Reference) {is=title K2Node_AsyncAction}
   n2 = "On Clicked (ViewCrew_Button_1)"!() -> (OutputDelegate, then, Button:Common Button Base Object Reference) {is=title K2Node_ComponentBoundEvent}
   n3 = "On Clicked (ManageSub_Button)"!() -> (OutputDelegate, then, Button:Common Button Base Object Reference) {is=title K2Node_ComponentBoundEvent}
   "On Clicked (ResolveIssue_Button)"!() -> (OutputDelegate, then, Button:Common Button Base Object Reference) {is=title K2Node_ComponentBoundEvent}
   > n4 = QueueModalWidgetClassAsync!(
    execute:Exec
    WidgetClass:Common Activatable Widget Class Reference=/Crew/UI/Subscription/SubscriptionResolveIssueModal.SubscriptionResolveIssueModal_C
   ) -> (then:Exec, OnModalWidgetQueued:Exec, ActivatableWidget:Common Activatable Widget Object Reference) {is=title K2Node_AsyncAction}
   n5 = Cast To SubscriptionResolveIssueModal!(execute=@n4.OnModalWidgetQueued, Object=@n4) -> (then, CastFailed, AsSubscription Resolve Issue Modal:Subscription Resolve Issue Modal Object Reference) {is=title K2Node_DynamicCast}
   > Set CrewSubscriptionModalInfoVM!(
    execute
    CrewSubscriptionModalInfoVM:Crew Subscription Modal Info VM Object Reference=GetModalViewModelById!(
     self:Crew VM Object Reference=Get CrewVM() -> (CrewVM:Crew VM Object Reference)
     ModalId:String=Get ModalId!(self:Subscription Modal Base Object Reference=@n5) -> (ModalId:String)
    ) -> (ReturnValue:Crew Subscription Modal Info VM Object Reference)
    self:Subscription Resolve Issue Modal Object Reference=@n5
   ) -> (then, Output_Get:Crew Subscription Modal Info VM Object Reference)
   Opening() {is=title K2Node_CustomEvent}
   > PlayAnimation(InAnimation=Get Intro(), PlaybackSpeed=2)
   Event Construct()
   > SetMobileStyle() {is=title K2Node_CallFunction}
   Event PreConstruct()
   Event Tick()
  ]}
 HandleSubState {nodes [
   n1 = HandleSubState!() -> (then, SubState:EMcpSubscriptionState Enum=Inactive) {is=title K2Node_FunctionEntry}
   > n2 = Switch on EMcpSubscriptionState!(execute, Selection:EMcpSubscriptionState Enum=@n1) -> (
    Inactive:Exec
    Active:Exec
    ActiveWithExternalBilling:Exec
    NotAutoRenewable:Exec
    Canceled:Exec
    PaymentProcessError:Exec
    PlatformAuthError:Exec
    RewardsDelayed:Exec
    Unavailable:Exec
    Unknown:Exec
   ) {is=title K2Node_SwitchEnum}
   n3 = SetVisibility(
    execute=[@n2.Active, @n2.Inactive, @n2.PlatformAuthError, @n2.RewardsDelayed, @n2.Unavailable, @n2.Unknown]
    self=Get PriceNextBill() -> (PriceNextBill:Common Text Object Reference)
    InVisibility=Collapsed
   )
   SetVisibility(execute=[@n3.then, @n4.then], self=Get ManageSub_Button(), InVisibility=SelfHitTestInvisible)
   > SetVisibility(self=[Get ViewCrew_Button_1(), Get ResolveIssue_Button()], InVisibility=Collapsed)
   "SetText (Text)"(execute=@n2.ActiveWithExternalBilling, self=Get PriceNextBill(), InText="LOCTEXT(\"Check your subscription platform for details.\")")
   > n4 = SetVisibility(self=Get PriceNextBill(), InVisibility=SelfHitTestInvisible)
   SetVisibility(execute=[@n2.Canceled, @n2.NotAutoRenewable], self=Get ViewCrew_Button_1(), InVisibility=SelfHitTestInvisible)
   > n5 = SetVisibility(self=[Get ManageSub_Button(), Get ResolveIssue_Button()], InVisibility=Collapsed)
   SetVisibility(execute=[@n5.then, @n6.then], self=Get PriceNextBill(), InVisibility=Collapsed)
   SetVisibility(execute=@n2.PaymentProcessError, self=[Get ManageSub_Button(), Get ResolveIssue_Button()], InVisibility=SelfHitTestInvisible)
   > n6 = SetVisibility(self=Get ViewCrew_Button_1(), InVisibility=Collapsed)
  ]}
 HeaderEnteredSubscreen {nodes [HeaderEnteredSubscreen() {is=title K2Node_FunctionEntry}]}
 IsSubscribedState {nodes [
   n1 = IsSubscribedState!() -> (then, isSubbed:bool=~) {is=title K2Node_FunctionEntry}
   > SetActiveWidgetIndex(self:Common Visibility Switcher Object Reference=Get HeaderSwitcher(), Index="To Integer (Boolean)"!(InBool:bool=@n1) -> (ReturnValue:int=0))
  ]}
 SetMobileStyle {nodes [
   SetMobileStyle() {is=title K2Node_FunctionEntry}
   > Branch(Condition=IsMobileGame())
   n1 = Get Text_Description() -> (Text_Description:Common Text Object Reference)
   > SetAutoWrapText!(execute, self:Text Object Reference=@n1, InAutoTextWrap:bool=false) -> (then)
   > SetWrapTextWidth!(execute, self:Common Text Object Reference=@n1, InWrapTextAt:int=1024) -> (then)
  ]}
]
variables [HeaderEnteredSubscreen {}]
widgets {
 info {
  inheritedWidgetCount=0
  parentClass=&CommonActivatableWidget
  rootWidgetClass=&CommonVisibilitySwitcher
  widgetCount=16
 }
 widgets [
  HeaderSwitcher {widgetClassPath=&CommonVisibilitySwitcher, bIsVariable=true, widget=HeaderSwitcher, properties {bOverride_Cursor=true}}
  VerticalBox_55 {widgetClassPath=&VerticalBox, parent=HeaderSwitcher, widget=VerticalBox_55, slot {class=CommonVisibilitySwitcherSlot}}
  Pass_Badge {
   widgetClassPath=&FortLazyImage
   bIsVariable=true
   parent=VerticalBox_55
   widget=Pass_Badge
   properties {brush {
     bIsDynamicallyLoaded=false
     drawAs=Image
     imageType=FullColor
     mirroring=NoMirror
     resourceName=None
     resourceObject=/Game/Athena/UI/Logos/Textures/M_UI_CrewLogo_Large
     tiling=NoTile
     imageSize=<432 128>
     margin=<ltrb 0 0 0 0>
     outlineSettings {
      bUseBrushTransparency=false
      roundingType=HalfHeightRadius
      width=0
      color {colorUseRule=UseColor_Specified, specifiedColor=<rgba 0 0 0 0>}
      cornerRadii=<0 0 0 0>
     }
     tintColor {colorUseRule=UseColor_Specified, specifiedColor=<rgba 1 1 1 1>}
     uVRegion {bIsValid=false, max=<0 0>, min=<0 0>}
    }}
   slot {class=VerticalBoxSlot, properties {horizontalAlignment=HAlign_Center, verticalAlignment=VAlign_Top, padding=<ltrb 0 0 0 24>}}
  }
  WBP_UIKit_Highlight {
   widgetClassPath=/FortUILibrary/UIKit/Highlight/WBP_UIKit_Highlight
   bIsVariable=true
   parent=VerticalBox_55
   widget=WBP_UIKit_Highlight
   properties {highlightText="Get more with Fortnite Crew!", highlightType=Deal, maxWidth=600, highlightPadding_Default=<ltrb 20 0 20 0>}
   slot {class=VerticalBoxSlot, properties {horizontalAlignment=HAlign_Center, verticalAlignment=VAlign_Top, padding=<ltrb 0 0 0 48>}}
  }
  Text_Description {
   widgetClassPath=&CommonTextBlock
   bIsVariable=true
   parent=VerticalBox_55
   widget=Text_Description
   properties {
    justification=Center
    mobileFontSizeMultiplier=0.75
    scrollStyle=/FortUILibrary/Styles/Typography/Foundation/TextScrollStyle-NoFade.TextScrollStyle-NoFade_C
    text=Crew description text
    textTransformPolicy=ToUpper
    wrapTextAt=700
    colorAndOpacity {colorUseRule=UseColor_Specified, specifiedColor=<rgba 1 1 1 0.75>}
    font {
     bForceMonospaced=false
     bMaterialIsStencil=false
     fontMaterial=null
     fontObject=&HeadingNow
     letterSpacing=-40
     monospacedWidth=1
     size=51.75
     skewAmount=0
     typefaceFontName=86 Bold Italic
     outlineSettings {
      bApplyOutlineToDropShadows=false
      bMiteredCorners=false
      bSeparateFillAlpha=false
      outlineMaterial=null
      outlineSize=0
      outlineColor=<rgba 0 0 0 1>
     }
    }
    shadowOffset=<0 0>
   }
   slot {class=VerticalBoxSlot, properties {horizontalAlignment=HAlign_Center, padding=<ltrb 0 0 0 48>}}
  }
  ViewCrew_Button {
   widgetClassPath=/FortUILibrary/UIKit/Buttons/WBP_UIKit_Button_Loud
   bIsVariable=true
   parent=VerticalBox_55
   widget=ViewCrew_Button
   properties {bHideInputAction=true, minWidth=458, text=View Offer}
   slot {class=VerticalBoxSlot, properties {horizontalAlignment=HAlign_Center}}
  }
  VerticalBox_28 {widgetClassPath=&VerticalBox, parent=HeaderSwitcher, widget=VerticalBox_28, slot {class=CommonVisibilitySwitcherSlot}}
  Pass_Badge_1 {
   widgetClassPath=&FortLazyImage
   bIsVariable=true
   parent=VerticalBox_28
   widget=Pass_Badge_1
   properties {brush {
     bIsDynamicallyLoaded=false
     drawAs=Image
     imageType=FullColor
     mirroring=NoMirror
     resourceName=None
     resourceObject=/Game/Athena/UI/Logos/Textures/T_Crew_Workmark_L
     tiling=NoTile
     imageSize=<432 128>
     margin=<ltrb 0 0 0 0>
     outlineSettings {
      bUseBrushTransparency=false
      roundingType=HalfHeightRadius
      width=0
      color {colorUseRule=UseColor_Specified, specifiedColor=<rgba 0 0 0 0>}
      cornerRadii=<0 0 0 0>
     }
     tintColor {colorUseRule=UseColor_Specified, specifiedColor=<rgba 1 1 1 1>}
     uVRegion {bIsValid=false, max=<0 0>, min=<0 0>}
    }}
   slot {class=VerticalBoxSlot, properties {horizontalAlignment=HAlign_Center, verticalAlignment=VAlign_Top}}
  }
  Spacer {widgetClassPath=&Spacer, parent=VerticalBox_28, widget=Spacer, properties {size=<1 40>}, slot {class=VerticalBoxSlot}}
  SubTag {
   widgetClassPath=/FortUILibrary/UIKit/Tag/WBP_UIKit_Tag
   bIsVariable=true
   parent=VerticalBox_28
   widget=SubTag
   properties {labelText="", tagText=SUBSCRIBED, tagType=Primary Tag, colorAndOpacity=<rgba 0.0395 1 0.2307 1>}
   slot {class=VerticalBoxSlot, properties {horizontalAlignment=HAlign_Center}}
  }
  Spacer_80 {widgetClassPath=&Spacer, parent=VerticalBox_28, widget=Spacer_80, properties {size=<1 120>}, slot {class=VerticalBoxSlot}}
  HorizontalBox_0 {
   widgetClassPath=&HorizontalBox
   parent=VerticalBox_28
   widget=HorizontalBox_0
   slot {class=VerticalBoxSlot, properties {horizontalAlignment=HAlign_Center}}
  }
  ViewCrew_Button_1 {
   widgetClassPath=&WBP_UIKit_Button_Loud
   bIsVariable=true
   parent=HorizontalBox_0
   widget=ViewCrew_Button_1
   properties {bHideInputAction=true, minWidth=458, text=View Offer}
   slot {class=HorizontalBoxSlot}
  }
  ResolveIssue_Button {
   widgetClassPath=&WBP_UIKit_Button_Loud
   bIsVariable=true
   parent=HorizontalBox_0
   widget=ResolveIssue_Button
   properties {bHideInputAction=true, minWidth=304, text=Resolve Issue}
   slot {class=HorizontalBoxSlot, properties {padding=<ltrb 0 0 16 0>}}
  }
  ManageSub_Button {
   widgetClassPath=&WBP_UIKit_Button_Regular
   bIsVariable=true
   parent=HorizontalBox_0
   widget=ManageSub_Button
   properties {bHideInputAction=true, minHeight=56, minWidth=425, text=Manage Subscription}
   slot {class=HorizontalBoxSlot}
  }
  PriceNextBill {
   widgetClassPath=&CommonTextBlock
   bIsVariable=true
   parent=VerticalBox_28
   widget=PriceNextBill
   properties {
    text=You will be charged on your next billing cycle.
    wrapTextAt=700
    colorAndOpacity {colorUseRule=UseColor_Specified, specifiedColor=<rgba 0.5776 0.5776 0.5776 1>}
    font {
     bForceMonospaced=false
     bMaterialIsStencil=false
     fontMaterial=null
     fontObject=&HeadingNow
     letterSpacing=0
     monospacedWidth=1
     size=11.25
     skewAmount=0
     typefaceFontName=74 Regular
     outlineSettings {
      bApplyOutlineToDropShadows=false
      bMiteredCorners=false
      bSeparateFillAlpha=false
      outlineMaterial=null
      outlineSize=0
      outlineColor=<rgba 0 0 0 1>
     }
    }
    shadowOffset=<0 0>
   }
   slot {class=VerticalBoxSlot, properties {horizontalAlignment=HAlign_Center, verticalAlignment=VAlign_Top, padding=<ltrb 0 16 0 0>}}
  }
 ]
}
