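# WBP_CrewLandingPage is the activatable Fortnite Crew landing page, showing a Crew tile beside a scrolling row of pass tiles.
# Crew view-model benefits populate that row, filtered to figment, music and LEGO passes.
# Tile selection switches between Crew and Season Pass headers, while clicks broadcast subscreen delegates and record focus to restore.
# It also handles mouse-and-keyboard scroll arrows, RTL, mobile padding, an intro animation and a multiple-subscription warning.
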WidgetBlueprint /SeasonPassHub/Widgets/Reboot/WBP_CrewLandingPage
animations [TilesIntro {
  binding_count=2
  length_seconds=0.35
  object=&self:TilesIntro
  sequence {bindings [CrewTiles {
     id=#1
     possessed_class=HorizontalBox
     tracks [
      {
       type=MovieSceneFloatTrack
       display_name=Render Opacity
       sections [{type=MovieSceneFloatSection, end_frame=6, start_frame=-1, channels [None {type=MovieSceneScriptingFloatChannel, keys [[2,0],[5.0003,1]]}]}]
      }
      {
       type=MovieScene2DTransformTrack
       sections [{
         type=MovieScene2DTransformSection
         end_frame=6
         start_frame=-1
         channels [
          Translation.X {type=MovieSceneScriptingFloatChannel, const [0,2]}
          Translation.Y {type=MovieSceneScriptingFloatChannel, keys [[2,140],[5.0003,0]]}
          Angle {type=MovieSceneScriptingFloatChannel, const [0,1]}
          Scale.X {type=MovieSceneScriptingFloatChannel, const [1,1]}
          Scale.Y {type=MovieSceneScriptingFloatChannel, const [1,1]}
          Shear.X {type=MovieSceneScriptingFloatChannel, const [0,1]}
          Shear.Y {type=MovieSceneScriptingFloatChannel, const [0,1]}
         ]
        }]
      }
     ]
    }]}
 }]
graphs [
 BP_GetDesiredFocusTarget {nodes [
   Get Desired Focus Target()
   n1 = Get Crew_Header() -> (Crew_Header:WBP Crew Header Object Reference)
   n2 = GetAllEntries!(self:Dynamic Entry Box Base Object Reference=Get Passes() -> (Passes:Dynamic Entry Box Object Reference)) -> (ReturnValue:Array of User Widget Object References)
   n3 = AND Boolean(
    A=Get HasEnteredCrewPassSubscreen() -> (HasEnteredCrewPassSubscreen:bool)
    B=Is Not Empty!(TargetArray:Array of User Widget Object References=@n2) -> (ReturnValue:bool)
   )
   n4 = "Get (a copy)"(Array:Array of User Widget Object References=@n2, Dimension 1=Get SelectedPassTileIndex() -> (SelectedPassTileIndex:int=0)) -> (Output:User Widget Object Reference)
   n5 = Get ReturnFromWidget() -> (ReturnFromWidget:Widget Object Reference)
   > Return Node!(
    execute
    ReturnValue:Widget Object Reference=Select(
     Option 0:Widget Object Reference=Select(Option 0=@n5, Option 1=@n4, Index=@n3)
     Option 1:Widget Object Reference=GetDesiredFocusTarget!(self:Common Activatable Widget Object Reference=@n1) -> (ReturnValue:Widget Object Reference)
     Index=Get IsPageNav() -> (IsPageNav:bool)
    ) -> (ReturnValue:Widget Object Reference)
   )
  ]}
 CrewPageActivated {nodes [CrewPageActivated() {is=title K2Node_FunctionEntry}]}
 CrewPassTileSelected {nodes [
   n1 = CrewPassTileSelected!() -> (then, ButtonSelected:WBP Crew Pass Button Object Reference, SeasonPassVM:Season Pass VM Object Reference) {is=title K2Node_FunctionEntry}
   > n2 = Set SelectedPassButton(SelectedPassButton:WBP Crew Pass Button Object Reference=@n1) -> (Output_Get:WBP Crew Pass Button Object Reference)
   > Call Selected Crew Pass Tile!(
    execute:Exec
    self:WBP Crew Landing Page Object Reference
    SeasonPassVM:Season Pass VM Object Reference=@n1.SeasonPassVM
    ButtonSelected:WBP Crew Pass Button Object Reference=@n2
   ) -> (then:Exec) {is=title K2Node_CallDelegate}
  ]}
 EnteredCrewSubscreen {nodes [EnteredCrewSubscreen() {is=title K2Node_FunctionEntry}]}
 EnteredPassSubscreen {nodes [EnteredPassSubscreen!() -> (then, PassVM:Season Pass VM Object Reference) {is=title K2Node_FunctionEntry}]}
 EventGraph {nodes [
   "Has Entered Buy Rewards Modal (SeasonPass_Header)"() {is=title K2Node_ComponentBoundEvent}
   > Set IsPageNav(IsPageNav:bool=false) -> (Output_Get:bool)
   > Set ReturnFromWidget(ReturnFromWidget:Widget Object Reference=Get BuyRewards_Button!(self:Season Pass Header Object Reference=Get SeasonPass_Header() -> (SeasonPass_Header:Season Pass Header Object Reference)) -> (BuyRewards_Button:WBP UIKit Button Loud Object Reference)) -> (Output_Get:Widget Object Reference)
   "On Clicked (ScrollRight)"!() -> (OutputDelegate, then, Button:Common Button Base Object Reference) {is=title K2Node_ComponentBoundEvent}
   n1 = Get ScrollBox() -> (ScrollBox:Athena Scroll Box Object Reference)
   n2 = GetScrollOffsetOfEnd!(self:Scroll Box Object Reference=@n1) -> (ReturnValue:float)
   n3 = GetScrollOffset!(self:Scroll Box Object Reference=@n1) -> (ReturnValue:float)
   n4 = float + float(A=@n3, B=Get ButtonScrollAmount() -> (ButtonScrollAmount:double))
   n5 = Get ScrollBox()
   > SetScrollOffset!(execute, self:Scroll Box Object Reference=@n5, NewScrollOffset:float=SelectFloat(A=@n2, B=@n4, bPickA="float >= float"(A=@n4, B=@n2))) -> (then)
   > n6 = Branch(Condition="Equal (Float)"(A=@n3, B=@n2))
   > n7 = SetVisibility(self=Get ScrollRight() -> (ScrollRight:WBP UIKit Button Regular Object Reference), InVisibility=Collapsed)
   SetVisibility(
    execute=[@n7.then, @n6.else]
    self=Get ScrollLEft() -> (ScrollLEft:WBP UIKit Button Regular Object Reference)
    InVisibility=SelfHitTestInvisible
   )
   "On Clicked (ScrollLEft)"!() -> (OutputDelegate, then, Button:Common Button Base Object Reference) {is=title K2Node_ComponentBoundEvent}
   n8 = GetScrollOffset(self=Get ScrollBox())
   n9 = float - float(A=@n8, B=Get ButtonScrollAmount())
   > SetScrollOffset(self=@n5, NewScrollOffset=SelectFloat(A=0, B=@n9, bPickA="float <= float"(A=@n9) -> (ReturnValue=~)))
   > n10 = Branch(Condition="Equal (Float)"(A=@n8))
   > n11 = SetVisibility(self=Get ScrollLEft(), InVisibility=Collapsed)
   SetVisibility(execute=[@n11.then, @n10.else], self=Get ScrollRight(), InVisibility=SelfHitTestInvisible)
   n12 = CrewPassButtonClicked!() -> (OutputDelegate, then, Button:WBP Crew Pass Button Object Reference, Index:int, SeasonPassVM:Season Pass VM Object Reference) {is=title K2Node_CustomEvent}
   > n13 = Branch(Condition=Get IsCrewTile!(self:WBP Crew Pass Button Object Reference=@n12.Button) -> (IsCrewTile:bool))
   > Set IsPageNav(IsPageNav=false)
   > Set ReturnFromWidget(ReturnFromWidget=Get Crew_Header())
   > Call Entered Crew Subscreen!(execute:Exec, self:WBP Crew Landing Page Object Reference) -> (then:Exec) {is=title K2Node_CallDelegate}
   Set SelectedPassTileIndex(execute=@n13.else, SelectedPassTileIndex:int=@n12.Index) -> (Output_Get:int=0)
   > Set IsPageNav(IsPageNav=false)
   > Set ReturnFromWidget(ReturnFromWidget=Get SelectedPassButton() -> (SelectedPassButton:WBP Crew Pass Button Object Reference))
   > Call Entered Pass Subscreen!(execute:Exec, self:WBP Crew Landing Page Object Reference, PassVM:Season Pass VM Object Reference=@n12.SeasonPassVM) -> (then:Exec) {is=title K2Node_CallDelegate}
   n14 = OpeningTile!() -> (OutputDelegate, then, SeasonPassVM:Season Pass VM Object Reference, ButtonSelected:WBP Crew Pass Button Object Reference) {is=title K2Node_CustomEvent}
   > UpdateHeaderContent!(
    execute
    self
    IsCrewTile:bool=NOT Boolean(A=Get HasEnteredCrewPassSubscreen())
    SeasonPassVM:Season Pass VM Object Reference=@n14.SeasonPassVM
   ) -> (then)
   > Call Selected Crew Pass Tile(SeasonPassVM=@n14.SeasonPassVM, ButtonSelected=@n14.ButtonSelected)
   > PlayAnimationForward(InAnimation=Get TilesIntro() -> (TilesIntro:Widget Animation Object Reference))
   OnViewCrewHeaderClicked() {is=title K2Node_CustomEvent}
   > Set IsPageNav(IsPageNav=false)
   > Set ReturnFromWidget(ReturnFromWidget=Get Crew_Header())
   > Call Entered Crew Subscreen()
   n15 = CrewPassButtonSelected!() -> (OutputDelegate, then, SeasonPassVM:Season Pass VM Object Reference, ButtonSelected:WBP Crew Pass Button Object Reference) {is=title K2Node_CustomEvent}
   > Call Selected Crew Pass Tile(SeasonPassVM=@n15.SeasonPassVM, ButtonSelected=@n15.ButtonSelected)
   > UpdateHeaderContent(IsCrewTile=Get IsCrewTile(self=@n15.ButtonSelected), SeasonPassVM=@n15.SeasonPassVM)
   OnViewPassHeaderClicked() {is=title K2Node_CustomEvent}
   > Set IsPageNav(IsPageNav=false)
   > Set ReturnFromWidget(ReturnFromWidget=Get SelectedPassButton())
   > Call Entered Pass Subscreen()
   Event Construct()
   n16 = Get CommonInputSubsystem()
   > Bind Event to On Input Method Changed(self=@n16, Delegate=Create Event())
   > HandleInputChanged!(execute, self, Input:ECommonInputType Enum=GetCurrentInputType(self=@n16)) -> (then)
   > SetMobileStyle() {is=title K2Node_CallFunction}
   > n17 = Branch(Condition=IsCultureRightToLeft!(Culture:String=GetCurrentCulture!() -> (ReturnValue:String)) -> (ReturnValue:bool))
   > n18 = HandleRTL()
   Bind Event to Header Entered Subscreen(execute=[@n18.then, @n17.else], self:WBP Crew Header Object Reference=Get Crew_Header(), Delegate=Create Event()) {is=title K2Node_AddDelegate}
   > Bind Event to Header Entered Subscreen(self:Season Pass Header Object Reference=Get SeasonPass_Header(), Delegate=Create Event())
   > SetStatusIndicator!(
    execute
    self:WBP UIKit Info Box Object Reference=Get MultipleSubError() -> (MultipleSubError:WBP UIKit Info Box Object Reference)
    Icon:Object Reference=/FortUILibrary/Textures/MaterialInstances/MI_UIKit_Icon_ExclamationMark
    StatusType:E_UI_StatusIndicator_Type Enum=NewEnumerator4
    Config:E_UI_StatusIndicator_Configuration Enum=NewEnumerator0
   ) -> (then)
   > Call Crew Page Activated!(execute:Exec, self:WBP Crew Landing Page Object Reference) -> (then:Exec) {is=title K2Node_CallDelegate}
   Event On Activated()
   > DelayUntilNextTick!(execute) -> (then)
   > DelayUntilNextTick()
   n19 = GetScrollOffsetOfEnd(self=Get ScrollBox())
   > SetVisibility(
    self=Get ScrollRight()
    InVisibility=Select(Option 0:ESlateVisibility Enum=Collapsed, Option 1:ESlateVisibility Enum=SelfHitTestInvisible, Index="float > float"(A=@n19)) -> (ReturnValue:ESlateVisibility Enum)
   )
   > Branch(Condition=Get IsPageNav())
   > CrewPassButtonSelected!(
    execute
    self
    SeasonPassVM:Season Pass VM Object Reference
    ButtonSelected:WBP Crew Pass Button Object Reference=Get CrewPassButton() -> (CrewPassButton:WBP Crew Pass Button Object Reference)
   ) -> (then) {is=title K2Node_CallFunction}
  ]}
 HandleInputChanged {nodes [
   n1 = HandleInputChanged!() -> (then, Input:ECommonInputType Enum) {is=title K2Node_FunctionEntry}
   > n2 = Branch(Condition="Equal (Enum)"(A=@n1, B=MouseAndKeyboard))
   n3 = Get MKBScrollArrows() -> (MKBScrollArrows:Horizontal Box Object Reference)
   > SetVisibility(self=@n3, InVisibility=SelfHitTestInvisible)
   SetVisibility(execute=@n2.else, self=@n3, InVisibility=Collapsed)
  ]}
 HandleNewPasses {nodes [
   HandleNewPasses!() -> (then, Passes:Array of Season Pass VM Object References) {is=title K2Node_FunctionEntry}
   n1 = Get BenefitsGroups!(self:Crew VM Object Reference=Get CrewVM() -> (CrewVM:Crew VM Object Reference)) -> (BenefitsGroups:Array of Crew Benefits Group VM Object References)
   > Branch(Condition=Is Not Empty(TargetArray:Array of Crew Benefits Group VM Object References=@n1))
   > n2 = Set Benefits(Benefits:Array of Crew Benefits Group VM Object References=@n1) -> (Output_Get:Array of Crew Benefits Group VM Object References)
   > SetupPassesList!(execute, self, Benefits:Array of Crew Benefits Group VM Object References=@n2) -> (then)
   > SetCrewTile!(execute, self, TileImageURL:String=Get TileImageURL!(self:Crew VM Object Reference=Get CrewVM()) -> (TileImageURL:String)) -> (then)
  ]}
 HandleRTL {nodes [
   HandleRTL() {is=title K2Node_FunctionEntry}
   > Set FlowDirectionPreference!(
    execute
    FlowDirectionPreference:EFlowDirectionPreference Enum=RightToLeft
    self:Widget Object Reference=Get Content() -> (Content:Vertical Box Object Reference)
   ) -> (then, Output_Get:EFlowDirectionPreference Enum=Inherit)
  ]}
 NavigateToHeaderButton {nodes [
   NavigateToHeaderButton!() -> (then, Navigation:EUINavigation Enum) {is=title K2Node_FunctionEntry}
   n1 = Get SeasonPass_Header()
   n2 = Get BuyRewards_Button(self=@n1)
   n3 = Get bIsEnabled!(self:Widget Object Reference=@n2) -> (bIsEnabled:bool)
   n4 = Get GIftPass_Button!(self:Season Pass Header Object Reference=@n1) -> (GIftPass_Button:WBP UIKit Button Regular Object Reference)
   n5 = Get SeasonPassVM() -> (SeasonPassVM:Season Pass VM Object Reference)
   > Return Node(ReturnValue=Select(
     Option 0:Widget Object Reference=Get PurchasePass_Button!(self:Season Pass Header Object Reference=@n1) -> (PurchasePass_Button:WBP UIKit Button Loud Object Reference)
     Option 1:Widget Object Reference=Select(Option 0=@n4, Option 1=@n2, Index=@n3)
     Index=Get bDoesOwnsPass!(self:Season Pass VM Object Reference=@n5) -> (bDoesOwnsPass:bool)
    ) -> (ReturnValue:Widget Object Reference))
  ]}
 NavigateToPassTile {nodes [
   NavigateToPassTile!() -> (then, Navigation:EUINavigation Enum) {is=title K2Node_FunctionEntry}
   > Return Node(ReturnValue=Get CrewPassButton())
  ]}
 SelectedCrewPassTile {nodes [SelectedCrewPassTile!() -> (then, SeasonPassVM:Season Pass VM Object Reference, ButtonSelected:WBP Crew Pass Button Object Reference) {is=title K2Node_FunctionEntry}]}
 SetCrewImage {nodes [
   n1 = SetCrewImage!() -> (then, CrewTileForegroundImage:Texture Object Reference) {is=title K2Node_FunctionEntry}
   n2 = Get Background!(self:WBP Crew Pass Button Object Reference=Get CrewPassButton()) -> (Background:Fort Lazy Image Object Reference)
   > SetBrushFromMaterial(self=@n2, Material=/Crew/UI/Materials/M_UI_CategoryButton_Crew)
   > n3 = GetDynamicMaterial(self=@n2)
   > SetTextureParameterValue(self=@n3, ParameterName=ForegroundImage, Value=@n1)
  ]}
 SetCrewSubsciptionPlatforms {nodes [
   n1 = SetCrewSubsciptionPlatforms!() -> (then, Subbed App Stores:Array of EAppStore Enums=DebugStore) {is=title K2Node_FunctionEntry}
   n2 = Length(TargetArray:Array of EAppStore Enums=@n1)
   > SetVisibility(
    self=Get MultipleSubError()
    InVisibility=Select(Option 0:ESlateVisibility Enum=Collapsed, Option 1:ESlateVisibility Enum=SelfHitTestInvisible, Index="integer > integer"(A=@n2, B=1)) -> (ReturnValue:ESlateVisibility Enum)
   )
  ]}
 SetCrewTile {nodes [
   SetCrewTile!() -> (then, TileImageURL:String) {is=title K2Node_FunctionEntry}
   > Set IsCrewTile!(execute, IsCrewTile:bool=true, self:WBP Crew Pass Button Object Reference=Get CrewPassButton()) -> (then, Output_Get:bool)
   > Set Double Width!(execute, self:WBP Crew Pass Button Object Reference=Get CrewPassButton(), Is Double Width:bool=true) -> (then) {is=title K2Node_CallFunction}
   > Set Text(
    Text="LOCTEXT(\"Fortnite Crew\")"
    self=Get PassName!(self:WBP Crew Pass Button Object Reference=Get CrewPassButton()) -> (PassName:Common Text Object Reference)
   )
   n1 = Get Text_Subscription!(self:WBP Crew Pass Button Object Reference=Get CrewPassButton()) -> (Text_Subscription:Common Text Object Reference)
   > Set Text(Text="LOCTEXT(\"Subscription\")", self=@n1)
   > SetVisibility(self=@n1, InVisibility=SelfHitTestInvisible)
   n2 = Get TileImageURL!() -> (TileImageURL)
   > SetMediaURL!(
    execute
    self:Epic CMSImage Object Reference=Get EpicCMSImage!(self:WBP Crew Pass Button Object Reference=Get CrewPassButton()) -> (EpicCMSImage:Epic CMSImage Object Reference)
    MediaUrl:String=@n2
   ) -> (then)
   n3 = IsEmpty(InString=@n2)
   > n4 = Branch(Condition=@n3)
   > n5 = SetCrewImage!(execute, self, CrewTileForegroundImage:Texture Object Reference=/SeasonPass/Textures/Ch7S3/T_UI_Nexus_CrewTile_4100) -> (then) {is=title K2Node_CallFunction}
   SetVisibility(
    execute=[@n5.then, @n4.else]
    self=Get EpicCMSImage(self=Get CrewPassButton())
    InVisibility=Select(Option 0=Visible, Option 1=Collapsed, Index=@n3)
   )
   n6 = Get CrewPassButton()
   > SetVisibility(
    self=[
     Get SeasonPassBadgeIcon!(self:WBP Crew Pass Button Object Reference=@n6) -> (SeasonPassBadgeIcon:Season Pass Badge Icon Object Reference)
     Get ProgressVisibilitySwitcher!(self:WBP Crew Pass Button Object Reference=@n6) -> (ProgressVisibilitySwitcher:Common Visibility Switcher Object Reference)
    ]
    InVisibility=Collapsed
   )
   > Bind Event to Crew Pass Button Clicked(self:WBP Crew Pass Button Object Reference=Get CrewPassButton(), Delegate=Create Event()) {is=title K2Node_AddDelegate}
   > Bind Event to Crew Pass Button Selected(self:WBP Crew Pass Button Object Reference=Get CrewPassButton(), Delegate=Create Event()) {is=title K2Node_AddDelegate}
  ]}
 SetMobileStyle {nodes [
   SetMobileStyle() {is=title K2Node_FunctionEntry}
   > Branch(Condition=IsMobileGame())
   > SetClipping!(execute, self:Widget Object Reference=Get ScrollBox(), InClipping:EWidgetClipping Enum=ClipToBounds) -> (then)
   n1 = GetViewportSize!() -> (ReturnValue_X:double, ReturnValue_Y:double)
   n2 = float / float(A=@n1, B=@n1.ReturnValue_Y)
   > SetPadding!(
    execute
    self:Vertical Box Slot Object Reference=SlotAsVerticalBoxSlot(Widget=Get Overlay_Passes() -> (Overlay_Passes:Overlay Object Reference))
    InPadding_Left:float=0
    InPadding_Top:float=SelectFloat(A=24, B=256, bPickA="float >= float"(A=@n2, B=1.34))
    InPadding_Right:float=0
    InPadding_Bottom:float=0
   ) -> (then)
   n3 = GetViewportSize()
   n4 = float / float(A=@n3, B=@n3.ReturnValue_Y)
   > SetPadding(
    self=SlotAsVerticalBoxSlot(Widget=Get SizeBox_Message() -> (SizeBox_Message:Size Box Object Reference))
    InPadding_Left=0
    InPadding_Top=SelectFloat(A=-80, B=24, bPickA="float >= float"(A=@n4, B=1.78))
    InPadding_Right=0
    InPadding_Bottom=0
   )
   > SetPadding(
    self:Scroll Box Slot Object Reference=SlotAsScrollBoxSlot!(Widget:Widget Object Reference=Get Passes()) -> (ReturnValue:Scroll Box Slot Object Reference)
    InPadding_Left=8
    InPadding_Top=8
    InPadding_Right=8
    InPadding_Bottom=8
   )
  ]}
 SetupPassesList {nodes [
   SetupPassesList!() -> (then, Benefits:Array of Crew Benefits Group VM Object References) {is=title K2Node_FunctionEntry}
   > Reset!(execute, self:Dynamic Entry Box Object Reference=Get Passes(), bDeleteWidgets:bool=true) -> (then)
   > n1 = For Each Loop(Array:Array of Crew Benefits Group VM Object References=Get Benefits() -> (Benefits:Array of Crew Benefits Group VM Object References)) -> (Array Element:Crew Benefits Group VM Object Reference)
   n2 = Get Benefits!(self:Crew Benefits Group VM Object Reference=@n1) -> (Benefits:Array of Crew Benefit VM Object References)
   > n3 = For Loop(LastIndex=int - int(A=Length(TargetArray:Array of Crew Benefit VM Object References=@n2)))
   n4 = "Get (a copy)"(Array:Array of Crew Benefit VM Object References=@n2, Dimension 1=@n3) -> (Output:Crew Benefit VM Object Reference)
   n5 = GetTemplateId!(self:Crew Benefit VM Object Reference=@n4) -> (ReturnValue:String)
   > n6 = Branch(Condition=OR Boolean!(
     A="Equal Exactly (String)"(A=@n5, B=figmentpassdata)
     B="Equal Exactly (String)"(A=@n5, B=musicpassdata)
     C:bool="Equal Exactly (String)"(A=@n5, B=legopassdata)
    ) -> (ReturnValue))
   Set Benefit(execute=[@n6.then, @n8.CastFailed], Benefit:Crew Benefit VM Object Reference=@n4) -> (Output_Get:Crew Benefit VM Object Reference)
   > n7 = Create Entry!(execute, self:Dynamic Entry Box Object Reference=Get Passes()) -> (then, ReturnValue:User Widget Object Reference)
   > n8 = Cast To WBP_CrewPassButton!(execute, Object=@n7) -> (then, CastFailed, AsWBP Crew Pass Button:WBP Crew Pass Button Object Reference) {is=title K2Node_DynamicCast}
   > n9 = Set PassButton(PassButton:WBP Crew Pass Button Object Reference=@n8) -> (Output_Get:WBP Crew Pass Button Object Reference)
   > SetBenefit!(
    execute
    self:WBP Crew Pass Button Object Reference=@n9
    Benefit:Crew Benefit VM Object Reference=Get Benefit() -> (Benefit:Crew Benefit VM Object Reference)
   ) -> (then)
   n10 = Get ButtonIndex() -> (ButtonIndex:int=0)
   > Set ButtonIndex(ButtonIndex:int=Select(Option 0:int=0, Option 1:int=int + int(A=@n10), Index=Get IsNotFirstIndex() -> (IsNotFirstIndex:bool)) -> (ReturnValue:int)) -> (Output_Get:int=0)
   > Set Index!(
    execute
    Index:int=Get ButtonIndex() -> (ButtonIndex=0)
    self:WBP Crew Pass Button Object Reference=Get PassButton() -> (PassButton:WBP Crew Pass Button Object Reference)
   ) -> (then, Output_Get:int=0)
   > Set IsNotFirstIndex(IsNotFirstIndex:bool=true) -> (Output_Get:bool)
   > Bind Event to Crew Pass Button Clicked(self=Get PassButton(), Delegate=Create Event())
   > Bind Event to Crew Pass Button Selected(self=Get PassButton(), Delegate=Create Event())
  ]}
 UpdateHeaderContent {nodes [
   n1 = UpdateHeaderContent!() -> (then, IsCrewTile:bool=~, SeasonPassVM:Season Pass VM Object Reference) {is=title K2Node_FunctionEntry}
   > n2 = Branch(Condition=@n1)
   > SetActiveWidget(
    self:Widget Switcher Object Reference=Get HeaderSwitcher() -> (HeaderSwitcher:Common Activatable Widget Switcher Object Reference)
    Widget=Get Crew_Header()
   )
   > Opening(self:WBP Crew Header Object Reference=Get Crew_Header())
   SetSeasonPassVM!(
    execute=@n2.else
    self:Season Pass Header Object Reference=Get SeasonPass_Header()
    Viewmodel:Season Pass VM Object Reference=@n1.SeasonPassVM
   ) -> (then)
   > SetActiveWidget(self=Get HeaderSwitcher(), Widget=Get SeasonPass_Header())
   > Opening(self:Season Pass Header Object Reference=Get SeasonPass_Header())
  ]}
]
variables [
 ButtonScrollAmount {type=number, default=1600}
 CrewPageActivated {}
 EnteredCrewSubscreen {}
 EnteredPassSubscreen {}
 HasEnteredCrewPassSubscreen {type=boolean}
 IsPageNav {type=boolean}
 ReturnFromWidget {type=Widget, default=null}
 SelectedCrewPassTile {}
 SelectedPassButton {type=WBP_CrewPassButton_C, default=null}
 SelectedPassTileIndex {type=integer}
 ShouldUpdateHeaderContent {type=boolean}
]
widgets {
 info {inheritedWidgetCount=0, parentClass=&CommonActivatableWidget, rootWidgetClass=&VerticalBox, widgetCount=20}
 widgets [
  Content {widgetClassPath=&VerticalBox, bIsVariable=true, widget=Content}
  VerticalBox_0 {
   widgetClassPath=&VerticalBox
   parent=Content
   widget=VerticalBox_0
   slot {class=VerticalBoxSlot, properties {horizontalAlignment=HAlign_Center, padding=<ltrb 0 40 0 70>, size {sizeRule=Fill, value=1}}}
  }
  GridPanel_0 {
   widgetClassPath=&GridPanel
   parent=VerticalBox_0
   widget=GridPanel_0
   properties {columnFill [1], rowFill [1]}
   slot {class=VerticalBoxSlot, properties {padding=<ltrb 0 0 0 40>, size {sizeRule=Fill, value=1}}}
  }
  HeaderSwitcher {
   widgetClassPath=&CommonActivatableWidgetSwitcher
   bIsVariable=true
   parent=GridPanel_0
   widget=HeaderSwitcher
   properties {transitionDuration=0}
   slot {class=GridSlot, properties {columnSpan=1, rowSpan=1, verticalAlignment=VAlign_Center}}
  }
  Crew_Header {
   widgetClassPath=&/WBP_Crew_Header
   bIsVariable=true
   parent=HeaderSwitcher
   widget=Crew_Header
   slot {class=WidgetSwitcherSlot, properties {horizontalAlignment=HAlign_Left}}
  }
  SeasonPass_Header {
   widgetClassPath=/SeasonPassHub/Widgets/SeasonPass_Header
   bIsVariable=true
   parent=HeaderSwitcher
   widget=SeasonPass_Header
   slot {class=WidgetSwitcherSlot, properties {horizontalAlignment=HAlign_Left}}
  }
  Message {
   widgetClassPath=&FortVisualAttachment
   bIsVariable=true
   parent=GridPanel_0
   widget=Message
   properties {contentAnchor=<1 1>}
   slot {class=GridSlot, properties {columnSpan=1, horizontalAlignment=HAlign_Right, rowSpan=1, verticalAlignment=VAlign_Bottom}}
  }
  SizeBox_Message {
   widgetClassPath=&SizeBox
   bIsVariable=true
   parent=Message
   widget=SizeBox_Message
   properties {bOverride_WidthOverride=true, widthOverride=592}
   slot {class=SizeBoxSlot}
  }
  VerticalBox_Errors {widgetClassPath=&VerticalBox, parent=SizeBox_Message, widget=VerticalBox_Errors, slot {class=SizeBoxSlot}}
  MultipleSubError {
   widgetClassPath=/FortUILibrary/UIKit/InfoBox/WBP_UIKit_InfoBox
   bIsVariable=true
   parent=VerticalBox_Errors
   widget=MultipleSubError
   properties {
    previewText="You are currently subscribed on multiple platforms. To avoid being billed multiple times, we recommend canceling extra subscriptions. \r\n\r\nTo view which platforms you are currently subscribed on, please visit: \r\n\r\nhttps://fn.gg/check-crew"
    showIcon=false
    showStatusIndicator=true
    use Preview Text=true
    visibility=Collapsed
   }
   slot {class=VerticalBoxSlot}
  }
  CrewEntitlementMessage {
   widgetClassPath=/Crew/UI/CrewEntitlementMessage
   bIsVariable=true
   parent=VerticalBox_Errors
   widget=CrewEntitlementMessage
   properties {padding=<ltrb 0 24 0 0>}
   slot {class=VerticalBoxSlot}
  }
  CrewTiles {widgetClassPath=&HorizontalBox, bIsVariable=true, parent=VerticalBox_0, widget=CrewTiles, slot {class=VerticalBoxSlot}}
  CrewPassButton {
   widgetClassPath=&/WBP_CrewPassButton
   bIsVariable=true
   parent=CrewTiles
   widget=CrewPassButton
   slot {class=HorizontalBoxSlot}
  }
  Separator {
   widgetClassPath=&Image
   bIsVariable=true
   parent=CrewTiles
   widget=Separator
   properties {
    renderOpacity=0.25
    visibility=SelfHitTestInvisible
    brush {
     bIsDynamicallyLoaded=false
     drawAs=Image
     imageType=FullColor
     mirroring=NoMirror
     resourceName=None
     resourceObject=/FortUILibrary/Textures/Icons/T_UI_Icon_UIKit_ContextMenu
     tiling=NoTile
     imageSize=<32 32>
     margin=<ltrb 0 0 0 0>
     outlineSettings {
      bUseBrushTransparency=false
      roundingType=HalfHeightRadius
      width=0
      color {colorUseRule=UseColor_Specified, specifiedColor=<rgba 0 0 0 0>}
      cornerRadii=<0 0 0 0>
     }
     tintColor {colorUseRule=UseColor_Specified, specifiedColor=<rgba 1 1 1 1>}
     uVRegion {bIsValid=false, max=<0 0>, min=<0 0>}
    }
   }
   slot {class=HorizontalBoxSlot, properties {horizontalAlignment=HAlign_Center, verticalAlignment=VAlign_Center, padding=<ltrb 32 16 32 16>}}
  }
  Overlay_Passes {
   widgetClassPath=&Overlay
   bIsVariable=true
   parent=CrewTiles
   widget=Overlay_Passes
   properties {navigation=&self:WidgetTree.Overlay_Passes.WidgetNavigation_0}
   slot {class=HorizontalBoxSlot}
  }
  ScrollBox {
   widgetClassPath=/Script/FortniteUI.AthenaScrollBox
   bIsVariable=true
   parent=Overlay_Passes
   widget=ScrollBox
   properties {
    bAllowRightClickDragScrolling=false
    clipping=Inherit
    consumeMouseWheel=Never
    orientation=Orient_Horizontal
    scrollBarVisibility=Collapsed
    widgetStyle {
     barThickness=9
     bottomShadowBrush {
      bIsDynamicallyLoaded=false
      drawAs=NoDrawType
      imageType=FullColor
      mirroring=NoMirror
      resourceName=../../../Engine/Content/Slate/Common/ScrollBoxShadowBottom.png
      resourceObject=null
      tiling=NoTile
      imageSize=<16 8>
      margin=<ltrb 0.5 0 0.5 1>
      outlineSettings {
       bUseBrushTransparency=false
       roundingType=HalfHeightRadius
       width=0
       color {colorUseRule=UseColor_Specified, specifiedColor=<rgba 0 0 0 0>}
       cornerRadii=<0 0 0 0>
      }
      tintColor {colorUseRule=UseColor_Specified, specifiedColor=<rgba 1 1 1 1>}
      uVRegion {bIsValid=false, max=<0 0>, min=<0 0>}
     }
     horizontalScrolledContentPadding=<ltrb 0 0 1 0>
     leftShadowBrush {
      bIsDynamicallyLoaded=false
      drawAs=NoDrawType
      imageType=FullColor
      mirroring=NoMirror
      resourceName=../../../Engine/Content/Slate/Common/ScrollBoxShadowLeft.png
      resourceObject=null
      tiling=NoTile
      imageSize=<8 16>
      margin=<ltrb 1 0.5 0 0.5>
      outlineSettings {
       bUseBrushTransparency=false
       roundingType=HalfHeightRadius
       width=0
       color {colorUseRule=UseColor_Specified, specifiedColor=<rgba 0 0 0 0>}
       cornerRadii=<0 0 0 0>
      }
      tintColor {colorUseRule=UseColor_Specified, specifiedColor=<rgba 1 1 1 1>}
      uVRegion {bIsValid=false, max=<0 0>, min=<0 0>}
     }
     rightShadowBrush {
      bIsDynamicallyLoaded=false
      drawAs=NoDrawType
      imageType=FullColor
      mirroring=NoMirror
      resourceName=../../../Engine/Content/Slate/Common/ScrollBoxShadowRight.png
      resourceObject=null
      tiling=NoTile
      imageSize=<8 16>
      margin=<ltrb 0 0.5 1 0.5>
      outlineSettings {
       bUseBrushTransparency=false
       roundingType=HalfHeightRadius
       width=0
       color {colorUseRule=UseColor_Specified, specifiedColor=<rgba 0 0 0 0>}
       cornerRadii=<0 0 0 0>
      }
      tintColor {colorUseRule=UseColor_Specified, specifiedColor=<rgba 1 1 1 1>}
      uVRegion {bIsValid=false, max=<0 0>, min=<0 0>}
     }
     topShadowBrush {
      bIsDynamicallyLoaded=false
      drawAs=NoDrawType
      imageType=FullColor
      mirroring=NoMirror
      resourceName=../../../Engine/Content/Slate/Common/ScrollBoxShadowTop.png
      resourceObject=null
      tiling=NoTile
      imageSize=<16 8>
      margin=<ltrb 0.5 1 0.5 0>
      outlineSettings {
       bUseBrushTransparency=false
       roundingType=HalfHeightRadius
       width=0
       color {colorUseRule=UseColor_Specified, specifiedColor=<rgba 0 0 0 0>}
       cornerRadii=<0 0 0 0>
      }
      tintColor {colorUseRule=UseColor_Specified, specifiedColor=<rgba 1 1 1 1>}
      uVRegion {bIsValid=false, max=<0 0>, min=<0 0>}
     }
     verticalScrolledContentPadding=<ltrb 0 0 0 1>
    }
   }
   slot {class=OverlaySlot, properties {horizontalAlignment=HAlign_Fill, verticalAlignment=VAlign_Bottom}}
  }
  Passes {
   widgetClassPath=&DynamicEntryBox
   bIsVariable=true
   parent=ScrollBox
   widget=Passes
   properties {
    entryVerticalAlignment=VAlign_Bottom
    entryWidgetClass=&/WBP_CrewPassButton.WBP_CrewPassButton_C
    navigation=&self:WidgetTree.Passes.WidgetNavigation_0
    entrySpacing=<24 0>
   }
   slot {class=ScrollBoxSlot}
  }
  MKBScrollArrows {
   widgetClassPath=&HorizontalBox
   bIsVariable=true
   parent=Overlay_Passes
   widget=MKBScrollArrows
   properties {flowDirectionPreference=LeftToRight}
   slot {class=OverlaySlot, properties {horizontalAlignment=HAlign_Fill, verticalAlignment=VAlign_Center, padding=<ltrb 25 0 25 0>}}
  }
  ScrollLEft {
   widgetClassPath=&WBP_UIKit_Button_Regular
   bIsVariable=true
   parent=MKBScrollArrows
   widget=ScrollLEft
   properties {iconMaterial=/Crew/UI/Materials/MI_UI_ScrollArrowLeft, text="", visibility=Collapsed}
   slot {class=HorizontalBoxSlot}
  }
  ScrollRight {
   widgetClassPath=&WBP_UIKit_Button_Regular
   bIsVariable=true
   parent=MKBScrollArrows
   widget=ScrollRight
   properties {iconMaterial=/Crew/UI/Materials/MI_UI_ScrollArrowRight, text="", visibility=Collapsed}
   slot {class=HorizontalBoxSlot, properties {horizontalAlignment=HAlign_Right, size {sizeRule=Fill, value=1}}}
  }
 ]
}
