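# Season pass hub category tile that shows a character image, theme color and owned-reward progress, and opens the reward track screen.
# Hover, focus or touch press drives highlighting, which animates the tile and selects its category in the selection view model.
# When the category can be unlocked, the button requires a hold whose progress scrubs the SetUnlock animation, then runs an async unlock.
# Collapsed bonus rewards switch it to double width with a combined image.
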WidgetBlueprint /SeasonPassHub/Widgets/Reboot/WBP_SeasonPass_CharacterButton
animations [
 OnHover {
  binding_count=7
  length_seconds=0.15
  object=&self:OnHover
  sequence {bindings [
    OverlaySlot_3 {
     id=#1
     parent_id=#2
     possessed_class=OverlaySlot
     tracks [{
       type=MovieSceneMarginTrack
       display_name=Padding
       sections [{
         type=MovieSceneMarginSection
         end_frame=3
         start_frame=0
         channels [
          Left {type=MovieSceneScriptingFloatChannel, const [0,1]}
          Top {type=MovieSceneScriptingFloatChannel, keys [[0,96],[3,64]]}
          Right {type=MovieSceneScriptingFloatChannel, const [0,1]}
          Bottom {type=MovieSceneScriptingFloatChannel, const [0,1]}
         ]
        }]
      }]
    }
    CharacterImage {
     id=#3
     possessed_class=FortLazyImage
     tracks [{
       type=MovieSceneWidgetMaterialTrack
       sections [{type=MovieSceneParameterSection, end_frame=3, start_frame=0, channels [Hover {type=MovieSceneScriptingFloatChannel, keys [[0,0],[3,1]]}]}]
      }]
    }
    Button_Hitbox {id=#4, possessed_class=WBP_UIKit_Button_Generic_C, tracks []}
    Block_Outline {id=#5, possessed_class=WBP_UIKit_Block_Outline_C, tracks []}
    OverlaySlot_1 {
     id=#6
     parent_id=#5
     possessed_class=OverlaySlot
     tracks [{
       type=MovieSceneMarginTrack
       display_name=Padding
       sections [{
         type=MovieSceneMarginSection
         end_frame=3
         start_frame=0
         channels [
          Left {type=MovieSceneScriptingFloatChannel, const [0,2]}
          Top {type=MovieSceneScriptingFloatChannel, keys [[0,96],[3,64]]}
          Right {type=MovieSceneScriptingFloatChannel, const [0,2]}
          Bottom {type=MovieSceneScriptingFloatChannel, const [0,2]}
         ]
        }]
      }]
    }
    Background {id=#2, possessed_class=FortLazyImage, tracks []}
    OverlaySlot_10 {
     id=#7
     parent_id=#4
     possessed_class=OverlaySlot
     tracks [{
       type=MovieSceneMarginTrack
       display_name=Padding
       sections [{
         type=MovieSceneMarginSection
         end_frame=3
         start_frame=0
         channels [
          Left {type=MovieSceneScriptingFloatChannel, const [0,2]}
          Top {type=MovieSceneScriptingFloatChannel, keys [[0,120],[3,72]]}
          Right {type=MovieSceneScriptingFloatChannel, const [0,2]}
          Bottom {type=MovieSceneScriptingFloatChannel, const [0,2]}
         ]
        }]
      }]
    }
   ]}
 }
 Intro {
  binding_count=4
  length_seconds=0.45
  object=&self:Intro
  sequence {bindings [
    SizeBox {
     id=#8
     possessed_class=SizeBox
     tracks [
      {
       type=MovieScene2DTransformTrack
       sections [{
         type=MovieScene2DTransformSection
         end_frame=8
         start_frame=-1
         channels [
          Translation.X {type=MovieSceneScriptingFloatChannel, default=0}
          Translation.Y {type=MovieSceneScriptingFloatChannel, keys [[-1,65],[3,65.6371],[5,-33.1266],[8,10]]}
          Angle {type=MovieSceneScriptingFloatChannel, default=0}
          Scale.X {type=MovieSceneScriptingFloatChannel, default=1}
          Scale.Y {type=MovieSceneScriptingFloatChannel, default=1}
          Shear.X {type=MovieSceneScriptingFloatChannel, default=0}
          Shear.Y {type=MovieSceneScriptingFloatChannel, default=0}
         ]
        }]
      }
      {
       type=MovieSceneWidgetMaterialTrack
       sections [{type=MovieSceneParameterSection, end_frame=3, start_frame=0, channels [Hover {type=MovieSceneScriptingFloatChannel, const [1,1]}]}]
      }
     ]
    }
    Background {
     id=#9
     possessed_class=FortLazyImage
     tracks [{
       type=MovieSceneFloatTrack
       display_name=Render Opacity
       sections [{type=MovieSceneFloatSection, end_frame=4, start_frame=0, channels [None {type=MovieSceneScriptingFloatChannel, keys [[0,0.5],[4,1]]}]}]
      }]
    }
    CharacterImage {
     id=#10
     possessed_class=FortLazyImage
     tracks [
      {
       type=MovieSceneWidgetMaterialTrack
       sections [{
         type=MovieSceneParameterSection
         end_frame=4
         start_frame=0
         channels [CharacterPositionY {type=MovieSceneScriptingFloatChannel, keys [[1.6,-0.7],[3.9997,-0.5]]}]
        }]
      }
      {
       type=MovieSceneFloatTrack
       display_name=Render Opacity
       sections [{type=MovieSceneFloatSection, end_frame=4, start_frame=0, channels [None {type=MovieSceneScriptingFloatChannel, const [1,1]}]}]
      }
     ]
    }
    WBP_SeasonPass_CharacterButton_C_4 {
     id=#11
     possessed_class=WBP_SeasonPass_CharacterButton_C
     tracks [{
       type=MovieSceneFloatTrack
       display_name=Render Opacity
       sections [{type=MovieSceneFloatSection, end_frame=3, start_frame=0, channels [None {type=MovieSceneScriptingFloatChannel, const [1,1]}]}]
      }]
    }
   ]}
 }
 SetUnlock {
  binding_count=4
  length_seconds=1.2
  object=&self:SetUnlock
  sequence {bindings [
    SetUnlockAnimationContainer {
     id=#12
     possessed_class=Overlay
     tracks [{
       type=MovieSceneByteTrack
       display_name=Visibility
       sections [{type=MovieSceneByteSection, end_frame=24, start_frame=0, channels [None {type=MovieSceneScriptingByteChannel, keys [[0,1],[1,4],[24,1]]}]}]
      }]
    }
    CharacterImage {
     id=#13
     possessed_class=FortLazyImage
     tracks [{
       type=MovieSceneWidgetMaterialTrack
       sections [{
         type=MovieSceneParameterSection
         end_frame=24
         start_frame=0
         channels [Whiteout {type=MovieSceneScriptingFloatChannel, keys [[0,0],[15,0.1],[24,0]]}]
        }]
      }]
    }
    Overlay_1 {
     id=#14
     possessed_class=Overlay
     tracks [{
       type=MovieScene2DTransformTrack
       sections [{
         type=MovieScene2DTransformSection
         end_frame=24
         start_frame=0
         channels [
          Translation.X {
           type=MovieSceneScriptingFloatChannel
           keys [[0,0],[9,-0.6455],[12,0.0717],[14,-0.7889],[15,0.502],[16,-2.0081],[17,1.5061],[18,0],[23.9997,0]]
          }
          Translation.Y {
           type=MovieSceneScriptingFloatChannel
           keys [[0,0],[9,-3.0348],[12,-11.6931],[14,-11.058],[15,-19.1404],[17,-15.3142],[18,-23.0521],[21,0],[23.9997,0]]
          }
          Angle {type=MovieSceneScriptingFloatChannel, const [0,2]}
          Scale.X {type=MovieSceneScriptingFloatChannel, thinned_from=5, keys [[0,1],[18,1.06],[21,0.9],[24,1]]}
          Scale.Y {type=MovieSceneScriptingFloatChannel, thinned_from=5, keys [[0,1],[18,1.06],[21,0.9],[24,1]]}
          Shear.X {type=MovieSceneScriptingFloatChannel, const [0,2]}
          Shear.Y {type=MovieSceneScriptingFloatChannel, const [0,2]}
         ]
        }]
      }]
    }
    Glow {
     id=#15
     possessed_class=Image
     tracks [{
       type=MovieSceneWidgetMaterialTrack
       sections [{type=MovieSceneParameterSection, end_frame=24, start_frame=0, channels [Reveal {type=MovieSceneScriptingFloatChannel, keys [[0,0],[24,1]]}]}]
      }]
    }
   ]}
 }
]
graphs [
 BonusTileFocused {nodes [BonusTileFocused() {is=title K2Node_FunctionEntry}]}
 CF_DetermineProgressVisibility {nodes [
   n1 = CF_DetermineProgressVisibility!() -> (then, OwnedRewards:int, TotalRewards:int, IsLocked:bool=~, IsBonusRewardsCollapsed:bool=~) {is=title K2Node_FunctionEntry}
   n2 = "Equal (Integer)"(A=@n1, B=@n1.TotalRewards)
   > Return Node!(
    execute
    Vis:ESlateVisibility Enum=Select(
     Option 0=SelfHitTestInvisible
     Option 1=Collapsed
     Index=OR Boolean!(A=@n2, B=@n1.IsLocked, C:bool=@n1.IsBonusRewardsCollapsed) -> (ReturnValue)
    )
   )
  ]}
 ChangeBackgroundForBonus {nodes [
   n1 = ChangeBackgroundForBonus!() -> (then, GroupName:String) {is=title K2Node_FunctionEntry}
   > SetBrushFromMaterial(
    self=Get Background() -> (Background:Fort Lazy Image Object Reference)
    Material=Select(
     Option 0:Material Interface Object Reference=/SeasonPassHub/Widgets/Reboot/Materials/M_UI_LandingPage_Tile
     Option 1:Material Interface Object Reference=/SeasonPassHub/Widgets/Reboot/Materials/MI_UI_LandingPage_Tile_Bonus
     Index="Equal Exactly (String)"(A=@n1, B=Bonus)
    ) -> (ReturnValue:Material Interface Object Reference)
   )
  ]}
 EventGraph {nodes [
   "On Unhovered (Button_Hitbox)"!() -> (OutputDelegate, then, Button:Common Button Base Object Reference) {is=title K2Node_ComponentBoundEvent}
   > SetIsHighlighted!(execute, self, NewIsHighlighted:bool=false) -> (then)
   "On Hovered (Button_Hitbox)"!() -> (OutputDelegate, then, Button:Common Button Base Object Reference) {is=title K2Node_ComponentBoundEvent}
   > SetIsHighlighted(NewIsHighlighted=true)
   "On Focused (Button_Hitbox)"!() -> (OutputDelegate, then, Button:Common Button Base Object Reference) {is=title K2Node_ComponentBoundEvent}
   > SetIsHighlighted(NewIsHighlighted=true)
   n1 = GetGroupName!(self:Season Pass Element Base VM Object Reference=Get SeasonPassCategoryVM() -> (SeasonPassCategoryVM:Season Pass Category VM Object Reference)) -> (ReturnValue:String)
   > n2 = Branch(Condition=OR Boolean!(A=Get IsBonusRewardCollapsed() -> (IsBonusRewardCollapsed:bool), B="Equal Exactly (String)"(A=@n1, B=Bonus)) -> (ReturnValue))
   > Call Bonus Tile Focused!(execute:Exec, self:WBP Season Pass Character Button Object Reference) -> (then:Exec) {is=title K2Node_CallDelegate}
   Call Main Category Tile Focused!(execute:Exec=@n2.else, self:WBP Season Pass Character Button Object Reference) -> (then:Exec) {is=title K2Node_CallDelegate}
   "On Unfocused (Button_Hitbox)"!() -> (OutputDelegate, then, Button:Common Button Base Object Reference) {is=title K2Node_ComponentBoundEvent}
   > SetIsHighlighted(NewIsHighlighted=false)
   n3 = Set Double Width!() -> (OutputDelegate, then, Is Double Width:bool=~) {is=title K2Node_CustomEvent}
   > Set IsDoubleWidth(IsDoubleWidth:bool=@n3.Is Double Width) -> (Output_Get:bool)
   > SetWidthOverride(self=Get SizeBox() -> (SizeBox:Size Box Object Reference), InWidthOverride=SelectFloat(A=320, B=128, bPickA=@n3.Is Double Width))
   Pressed() {is=title K2Node_CustomEvent}
   > Branch(Condition=IsInputType!(InputType:ECommonInputType Enum=Touch, IsNot:bool=false) -> (Result:bool))
   > SetIsHighlighted(NewIsHighlighted=true)
   n4 = IntroAnim!() -> (OutputDelegate, then, Duration:float=~) {is=title K2Node_CustomEvent}
   > n5 = Set IntroDelay(IntroDelay:float=@n4.Duration) -> (Output_Get:float)
   n6 = PlayAnim() {is=title K2Node_CustomEvent}
   Set Timer by Event(execute=@n5.then, Delegate=@n6, Time=@n5, bLooping=false)
   StopQueuedAnims() {is=title K2Node_CustomEvent}
   > Check and Stop Animation!(execute:Exec, Animation:Widget Animation Object Reference=Get Intro()) -> (then:Exec) {is=title K2Node_MacroInstance}
   > DelayUntilNextTick!(execute) -> (then)
   > InstantItemcardTransition!(execute, self, Show:bool=true) -> (then)
   Released!() -> (OutputDelegate, then, Button:WBP UIKit Button Generic Base Object Reference) {is=title K2Node_CustomEvent}
   > n7 = Branch(Condition=IsInputType(InputType=Touch, IsNot=false))
   > n8 = SetIsHighlighted(NewIsHighlighted=false)
   Branch(
    execute=[@n8.then, @n7.else]
    Condition=Get bCanUnlock!(self:Season Pass Element Base VM Object Reference=Get SeasonPassCategoryVM()) -> (bCanUnlock:bool)
   )
   > TryOpenRewardTrack()
   OpenRewardTrack() {is=title K2Node_CustomEvent}
   > Set SelectedCategory!(
    execute
    SelectedCategory:Season Pass Category VM Object Reference=Get SeasonPassCategoryVM()
    self:Season Pass Selection VM Object Reference=Get SeasonPassSelectionVM() -> (SeasonPassSelectionVM:Season Pass Selection VM Object Reference)
   ) -> (then, Output_Get:Season Pass Category VM Object Reference)
   > Call Season Pass Category Entered!(execute:Exec, self:WBP Season Pass Character Button Object Reference) -> (then:Exec) {is=title K2Node_CallDelegate}
   > n9 = PushPrimaryContentClassAsync!(
    execute:Exec
    WidgetClass:Common Activatable Widget Class Reference=/SeasonPass/Widgets/Reboot/WBP_RewardTrack_Screen.WBP_RewardTrack_Screen_C
   ) -> (then:Exec, OnPrimaryContentPushed:Exec, ActivatableWidget:Common Activatable Widget Object Reference) {is=title K2Node_AsyncAction}
   n10 = GetCategoryPageTitle!(self:Season Pass Category VM Object Reference=Get SeasonPassCategoryVM()) -> (ReturnValue:Text)
   Fire Event Element Interacted(
    execute=@n9.OnPrimaryContentPushed
    Element=@n9
    AdditionalInfo="Add String (Non-Localized)"!(
     Info:"Flow Analytics Additional Info Structure (by ref)"=Make Flow Analytics Additional Info!() -> (FlowAnalyticsAdditionalInfo:Flow Analytics Additional Info Structure) {is=title K2Node_MakeStruct}
     Key:String=CategoryTitle
     Value:String=Get Source String from Text!(Text:Text=@n10) -> (ReturnValue:String) {is=title K2Node_CallFunction}
    ) -> (ReturnValue:Flow Analytics Additional Info Structure)
   )
   OnHoldCompleteButton() {is=title K2Node_CustomEvent}
   n11 = Get SeasonPassCategoryVM()
   > n12 = TryUnlockElement!(execute:Exec, PassElementVM:Season Pass Element Base VM Object Reference=@n11) -> (then:Exec, OnOperationCompleted:Exec, Success:bool=~) {is=title K2Node_AsyncAction}
   n13 = Get Button_Hitbox() -> (Button_Hitbox:WBP UIKit Button Generic Object Reference)
   > SetIsEnabled(self=@n13, bInIsEnabled=false)
   SetIsEnabled(execute=@n12.OnOperationCompleted, self=@n13, bInIsEnabled=true)
   > SetUnlockAnimationTime!(execute, self, AnimationNormalizedTime:double=0) -> (then)
   > Branch(Condition=@n12)
   > Call Unlocked Set!(execute:Exec, self:WBP Season Pass Character Button Object Reference, Category:Season Pass Category VM Object Reference=@n11) -> (then:Exec) {is=title K2Node_CallDelegate}
   OnInputMethodChanged!() -> (OutputDelegate, then, bNewInputType:ECommonInputType Enum) {is=title K2Node_CustomEvent}
   > UpdateButtonPromptVisibility()
   n14 = OnHoldActionProgressButton!() -> (OutputDelegate, then, HoldPercentage:double=~) {is=title K2Node_CustomEvent}
   > SetUnlockAnimationTime(AnimationNormalizedTime=@n14.HoldPercentage)
   OnClickedButton!() -> (OutputDelegate, then, Button:Common Button Base Object Reference) {is=title K2Node_CustomEvent}
   > TryOpenRewardTrack()
   Event Destruct()
   n15 = Create Event()
   n16 = Get Button_Hitbox()
   > Unbind Event from On Pressed(self:WBP UIKit Button Generic Object Reference=@n16, Delegate=@n15) {is=title K2Node_RemoveDelegate}
   n17 = Create Event()
   > Unbind Event from On Released(self:WBP UIKit Button Generic Base Object Reference=@n16, Delegate=@n17) {is=title K2Node_RemoveDelegate}
   n18 = Create Event()
   > Unbind Event from On Clicked(self:Common Button Base Object Reference=@n16, Delegate=@n18) {is=title K2Node_RemoveDelegate}
   n19 = Create Event()
   > Unbind Event from On Hold Action Progress(self:WBP UIKit Button Generic Base Object Reference=@n16, Delegate=@n19) {is=title K2Node_RemoveDelegate}
   n20 = Create Event()
   > Unbind Event from On Hold Completed(self:WBP UIKit Button Generic Base Object Reference=@n16, Delegate=@n20) {is=title K2Node_RemoveDelegate}
   n21 = Create Event()
   n22 = Get CommonInputSubsystem()
   > Unbind Event from On Input Method Changed(self=@n22, Delegate=@n21)
   Event Construct()
   n23 = Get Button_Hitbox()
   > Bind Event to On Pressed(self:WBP UIKit Button Generic Object Reference=@n23, Delegate=@n15) {is=title K2Node_AddDelegate}
   > Bind Event to On Released(self:WBP UIKit Button Generic Base Object Reference=@n23, Delegate=@n17) {is=title K2Node_AddDelegate}
   > Bind Event to On Clicked(self:Common Button Base Object Reference=@n23, Delegate=@n18) {is=title K2Node_AddDelegate}
   > Bind Event to On Hold Action Progress(self:WBP UIKit Button Generic Base Object Reference=@n23, Delegate=@n19) {is=title K2Node_AddDelegate}
   > Bind Event to On Hold Completed(self:WBP UIKit Button Generic Base Object Reference=@n23, Delegate=@n20) {is=title K2Node_AddDelegate}
   > Bind Event to On Input Method Changed(self=@n22, Delegate=@n21)
   > Set HoldTime(HoldTime:double=GetRequiredHoldTime!(self:Common Button Base Object Reference=Get Button_Hitbox()) -> (ReturnValue:float)) -> (Output_Get:double)
   n24 = Get Intro()
   Set Timer by Event(
    execute=@n25.then
    Delegate=Create Event()
    Time=float + float(A=Get IntroDelay() -> (IntroDelay:float), B=GetEndTime(self=@n24))
    bLooping=false
   )
   n25 = PlayAnimationForward(execute=@n6.then, InAnimation=Get Intro())
  ]}
 InstantItemcardTransition {nodes [
   InstantItemcardTransition!() -> (then, Show:bool=~) {is=title K2Node_FunctionEntry}
   n1 = SelectFloat(A=1, B=0, bPickA=Get Show() -> (Show:bool))
   > SetRenderOpacity(self=Get Background(), InOpacity=@n1)
   n2 = Get CharacterImage() -> (CharacterImage:Fort Lazy Image Object Reference)
   > SetRenderOpacity(self=@n2, InOpacity=@n1)
   > n3 = GetDynamicMaterial(self=@n2)
   > SetScalarParameterValue(self=@n3, ParameterName=CharacterPositionY, Value=SelectFloat(A=-0.5, B=-0.7, bPickA=Get Show()))
   > SetRenderTransform!(
    execute
    self:Widget Object Reference=Get SizeBox()
    InTransform_Translation_X:double=0
    InTransform_Translation_Y:double=SelectFloat(A=10, B=65, bPickA=Get Show())
    InTransform_Scale:Vector 2D Structure=(X=1,Y=1)
    InTransform_Shear:Vector 2D Structure=(X=0,Y=0)
    InTransform_Angle:float=0
   ) -> (then)
  ]}
 MainCategoryTileFocused {nodes [MainCategoryTileFocused() {is=title K2Node_FunctionEntry}]}
 OnCanUnlockChanged {nodes [
   n1 = OnCanUnlockChanged!() -> (then, NewCanUnlock:bool=~) {is=title K2Node_FunctionEntry}
   > SetRequiresHold!(execute, self:Common Button Base Object Reference=Get Button_Hitbox(), bInRequiresHold:bool=@n1) -> (then)
   > UpdatePrompt!(
    execute
    self:WBP UIKit Button Generic Base Object Reference=Get Button_Hitbox()
    InputType:ECommonInputType Enum=GetCurrentInputType(self=Get CommonInputSubsystem())
   ) -> (then)
   > UpdateButtonPromptVisibility()
   > Branch(Condition=@n1)
   > ResetLableSwitcher()
  ]}
 ResetCharacterImageBrush {nodes [
   ResetCharacterImageBrush() {is=title K2Node_FunctionEntry}
   n1 = Get CharacterImage()
   > SetBrushFromMaterial(self=@n1, Material=/SeasonPassHub/Widgets/Reboot/Materials/M_UI_LandingPage_Tile_Content)
   > SetMaterialTextureParamName!(execute, self:Common Lazy Image Object Reference=@n1, TextureParamName:Name=RewardTexture) -> (then)
  ]}
 ResetLableSwitcher {nodes [
   ResetLableSwitcher() {is=title K2Node_FunctionEntry}
   > Branch(Condition=Get CategoryInTandem() -> (CategoryInTandem:bool))
   > Set CategoryInTandem(CategoryInTandem:bool=false) -> (Output_Get:bool)
   n1 = Get LabelSwitcher() -> (LabelSwitcher:WBP Season Pass Tile Label Switcher Object Reference)
   > AddChild(self=Get Overlay_LabelSwitcher() -> (Overlay_LabelSwitcher:Overlay Object Reference), Content=@n1)
   > SetHorizontalAlignment(self=SlotAsOverlaySlot(Widget=@n1), InHorizontalAlignment=HAlign_Center)
   n2 = Get LabelSwitcher()
   > SetIsInTandem!(execute, self:WBP Season Pass Tile Label Switcher Object Reference=@n2, "InIsInTandem ":bool=false, Amount:int=0) -> (then)
   > SetVisibility(self=@n2, InVisibility=SelfHitTestInvisible)
  ]}
 SeasonPassCategoryEntered {nodes [SeasonPassCategoryEntered() {is=title K2Node_FunctionEntry}]}
 SetBonusRewardsCollapsed {nodes [
   n1 = SetBonusRewardsCollapsed!() -> (then, IsBonusRewardsCollapsed:bool=~) {is=title K2Node_FunctionEntry}
   > Set Double Width!(execute, self, Is Double Width:bool=@n1) -> (then) {is=title K2Node_CallFunction}
   n2 = Break Season Pass Theme Data!(SeasonPassThemeData:"Season Pass Theme Data Structure (by ref)"=GetThemeData!(self:Season Pass VM Object Reference=Get SeasonPassVM() -> (SeasonPassVM:Season Pass VM Object Reference)) -> (ReturnValue:"Season Pass Theme Data Structure (by ref)")) -> (
    TitleColor:Linear Color Structure
    TitleMaterial:Soft Object Reference
    TitleOutline:Font Outline Settings Structure
    SeasonColor:Linear Color Structure
    ProgressionColor:Linear Color Structure
    OverlayColor:Linear Color Structure
    VaultBackgroundMaterial:Soft Object Reference
    GiftboxVaultBackgroundMaterial:Soft Object Reference
    UIBackgroundMaterial:Soft Object Reference
    UpsellBackgroundMaterial:Soft Object Reference
    CrewCategoryBackgroundMaterial:Soft Object Reference
    CustomLogo:Soft Object Reference
    CustomSeasonTitle:Text
    CustomSeasonTitleColor:Linear Color Structure
    Upsell2BackgrounMaterial:Soft Object Reference
    ImageReferenceTable:Fort Image Reference Table Structure
    ColorReferenceTable:Fort Color Reference Table Structure
    TextReferenceTable:Fort Text Reference Table Structure
   ) {is=title K2Node_BreakStruct}
   n3 = TryLookupColor!(Table:"Fort Color Reference Table Structure (by ref)"=@n2.ColorReferenceTable, Name:String=CharacterColor) -> (OutColor:Linear Color Structure, ReturnValue:bool)
   > Set CharacterColor(CharacterColor:Linear Color Structure=Select(Option 0:Linear Color Structure=(R=0.0802,G=1,B=0,A=1), Option 1:Linear Color Structure=@n3, Index=@n3.ReturnValue) -> (ReturnValue:Linear Color Structure)) -> (Output_Get:Linear Color Structure)
   > Set with Broadcast IsBonusRewardCollapsed!(execute, IsBonusRewardCollapsed:bool=Get IsBonusRewardsCollapsed() -> (IsBonusRewardsCollapsed:bool)) -> (then, Output_Get:bool)
   n4 = TryLookupImage!(Table:"Fort Image Reference Table Structure (by ref)"=@n2.ImageReferenceTable, Name:String=BonusRewardsCombined) -> (OutImageReference:Fort Image Reference Structure, ReturnValue:bool)
   > Branch(Condition=@n4.ReturnValue)
   > Set CachedCharacterImageReference(CachedCharacterImageReference:Fort Image Reference Structure=@n4) -> (Output_Get:Fort Image Reference Structure)
   > ResetCharacterImageBrush() {is=title K2Node_CallFunction}
   > UpdateCharacterImageBrush()
  ]}
 SetCategoryThemeData {nodes [
   n1 = SetCategoryThemeData!() -> (then, ThemeData:Season Pass Category Theme Data Structure) {is=title K2Node_FunctionEntry}
   n2 = TryLookupColor(
    Table=Break Season Pass Category Theme Data!(SeasonPassCategoryThemeData:"Season Pass Category Theme Data Structure (by ref)"=@n1) -> (
     ImageReferenceTable:Fort Image Reference Table Structure
     ColorReferenceTable:Fort Color Reference Table Structure
     TextReferenceTable:Fort Text Reference Table Structure
    ) {is=title K2Node_BreakStruct}.ColorReferenceTable
    Name=CharacterColor
   )
   > Set CharacterColor(CharacterColor=Select(Option 0=(R=0.0802,G=1,B=0,A=1), Option 1=@n2, Index=@n2.ReturnValue))
   > UpdateProgressVisual()
  ]}
 SetInTandem {nodes [
   n1 = SetInTandem!() -> (then, InbShowLevel:bool=~, TandemAmount:int) {is=title K2Node_FunctionEntry}
   n2 = "integer > integer"(A=@n1.TandemAmount)
   > Set CategoryInTandem(CategoryInTandem=@n2)
   > Branch(Condition=@n2)
   n3 = Get LabelSwitcher()
   > SetVisibility(
    self=@n3
    InVisibility=Select(Option 0:ESlateVisibility Enum=Collapsed, Option 1:ESlateVisibility Enum=SelfHitTestInvisible, Index=@n1) -> (ReturnValue:ESlateVisibility Enum)
   )
   > SetIsInTandem(self=@n3, "InIsInTandem "=true, Amount=Get TandemAmount() -> (TandemAmount:int=0))
   > AddChild(self=Get Tandem_VisualAttachment() -> (Tandem_VisualAttachment:Fort Visual Attachment Object Reference), Content=Get LabelSwitcher())
  ]}
 SetIsHighlighted {nodes [
   n1 = SetIsHighlighted!() -> (then, NewIsHighlighted:bool=~) {is=title K2Node_FunctionEntry}
   > n2 = Branch(Condition="Not Equal (Boolean)"(A=@n1, B=Get IsHighlighted() -> (IsHighlighted:bool)) -> (ReturnValue=~))
   > n3 = Set IsHighlighted(IsHighlighted:bool=Get NewIsHighlighted() -> (NewIsHighlighted:bool)) -> (Output_Get:bool)
   > UpdateButtonPromptVisibility()
   > n4 = Branch(Condition=@n3)
   > RequestTransition!(
    execute
    self:UIKit Block Object Reference=Get Block_Outline() -> (Block_Outline:WBP UIKit Block Outline Object Reference)
    State:EUIKitBlockVisualState Enum=Focused
   ) -> (then)
   n5 = Get OnHover() -> (OnHover:Widget Animation Object Reference)
   > PlayAnimation(
    InAnimation=@n5
    StartAtTime=GetAnimationCurrentTime!(self:User Widget Object Reference, InAnimation:Widget Animation Object Reference=@n5) -> (ReturnValue:float)
   )
   > Set SelectedCategory(SelectedCategory=Get SeasonPassCategoryVM(), self=Get SeasonPassSelectionVM())
   > n6 = GetDynamicMaterial(self=Get CharacterImage())
   > SetVectorParameterValue(self=@n6, ParameterName=CharacterColor, Value=Get CharacterColor() -> (CharacterColor:Linear Color Structure))
   RequestTransition(execute=@n4.else, self=Get Block_Outline(), State=Unfocused)
   n7 = Get OnHover()
   > PlayAnimation(InAnimation=@n7, StartAtTime=GetAnimationCurrentTime(InAnimation=@n7), PlayMode=Reverse)
   UpdateButtonPromptVisibility(execute=@n2.else)
  ]}
 SetLazyCharacterImage {nodes [
   n1 = SetLazyCharacterImage!() -> (then, Image:Soft Object Reference) {is=title K2Node_FunctionEntry}
   > Branch(Condition=IsValidSoftObjectReference(SoftObjectReference=@n1))
   > Set CachedLazyCharacterImage(CachedLazyCharacterImage:Soft Object Reference=Get Image() -> (Image:Soft Object Reference)) -> (Output_Get:Soft Object Reference)
   > ResetCharacterImageBrush()
   > UpdateCharacterImageBrush()
  ]}
 SetOwnedRewardsCount {nodes [
   n1 = SetOwnedRewardsCount!() -> (then, Owned:int) {is=title K2Node_FunctionEntry}
   > Set OwnedRewardsCount(OwnedRewardsCount:int=@n1) -> (Output_Get:int=0)
   > UpdateProgressVisual()
  ]}
 SetProgressionColor {nodes [
   n1 = SetProgressionColor!() -> (then, Color:Linear Color Structure) {is=title K2Node_FunctionEntry}
   > Set ProgressionColor(ProgressionColor:Linear Color Structure=@n1) -> (Output_Get:Linear Color Structure)
   > UpdateProgressVisual()
  ]}
 SetTotalRewardsCount {nodes [
   n1 = SetTotalRewardsCount!() -> (then, Total:int) {is=title K2Node_FunctionEntry}
   > Set TotalRewardsCount(TotalRewardsCount:int=@n1) -> (Output_Get:int=0)
   > UpdateProgressVisual()
  ]}
 SetUnlockAnimationTime {nodes [
   n1 = SetUnlockAnimationTime!() -> (then, AnimationNormalizedTime:double=~) {is=title K2Node_FunctionEntry}
   n2 = Get SetUnlock() -> (SetUnlock:Widget Animation Object Reference)
   n3 = "Clamp (Float)"(Value=float * float(A=@n1, B=Get HoldTime() -> (HoldTime:double)) {is=title K2Node_PromotableOperator}, Min=0, Max=Get HoldTime())
   > n4 = Branch(Condition="Nearly Equal (Float)"(A=@n3, B=GetAnimationCurrentTime(InAnimation=@n2), ErrorTolerance=0))
   PlayAnimation(execute=@n4.else, InAnimation=Get SetUnlock(), StartAtTime=@n3)
   > PauseAnimation!(execute, self:User Widget Object Reference, InAnimation:Widget Animation Object Reference=Get SetUnlock()) -> (then, ReturnValue:float)
  ]}
 TryOpenRewardTrack {nodes [
   TryOpenRewardTrack() {is=title K2Node_FunctionEntry}
   n1 = Get bCanUnlock(self=Get SeasonPassCategoryVM())
   n2 = "float < float"(A=GetAnimationCurrentTime(InAnimation=Get SetUnlock()), B=0.2)
   > n3 = Branch(Condition=OR Boolean(A=AND Boolean(A=@n2, B=@n1), B=NOT Boolean(A=@n1)))
   > OpenRewardTrack() {is=title K2Node_CallFunction}
   UpdateButtonPromptVisibility(execute=@n3.else)
  ]}
 UnlockSetVisuals {nodes [
   UnlockSetVisuals() {is=title K2Node_FunctionEntry}
   > SetVisibility(self=Get SetUnlockAnimationContainer() -> (SetUnlockAnimationContainer:Overlay Object Reference), InVisibility=SelfHitTestInvisible)
   > PlayAnimation(InAnimation=Get SetUnlock())
  ]}
 UnlockedSet {nodes [UnlockedSet!() -> (then, Category:Season Pass Category VM Object Reference) {is=title K2Node_FunctionEntry}]}
 UpdateButtonPromptVisibility {nodes [
   UpdateButtonPromptVisibility() {is=title K2Node_FunctionEntry}
   n1 = Get SeasonPassCategoryVM()
   n2 = Select(Option 0=Collapsed, Option 1=Visible, Index=AND Boolean(A=Get IsHighlighted(), B=Get bCanUnlock(self=@n1)))
   > SetVisibility(
    self=Get Prompt!(self:WBP UIKit Button Generic Base Object Reference=Get Button_Hitbox()) -> (Prompt:Grid Panel Object Reference)
    InVisibility=@n2
   )
   > Set Visibility(Visibility=@n2, self=Get UnlockSetLabel() -> (UnlockSetLabel:Common Text Object Reference))
  ]}
 UpdateCharacterImageBrush {nodes [
   UpdateCharacterImageBrush() {is=title K2Node_FunctionEntry}
   > n1 = Branch(Condition=Get IsBonusRewardCollapsed())
   > SetBrushFromImageReference!(
    execute
    self:Fort Lazy Image Object Reference=Get CharacterImage()
    ImageReference:"Fort Image Reference Structure (by ref)"=Get CachedCharacterImageReference() -> (CachedCharacterImageReference:Fort Image Reference Structure)
    bMatchSize:bool=false
    DownloadPriority:ECMSDownloadPriority Enum=Normal
   ) -> (then)
   SetBrushFromLazyDisplayAsset!(
    execute=@n1.else
    self:Common Lazy Image Object Reference=Get CharacterImage()
    LazyObject:"Soft Object Reference (by ref)"=Get CachedLazyCharacterImage() -> (CachedLazyCharacterImage:Soft Object Reference)
    bMatchTextureSize:bool=false
   ) -> (then)
  ]}
 UpdateProgressVisual {nodes [
   UpdateProgressVisual() {is=title K2Node_FunctionEntry}
   > n1 = Sequence!(execute) -> (then_0, then_1, then_2:Exec)
   > n2 = Branch(Condition="integer > integer"(A=Get TotalRewardsCount() -> (TotalRewardsCount:int=0)))
   n3 = Get TotalRewardsCount() -> (TotalRewardsCount=0)
   n4 = Get OwnedRewardsCount() -> (OwnedRewardsCount:int=0)
   > n5 = SetProgress!(
    execute
    self:WBP UIKit Progress Bar Base Object Reference=Get ProgressBar() -> (ProgressBar:WBP UIKit Progress Bar Object Reference)
    Progress:double=float / float(A="To Float (Integer)"(InInt=@n4), B="To Float (Integer)"(InInt=@n3))
   ) -> (then)
   SetVectorParameterValue(
    execute=[@n5.then, @n6.then]
    self=Get ProgressBarMID!(self:WBP UIKit Progress Bar Base Object Reference=Get ProgressBar()) -> (ProgressBarMID:Material Instance Dynamic Object Reference)
    ParameterName=Primary Color
    Value=Get CharacterColor()
   )
   n6 = SetProgress(execute=@n2.else, self=Get ProgressBar(), Progress=0)
   Set Text(
    execute=@n1.then_1
    Text="To Text (Integer)"(Value=Get OwnedRewardsCount() -> (OwnedRewardsCount=0))
    self=Get ProgressTextX() -> (ProgressTextX:Common Text Object Reference)
   )
   n7 = Get TotalRewardsCount() -> (TotalRewardsCount=0)
   > Set Text(
    Text="To Text (String)"(InString="To String (Integer)"(InInt=@n7))
    self=Get ProgressTextOf() -> (ProgressTextOf:Common Text Object Reference)
   )
   n8 = GetDynamicFontMaterial!(execute=@n1.then_2, self:Text Object Reference=Get ProgressTextX()) -> (then, ReturnValue:Material Instance Dynamic Object Reference)
   n9 = "Equal (Integer)"(A=Get OwnedRewardsCount() -> (OwnedRewardsCount=0), B=Get TotalRewardsCount() -> (TotalRewardsCount=0))
   > SetVectorParameterValue(
    self=@n8
    ParameterName=CharacterColor
    Value=SelectColor(A=Get CompletionColor() -> (CompletionColor:Linear Color Structure), B=Get CharacterColor(), bPickA=@n9)
   )
   > SetColorAndOpacity!(
    execute
    self=[Get ProgressTextOf(), Get Divider() -> (Divider:Common Text Object Reference)]
    InColorAndOpacity_SpecifiedColor:Linear Color Structure=SelectColor(A=Get CompletionColor(), B=(R=1,G=1,B=1,A=1), bPickA=@n9)
    InColorAndOpacity_ColorUseRule:ESlateColorStylingMode Enum
   ) -> (then)
   SetFillColors!(
    execute
    self:WBP UIKit Progress Bar Base Object Reference=Get ProgressBar()
    ColorA:Linear Color Structure=Get CharacterColor()
    ColorB:Linear Color Structure=(R=1,G=1,B=1,A=1)
    ColorC:Linear Color Structure=(R=1,G=1,B=1,A=1)
   ) -> (then)
  ]}
]
variables [
 BaseCharacterColor {type=LinearColor, default=<rgba 0.347 0.913 1 1>}
 BaseLabelBackgroundColor {type=LinearColor, default=<rgba 0.005 0.004 0.012 1>}
 BonusTileFocused {}
 CachedCharacterImageReference {type=FortImageReference, default {image=null, imageUrl="", materialTextureParamName=None}}
 CachedLazyCharacterImage {type=Object, default=null}
 CanUnlock {type=boolean}
 Categories {type=boolean}
 CategoryInTandem {type=boolean}
 CharacterColor {type=LinearColor, default=<rgba 0 0 0 0>}
 CompletionColor {type=LinearColor, default=<rgba 0.04 1 0.2307 1>}
 HoldTime {type=number, default=0.5}
 IntroDelay {type=number}
 IsBonusRewardCollapsed {type=boolean}
 IsDoubleWidth {type=boolean}
 IsHighlighted {type=boolean}
 IsLocked {type=boolean}
 IsLockedByLevel {type=boolean}
 IsSetCompleted {type=boolean}
 LevelToUnlock {type=integer}
 MainCategoryTileFocused {}
 OwnedRewardsCount {type=integer}
 ProgressionColor {type=LinearColor, default=<rgba 1 0 0.7096 1>}
 SeasonPassCategoryEntered {}
 TotalRewardsCount {type=integer}
 UnlockedSet {}
]
widgets {
 info {
  inheritedWidgetCount=0
  namedSlotCount=1
  parentClass=&CommonActivatableWidget
  rootWidgetClass=&SizeBox
  widgetCount=21
 }
 widgets [
  SizeBox {
   widgetClassPath=&SizeBox
   bIsVariable=true
   widget=SizeBox
   properties {
    bOverride_HeightOverride=true
    bOverride_WidthOverride=true
    heightOverride=264
    maxDesiredWidth=312
    minDesiredHeight=168
    minDesiredWidth=312
    widthOverride=128
   }
  }
  Overlay_1 {widgetClassPath=&Overlay, parent=SizeBox, widget=Overlay_1, slot {class=SizeBoxSlot}}
  Block_Outline {
   widgetClassPath=/FortUILibrary/UIKit/Buttons/Blocks/WBP_UIKit_Block_Outline
   bIsVariable=true
   parent=Overlay_1
   widget=Block_Outline
   properties {
    bottomLeftCornerRadius=20
    bottomRightCornerRadius=20
    radiusInPixels=true
    topLeftCornerRadius=20
    topRightCornerRadius=20
    visibility=SelfHitTestInvisible
   }
   slot {class=OverlaySlot, properties {horizontalAlignment=HAlign_Fill, verticalAlignment=VAlign_Fill, padding=<ltrb 0 96 0 0>}}
  }
  Overlay_0 {
   widgetClassPath=&Overlay
   parent=Overlay_1
   widget=Overlay_0
   slot {class=OverlaySlot, properties {horizontalAlignment=HAlign_Fill, verticalAlignment=VAlign_Fill}}
  }
  Background {
   widgetClassPath=&FortLazyImage
   bIsVariable=true
   parent=Overlay_0
   widget=Background
   properties {
    brush {
     bIsDynamicallyLoaded=false
     drawAs=Box
     imageType=FullColor
     mirroring=NoMirror
     resourceName=None
     resourceObject=&M_UI_LandingPage_Tile
     tiling=NoTile
     imageSize=<128 128>
     margin=<ltrb 0 0 0 0>
     outlineSettings {
      bUseBrushTransparency=false
      roundingType=HalfHeightRadius
      width=0
      color {colorUseRule=UseColor_Specified, specifiedColor=<rgba 0 0 0 0>}
      cornerRadii=<0 0 0 0>
     }
     tintColor {colorUseRule=UseColor_Specified, specifiedColor=<rgba 1 1 1 1>}
     uVRegion {bIsValid=false, max=<0 0>, min=<0 0>}
    }
    loadingBackgroundBrush {
     bIsDynamicallyLoaded=false
     drawAs=Box
     imageType=NoImage
     mirroring=NoMirror
     resourceName=None
     resourceObject=null
     tiling=NoTile
     imageSize=<32 32>
     margin=<ltrb 0 0 0 0>
     outlineSettings {
      bUseBrushTransparency=false
      roundingType=HalfHeightRadius
      width=0
      color {colorUseRule=UseColor_Specified, specifiedColor=<rgba 0 0 0 0>}
      cornerRadii=<0 0 0 0>
     }
     tintColor {colorUseRule=UseColor_Specified, specifiedColor=<rgba 1 1 1 1>}
     uVRegion {bIsValid=false, max=<0 0>, min=<0 0>}
    }
    loadingThrobberBrush {
     bIsDynamicallyLoaded=false
     drawAs=Box
     imageType=NoImage
     mirroring=NoMirror
     resourceName=None
     resourceObject=null
     tiling=NoTile
     imageSize=<32 32>
     margin=<ltrb 0 0 0 0>
     outlineSettings {
      bUseBrushTransparency=false
      roundingType=HalfHeightRadius
      width=0
      color {colorUseRule=UseColor_Specified, specifiedColor=<rgba 0 0 0 0>}
      cornerRadii=<0 0 0 0>
     }
     tintColor {colorUseRule=UseColor_Specified, specifiedColor=<rgba 1 1 1 1>}
     uVRegion {bIsValid=false, max=<0 0>, min=<0 0>}
    }
   }
   slot {class=OverlaySlot, properties {horizontalAlignment=HAlign_Fill, verticalAlignment=VAlign_Fill, padding=<ltrb 0 96 0 0>}}
  }
  CharacterImage {
   widgetClassPath=&FortLazyImage
   bIsVariable=true
   parent=Overlay_0
   widget=CharacterImage
   properties {
    materialTextureParamName=RewardTexture
    brush {
     bIsDynamicallyLoaded=false
     drawAs=Box
     imageType=FullColor
     mirroring=NoMirror
     resourceName=None
     resourceObject=&M_UI_LandingPage_Tile_Content
     tiling=NoTile
     imageSize=<128 128>
     margin=<ltrb 0 0 0 0>
     outlineSettings {
      bUseBrushTransparency=false
      roundingType=HalfHeightRadius
      width=0
      color {colorUseRule=UseColor_Specified, specifiedColor=<rgba 0 0 0 0>}
      cornerRadii=<0 0 0 0>
     }
     tintColor {colorUseRule=UseColor_Specified, specifiedColor=<rgba 1 1 1 1>}
     uVRegion {bIsValid=false, max=<0 0>, min=<0 0>}
    }
    loadingBackgroundBrush {
     bIsDynamicallyLoaded=false
     drawAs=Box
     imageType=NoImage
     mirroring=NoMirror
     resourceName=None
     resourceObject=null
     tiling=NoTile
     imageSize=<32 32>
     margin=<ltrb 0 0 0 0>
     outlineSettings {
      bUseBrushTransparency=false
      roundingType=HalfHeightRadius
      width=0
      color {colorUseRule=UseColor_Specified, specifiedColor=<rgba 0 0 0 0>}
      cornerRadii=<0 0 0 0>
     }
     tintColor {colorUseRule=UseColor_Specified, specifiedColor=<rgba 1 1 1 1>}
     uVRegion {bIsValid=false, max=<0 0>, min=<0 0>}
    }
    loadingThrobberBrush {
     bIsDynamicallyLoaded=false
     drawAs=Box
     imageType=NoImage
     mirroring=NoMirror
     resourceName=None
     resourceObject=null
     tiling=NoTile
     imageSize=<32 32>
     margin=<ltrb 0 0 0 0>
     outlineSettings {
      bUseBrushTransparency=false
      roundingType=HalfHeightRadius
      width=0
      color {colorUseRule=UseColor_Specified, specifiedColor=<rgba 0 0 0 0>}
      cornerRadii=<0 0 0 0>
     }
     tintColor {colorUseRule=UseColor_Specified, specifiedColor=<rgba 1 1 1 1>}
     uVRegion {bIsValid=false, max=<0 0>, min=<0 0>}
    }
   }
   slot {class=OverlaySlot, properties {horizontalAlignment=HAlign_Fill, verticalAlignment=VAlign_Fill}}
  }
  SetUnlockAnimationContainer {
   widgetClassPath=&Overlay
   bIsVariable=true
   parent=Overlay_0
   widget=SetUnlockAnimationContainer
   properties {bOverride_Cursor=true, visibility=Collapsed}
   slot {class=OverlaySlot, properties {horizontalAlignment=HAlign_Center, verticalAlignment=VAlign_Center, padding=<ltrb 0 72 0 24>}}
  }
  Glow {
   widgetClassPath=&Image
   bIsVariable=true
   parent=SetUnlockAnimationContainer
   widget=Glow
   properties {
    visibility=SelfHitTestInvisible
    brush {
     bIsDynamicallyLoaded=false
     drawAs=Image
     imageType=FullColor
     mirroring=NoMirror
     resourceName=None
     resourceObject=/SeasonPassHub/Widgets/Reboot/Materials/SetUnlocking/MI_UI_BPReboot_SetUnlocking_LandingPage
     tiling=NoTile
     imageSize=<600 600>
     margin=<ltrb 0 0 0 0>
     outlineSettings {
      bUseBrushTransparency=false
      roundingType=HalfHeightRadius
      width=0
      color {colorUseRule=UseColor_Specified, specifiedColor=<rgba 0 0 0 0>}
      cornerRadii=<0 0 0 0>
     }
     tintColor {colorUseRule=UseColor_Specified, specifiedColor=<rgba 1 1 1 1>}
     uVRegion {bIsValid=false, max=<0 0>, min=<0 0>}
    }
   }
   slot {class=OverlaySlot, properties {horizontalAlignment=HAlign_Center, verticalAlignment=VAlign_Center}}
  }
  Main {
   widgetClassPath=&VerticalBox
   parent=Overlay_1
   widget=Main
   properties {visibility=HitTestInvisible}
   slot {class=OverlaySlot, properties {horizontalAlignment=HAlign_Fill, verticalAlignment=VAlign_Bottom, padding=<ltrb 0 4 0 8>}}
  }
  Overlay_LabelSwitcher {
   widgetClassPath=&Overlay
   bIsVariable=true
   parent=Main
   widget=Overlay_LabelSwitcher
   properties {flowDirectionPreference=LeftToRight}
   slot {class=VerticalBoxSlot, properties {verticalAlignment=VAlign_Center}}
  }
  Tandem_VisualAttachment {
   widgetClassPath=&FortVisualAttachment
   bIsVariable=true
   parent=Overlay_LabelSwitcher
   widget=Tandem_VisualAttachment
   properties {contentAnchor=<1 1>}
   slot {class=OverlaySlot, properties {padding=<ltrb 80 0 0 0>}}
  }
  LabelSwitcher {
   widgetClassPath=&/WBP_SeasonPass_TileLabelSwitcher
   bIsVariable=true
   parent=Overlay_LabelSwitcher
   widget=LabelSwitcher
   properties {flowDirectionPreference=Culture}
   slot {class=OverlaySlot, properties {horizontalAlignment=HAlign_Center, verticalAlignment=VAlign_Fill}}
  }
  Progress {
   widgetClassPath=&VerticalBox
   bIsVariable=true
   parent=Main
   widget=Progress
   properties {bOverride_Cursor=true, visibility=Collapsed}
   slot {class=VerticalBoxSlot, properties {padding=<ltrb 8 0 8 0>}}
  }
  ProgressLabel {
   widgetClassPath=&HorizontalBox
   bIsVariable=true
   parent=Progress
   widget=ProgressLabel
   slot {class=VerticalBoxSlot, properties {horizontalAlignment=HAlign_Center}}
  }
  ProgressTextX {
   widgetClassPath=&CommonTextBlock
   bIsVariable=true
   parent=ProgressLabel
   widget=ProgressTextX
   properties {
    text=0
    font {
     bForceMonospaced=false
     bMaterialIsStencil=false
     fontMaterial=/SeasonPassHub/Widgets/Reboot/Materials/M_UI_LandingPage_Progress_Text
     fontObject=&HeadingNow
     letterSpacing=10
     monospacedWidth=1
     size=15.75
     skewAmount=0
     typefaceFontName=76 Bold
     outlineSettings {
      bApplyOutlineToDropShadows=false
      bMiteredCorners=false
      bSeparateFillAlpha=false
      outlineMaterial=null
      outlineSize=0
      outlineColor=<rgba 0 0 0 1>
     }
    }
    shadowOffset=<0 0>
   }
   slot {class=HorizontalBoxSlot, properties {padding=<ltrb 0 0 0 4>}}
  }
  Divider {
   widgetClassPath=&CommonTextBlock
   bIsVariable=true
   parent=ProgressLabel
   widget=Divider
   properties {
    text=/
    font {
     bForceMonospaced=false
     bMaterialIsStencil=false
     fontMaterial=null
     fontObject=&HeadingNow
     letterSpacing=10
     monospacedWidth=1
     size=15.75
     skewAmount=0
     typefaceFontName=76 Bold
     outlineSettings {
      bApplyOutlineToDropShadows=false
      bMiteredCorners=false
      bSeparateFillAlpha=false
      outlineMaterial=null
      outlineSize=0
      outlineColor=<rgba 0 0 0 1>
     }
    }
    shadowOffset=<0 0>
   }
   slot {class=HorizontalBoxSlot}
  }
  ProgressTextOf {
   widgetClassPath=&CommonTextBlock
   bIsVariable=true
   parent=ProgressLabel
   widget=ProgressTextOf
   properties {
    text=100
    font {
     bForceMonospaced=false
     bMaterialIsStencil=false
     fontMaterial=null
     fontObject=&NotoSans
     letterSpacing=0
     monospacedWidth=1
     size=24
     skewAmount=0
     typefaceFontName=NotoSans-Regular
     outlineSettings {
      bApplyOutlineToDropShadows=false
      bMiteredCorners=false
      bSeparateFillAlpha=false
      outlineMaterial=null
      outlineSize=0
      outlineColor=<rgba 0 0 0 1>
     }
    }
    shadowOffset=<0 0>
   }
   slot {class=HorizontalBoxSlot}
  }
  ProgressBar {
   widgetClassPath=/FortUILibrary/UIKit/ProgressBar/WBP_UIKit_ProgressBar
   bIsVariable=true
   parent=Progress
   widget=ProgressBar
   properties {
    height=16
    linearMaterial=/SeasonPass/Materials/ProgressBar/MI_UI_NexusProgerssBar
    outline_PX=6
    progressValue=0.5
    useCircular=false
    useOutline=true
    colorB=<rgba 1 0.9857 0.9484 1>
    emptyColor=<rgba 0.3573 0.3573 0.3573 0.305>
   }
   slot {class=VerticalBoxSlot}
  }
  Button_Hitbox {
   widgetClassPath=&WBP_UIKit_Button_Generic
   bIsVariable=true
   parent=Overlay_1
   widget=Button_Hitbox
   properties {
    bRequiresHold=true
    bSimulateHoverOnTouchInput=false
    contentHorizontalAlignment=HAlign_Fill
    contentVerticalAlignment=VAlign_Fill
    holdCompletedSound=/Game/Sounds/Fort_UI_Sounds/FrontEnd/DefaultSounds/HoldableButtons/Holdable_Loud/MSS_UI_Button_Holdable_Loud_Complete
    holdData=/FortUILibrary/Input/PurchaseInteraction_HoldData.PurchaseInteraction_HoldData_C
    holdPressedSound=/Game/Sounds/Fort_UI_Sounds/FrontEnd/DefaultSounds/HoldableButtons/Holdable_Loud/MSS_UI_Button_Holdable_Loud_Press
    holdReleasedSound=/Game/Sounds/Fort_UI_Sounds/FrontEnd/DefaultSounds/HoldableButtons/Holdable_Loud/MSS_UI_Button_Holdable_Loud_Release
    holdSound=/Game/Sounds/Fort_UI_Sounds/FrontEnd/DefaultSounds/HoldableButtons/Holdable_Loud/MSS_UI_Button_Holdable_Loud_Hold_lp
    navigation=&self:WidgetTree.Button_Hitbox.WidgetNavigation_0
    promptVerticalAlignment=VAlign_Top
    secondText=""
    text=Unlock Set
    textBlockAlignment=Top
    touchMethod=PreciseTap
    useBackgroundBlock=false
    useIconBlock=false
    useTextBlock=false
    visibility=Visible
    font {
     bForceMonospaced=false
     bMaterialIsStencil=false
     fontMaterial=/FortUILibrary/UIKit/Buttons/MaterialInstances/CTAButtons/MI_UIKit_TextQuiet
     fontObject=&HeadingNow
     letterSpacing=48
     monospacedWidth=1
     size=18
     skewAmount=0
     typefaceFontName=76 Bold
     outlineSettings {
      bApplyOutlineToDropShadows=false
      bMiteredCorners=false
      bSeparateFillAlpha=false
      outlineMaterial=null
      outlineSize=0
      outlineColor=<rgba 0 0 0 1>
     }
    }
    promptNudge=<-50 -80>
    textPadding=<ltrb -20 0 -20 0>
   }
   slot {class=OverlaySlot, properties {horizontalAlignment=HAlign_Fill, verticalAlignment=VAlign_Fill, padding=<ltrb 0 72 0 0>}}
  }
  Overlay_3 {widgetClassPath=&Overlay, namedSlotHost=Button_Hitbox, widget=Overlay_3}
  UnlockSetLabel {
   widgetClassPath=&CommonTextBlock
   bIsVariable=true
   parent=Overlay_3
   widget=UnlockSetLabel
   properties {
    text=Unlock Set
    textTransformPolicy=ToUpper
    font {
     bForceMonospaced=false
     bMaterialIsStencil=false
     fontMaterial=null
     fontObject=&HeadingNow
     letterSpacing=0
     monospacedWidth=1
     size=18
     skewAmount=0
     typefaceFontName=76 Bold
     outlineSettings {
      bApplyOutlineToDropShadows=false
      bMiteredCorners=false
      bSeparateFillAlpha=false
      outlineMaterial=null
      outlineSize=0
      outlineColor=<rgba 0 0 0 1>
     }
    }
    shadowOffset=<0 0>
   }
   slot {class=OverlaySlot, properties {horizontalAlignment=HAlign_Center, padding=<ltrb 0 -70 0 0>}}
  }
 ]
}
