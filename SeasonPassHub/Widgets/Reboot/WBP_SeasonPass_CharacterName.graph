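# Season pass widget that briefly displays a character's name in stylized text over a soft radial-glow shadow.
# SetCharacterName sets the text, stops all animations and rewinds the intro, while PlayCharacterNameAnimation plays it separately.
# The CharacterNameIntro animation fades the initially transparent root in and back out.
# It also animates the font material's Reveal Amount, Outline Reveal Amount and Flip Mask parameters to wipe the name in and out.
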WidgetBlueprint /SeasonPassHub/Widgets/Reboot/WBP_SeasonPass_CharacterName
animations [CharacterNameIntro {
  binding_count=3
  length_seconds=2.2
  object=&self:CharacterNameIntro
  sequence {bindings [
    RootContainer {
     id=#1
     possessed_class=Overlay
     tracks [{
       type=MovieSceneFloatTrack
       display_name=Render Opacity
       sections [{
         type=MovieSceneFloatSection
         end_frame=44
         start_frame=0
         channels [None {type=MovieSceneScriptingFloatChannel, keys [[0,0],[2,1],[40,1],[44,0]]}]
        }]
      }]
    }
    CharacterName {
     id=#2
     possessed_class=CommonTextBlock
     tracks [{
       type=MovieSceneWidgetMaterialTrack
       display_name=Font.FontMaterial
       sections [{
         type=MovieSceneParameterSection
         end_frame=44
         start_frame=0
         channels [
          Reveal Amount {type=MovieSceneScriptingFloatChannel, keys [[0,0],[25,1],[36,1],[44,0]]}
          Outline Reveal Amount {type=MovieSceneScriptingFloatChannel, keys [[0,0],[11,1]]}
          Flip Mask {type=MovieSceneScriptingFloatChannel, keys [[32,0],[35,1]]}
         ]
        }]
      }]
    }
    Shadow {id=#3, possessed_class=Image, tracks []}
   ]}
 }]
graphs [
 EventGraph {nodes [
   Event PreConstruct()
   Event Construct()
   Event Tick()
  ]}
 PlayCharacterNameAnimation {nodes [
   PlayCharacterNameAnimation() {is=title K2Node_FunctionEntry}
   > PlayAnimation(InAnimation=Get CharacterNameIntro() -> (CharacterNameIntro:Widget Animation Object Reference))
  ]}
 SetCharacterName {nodes [
   n1 = SetCharacterName!() -> (then, CharacterName:Text) {is=title K2Node_FunctionEntry}
   > "SetText (Text)"(self=Get CharacterName() -> (CharacterName:Common Text Object Reference), InText=@n1)
   > StopAllAnimations(self:User Widget Object Reference)
   > SetAnimationCurrentTime!(execute, self:User Widget Object Reference, InAnimation:Widget Animation Object Reference=Get CharacterNameIntro(), InTime:float=0) -> (then)
  ]}
]
variables []
widgets {
 info {inheritedWidgetCount=0, parentClass=&UserWidget, rootWidgetClass=&Overlay, widgetCount=3}
 widgets [
  RootContainer {widgetClassPath=&Overlay, widget=RootContainer, properties {renderOpacity=0}}
  Shadow {
   widgetClassPath=&Image
   parent=RootContainer
   widget=Shadow
   properties {
    visibility=SelfHitTestInvisible
    brush {
     bIsDynamicallyLoaded=false
     drawAs=Image
     imageType=FullColor
     mirroring=NoMirror
     resourceName=None
     resourceObject=/FortUILibrary/UIKit/Buttons/Materials/M_UIKit_RadialGlow
     tiling=NoTile
     imageSize=<32 200>
     margin=<ltrb 0 0 0 0>
     outlineSettings {
      bUseBrushTransparency=false
      roundingType=HalfHeightRadius
      width=0
      color {colorUseRule=UseColor_Specified, specifiedColor=<rgba 0 0 0 0>}
      cornerRadii=<0 0 0 0>
     }
     tintColor {colorUseRule=UseColor_Specified, specifiedColor=<rgba 0 0 0 0.2>}
     uVRegion {bIsValid=false, max=<0 0>, min=<0 0>}
    }
   }
   slot {class=OverlaySlot, properties {horizontalAlignment=HAlign_Fill, verticalAlignment=VAlign_Center}}
  }
  CharacterName {
   widgetClassPath=&CommonTextBlock
   bIsVariable=true
   parent=RootContainer
   widget=CharacterName
   properties {
    text="Character  Name"
    font {
     bForceMonospaced=false
     bMaterialIsStencil=true
     fontMaterial=/SeasonPass/Widgets/Reboot/Materials/M_UI_BPReboot_CharacterName_Font
     fontObject=&HeadingNow
     letterSpacing=-40
     monospacedWidth=1
     size=32.25
     skewAmount=0
     typefaceFontName=86 Bold Italic
     outlineSettings {
      bApplyOutlineToDropShadows=false
      bMiteredCorners=false
      bSeparateFillAlpha=false
      outlineMaterial=null
      outlineSize=0
      outlineColor=<rgba 0 0 0 1>
     }
    }
    shadowOffset=<0 0>
   }
   slot {class=OverlaySlot, properties {horizontalAlignment=HAlign_Center, verticalAlignment=VAlign_Center, padding=<ltrb 150 0 150 0>}}
  }
 ]
}
