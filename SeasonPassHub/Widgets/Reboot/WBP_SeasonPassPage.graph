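# Activatable Season Pass hub page showing a header above a horizontally scrolling row of main and bonus reward-set tiles.
# Category view models fill two DynamicEntryBoxes by group name, with linked categories marked as tandem sets.
# Bonus sets collapse into one tile until all main sets are claimed; completion changes also trigger scrolling between sections.
# It also handles header/tile focus navigation, staggered intro animations, input-dependent scroll arrows, RTL and compact layouts.
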WidgetBlueprint /SeasonPassHub/Widgets/Reboot/WBP_SeasonPassPage
animations [Anim_Intro {
  binding_count=4
  length_seconds=1
  object=&self:Anim_Intro
  sequence {
   bindings [
    ScrollBox_0 {
     id=#1
     possessed_class=ScrollBox
     tracks [
      {
       type=MovieSceneFloatTrack
       display_name=Render Opacity
       sections [{type=MovieSceneFloatSection, end_frame=60, start_frame=0, channels [None {type=MovieSceneScriptingFloatChannel, keys [[0,0],[42,0],[60,1]]}]}]
      }
      {
       type=MovieScene2DTransformTrack
       sections [{
         type=MovieScene2DTransformSection
         end_frame=60
         start_frame=0
         channels [
          Translation.Y {type=MovieSceneScriptingFloatChannel, default=0}
          Scale.X {type=MovieSceneScriptingFloatChannel, keys [[0,1.2],[42,1.2],[60,1]]}
          Scale.Y {type=MovieSceneScriptingFloatChannel, keys [[0,1.2],[42,1.2],[60,1]]}
         ]
        }]
      }
     ]
    }
    SeasonPass_Header {
     id=#2
     possessed_class=WBP_SeasonPass_Header_C
     tracks [
      {
       type=MovieSceneFloatTrack
       display_name=Render Opacity
       sections [{type=MovieSceneFloatSection, channels [None {type=MovieSceneScriptingFloatChannel, keys [[0,0],[33,0],[60,1]]}]}]
      }
      {
       type=MovieScene2DTransformTrack
       sections [{
         type=MovieScene2DTransformSection
         end_frame=60
         start_frame=0
         channels [
          Scale.X {type=MovieSceneScriptingFloatChannel, keys [[0,1.2],[33,1.2],[60,1]]}
          Scale.Y {type=MovieSceneScriptingFloatChannel, keys [[0,1.2],[33,1.2],[60,1]]}
         ]
        }]
      }
     ]
    }
    MKBScrollArrows {
     id=#3
     possessed_class=HorizontalBox
     tracks [
      {
       type=MovieSceneFloatTrack
       display_name=Render Opacity
       sections [{type=MovieSceneFloatSection, channels [None {type=MovieSceneScriptingFloatChannel, keys [[0,0],[33,0],[60,1]]}]}]
      }
      {
       type=MovieScene2DTransformTrack
       sections [{
         type=MovieScene2DTransformSection
         end_frame=60
         start_frame=0
         channels [
          Scale.X {type=MovieSceneScriptingFloatChannel, keys [[0,1.2],[33,1.2],[60,1]]}
          Scale.Y {type=MovieSceneScriptingFloatChannel, keys [[0,1.2],[33,1.2],[60,1]]}
         ]
        }]
      }
     ]
    }
    SeasonPassBubble {
     id=#4
     possessed_class=WBP_SeasonPassBubble_C
     tracks [
      {
       type=MovieSceneFloatTrack
       display_name=Render Opacity
       sections [{type=MovieSceneFloatSection, channels [None {type=MovieSceneScriptingFloatChannel, keys [[0,0],[33,0],[60,1]]}]}]
      }
      {
       type=MovieScene2DTransformTrack
       sections [{
         type=MovieScene2DTransformSection
         end_frame=60
         start_frame=0
         channels [
          Scale.X {type=MovieSceneScriptingFloatChannel, keys [[0,1.2],[33,1.2],[60,1]]}
          Scale.Y {type=MovieSceneScriptingFloatChannel, keys [[0,1.2],[33,1.2],[60,1]]}
         ]
        }]
      }
     ]
    }
   ]
   display_rate {numerator=60}
  }
 }]
graphs [
 BP_GetDesiredFocusTarget {nodes [
   Get Desired Focus Target()
   > n1 = Is Valid(InputObject=Get SeasonPassVM() -> (SeasonPassVM:Season Pass VM Object Reference))
   n2 = Get IsActivatingFromSubscreen() -> (IsActivatingFromSubscreen:bool)
   n3 = Get IsPageUp() -> (IsPageUp:bool)
   n4 = Get LastTileFocused() -> (LastTileFocused:User Widget Object Reference)
   n5 = IsValid(Object=@n4)
   n6 = GetFirstListEntryCurrentSection!(self) -> (ReturnValue:User Widget Object Reference)
   n7 = Get SeasonPass_Header() -> (SeasonPass_Header:WBP Season Pass Header Object Reference)
   n8 = Select(
    Option 0:WBP UIKit Button Loud Object Reference=Get PurchasePass_Button!(self:WBP Season Pass Header Object Reference=@n7) -> (PurchasePass_Button:WBP UIKit Button Loud Object Reference)
    Option 1:WBP UIKit Button Loud Object Reference=Get BuyRewards_Button!(self:WBP Season Pass Header Object Reference=@n7) -> (BuyRewards_Button:WBP UIKit Button Loud Object Reference)
    Index=Get bDoesOwnsPass!(self:Season Pass VM Object Reference=Get SeasonPassVM()) -> (bDoesOwnsPass:bool)
   ) -> (ReturnValue:WBP UIKit Button Loud Object Reference)
   n9 = Select(
    Option 0:User Widget Object Reference=Get GIftPass_Button!(self:WBP Season Pass Header Object Reference=Get SeasonPass_Header()) -> (GIftPass_Button:WBP UIKit Button Regular Object Reference)
    Option 1:User Widget Object Reference=@n8
    Index=Get bIsEnabled!(self:Widget Object Reference=@n8) -> (bIsEnabled:bool)
   ) -> (ReturnValue:User Widget Object Reference)
   n10 = Select(Option 0=@n9, Option 1=@n6, Index=IsValid(Object=@n6))
   > Return Node!(
    execute
    ReturnValue:Widget Object Reference=Select(Option 0=@n9, Option 1=Select(Option 0=@n10, Option 1=@n4, Index=@n5), Index=OR Boolean(A=@n3, B=@n2))
   )
   Return Node(execute=@n1.Is Not Valid, ReturnValue=Get PurchasePass_Button(self=Get SeasonPass_Header()))
  ]}
 CF_DetermineBubbleVis {nodes [
   n1 = CF_DetermineBubbleVis!() -> (then, Text:Text, IsLevelLocked:bool=~, MainSetClaimed:bool=~, CategoryGroup:String) {is=title K2Node_FunctionEntry}
   n2 = NOT Boolean(A=TextIsEmpty(InText=@n1))
   n3 = NOT Boolean(A=@n1.MainSetClaimed)
   n4 = OR Boolean(A=@n1.IsLevelLocked, B="Equal, Case Insensitive (String)"(A=@n1.CategoryGroup, B=bonus))
   > Return Node!(
    execute
    Visibility:ESlateVisibility Enum=Select(
     Option 0:ESlateVisibility Enum=Collapsed
     Option 1:ESlateVisibility Enum=SelfHitTestInvisible
     Index=AND Boolean!(A=@n4, B=@n2, C:bool=@n3) -> (ReturnValue)
    ) -> (ReturnValue:ESlateVisibility Enum)
   )
  ]}
 CF_GetLastFocusedTile {nodes [
   CF_GetLastFocusedTile!() -> (then, SelectedCategory:Object Reference) {is=title K2Node_FunctionEntry}
   > n1 = Sequence!(execute) -> (then_0, then_1, then_2:Exec)
   n2 = GetAllEntries!(self:Dynamic Entry Box Base Object Reference=Get BonusSets() -> (BonusSets:Dynamic Entry Box Object Reference)) -> (ReturnValue:Array of User Widget Object References)
   n3 = For Each Loop(Exec=@n1.then_1, Array:Array of User Widget Object References=@n2) -> (Array Element:User Widget Object Reference)
   > n4 = Cast To WBP_SeasonPass_CharacterButton!(execute, Object=@n3) -> (then, CastFailed, AsWBP Season Pass Character Button:WBP Season Pass Character Button Object Reference) {is=title K2Node_DynamicCast}
   > Branch(Condition="Equal (Object)"(
     A=Get SeasonPassCategoryVM!(self:WBP Season Pass Character Button Object Reference=@n4) -> (SeasonPassCategoryVM:Season Pass Category VM Object Reference)
     B=Get SelectedCategory() -> (SelectedCategory:Object Reference)
    ))
   n5 = GetConfigDataBoolWithDefault!(EntryName:String=FocusBonusRewardsTileIfMainSetIncomplete, bDefaultValue:bool=true) -> (ReturnValue:bool)
   n6 = NOT Boolean(A=Get IsMainSetFullyClaimed() -> (IsMainSetFullyClaimed:bool))
   > Return Node!(
    execute
    LastSelectedCharacterTile:User Widget Object Reference=Select(
     Option 0:User Widget Object Reference=@n4
     Option 1:User Widget Object Reference="Get (a copy)"(Array:Array of User Widget Object References=@n2) -> (Output:User Widget Object Reference)
     Index=AND Boolean!(A=@n5, B=@n6) -> (ReturnValue)
    ) -> (ReturnValue:User Widget Object Reference)
   )
   Return Node(execute=@n1.then_2)
   n7 = For Each Loop(Exec=@n1.then_0, Array=GetAllEntries(self=Get MainSets() -> (MainSets:Dynamic Entry Box Object Reference)))
   > n8 = Cast To WBP_SeasonPass_CharacterButton(Object=@n7)
   > Branch(Condition="Equal (Object)"(A=Get SeasonPassCategoryVM(self=@n8), B=Get SelectedCategory()))
   > Return Node(LastSelectedCharacterTile=@n8)
  ]}
 Categories Changed {nodes [
   n1 = Categories Changed!() -> (then, Categories:Array of Season Pass Category VM Object References) {is=title K2Node_FunctionEntry}
   > Branch(Condition=Is Not Empty!(TargetArray:Array of Season Pass Category VM Object References=@n1) -> (ReturnValue:bool))
   "Get (a copy)"(Array:Array of Season Pass Category VM Object References=@n1) -> (Output:Season Pass Category VM Object Reference)
  ]}
 Clear Preview {nodes [Clear Preview() {is=title K2Node_FunctionEntry}]}
 EnteredPassSubscreen {nodes [EnteredPassSubscreen() {is=title K2Node_FunctionEntry}]}
 EventGraph {nodes [
   n1 = "On Clicked (ScrollRight)"!() -> (OutputDelegate, then, Button:Common Button Base Object Reference) {is=title K2Node_ComponentBoundEvent}
   ScrollSets!(execute=[@n1.then, @n22.then], self, ScrollToBonus:bool=true, AnimateScroll:bool=true) -> (then)
   n2 = "On Clicked (ScrollLeft)"!() -> (OutputDelegate, then, Button:Common Button Base Object Reference) {is=title K2Node_ComponentBoundEvent}
   ScrollSets(execute=[@n2.then, @n23.then], ScrollToBonus=false, AnimateScroll=true)
   "Clear Preview (SeasonPass_Header)"() {is=title K2Node_ComponentBoundEvent}
   > Call Clear Preview!(execute:Exec, self:WBP Season Pass Page Object Reference) -> (then:Exec) {is=title K2Node_CallDelegate}
   > SetVisibility(self=Get SeasonPassBubble() -> (SeasonPassBubble:WBP Season Pass Bubble Object Reference), InVisibility=Collapsed)
   n3 = CategoriesListChanged!() -> (OutputDelegate, then, SeasonPassCategoryVMs:Array of Season Pass Category VM Object References) {is=title K2Node_CustomEvent}
   > n4 = Set Season Pass Category VMs(Season Pass Category VMs:Array of Season Pass Category VM Object References=@n3.SeasonPassCategoryVMs) -> (Output_Get:Array of Season Pass Category VM Object References)
   n5 = Length(TargetArray:Array of Season Pass Category VM Object References=@n4)
   > FortBlueprintLog!(
    execute
    InString:String="To String (Text)"(InText=Format Text!(Format="INVTEXT(\"CategoriesListChanged: {0} categories provided\")", 0:int=@n5) -> (Result))
   ) -> (then)
   > Reset!(execute, self:Dynamic Entry Box Object Reference=[Get MainSets(), Get BonusSets()], bDeleteWidgets:bool=true) -> (then)
   > n6 = For Each Loop(Array:Array of Season Pass Category VM Object References=Get Season Pass Category VMs() -> (Season Pass Category VMs:Array of Season Pass Category VM Object References)) -> (Array Element:Season Pass Category VM Object Reference)
   n7 = GetNumEntries!(self:Dynamic Entry Box Base Object Reference=Get MainSets()) -> (ReturnValue:int=0)
   n8 = GetNumEntries(self=Get BonusSets()) -> (ReturnValue=0)
   FortBlueprintLog(
    execute=@n6.Completed
    InString="To String (Text)"(InText=Format Text!(Format="INVTEXT(\"   completed! Main Sets: {0}, Bonus Sets: {1}\")", 0=@n7, 1:int=@n8) -> (Result))
   )
   > HandleBonusRewardsCollapsedTile()
   > RequestRefreshFocus(self:Common Activatable Widget Object Reference)
   Set CategoryListElementVM(execute=@n6.LoopBody, CategoryListElementVM:Season Pass Category VM Object Reference=@n6) -> (Output_Get:Season Pass Category VM Object Reference)
   > Set CategoryListIndex(CategoryListIndex:int=@n6.Array Index) -> (Output_Get:int=0)
   n9 = GetGroupName!(self:Season Pass Element Base VM Object Reference=Get CategoryListElementVM() -> (CategoryListElementVM:Season Pass Category VM Object Reference)) -> (ReturnValue:String)
   > n10 = Create Entry!(
    execute
    self:Dynamic Entry Box Object Reference=Select(
     Option 0:Dynamic Entry Box Object Reference=Get MainSets()
     Option 1:Dynamic Entry Box Object Reference=Get BonusSets()
     Index="Equal Exactly (String)"(A=@n9, B=Bonus)
    ) -> (ReturnValue:Dynamic Entry Box Object Reference)
   ) -> (then, ReturnValue:User Widget Object Reference)
   > n11 = Cast To WBP_SeasonPass_CharacterButton(Object=@n10)
   n12 = Get CategoryListElementVM()
   > FortBlueprintLog(InString="To String (Text)"(InText=Format Text(
      Format="INVTEXT(\"   category widget created: {0} - {1}\")"
      0:WBP Season Pass Character Button Object Reference=@n11
      1:Season Pass Category VM Object Reference=@n12
     )))
   > n13 = Set CategoryListButton(CategoryListButton:WBP Season Pass Character Button Object Reference=@n11) -> (Output_Get:WBP Season Pass Character Button Object Reference)
   > SetCategoryTandemSet!(
    execute
    self
    CategoryVM:Season Pass Element Base VM Object Reference=Get CategoryListElementVM()
    CategoryIndex:int=Get CategoryListIndex() -> (CategoryListIndex:int=0)
    Button:WBP Season Pass Character Button Object Reference=@n11
   ) -> (then)
   > Bind Event to Bonus Tile Focused(self:WBP Season Pass Character Button Object Reference=@n11, Delegate=Create Event()) {is=title K2Node_AddDelegate}
   > Bind Event to Main Category Tile Focused(self:WBP Season Pass Character Button Object Reference=@n11, Delegate=Create Event()) {is=title K2Node_AddDelegate}
   > Bind Event to Unlocked Set(self:WBP Season Pass Character Button Object Reference=@n11, Delegate=Create Event()) {is=title K2Node_AddDelegate}
   > SetSeasonPassCategoryVM!(
    execute
    self:WBP Season Pass Character Button Object Reference=@n13
    Viewmodel:Season Pass Category VM Object Reference=Get CategoryListElementVM()
   ) -> (then)
   > SetSeasonPassVM!(
    execute
    self:WBP Season Pass Character Button Object Reference=Get CategoryListButton() -> (CategoryListButton:WBP Season Pass Character Button Object Reference)
    Viewmodel:Season Pass VM Object Reference=Get SeasonPassVM()
   ) -> (then)
   > Set Double Width!(execute, self:WBP Season Pass Character Button Object Reference=Get CategoryListButton(), Is Double Width:bool=false) -> (then) {is=title K2Node_CallFunction}
   > Assign Season Pass Category Entered!(execute:Exec, self:WBP Season Pass Character Button Object Reference=Get CategoryListButton(), Delegate:"Delegate (by ref)"=Create Event()) -> (then:Exec) {is=title K2Node_AssignDelegate}
   Opening() {is=title K2Node_CustomEvent}
   > PlayAnimationForward(InAnimation=Get Anim_Intro() -> (Anim_Intro:Widget Animation Object Reference))
   > n14 = Get!(execute:Exec) -> (then:Exec, else:Exec, SeasonPassVM:Season Pass VM Object Reference)
   > n15 = SetIsMainSetComplete!(
    execute
    self
    SeasonPassCategoryVM:Array of Season Pass Category VM Object References=Get Categories!(self:Season Pass VM Object Reference=@n14) -> (Categories:Array of Season Pass Category VM Object References)
   ) -> (then)
   Delay(execute=[@n15.then, @n21.then], Duration=0.6)
   > Set CascadeIntroDelayTotal(CascadeIntroDelayTotal:double=0) -> (Output_Get:double)
   > n16 = For Each Loop(Array:Array of User Widget Object References=GetAllEntries(self=Get MainSets())) -> (Array Element:User Widget Object Reference)
   > n17 = Cast To WBP_SeasonPass_CharacterButton(Object=@n16)
   > PlayCascadeIntroAnimWithDelay!(execute, self, CharacterButton:WBP Season Pass Character Button Object Reference=@n17) -> (then)
   n18 = GetCurrentCulture!() -> (ReturnValue:String)
   > Branch(Condition="Not Equal (Boolean)"(A=IsCultureRightToLeft!(Culture:String=@n18) -> (ReturnValue:bool), B=Get IsCultureRightToLeft() -> (IsCultureRightToLeft:bool)) -> (ReturnValue=~))
   > SetCategoryTandemSet(CategoryVM=Get SeasonPassCategoryVM(self=@n17), CategoryIndex=@n16.Array Index, Button=@n17)
   n19 = For Each Loop(Exec=@n16.Completed, Array=GetAllEntries(self=Get BonusSets()))
   Set IsCultureRightToLeft(execute=@n19.Completed, IsCultureRightToLeft:bool=IsCultureRightToLeft(Culture=GetCurrentCulture())) -> (Output_Get:bool)
   n20 = Cast To WBP_SeasonPass_CharacterButton(execute=@n19.LoopBody, Object=@n19)
   > PlayCascadeIntroAnimWithDelay(CharacterButton=@n20)
   n21 = HandleBonusRewardsCollapsedTile(execute=@n14.else)
   OnEnteredSubscreen() {is=title K2Node_CustomEvent}
   > Call Entered Pass Subscreen!(execute:Exec, self:WBP Season Pass Page Object Reference) -> (then:Exec) {is=title K2Node_CallDelegate}
   n22 = BonusTileFocused() {is=title K2Node_CustomEvent}
   n23 = MainCategoryTileFocused() {is=title K2Node_CustomEvent}
   n24 = OnSetUnlocked!() -> (OutputDelegate, then, Category:Season Pass Category VM Object Reference) {is=title K2Node_CustomEvent}
   > Call Set Unlocked!(execute:Exec, self:WBP Season Pass Page Object Reference, Category:Season Pass Category VM Object Reference=@n24.Category) -> (then:Exec) {is=title K2Node_CallDelegate}
   Event On Activated()
   n25 = Create Event()
   n26 = Get CommonInputSubsystem()
   > Bind Event to On Input Method Changed(self=@n26, Delegate=@n25)
   > HandleInputChanged!(execute, self, bNewInputType:ECommonInputType Enum=GetCurrentInputType(self=@n26)) -> (then)
   Event PreConstruct()
   > SetCompactUIStyle() {is=title K2Node_CallFunction}
   > Bind Event to Header Entered Subscreen(self:WBP Season Pass Header Object Reference=Get SeasonPass_Header(), Delegate=Create Event()) {is=title K2Node_AddDelegate}
   n27 = IsCultureRightToLeft(Culture=GetCurrentCulture())
   > Set IsCultureRightToLeft(IsCultureRightToLeft=@n27)
   > Branch(Condition=@n27)
   > HandleRTL()
   Event On Deactivated()
   > Unbind Event from On Input Method Changed(self=@n26, Delegate=@n25)
  ]}
 GetFirstListEntryCurrentSection {nodes [
   GetFirstListEntryCurrentSection() {is=title K2Node_FunctionEntry}
   n1 = GetAllEntries(self=Get BonusSets())
   n2 = GetAllEntries(self=Get MainSets())
   > Return Node!(
    execute
    ReturnValue:User Widget Object Reference=Select(
     Option 0:User Widget Object Reference="Get (a copy)"(Array:Array of User Widget Object References=@n1) -> (Output:User Widget Object Reference)
     Option 1:User Widget Object Reference="Get (a copy)"(Array=@n2)
     Index=Get IsShowingMainSet() -> (IsShowingMainSet:bool)
    ) -> (ReturnValue:User Widget Object Reference)
   )
  ]}
 HandleBonusRewardsCollapsedTile {nodes [
   HandleBonusRewardsCollapsedTile() {is=title K2Node_FunctionEntry}
   > n1 = Branch(Condition=Get IsMainSetFullyClaimed())
   > n2 = Set BonusRewardEntries(BonusRewardEntries:Array of User Widget Object References=GetAllEntries(self=Get BonusSets())) -> (Output_Get:Array of User Widget Object References)
   > n3 = For Each Loop(Array=@n2)
   > Set Visibility(Visibility=SelfHitTestInvisible, self=@n3)
   > n4 = Cast To WBP_SeasonPass_CharacterButton(Object=@n3)
   > Set Double Width(self=@n4, Is Double Width=false)
   > SetBonusRewardsCollapsed!(execute, self:WBP Season Pass Character Button Object Reference=@n4, IsBonusRewardsCollapsed:bool=false) -> (then)
   > SetBonusRewardsCollapsed!(
    execute
    self:WBP Season Pass Tile Label Switcher Object Reference=Get LabelSwitcher!(self:WBP Season Pass Character Button Object Reference=@n4) -> (LabelSwitcher:WBP Season Pass Tile Label Switcher Object Reference)
    IsBonusCollapsed:bool=false
   ) -> (then)
   > SetVisibility(
    self=[Get EndSpacerForBonusSets() -> (EndSpacerForBonusSets:Spacer Object Reference), Get FrontSpacer() -> (FrontSpacer:Spacer Object Reference)]
    InVisibility=SelfHitTestInvisible
   )
   n5 = Set BonusRewardEntries(execute=@n1.else, BonusRewardEntries=GetAllEntries(self=Get BonusSets()))
   > n6 = For Each Loop(Array=@n5)
   > Set BonusRewardEntry(BonusRewardEntry:User Widget Object Reference=@n6) -> (Output_Get:User Widget Object Reference)
   > n7 = Branch(Condition="Equal (Integer)"(A=@n6.Array Index))
   > Set Visibility(Visibility=SelfHitTestInvisible, self=Get BonusRewardEntry() -> (BonusRewardEntry:User Widget Object Reference))
   > n8 = Cast To WBP_SeasonPass_CharacterButton(Object=Get BonusRewardEntry())
   > Set BonusRewardCharacterButton(BonusRewardCharacterButton:WBP Season Pass Character Button Object Reference=@n8) -> (Output_Get:WBP Season Pass Character Button Object Reference)
   n9 = Get BonusRewardCharacterButton() -> (BonusRewardCharacterButton:WBP Season Pass Character Button Object Reference)
   > SetBonusRewardsCollapsed!(execute, self:WBP Season Pass Character Button Object Reference=@n9, IsBonusRewardsCollapsed:bool=true) -> (then)
   > SetBonusRewardsCollapsed!(execute, self:WBP Season Pass Tile Label Switcher Object Reference=Get LabelSwitcher(self=@n9), IsBonusCollapsed:bool=true) -> (then)
   > SetVisibility(self=[Get EndSpacerForBonusSets(), Get FrontSpacer()], InVisibility=Collapsed)
   Set Visibility(execute=@n7.else, Visibility=Collapsed, self=Get BonusRewardEntry())
  ]}
 HandleInputChanged {nodes [
   n1 = HandleInputChanged!() -> (then, bNewInputType:ECommonInputType Enum) {is=title K2Node_FunctionEntry}
   > n2 = Branch(Condition="Equal (Enum)"(A=@n1, B=MouseAndKeyboard))
   n3 = Get MKBScrollArrows() -> (MKBScrollArrows:Horizontal Box Object Reference)
   > SetVisibility(self=@n3, InVisibility=SelfHitTestInvisible)
   SetVisibility(execute=@n2.else, self=@n3, InVisibility=Collapsed)
  ]}
 HandleRTL {nodes [
   HandleRTL() {is=title K2Node_FunctionEntry}
   > Set FlowDirectionPreference!(
    execute
    FlowDirectionPreference:EFlowDirectionPreference Enum=Culture
    self:Widget Object Reference=Get Content() -> (Content:Vertical Box Object Reference)
   ) -> (then, Output_Get:EFlowDirectionPreference Enum=Inherit)
  ]}
 NavUpButNoButtonsLeft {nodes [NavUpButNoButtonsLeft() {is=title K2Node_FunctionEntry}]}
 NavigateToHeader {nodes [
   NavigateToHeader!() -> (then, Navigation:EUINavigation Enum) {is=title K2Node_FunctionEntry}
   n1 = Get SeasonPass_Header()
   n2 = Get PassButtonStates!(self:WBP Season Pass Header Object Reference=@n1) -> (PassButtonStates:Common Visibility Switcher Object Reference)
   n3 = "Equal (Enum)"(
    A:ESlateVisibility Enum=GetVisibility!(self:Widget Object Reference=@n2) -> (ReturnValue:ESlateVisibility Enum=Visible)
    B:ESlateVisibility Enum=Collapsed
   )
   n4 = Select(
    Option 0:Widget Object Reference=Get PurchasePass_Button(self=@n1)
    Option 1:Widget Object Reference=Get BuyRewards_Button(self=@n1)
    Index:int=GetActiveWidgetIndex!(self:Common Visibility Switcher Object Reference=@n2) -> (ReturnValue:int=0)
   ) -> (ReturnValue:Widget Object Reference)
   n5 = "Equal (Boolean)"(A:bool=GetIsEnabled!(self:Widget Object Reference=@n4) -> (ReturnValue:bool), B:bool) -> (ReturnValue:bool=~) {is=title K2Node_PromotableOperator}
   n6 = Get GIftPass_Button(self=Get SeasonPass_Header())
   n7 = GetIsEnabled(self=@n6)
   n8 = "Equal (Enum)"(A=GetVisibility(self=@n6) -> (ReturnValue=Visible), B=SelfHitTestInvisible)
   n9 = Get SeasonPass_Header()
   n10 = "float > float"(
    A=GetDesiredSize!(self:Widget Object Reference=Get HorizontalBox!(self:WBP Season Pass Header Object Reference=@n9) -> (HorizontalBox:Horizontal Box Object Reference)) -> (ReturnValue_X:double, ReturnValue_Y:double).ReturnValue_Y
    B=80
   )
   n11 = Select(
    Option 0=@n4
    Option 1=Get GIftPass_Button(self=Get SeasonPass_Header())
    Index:bool=OR Boolean(A=OR Boolean(A=@n3, B=@n5), B=AND Boolean!(A=@n8, B=@n7, C:bool=@n10) -> (ReturnValue))
   )
   n12 = GetVisibility(self=@n11) -> (ReturnValue=Visible)
   n13 = GetIsEnabled(self=@n11)
   > n14 = Branch(Condition=OR Boolean(A="Equal (Boolean)"(A=@n13) -> (ReturnValue=~), B="Equal (Enum)"(A=@n12, B=Collapsed)))
   > Call Nav Up But No Buttons Left!(execute:Exec, self:WBP Season Pass Page Object Reference) -> (then:Exec) {is=title K2Node_CallDelegate}
   > Return Node(ReturnValue:Widget Object Reference=GetDesiredFocusTarget!(self:Common Activatable Widget Object Reference) -> (ReturnValue:Widget Object Reference))
   Return Node(execute=@n14.else, ReturnValue=@n11)
  ]}
 NavigateToTiles {nodes [
   NavigateToTiles!() -> (then, Navigation:EUINavigation Enum) {is=title K2Node_FunctionEntry}
   > n1 = Set TargetWidget(TargetWidget:User Widget Object Reference=GetFirstListEntryCurrentSection() {is=title K2Node_CallFunction}) -> (Output_Get:User Widget Object Reference)
   > n2 = Is Valid(InputObject=@n1)
   Call On Navigate to Tiles Without Pass!(execute:Exec=@n2.Is Not Valid, self:WBP Season Pass Page Object Reference) -> (then:Exec) {is=title K2Node_CallDelegate}
   > Return Node(ReturnValue=Get FallbackFocusWidgetWhenPassIsMissing() -> (FallbackFocusWidgetWhenPassIsMissing:Widget Object Reference))
   Return Node(execute=@n2.Is Valid, ReturnValue=Get TargetWidget() -> (TargetWidget:User Widget Object Reference))
  ]}
 OnNavigateToTilesWithoutPass {nodes [OnNavigateToTilesWithoutPass() {is=title K2Node_FunctionEntry}]}
 PlayCascadeIntroAnimWithDelay {nodes [
   n1 = PlayCascadeIntroAnimWithDelay!() -> (then, CharacterButton:WBP Season Pass Character Button Object Reference) {is=title K2Node_FunctionEntry}
   > SetRenderOpacity(self=@n1, InOpacity=0)
   n2 = Get CascadeIntroDelayTotal() -> (CascadeIntroDelayTotal:double)
   n3 = Get CascadeIntroDelayAmt() -> (CascadeIntroDelayAmt:double)
   > n4 = Set CascadeIntroDelayTotal(CascadeIntroDelayTotal=float * float(A=float + float(A=@n2, B=@n3), B=0.8) {is=title K2Node_PromotableOperator})
   > IntroAnim!(execute, self:WBP Season Pass Character Button Object Reference=@n1, Duration:float=@n4) -> (then)
  ]}
 ScrollSets {nodes [
   n1 = ScrollSets!() -> (then, ScrollToBonus:bool=~, AnimateScroll:bool=true) {is=title K2Node_FunctionEntry}
   > ScrollWidgetIntoView!(
    execute
    self:Scroll Box Object Reference=Get ScrollBox_0() -> (ScrollBox_0:Scroll Box Object Reference)
    WidgetToFind:Widget Object Reference=Select(Option 0=Get MainSets(), Option 1=Get BonusSets(), Index=@n1)
    AnimateScroll:bool=@n1.AnimateScroll
    ScrollDestination:EDescendantScrollDestination Enum=Center
    Padding:float=0
   ) -> (then)
   > n2 = Set IsShowingMainSet(IsShowingMainSet:bool=NOT Boolean(A=@n1)) -> (Output_Get:bool)
   > Branch(Condition=Get IsMainSetFullyClaimed())
   > Set Visibility(
    Visibility=Select(Option 0:ESlateVisibility Enum=Hidden, Option 1:ESlateVisibility Enum=Visible, Index=@n2) -> (ReturnValue:ESlateVisibility Enum)
    self=Get ScrollRight() -> (ScrollRight:WBP UIKit Button Regular Object Reference)
   )
   > Set Visibility(
    Visibility=Select(Option 0=Visible, Option 1=Hidden, Index=@n2)
    self=Get ScrollLeft() -> (ScrollLeft:WBP UIKit Button Regular Object Reference)
   )
   > Return Node!(execute)
  ]}
 SeasonPassVmChanged {nodes [
   n1 = SeasonPassVmChanged!() -> (then, PassVM:Season Pass VM Object Reference) {is=title K2Node_FunctionEntry}
   > SetSeasonPassVM(self:WBP Season Pass Header Object Reference=Get SeasonPass_Header(), Viewmodel=@n1)
  ]}
 SetCategoryTandemSet {nodes [
   SetCategoryTandemSet!() -> (then, CategoryVM:Season Pass Element Base VM Object Reference, CategoryIndex:int, Button:WBP Season Pass Character Button Object Reference) {is=title K2Node_FunctionEntry}
   > n1 = Branch(Condition=IsCultureRightToLeft(Culture=GetCurrentCulture()))
   > n2 = Set MaxIndex(MaxIndex:int=100) -> (Output_Get:int=0)
   n3 = GetLinkedElements!(self:Season Pass Element Base VM Object Reference=Get CategoryVM() -> (CategoryVM:Season Pass Element Base VM Object Reference)) -> (ReturnValue:Array of Season Pass Element Base VM Object References)
   n4 = Length(TargetArray:Array of Season Pass Element Base VM Object References=@n3)
   n5 = Branch(execute=[@n14.then, @n2.then], Condition="integer > integer"(A=@n4))
   > n6 = For Each Loop(Array:Array of Season Pass Element Base VM Object References=@n3) -> (Array Element:Season Pass Element Base VM Object Reference)
   n7 = GetCurrentCulture()
   n8 = Get MaxIndex() -> (MaxIndex:int=0)
   n9 = Get CategoryIndex() -> (CategoryIndex:int=0)
   SetInTandem!(
    execute=@n6.Completed
    self:WBP Season Pass Character Button Object Reference=Get Button() -> (Button:WBP Season Pass Character Button Object Reference)
    InbShowLevel:bool=Select(Option 0:bool="integer >= integer"(A=@n9, B=@n8), Option 1:bool="integer <= integer"(A=@n9, B=@n8), Index=IsCultureRightToLeft(Culture=@n7)) -> (ReturnValue:bool)
    TandemAmount:int=@n4
   ) -> (then)
   n10 = Cast To SeasonPassCategoryVM!(execute=@n6.LoopBody, Object=@n6) -> (then, CastFailed, AsSeason Pass Category VM:Season Pass Category VM Object Reference) {is=title K2Node_DynamicCast}
   n11 = GetCurrentCulture()
   n12 = Get MaxIndex() -> (MaxIndex=0)
   n13 = Find Item!(
    TargetArray:Array of Season Pass Category VM Object References=Get Season Pass Category VMs()
    ItemToFind:"Season Pass Category VM Object Reference (by ref)"=@n10
   ) -> (ReturnValue)
   > Branch(Condition=Select(Option 0="integer > integer"(A=@n13, B=@n12), Option 1="integer < integer"(A=@n13, B=@n12), Index=IsCultureRightToLeft(Culture=@n11)))
   > Set MaxIndex(MaxIndex=@n13) -> (Output_Get=0)
   ResetLableSwitcher(execute=@n5.else, self:WBP Season Pass Character Button Object Reference=Get Button())
   n14 = Set MaxIndex(execute=@n1.else, MaxIndex=0) -> (Output_Get=0)
  ]}
 SetCompactUIStyle {nodes [
   SetCompactUIStyle() {is=title K2Node_FunctionEntry}
   > Branch(Condition=IsUsingCompactUI())
   n1 = CompareToScreenAspectRatio!(Comparison:EAspectRatioComparison Enum=GreaterThanOrEqualTo, AspectRatio:EAspectRatio Enum=Ratio_10_7) -> (ReturnValue:bool)
   > SetPadding!(
    execute
    self:Overlay Slot Object Reference=SlotAsOverlaySlot(Widget=Get SeasonPass_Header())
    InPadding_Left:float=0
    InPadding_Top:float=0
    InPadding_Right:float=0
    InPadding_Bottom:float=SelectFloat(A=112, B=0, bPickA=@n1)
   ) -> (then)
   n2 = SlotAsVerticalBoxSlot(Widget=Get Overlay_1() -> (Overlay_1:Overlay Object Reference))
   n3 = Get!(self:Vertical Box Slot Object Reference=@n2) -> (Padding_Left:float=~, Padding_Top:float=~, Padding_Right:float=~, Padding_Bottom:float=~)
   > SetPadding(
    self:Vertical Box Slot Object Reference=@n2
    InPadding_Left=@n3
    InPadding_Top=SelectFloat(A=24, B=60, bPickA=@n1)
    InPadding_Right=@n3.Padding_Right
    InPadding_Bottom=72
   )
   > Branch(Condition=@n1)
   n4 = SlotAsScrollBoxSlot!(Widget:Widget Object Reference=Get HB_Sets() -> (HB_Sets:Horizontal Box Object Reference)) -> (ReturnValue:Scroll Box Slot Object Reference)
   n5 = Get(self:Scroll Box Slot Object Reference=@n4) -> (Padding_Left=~, Padding_Top=~, Padding_Right=~, Padding_Bottom=~)
   > SetPadding(self:Scroll Box Slot Object Reference=@n4, InPadding_Left=@n5, InPadding_Top=54, InPadding_Right=@n5.Padding_Right, InPadding_Bottom=16)
  ]}
 SetIsMainSetComplete {nodes [
   n1 = SetIsMainSetComplete!() -> (then, SeasonPassCategoryVM:Array of Season Pass Category VM Object References) {is=title K2Node_FunctionEntry}
   > n2 = Set SeasonPassCategories(SeasonPassCategories:Array of Season Pass Category VM Object References=@n1) -> (Output_Get:Array of Season Pass Category VM Object References)
   > n3 = For Each Loop(Array:Array of Season Pass Category VM Object References=@n2) -> (Array Element:Season Pass Category VM Object Reference)
   n4 = "Equal (Integer)"(
    A=Get FullyClaimedMainCategorySets() -> (FullyClaimedMainCategorySets:int=0)
    B=Get NumberOfMainCategorySets() -> (NumberOfMainCategorySets:int=0)
   )
   Set MainSetFullyClaimedChanged(execute=@n3.Completed, MainSetFullyClaimedChanged:bool="Not Equal (Boolean)"(A=@n4, B=Get IsMainSetFullyClaimed()) -> (ReturnValue=~)) -> (Output_Get:bool)
   > Set with Broadcast IsMainSetFullyClaimed!(execute, IsMainSetFullyClaimed:bool=@n4) -> (then, Output_Get:bool)
   > n5 = Branch(Condition=Get MainSetFullyClaimedChanged() -> (MainSetFullyClaimedChanged:bool))
   > n6 = ScrollSets(ScrollToBonus=Get IsMainSetFullyClaimed(), AnimateScroll=false) {is=title K2Node_CallFunction}
   HandleBonusRewardsCollapsedTile(execute=[@n6.then, @n5.else]) {is=title K2Node_CallFunction}
   Branch(execute=@n3.LoopBody, Condition="Equal Exactly (String)"(A=GetGroupName(self=@n3), B=Default))
   > Increment Int(Value=Get NumberOfMainCategorySets() -> (NumberOfMainCategorySets=0))
   > Branch(Condition=Get bIsFullyClaimed!(self:Season Pass Category VM Object Reference=@n3) -> (bIsFullyClaimed:bool))
   > Increment Int(Value=Get FullyClaimedMainCategorySets() -> (FullyClaimedMainCategorySets=0))
  ]}
 SetUnlocked {nodes [SetUnlocked!() -> (then, Category:Season Pass Category VM Object Reference) {is=title K2Node_FunctionEntry}]}
]
variables [
 CascadeIntroDelayAmt {type=number, default=0.12}
 CascadeIntroDelayTotal {type=number}
 CategoryListButton {type=WBP_SeasonPass_CharacterButton_C, default=null}
 CategoryListElementVM {type=SeasonPassCategoryVM, default=null}
 CategoryListIndex {type=integer}
 Clear Preview {}
 EnteredPassSubscreen {}
 FallbackFocusWidgetWhenPassIsMissing {type=Widget, default=null}
 IsActivatingFromSubscreen {type=boolean}
 IsCultureRightToLeft {type=boolean}
 IsMainSetFullyClaimed {type=boolean}
 IsPageUp {type=boolean}
 IsShowingMainSet {type=boolean, default=true}
 LastTileFocused {type=UserWidget, default=null}
 NavUpButNoButtonsLeft {}
 OnNavigateToTilesWithoutPass {}
 Season Pass Category VMs {type="array<SeasonPassCategoryVM>", default []}
 SequenceTimerHandler {type=TimerHandle}
 SetUnlocked {}
]
widgets {
 info {inheritedWidgetCount=0, parentClass=&CommonActivatableWidget, rootWidgetClass=&VerticalBox, widgetCount=18}
 widgets [
  Content {widgetClassPath=&VerticalBox, bIsVariable=true, widget=Content}
  Overlay_3 {
   widgetClassPath=&Overlay
   parent=Content
   widget=Overlay_3
   slot {class=VerticalBoxSlot, properties {size {sizeRule=Fill, value=1}}}
  }
  SeasonPass_Header {
   widgetClassPath=&/WBP_SeasonPass_Header
   bIsVariable=true
   parent=Overlay_3
   widget=SeasonPass_Header
   properties {navigation=&self:WidgetTree.SeasonPass_Header.WidgetNavigation_1, renderTransformPivot=<1 1>}
   slot {class=OverlaySlot, properties {padding=<ltrb 0 80 0 0>}}
  }
  SeasonPassBubble {
   widgetClassPath=/SeasonPass/Widgets/Reboot/WBP_SeasonPassBubble
   bIsVariable=true
   parent=Overlay_3
   widget=SeasonPassBubble
   properties {visibility=Hidden, renderTransformPivot=<0 1>}
   slot {class=OverlaySlot, properties {horizontalAlignment=HAlign_Right, verticalAlignment=VAlign_Bottom}}
  }
  Overlay_1 {
   widgetClassPath=&Overlay
   bIsVariable=true
   parent=Content
   widget=Overlay_1
   slot {class=VerticalBoxSlot, properties {padding=<ltrb 0 0 0 86>}}
  }
  PostBufferUpdate_92 {
   widgetClassPath=/Script/UMG.PostBufferUpdate
   parent=Overlay_1
   widget=PostBufferUpdate_92
   properties {
    visibility=SelfHitTestInvisible
    renderTransform {angle=0, scale=<1 1.1>, shear=<0 0>, translation=<0 0>}
    updateBufferInfos [{bufferToUpdate=ESlatePostRT_1, postParamUpdater=null}]
   }
   slot {class=OverlaySlot, properties {horizontalAlignment=HAlign_Fill, verticalAlignment=VAlign_Fill}}
  }
  WBP_SeasonPass_CharacterName {
   widgetClassPath=&/WBP_SeasonPass_CharacterName
   bIsVariable=true
   parent=Overlay_1
   widget=WBP_SeasonPass_CharacterName
   slot {class=OverlaySlot, properties {horizontalAlignment=HAlign_Center, verticalAlignment=VAlign_Center, padding=<ltrb 0 0 0 140>}}
  }
  Overlay_Sets {
   widgetClassPath=&Overlay
   parent=Overlay_1
   widget=Overlay_Sets
   properties {navigation=&self:WidgetTree.Overlay_Sets.WidgetNavigation_0}
   slot {class=OverlaySlot, properties {horizontalAlignment=HAlign_Fill, verticalAlignment=VAlign_Center}}
  }
  ScrollBox_0 {
   widgetClassPath=&ScrollBox
   bIsVariable=true
   parent=Overlay_Sets
   widget=ScrollBox_0
   properties {
    consumeMouseWheel=Never
    navigationDestination=Center
    orientation=Orient_Horizontal
    scrollBarVisibility=Collapsed
    renderTransformPivot=<0.5 0>
    widgetBarStyle {
     thickness=0
     draggedThumbImage {
      bIsDynamicallyLoaded=false
      drawAs=Box
      imageType=FullColor
      mirroring=NoMirror
      resourceName=../../../Engine/Content/Slate/Common/Scrollbar_Thumb.png
      resourceObject=null
      tiling=NoTile
      imageSize=<0 0>
      margin=<ltrb 0.25 0.25 0.25 0.25>
      outlineSettings {
       bUseBrushTransparency=false
       roundingType=HalfHeightRadius
       width=0
       color {colorUseRule=UseColor_Specified, specifiedColor=<rgba 0 0 0 0>}
       cornerRadii=<0 0 0 0>
      }
      tintColor {colorUseRule=UseColor_Specified, specifiedColor=<rgba 1 1 1 1>}
      uVRegion {bIsValid=false, max=<0 0>, min=<0 0>}
     }
     horizontalBackgroundImage {
      bIsDynamicallyLoaded=false
      drawAs=NoDrawType
      imageType=NoImage
      mirroring=NoMirror
      resourceName=None
      resourceObject=null
      tiling=NoTile
      imageSize=<0 0>
      margin=<ltrb 0 0 0 0>
      outlineSettings {
       bUseBrushTransparency=false
       roundingType=HalfHeightRadius
       width=0
       color {colorUseRule=UseColor_Specified, specifiedColor=<rgba 0 0 0 0>}
       cornerRadii=<0 0 0 0>
      }
      tintColor {colorUseRule=UseColor_Specified, specifiedColor=<rgba 1 1 1 1>}
      uVRegion {bIsValid=false, max=<0 0>, min=<0 0>}
     }
     horizontalBottomSlotImage {
      bIsDynamicallyLoaded=false
      drawAs=Image
      imageType=FullColor
      mirroring=NoMirror
      resourceName=../../../Engine/Content/Slate/Common/Scrollbar_Background_Horizontal.png
      resourceObject=null
      tiling=NoTile
      imageSize=<8 8>
      margin=<ltrb 0 0 0 0>
      outlineSettings {
       bUseBrushTransparency=false
       roundingType=HalfHeightRadius
       width=0
       color {colorUseRule=UseColor_Specified, specifiedColor=<rgba 0 0 0 0>}
       cornerRadii=<0 0 0 0>
      }
      tintColor {colorUseRule=UseColor_Specified, specifiedColor=<rgba 1 1 1 1>}
      uVRegion {bIsValid=false, max=<0 0>, min=<0 0>}
     }
     horizontalTopSlotImage {
      bIsDynamicallyLoaded=false
      drawAs=Image
      imageType=FullColor
      mirroring=NoMirror
      resourceName=../../../Engine/Content/Slate/Common/Scrollbar_Background_Horizontal.png
      resourceObject=null
      tiling=NoTile
      imageSize=<8 8>
      margin=<ltrb 0 0 0 0>
      outlineSettings {
       bUseBrushTransparency=false
       roundingType=HalfHeightRadius
       width=0
       color {colorUseRule=UseColor_Specified, specifiedColor=<rgba 0 0 0 0>}
       cornerRadii=<0 0 0 0>
      }
      tintColor {colorUseRule=UseColor_Specified, specifiedColor=<rgba 1 1 1 1>}
      uVRegion {bIsValid=false, max=<0 0>, min=<0 0>}
     }
     hoveredThumbImage {
      bIsDynamicallyLoaded=false
      drawAs=Box
      imageType=FullColor
      mirroring=NoMirror
      resourceName=../../../Engine/Content/Slate/Common/Scrollbar_Thumb.png
      resourceObject=null
      tiling=NoTile
      imageSize=<0 0>
      margin=<ltrb 0.25 0.25 0.25 0.25>
      outlineSettings {
       bUseBrushTransparency=false
       roundingType=HalfHeightRadius
       width=0
       color {colorUseRule=UseColor_Specified, specifiedColor=<rgba 0 0 0 0>}
       cornerRadii=<0 0 0 0>
      }
      tintColor {colorUseRule=UseColor_Specified, specifiedColor=<rgba 1 1 1 1>}
      uVRegion {bIsValid=false, max=<0 0>, min=<0 0>}
     }
     normalThumbImage {
      bIsDynamicallyLoaded=false
      drawAs=Box
      imageType=FullColor
      mirroring=NoMirror
      resourceName=../../../Engine/Content/Slate/Common/Scrollbar_Thumb.png
      resourceObject=null
      tiling=NoTile
      imageSize=<0 0>
      margin=<ltrb 0.25 0.25 0.25 0.25>
      outlineSettings {
       bUseBrushTransparency=false
       roundingType=HalfHeightRadius
       width=0
       color {colorUseRule=UseColor_Specified, specifiedColor=<rgba 0 0 0 0>}
       cornerRadii=<0 0 0 0>
      }
      tintColor {colorUseRule=UseColor_Specified, specifiedColor=<rgba 1 1 1 1>}
      uVRegion {bIsValid=false, max=<0 0>, min=<0 0>}
     }
     verticalBackgroundImage {
      bIsDynamicallyLoaded=false
      drawAs=NoDrawType
      imageType=NoImage
      mirroring=NoMirror
      resourceName=None
      resourceObject=null
      tiling=NoTile
      imageSize=<0 0>
      margin=<ltrb 0 0 0 0>
      outlineSettings {
       bUseBrushTransparency=false
       roundingType=HalfHeightRadius
       width=0
       color {colorUseRule=UseColor_Specified, specifiedColor=<rgba 0 0 0 0>}
       cornerRadii=<0 0 0 0>
      }
      tintColor {colorUseRule=UseColor_Specified, specifiedColor=<rgba 1 1 1 1>}
      uVRegion {bIsValid=false, max=<0 0>, min=<0 0>}
     }
     verticalBottomSlotImage {
      bIsDynamicallyLoaded=false
      drawAs=Image
      imageType=FullColor
      mirroring=NoMirror
      resourceName=../../../Engine/Content/Slate/Common/Scrollbar_Background_Vertical.png
      resourceObject=null
      tiling=NoTile
      imageSize=<8 8>
      margin=<ltrb 0 0 0 0>
      outlineSettings {
       bUseBrushTransparency=false
       roundingType=HalfHeightRadius
       width=0
       color {colorUseRule=UseColor_Specified, specifiedColor=<rgba 0 0 0 0>}
       cornerRadii=<0 0 0 0>
      }
      tintColor {colorUseRule=UseColor_Specified, specifiedColor=<rgba 1 1 1 1>}
      uVRegion {bIsValid=false, max=<0 0>, min=<0 0>}
     }
     verticalTopSlotImage {
      bIsDynamicallyLoaded=false
      drawAs=Image
      imageType=FullColor
      mirroring=NoMirror
      resourceName=../../../Engine/Content/Slate/Common/Scrollbar_Background_Vertical.png
      resourceObject=null
      tiling=NoTile
      imageSize=<8 8>
      margin=<ltrb 0 0 0 0>
      outlineSettings {
       bUseBrushTransparency=false
       roundingType=HalfHeightRadius
       width=0
       color {colorUseRule=UseColor_Specified, specifiedColor=<rgba 0 0 0 0>}
       cornerRadii=<0 0 0 0>
      }
      tintColor {colorUseRule=UseColor_Specified, specifiedColor=<rgba 1 1 1 1>}
      uVRegion {bIsValid=false, max=<0 0>, min=<0 0>}
     }
    }
    widgetStyle {
     barThickness=0
     bottomShadowBrush {
      bIsDynamicallyLoaded=false
      drawAs=NoDrawType
      imageType=FullColor
      mirroring=NoMirror
      resourceName=../../../Engine/Content/Slate/Common/ScrollBoxShadowBottom.png
      resourceObject=null
      tiling=NoTile
      imageSize=<16 8>
      margin=<ltrb 0.5 0 0.5 1>
      outlineSettings {
       bUseBrushTransparency=false
       roundingType=HalfHeightRadius
       width=0
       color {colorUseRule=UseColor_Specified, specifiedColor=<rgba 0 0 0 0>}
       cornerRadii=<0 0 0 0>
      }
      tintColor {colorUseRule=UseColor_Specified, specifiedColor=<rgba 1 1 1 1>}
      uVRegion {bIsValid=false, max=<0 0>, min=<0 0>}
     }
     horizontalScrolledContentPadding=<ltrb 0 0 1 0>
     leftShadowBrush {
      bIsDynamicallyLoaded=false
      drawAs=Image
      imageType=FullColor
      mirroring=NoMirror
      resourceName=None
      resourceObject=/SeasonPassHub/Widgets/Reboot/Materials/MI_UI_SeasonPass_ListFade_Left
      tiling=NoTile
      imageSize=<96 32>
      margin=<ltrb 1 0.5 0 0.5>
      outlineSettings {
       bUseBrushTransparency=false
       roundingType=HalfHeightRadius
       width=0
       color {colorUseRule=UseColor_Specified, specifiedColor=<rgba 0 0 0 0>}
       cornerRadii=<0 0 0 0>
      }
      tintColor {colorUseRule=UseColor_Specified, specifiedColor=<rgba 1 1 1 1>}
      uVRegion {bIsValid=false, max=<0 0>, min=<0 0>}
     }
     rightShadowBrush {
      bIsDynamicallyLoaded=false
      drawAs=Image
      imageType=FullColor
      mirroring=NoMirror
      resourceName=None
      resourceObject=/SeasonPassHub/Widgets/Reboot/Materials/MI_UI_SeasonPass_ListFade_Right
      tiling=NoTile
      imageSize=<96 32>
      margin=<ltrb 0 0.5 1 0.5>
      outlineSettings {
       bUseBrushTransparency=false
       roundingType=HalfHeightRadius
       width=0
       color {colorUseRule=UseColor_Specified, specifiedColor=<rgba 0 0 0 0>}
       cornerRadii=<0 0 0 0>
      }
      tintColor {colorUseRule=UseColor_Specified, specifiedColor=<rgba 1 1 1 1>}
      uVRegion {bIsValid=false, max=<0 0>, min=<0 0>}
     }
     topShadowBrush {
      bIsDynamicallyLoaded=false
      drawAs=NoDrawType
      imageType=FullColor
      mirroring=NoMirror
      resourceName=../../../Engine/Content/Slate/Common/ScrollBoxShadowTop.png
      resourceObject=null
      tiling=NoTile
      imageSize=<16 8>
      margin=<ltrb 0.5 1 0.5 0>
      outlineSettings {
       bUseBrushTransparency=false
       roundingType=HalfHeightRadius
       width=0
       color {colorUseRule=UseColor_Specified, specifiedColor=<rgba 0 0 0 0>}
       cornerRadii=<0 0 0 0>
      }
      tintColor {colorUseRule=UseColor_Specified, specifiedColor=<rgba 1 1 1 1>}
      uVRegion {bIsValid=false, max=<0 0>, min=<0 0>}
     }
     verticalScrolledContentPadding=<ltrb 0 0 0 1>
    }
   }
   slot {class=OverlaySlot, properties {horizontalAlignment=HAlign_Center, verticalAlignment=VAlign_Fill}}
  }
  HB_Sets {
   widgetClassPath=&HorizontalBox
   bIsVariable=true
   parent=ScrollBox_0
   widget=HB_Sets
   slot {class=ScrollBoxSlot, properties {verticalAlignment=VAlign_Bottom, padding=<ltrb 32 54 32 32>}}
  }
  FrontSpacer {
   widgetClassPath=&Spacer
   bIsVariable=true
   parent=HB_Sets
   widget=FrontSpacer
   properties {size=<750 1>}
   slot {class=HorizontalBoxSlot}
  }
  MainSets {
   widgetClassPath=&DynamicEntryBox
   bIsVariable=true
   parent=HB_Sets
   widget=MainSets
   properties {
    entryWidgetClass=&/WBP_SeasonPass_CharacterButton.WBP_SeasonPass_CharacterButton_C
    navigation=&self:WidgetTree.MainSets.WidgetNavigation_0
    numDesignerPreviewEntries=8
    entrySpacing=<24 0>
   }
   slot {class=HorizontalBoxSlot, properties {horizontalAlignment=HAlign_Center, verticalAlignment=VAlign_Center}}
  }
  Divider {
   widgetClassPath=&Image
   bIsVariable=true
   parent=HB_Sets
   widget=Divider
   properties {
    renderOpacity=0.25
    visibility=SelfHitTestInvisible
    brush {
     bIsDynamicallyLoaded=false
     drawAs=Image
     imageType=FullColor
     mirroring=NoMirror
     resourceName=None
     resourceObject=/FortUILibrary/Textures/Icons/T_UI_Icon_UIKit_ContextMenu
     tiling=NoTile
     imageSize=<32 32>
     margin=<ltrb 0 0 0 0>
     outlineSettings {
      bUseBrushTransparency=false
      roundingType=HalfHeightRadius
      width=0
      color {colorUseRule=UseColor_Specified, specifiedColor=<rgba 0 0 0 0>}
      cornerRadii=<0 0 0 0>
     }
     tintColor {colorUseRule=UseColor_Specified, specifiedColor=<rgba 1 1 1 1>}
     uVRegion {bIsValid=false, max=<0 0>, min=<0 0>}
    }
   }
   slot {class=HorizontalBoxSlot, properties {horizontalAlignment=HAlign_Center, verticalAlignment=VAlign_Center, padding=<ltrb 28 48 28 0>}}
  }
  BonusSets {
   widgetClassPath=&DynamicEntryBox
   bIsVariable=true
   parent=HB_Sets
   widget=BonusSets
   properties {
    entryWidgetClass=&WBP_SeasonPass_CharacterButton_C
    navigation=&self:WidgetTree.BonusSets.WidgetNavigation_0
    numDesignerPreviewEntries=2
    entrySpacing=<24 0>
   }
   slot {class=HorizontalBoxSlot}
  }
  EndSpacerForBonusSets {
   widgetClassPath=&Spacer
   bIsVariable=true
   parent=HB_Sets
   widget=EndSpacerForBonusSets
   properties {size=<750 1>}
   slot {class=HorizontalBoxSlot}
  }
  MKBScrollArrows {
   widgetClassPath=&HorizontalBox
   bIsVariable=true
   parent=Overlay_Sets
   widget=MKBScrollArrows
   slot {class=OverlaySlot, properties {horizontalAlignment=HAlign_Fill, verticalAlignment=VAlign_Bottom, padding=<ltrb 100 0 100 75>}}
  }
  ScrollLeft {
   widgetClassPath=&WBP_UIKit_Button_Regular
   bIsVariable=true
   parent=MKBScrollArrows
   widget=ScrollLeft
   properties {iconMaterial=/Crew/UI/Materials/MI_UI_ScrollArrowLeft, text="", visibility=Collapsed}
   slot {class=HorizontalBoxSlot}
  }
  ScrollRight {
   widgetClassPath=&WBP_UIKit_Button_Regular
   bIsVariable=true
   parent=MKBScrollArrows
   widget=ScrollRight
   properties {iconMaterial=/Crew/UI/Materials/MI_UI_ScrollArrowRight, text="", visibility=Collapsed}
   slot {class=HorizontalBoxSlot, properties {horizontalAlignment=HAlign_Right, size {sizeRule=Fill, value=1}}}
  }
 ]
}
