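# Season pass hub page navigation widget: a labeled prompt with an animated directional chevron and an invisible full-size hitbox button.
# SetData fills title and description, collapsing an empty description, and shows one of four chevrons by enum.
# Clicking or focusing the hitbox fires an element-interacted analytics event tagged with the label, plays the pressed sound and broadcasts Navigate.
# Hover and unhover play the scale animation forward or reversed, with sounds.
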WidgetBlueprint /SeasonPassHub/Widgets/Reboot/WBP_SeasonPassPageNav
animations [OnHover {
  binding_count=1
  length_seconds=0.2
  object=&self:OnHover
  sequence {
   bindings [Overlay_0 {
     id=#1
     possessed_class=Overlay
     tracks [{
       type=MovieScene2DTransformTrack
       sections [{
         type=MovieScene2DTransformSection
         end_frame=12
         start_frame=0
         channels [
          Scale.X {type=MovieSceneScriptingFloatChannel, keys [[0,1],[12,1.1]]}
          Scale.Y {type=MovieSceneScriptingFloatChannel, keys [[0,1],[12,1.1]]}
         ]
        }]
      }]
    }]
   display_rate {numerator=60}
  }
 }]
graphs [
 EventGraph {nodes [
   n1 = "On Clicked (Nav_Button_Hitbox)"!() -> (OutputDelegate, then, Button:Common Button Base Object Reference) {is=title K2Node_ComponentBoundEvent}
   n2 = Get Text(self:Text Object Reference=Get Label() -> (Label:Common Text Object Reference))
   Fire Event Element Interacted(
    execute=[@n1.then, @n3.then]
    Element=@n1.Button
    AdditionalInfo="Add String (Non-Localized)"!(
     Info:"Flow Analytics Additional Info Structure (by ref)"=Make Flow Analytics Additional Info!() -> (FlowAnalyticsAdditionalInfo:Flow Analytics Additional Info Structure) {is=title K2Node_MakeStruct}
     Key:String=Nav
     Value:String="To String (Text)"(InText=@n2)
    ) -> (ReturnValue:Flow Analytics Additional Info Structure)
   )
   > PlaySound2D(Sound=Get PressedSound() -> (PressedSound:Sound Base Object Reference))
   > Call Navigate!(execute:Exec, self:WBP Season Pass Page Nav Object Reference) -> (then:Exec) {is=title K2Node_CallDelegate}
   "On Hovered (Nav_Button_Hitbox)"!() -> (OutputDelegate, then, Button:Common Button Base Object Reference) {is=title K2Node_ComponentBoundEvent}
   > PlayAnimationForward(InAnimation=Get OnHover() -> (OnHover:Widget Animation Object Reference))
   > PlaySound2D(Sound=Get HoveredSound() -> (HoveredSound:Sound Base Object Reference))
   "On Unhovered (Nav_Button_Hitbox)"!() -> (OutputDelegate, then, Button:Common Button Base Object Reference) {is=title K2Node_ComponentBoundEvent}
   > PlayAnimationReverse(InAnimation=Get OnHover())
   > PlaySound2D(Sound=Get UnhoveredSound() -> (UnhoveredSound:Sound Base Object Reference))
   n3 = "On Focused (Nav_Button_Hitbox)"!() -> (OutputDelegate, then, Button:Common Button Base Object Reference) {is=title K2Node_ComponentBoundEvent}
  ]}
 Navigate {nodes [Navigate() {is=title K2Node_FunctionEntry}]}
 SetChevronPosition {nodes [
   n1 = SetChevronPosition!() -> (then, Position:E_ChevronPosition Enum=NewEnumerator0) {is=title K2Node_FunctionEntry}
   > SetVisibility(
    self=[
     Get Chevron_Icon_Right() -> (Chevron_Icon_Right:Image Object Reference)
     Get Chevron_Icon_Left() -> (Chevron_Icon_Left:Image Object Reference)
     Get Chevron_Icon_Top() -> (Chevron_Icon_Top:Image Object Reference)
     Get Chevron_Icon_Bottom() -> (Chevron_Icon_Bottom:Image Object Reference)
    ]
    InVisibility=Collapsed
   )
   > n2 = Switch on E_ChevronPosition!(execute, Selection:E_ChevronPosition Enum=@n1) -> (NewEnumerator0:Exec, NewEnumerator1:Exec, NewEnumerator2:Exec, NewEnumerator3:Exec) {is=title K2Node_SwitchEnum}
   > SetVisibility(self=Get Chevron_Icon_Top(), InVisibility=SelfHitTestInvisible)
   SetVisibility(execute=@n2.NewEnumerator1, self=Get Chevron_Icon_Bottom(), InVisibility=SelfHitTestInvisible)
   SetVisibility(execute=@n2.NewEnumerator2, self=Get Chevron_Icon_Left(), InVisibility=SelfHitTestInvisible)
   SetVisibility(execute=@n2.NewEnumerator3, self=Get Chevron_Icon_Right(), InVisibility=SelfHitTestInvisible)
  ]}
 SetData {nodes [
   n1 = SetData!() -> (then, PageNavData:F Page Nav Data Structure) {is=title K2Node_FunctionEntry}
   n2 = Break F Page Nav Data!(F_PageNav_Data:"F Page Nav Data Structure (by ref)"=@n1) -> (
    Title_5_#2:Text
    Description_7_#3:Text
    ChevronPosition_2_#4:E_ChevronPosition Enum
   ) {is=title K2Node_BreakStruct}
   > SetText!(execute, self:Self Object Reference, Title:Text=@n2, Description:Text=@n2.Description_7_#3) -> (then)
   > SetChevronPosition!(execute, self, Position:E_ChevronPosition Enum=@n2.ChevronPosition_2_#4) -> (then) {is=title K2Node_CallFunction}
  ]}
 SetText {nodes [
   n1 = SetText!() -> (then, Title:Text, Description:Text) {is=title K2Node_FunctionEntry}
   > Set Text(Text=@n1, self=Get Label())
   > n2 = Set Text(Text=@n1.Description, self=Get Description() -> (Description:Common Text Object Reference))
   n3 = "To String (Text)"(InText=@n2)
   > SetVisibility(self=Get Description(), InVisibility=Select(Option 0=SelfHitTestInvisible, Option 1=Collapsed, Index="Equal Exactly (String)"(A=@n3)))
  ]}
]
variables [
 HoveredSound {type=SoundBase, default=null}
 Navigate {}
 PressedSound {type=SoundBase, default=null}
 UnhoveredSound {type=SoundBase, default=null}
]
widgets {
 info {inheritedWidgetCount=0, parentClass=&CommonActivatableWidget, rootWidgetClass=&Overlay, widgetCount=11}
 widgets [
  Overlay_0 {widgetClassPath=&Overlay, widget=Overlay_0}
  VerticalBox_0 {widgetClassPath=&VerticalBox, parent=Overlay_0, widget=VerticalBox_0, slot {class=OverlaySlot}}
  Chevron_Icon_Top {
   widgetClassPath=&Image
   bIsVariable=true
   parent=VerticalBox_0
   widget=Chevron_Icon_Top
   properties {
    visibility=Collapsed
    brush {
     bIsDynamicallyLoaded=false
     drawAs=Image
     imageType=FullColor
     mirroring=NoMirror
     resourceName=None
     resourceObject=/SeasonPassHub/Widgets/Materials/M_UI_AnimatedChevron
     tiling=NoTile
     imageSize=<32 32>
     margin=<ltrb 0 0 0 0>
     outlineSettings {
      bUseBrushTransparency=false
      roundingType=HalfHeightRadius
      width=0
      color {colorUseRule=UseColor_Specified, specifiedColor=<rgba 0 0 0 0>}
      cornerRadii=<0 0 0 0>
     }
     tintColor {colorUseRule=UseColor_Specified, specifiedColor=<rgba 1 1 1 1>}
     uVRegion {bIsValid=false, max=<0 0>, min=<0 0>}
    }
    renderTransform {angle=-90, scale=<1 1>, shear=<0 0>, translation=<0 0>}
   }
   slot {class=VerticalBoxSlot, properties {horizontalAlignment=HAlign_Center, verticalAlignment=VAlign_Bottom, padding=<ltrb 0 0 0 8>}}
  }
  HorizontalBox_0 {
   widgetClassPath=&HorizontalBox
   parent=VerticalBox_0
   widget=HorizontalBox_0
   slot {class=VerticalBoxSlot, properties {horizontalAlignment=HAlign_Center}}
  }
  Chevron_Icon_Left {
   widgetClassPath=&Image
   bIsVariable=true
   parent=HorizontalBox_0
   widget=Chevron_Icon_Left
   properties {
    visibility=Collapsed
    brush {
     bIsDynamicallyLoaded=false
     drawAs=Image
     imageType=FullColor
     mirroring=NoMirror
     resourceName=None
     resourceObject=&M_UI_AnimatedChevron
     tiling=NoTile
     imageSize=<32 32>
     margin=<ltrb 0 0 0 0>
     outlineSettings {
      bUseBrushTransparency=false
      roundingType=HalfHeightRadius
      width=0
      color {colorUseRule=UseColor_Specified, specifiedColor=<rgba 0 0 0 0>}
      cornerRadii=<0 0 0 0>
     }
     tintColor {colorUseRule=UseColor_Specified, specifiedColor=<rgba 1 1 1 1>}
     uVRegion {bIsValid=false, max=<0 0>, min=<0 0>}
    }
    renderTransform {angle=180, scale=<1 1>, shear=<0 0>, translation=<0 0>}
   }
   slot {class=HorizontalBoxSlot, properties {horizontalAlignment=HAlign_Left, verticalAlignment=VAlign_Center, padding=<ltrb 0 0 8 0>}}
  }
  VerticalBox_1 {widgetClassPath=&VerticalBox, parent=HorizontalBox_0, widget=VerticalBox_1, slot {class=HorizontalBoxSlot}}
  Label {
   widgetClassPath=&CommonTextBlock
   bIsVariable=true
   parent=VerticalBox_1
   widget=Label
   properties {
    mobileFontSizeMultiplier=1.19
    style=&TextStyle-Base_C
    text=Navigate To
    font {
     bForceMonospaced=false
     bMaterialIsStencil=false
     fontMaterial=null
     fontObject=&NotoSans
     letterSpacing=0
     monospacedWidth=1
     size=24
     skewAmount=0
     typefaceFontName=NotoSans-Regular
     outlineSettings {
      bApplyOutlineToDropShadows=false
      bMiteredCorners=false
      bSeparateFillAlpha=false
      outlineMaterial=null
      outlineSize=0
      outlineColor=<rgba 0 0 0 1>
     }
    }
    shadowOffset=<0 0>
   }
   slot {class=VerticalBoxSlot, properties {horizontalAlignment=HAlign_Center, verticalAlignment=VAlign_Top}}
  }
  Description {
   widgetClassPath=&CommonTextBlock
   bIsVariable=true
   parent=VerticalBox_1
   widget=Description
   properties {
    text=Description
    visibility=Collapsed
    font {
     bForceMonospaced=false
     bMaterialIsStencil=false
     fontMaterial=null
     fontObject=&HeadingNow
     letterSpacing=0
     monospacedWidth=1
     size=18
     skewAmount=0
     typefaceFontName=75 Medium
     outlineSettings {
      bApplyOutlineToDropShadows=false
      bMiteredCorners=false
      bSeparateFillAlpha=false
      outlineMaterial=null
      outlineSize=0
      outlineColor=<rgba 0 0 0 1>
     }
    }
    shadowOffset=<0 0>
   }
   slot {class=VerticalBoxSlot, properties {horizontalAlignment=HAlign_Center, verticalAlignment=VAlign_Top}}
  }
  Chevron_Icon_Right {
   widgetClassPath=&Image
   bIsVariable=true
   parent=HorizontalBox_0
   widget=Chevron_Icon_Right
   properties {
    visibility=Collapsed
    brush {
     bIsDynamicallyLoaded=false
     drawAs=Image
     imageType=FullColor
     mirroring=NoMirror
     resourceName=None
     resourceObject=&M_UI_AnimatedChevron
     tiling=NoTile
     imageSize=<32 32>
     margin=<ltrb 0 0 0 0>
     outlineSettings {
      bUseBrushTransparency=false
      roundingType=HalfHeightRadius
      width=0
      color {colorUseRule=UseColor_Specified, specifiedColor=<rgba 0 0 0 0>}
      cornerRadii=<0 0 0 0>
     }
     tintColor {colorUseRule=UseColor_Specified, specifiedColor=<rgba 1 1 1 1>}
     uVRegion {bIsValid=false, max=<0 0>, min=<0 0>}
    }
   }
   slot {class=HorizontalBoxSlot, properties {horizontalAlignment=HAlign_Right, verticalAlignment=VAlign_Center, padding=<ltrb 8 0 0 0>}}
  }
  Chevron_Icon_Bottom {
   widgetClassPath=&Image
   bIsVariable=true
   parent=VerticalBox_0
   widget=Chevron_Icon_Bottom
   properties {
    visibility=Collapsed
    brush {
     bIsDynamicallyLoaded=false
     drawAs=Image
     imageType=FullColor
     mirroring=NoMirror
     resourceName=None
     resourceObject=&M_UI_AnimatedChevron
     tiling=NoTile
     imageSize=<32 32>
     margin=<ltrb 0 0 0 0>
     outlineSettings {
      bUseBrushTransparency=false
      roundingType=HalfHeightRadius
      width=0
      color {colorUseRule=UseColor_Specified, specifiedColor=<rgba 0 0 0 0>}
      cornerRadii=<0 0 0 0>
     }
     tintColor {colorUseRule=UseColor_Specified, specifiedColor=<rgba 1 1 1 1>}
     uVRegion {bIsValid=false, max=<0 0>, min=<0 0>}
    }
    renderTransform {angle=90, scale=<1 1>, shear=<0 0>, translation=<0 0>}
   }
   slot {class=VerticalBoxSlot, properties {horizontalAlignment=HAlign_Center, verticalAlignment=VAlign_Top, padding=<ltrb 0 8 0 0>}}
  }
  Nav_Button_Hitbox {
   widgetClassPath=&WBP_UIKit_Button_Generic
   bIsVariable=true
   parent=Overlay_0
   widget=Nav_Button_Hitbox
   properties {
    bHideInputAction=true
    bSimulateHoverOnTouchInput=false
    contentHorizontalAlignment=HAlign_Fill
    contentVerticalAlignment=VAlign_Fill
    secondText=""
    touchMethod=PreciseTap
    useBackgroundBlock=false
    useIconBlock=false
    useTextBlock=false
    visibility=Visible
   }
   slot {class=OverlaySlot, properties {horizontalAlignment=HAlign_Fill, verticalAlignment=VAlign_Fill}}
  }
 ]
}
