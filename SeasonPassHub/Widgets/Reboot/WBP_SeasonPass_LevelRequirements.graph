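# Season pass label widget that shows a "Choose a Set to Unlock at LVL" heading beside a level number.
# CF_GetVisibilityForSelf returns HitTestInvisible when the next unlock level is above the current level and below a cap, otherwise Collapsed.
# The cap is higher when the pass is owned.
# The event graph is empty, so nothing here calls the function or updates the placeholder number.
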WidgetBlueprint /SeasonPassHub/Widgets/Reboot/WBP_SeasonPass_LevelRequirements
graphs [
 CF_GetVisibilityForSelf {nodes [
   n1 = CF_GetVisibilityForSelf!() -> (then, LevelNextUnlock:int, CurrentLevel:int, OwnsPass:bool=~) {is=title K2Node_FunctionEntry}
   n2 = "integer > integer"(A=@n1, B=@n1.CurrentLevel)
   n3 = "integer < integer"(A=@n1, B=Select(Option 0:int=100, Option 1:int=151, Index=@n1.OwnsPass) -> (ReturnValue:int))
   > Return Node!(
    execute
    Vis:ESlateVisibility Enum=Select(Option 0:ESlateVisibility Enum=Collapsed, Option 1:ESlateVisibility Enum=HitTestInvisible, Index=AND Boolean(A=@n2, B=@n3)) -> (ReturnValue:ESlateVisibility Enum)
   )
  ]}
 EventGraph {links [], nodes []}
]
variables []
widgets {
 info {inheritedWidgetCount=0, parentClass=&CommonUserWidget, rootWidgetClass=&HorizontalBox, widgetCount=3}
 widgets [
  Container {widgetClassPath=&HorizontalBox, bIsVariable=true, widget=Container}
  HeadingText {
   widgetClassPath=&CommonTextBlock
   bIsVariable=true
   parent=Container
   widget=HeadingText
   properties {
    mobileFontSizeMultiplier=1.15
    text=Choose a Set to Unlock at LVL
    font {
     bForceMonospaced=false
     bMaterialIsStencil=false
     fontMaterial=null
     fontObject=&NotoSans
     letterSpacing=0
     monospacedWidth=1
     size=24
     skewAmount=0
     typefaceFontName=NotoSans-Regular
     outlineSettings {
      bApplyOutlineToDropShadows=false
      bMiteredCorners=false
      bSeparateFillAlpha=false
      outlineMaterial=null
      outlineSize=0
      outlineColor=<rgba 0 0 0 1>
     }
    }
    shadowOffset=<0 0>
   }
   slot {class=HorizontalBoxSlot, properties {horizontalAlignment=HAlign_Left, verticalAlignment=VAlign_Center, padding=<ltrb 0 0 8 0>}}
  }
  LevelText {
   widgetClassPath=&CommonTextBlock
   bIsVariable=true
   parent=Container
   widget=LevelText
   properties {
    mobileFontSizeMultiplier=1.15
    text=35
    font {
     bForceMonospaced=false
     bMaterialIsStencil=false
     fontMaterial=null
     fontObject=&NotoSans
     letterSpacing=0
     monospacedWidth=1
     size=24
     skewAmount=0
     typefaceFontName=NotoSans-Regular
     outlineSettings {
      bApplyOutlineToDropShadows=false
      bMiteredCorners=false
      bSeparateFillAlpha=false
      outlineMaterial=null
      outlineSize=0
      outlineColor=<rgba 0 0 0 1>
     }
    }
    shadowOffset=<0 0>
   }
   slot {class=HorizontalBoxSlot, properties {horizontalAlignment=HAlign_Left, verticalAlignment=VAlign_Center}}
  }
 ]
}
