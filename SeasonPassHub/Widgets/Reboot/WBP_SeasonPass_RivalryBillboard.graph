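# Season pass rivalry billboard widget: a hold-to-activate button with a material-driven character image, logo image or text, shadow and unlock prompt that broadcasts OnBillboardClicked.
# Setters load soft assets synchronously with fallbacks and feed the billboard's dynamic material.
# SetUIFeature maps a string to an EFortUIFeature, then refreshes visibility (hidden if invalid or in compact UI), direction and intro animation.
# SetDirection mirrors material and layout by alignment, inverted for Arabic culture.
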WidgetBlueprint /SeasonPassHub/Widgets/Reboot/WBP_SeasonPass_RivalryBillboard
animations [Intro {
  binding_count=4
  length_seconds=0.9
  object=&self:Intro
  sequence {bindings [
    Shadow_img {
     id=#1
     possessed_class=Image
     tracks [{
       type=MovieSceneFloatTrack
       display_name=Render Opacity
       sections [{type=MovieSceneFloatSection, end_frame=18, start_frame=0, channels [None {type=MovieSceneScriptingFloatChannel, keys [[0,0],[1,0],[18,1]]}]}]
      }]
    }
    Logo_Image {
     id=#2
     possessed_class=Image
     tracks [
      {
       type=MovieSceneFloatTrack
       display_name=Render Opacity
       sections [{type=MovieSceneFloatSection, end_frame=17, start_frame=0, channels [None {type=MovieSceneScriptingFloatChannel, keys [[5,0],[17,1]]}]}]
      }
      {
       type=MovieScene2DTransformTrack
       sections [{
         type=MovieScene2DTransformSection
         end_frame=17
         start_frame=0
         channels [
          Translation.X {type=MovieSceneScriptingFloatChannel, keys [[5,-64],[17,0]]}
          Translation.Y {type=MovieSceneScriptingFloatChannel, const [0,2]}
          Angle {type=MovieSceneScriptingFloatChannel, default=0}
          Scale.X {type=MovieSceneScriptingFloatChannel, default=1}
          Scale.Y {type=MovieSceneScriptingFloatChannel, default=1}
          Shear.X {type=MovieSceneScriptingFloatChannel, default=0}
          Shear.Y {type=MovieSceneScriptingFloatChannel, default=0}
         ]
        }]
      }
     ]
    }
    Button_Hitbox {
     id=#3
     possessed_class=WBP_UIKit_Button_Generic_C
     tracks [
      {
       type=MovieSceneFloatTrack
       display_name=Render Opacity
       sections [{type=MovieSceneFloatSection, end_frame=17, start_frame=0, channels [None {type=MovieSceneScriptingFloatChannel, keys [[0,0],[4,1]]}]}]
      }
      {
       type=MovieScene2DTransformTrack
       sections [{
         type=MovieScene2DTransformSection
         end_frame=17
         start_frame=0
         channels [
          Translation.X {type=MovieSceneScriptingFloatChannel, keys [[2,-80],[17,0]]}
          Translation.Y {type=MovieSceneScriptingFloatChannel, const [0,2]}
          Angle {type=MovieSceneScriptingFloatChannel, const [0,1]}
          Scale.X {type=MovieSceneScriptingFloatChannel, const [1,1]}
          Scale.Y {type=MovieSceneScriptingFloatChannel, const [1,1]}
          Shear.X {type=MovieSceneScriptingFloatChannel, const [0,1]}
          Shear.Y {type=MovieSceneScriptingFloatChannel, const [0,1]}
         ]
        }]
      }
     ]
    }
    Billboard {
     id=#4
     possessed_class=FortLazyImage
     tracks [{
       type=MovieSceneWidgetMaterialTrack
       sections [{
         type=MovieSceneParameterSection
         end_frame=18
         start_frame=0
         channels [
          Reveal Amount {type=MovieSceneScriptingFloatChannel, keys [[0,0],[9,1]]}
          TexturePosX_Anim {type=MovieSceneScriptingFloatChannel, keys [[0,2],[9,0]]}
          TextureMask_Size {type=MovieSceneScriptingFloatChannel, keys [[0,2.2],[9,0]]}
          SkewedGlow_1_Mask {type=MovieSceneScriptingFloatChannel, keys [[2,0],[5,0],[18,1]]}
          SkewedGlow_2_Mask {type=MovieSceneScriptingFloatChannel, keys [[2,0],[8,0],[14,1]]}
          TextureOpacity {type=MovieSceneScriptingFloatChannel, keys [[4,0],[15,1]]}
          TextureColorMask {type=MovieSceneScriptingFloatChannel, keys [[6,0],[15,1]]}
          BackgroundGlow_Opacity {type=MovieSceneScriptingFloatChannel, keys [[2,0],[4,0],[18,1]]}
         ]
        }]
      }]
    }
   ]}
 }]
graphs [
 EventGraph {nodes [
   "On Clicked (Button_Hitbox)"!() -> (OutputDelegate, then, Button:Common Button Base Object Reference) {is=title K2Node_ComponentBoundEvent}
   > Call On Billboard Clicked!(execute:Exec, self:WBP Season Pass Rivalry Billboard Object Reference) -> (then:Exec) {is=title K2Node_CallDelegate}
   n1 = SetBillboard!() -> (OutputDelegate, then, IsValidUIFeature:bool=~) {is=title K2Node_CustomEvent}
   > UpdateVisibility!(execute, self, IsValidUIFeature:bool=@n1.IsValidUIFeature) -> (then)
   > SetDirection()
   > IntroAnim()
   n2 = SetLogoImage!() -> (OutputDelegate, then, InLogoImage:Soft Object Reference) {is=title K2Node_CustomEvent}
   n3 = IsValidSoftObjectReference(SoftObjectReference=@n2.InLogoImage)
   > SetLogoActiveWidget!(execute, self, IsTextValid:bool=NOT Boolean(A=@n3)) -> (then)
   > n4 = Branch(Condition=@n3)
   > n5 = LoadAsset_Blocking(Asset=@n2.InLogoImage)
   > n6 = Cast To MaterialInstance!(execute, Object=@n5) -> (then, CastFailed, AsMaterial Instance:Material Instance Object Reference) {is=title K2Node_DynamicCast}
   > SetBrushFromMaterial(self=Get Logo_Image() -> (Logo_Image:Image Object Reference), Material=@n6)
   n7 = Cast To Texture2D(execute=@n6.CastFailed, Object=@n5)
   > SetBrushFromMaterial(self=Get Logo_Image(), Material=/SeasonPassHub/Widgets/Reboot/Materials/M_UI_DebugLogo)
   > n8 = GetDynamicMaterial(self=Get Logo_Image())
   > SetTextureParameterValue(self=@n8, ParameterName=LogoTexture, Value=@n7)
   SetBrushFromMaterial(execute=[@n7.CastFailed, @n4.else], self=Get Logo_Image(), Material=&M_UI_DebugLogo)
   n9 = SetCharacterImage!() -> (OutputDelegate, then, InCharacterImage:Soft Object Reference) {is=title K2Node_CustomEvent}
   > n10 = Branch(Condition=IsValidSoftObjectReference(SoftObjectReference=@n9.InCharacterImage))
   > n11 = LoadAsset_Blocking(Asset=@n9.InCharacterImage)
   > n12 = Cast To Texture2D(Object=@n11)
   > SetTextureParameterValue(self=Get MIDBillboard() -> (MIDBillboard:Material Instance Dynamic Object Reference), ParameterName=Character Texture, Value=@n12)
   SetTextureParameterValue(
    execute=[@n12.CastFailed, @n10.else]
    self=Get MIDBillboard()
    ParameterName=Character Texture
    Value=/FortUILibrary/Input/PeripheralGuitar/Source/Empty1x1
   )
   n13 = Event PreConstruct()
   > SetUpPromptLayout()
   > n14 = GetDynamicMaterial(self=Get Billboard() -> (Billboard:Fort Lazy Image Object Reference))
   > Set MIDBillboard(MIDBillboard:Material Instance Dynamic Object Reference=@n14) -> (Output_Get:Material Instance Dynamic Object Reference)
   > SetDirection()
   > SetVisibility(
    self=Get Shadow_img() -> (Shadow_img:Image Object Reference)
    InVisibility=Get ShadowVisibility() -> (ShadowVisibility:ESlateVisibility Enum=Visible)
   )
   > Branch(Condition=@n13.IsDesignTime)
   > n15 = AddChildToGrid!(
    execute
    self:Grid Panel Object Reference=Get Prompt!(self:WBP UIKit Button Generic Base Object Reference=Get Button_Hitbox() -> (Button_Hitbox:WBP UIKit Button Generic Object Reference)) -> (Prompt:Grid Panel Object Reference)
    Content:Widget Object Reference=Get UnlockSetLabel() -> (UnlockSetLabel:Common Rich Text Block Object Reference)
    InRow:int=0
    InColumn:int=1
   ) -> (then, ReturnValue:Grid Slot Object Reference)
   > SetVerticalAlignment(self:Grid Slot Object Reference=@n15, InVerticalAlignment=VAlign_Center)
  ]}
 IntroAnim {nodes [
   IntroAnim() {is=title K2Node_FunctionEntry}
   > PlayAnimationForward(InAnimation=Get Intro())
  ]}
 OnBillboardClicked {nodes [OnBillboardClicked() {is=title K2Node_FunctionEntry}]}
 SetColor {nodes [
   n1 = SetColor!() -> (then, InColor:Linear Color Structure) {is=title K2Node_FunctionEntry}
   > SetVectorParameterValue(self=Get MIDBillboard(), ParameterName=Base Color, Value=@n1)
  ]}
 SetDirection {nodes [
   SetDirection() {is=title K2Node_FunctionEntry}
   > n1 = GetDynamicMaterial(self=Get Billboard())
   n2 = "Equal (Enum)"(
    A:EHorizontalAlignment Enum=Get WidgetDirection() -> (WidgetDirection:EHorizontalAlignment Enum=HAlign_Fill)
    B:EHorizontalAlignment Enum=HAlign_Left
   )
   n3 = Select(Option 0:bool=@n2, Option 1:bool=NOT Boolean(A=@n2), Index="Equal Exactly (String)"(A=GetCurrentCulture!() -> (ReturnValue:String), B=ar)) -> (ReturnValue:bool)
   > SetScalarParameterValue(self=@n1, ParameterName=Mirror, Value="To Float (Boolean)"(InBool=@n3))
   > SetPadding!(
    execute
    self:Overlay Slot Object Reference=SlotAsOverlaySlot(Widget=Get LogoVisibilitySwitcher() -> (LogoVisibilitySwitcher:Common Visibility Switcher Object Reference))
    InPadding:Margin Structure=Select!(
     Option 0_Left:float=0
     Option 0_Top:float=0
     Option 0_Right:float=70
     Option 0_Bottom:float=0
     Option 1_Left:float=70
     Option 1_Top:float=0
     Option 1_Right:float=0
     Option 1_Bottom:float=0
     Index=@n3
    ) -> (ReturnValue:Margin Structure)
   ) -> (then)
   > SetPadding(
    self=SlotAsOverlaySlot(Widget=Get Button_Hitbox())
    InPadding=Select(
     Option 0_Left=0
     Option 0_Top=0
     Option 0_Right=20
     Option 0_Bottom=0
     Option 1_Left=20
     Option 1_Top=0
     Option 1_Right=0
     Option 1_Bottom=0
     Index=@n3
    )
   )
   n4 = Get Prompt(self=Get Button_Hitbox())
   > SetGridSlotAlignment!(
    execute
    Widget:Widget Object Reference=@n4
    Horizontal:EHorizontalAlignment Enum=Select!(Option 0:EHorizontalAlignment Enum=HAlign_Right, Option 1:EHorizontalAlignment Enum=HAlign_Left, Index=@n3) -> (ReturnValue:EHorizontalAlignment Enum)
    Vertical:EVerticalAlignment Enum=VAlign_Fill
   ) -> (then)
   > SetGridSlotPadding!(
    execute
    Widget:Widget Object Reference=@n4
    Padding:Margin Structure=Select!(
     Option 0_Left:float=0
     Option 0_Top:float=0
     Option 0_Right:float=35
     Option 0_Bottom:float=0
     Option 1_Left:float=0
     Option 1_Top:float=0
     Option 1_Right:float=0
     Option 1_Bottom:float=0
     Index=@n3
    ) -> (ReturnValue:Margin Structure)
   ) -> (then)
   > SetPadding!(
    execute
    self=SlotAsOverlaySlot(Widget=Get UnlockSetLabel())
    InPadding_Left:float=4
    InPadding_Top:float=0
    InPadding_Right:float=4
    InPadding_Bottom:float=0
   ) -> (then)
  ]}
 SetInputAction {nodes [
   n1 = SetInputAction!() -> (then, InButtonInput:Data Table Row Handle Structure) {is=title K2Node_FunctionEntry}
   n2 = Break Data Table Row Handle!(DataTableRowHandle:"Data Table Row Handle Structure (by ref)"=@n1) -> (DataTable:Data Table Object Reference, RowName:Name) {is=title K2Node_BreakStruct}
   n3 = AND Boolean(A=IsValid(Object=@n2), B="Not Equal (Name)"(A=@n2.RowName) -> (ReturnValue=~))
   n4 = Get Prompt(self=Get Button_Hitbox())
   > SetVisibility(
    self=@n4
    InVisibility=Select!(Option 0:ESlateVisibility Enum=Collapsed, Option 1:ESlateVisibility Enum=SelfHitTestInvisible, Index=@n3) -> (ReturnValue:ESlateVisibility Enum)
   )
   > SetTriggeringInputAction!(
    execute
    self:Common Button Base Object Reference=Get Button_Hitbox()
    InputActionRow:"Data Table Row Handle Structure (by ref)"=Get InButtonInput() -> (InButtonInput:Data Table Row Handle Structure)
   ) -> (then)
   > n5 = Branch(Condition=@n3)
   > n6 = AddChildToGrid(self=@n4, Content=Get UnlockSetLabel(), InRow=0, InColumn=1)
   > SetVerticalAlignment(self=@n6, InVerticalAlignment=VAlign_Center)
   n7 = AddChildToGrid(
    execute=@n5.else
    self=Get Root!(self:WBP UIKit Button Generic Base Object Reference=Get Button_Hitbox()) -> (Root:Grid Panel Object Reference)
    Content=Get UnlockSetLabel()
    InRow=2
    InColumn=1
   )
   > SetVerticalAlignment(self=@n7, InVerticalAlignment=VAlign_Fill)
  ]}
 SetLogoActiveWidget {nodes [
   n1 = SetLogoActiveWidget!() -> (then, IsTextValid:bool=~) {is=title K2Node_FunctionEntry}
   > SetActiveWidgetIndex(
    self:Common Visibility Switcher Object Reference=Get LogoVisibilitySwitcher()
    Index="To Integer (Boolean)"!(InBool:bool=@n1) -> (ReturnValue:int=0)
   )
  ]}
 SetLogoText {nodes [
   n1 = SetLogoText!() -> (then, LogoText:Text) {is=title K2Node_FunctionEntry}
   > SetLogoActiveWidget(IsTextValid=NOT Boolean(A=TextIsEmpty(InText=@n1))) {is=title K2Node_CallFunction}
   > "SetText (Text)"(self=Get Logo_Text() -> (Logo_Text:Common Text Object Reference), InText=@n1)
  ]}
 SetNavigation {nodes [
   n1 = SetNavigation!() -> (then, UIFeature:String, IsValidUIFeature:bool=~) {is=title K2Node_FunctionEntry}
   > Set NavigateToFeatureTab(NavigateToFeatureTab:EFortUIFeature Enum=Max_None) -> (Output_Get:EFortUIFeature Enum=ShowHome)
   > Branch(Condition=@n1.IsValidUIFeature)
   > n2 = ForEach EFortUIFeature!(execute:Exec, SkipHidden:bool) -> (LoopBody:Exec, EnumValue:EFortUIFeature Enum, then:Exec) {is=title K2Node_ForEachElementInEnum}
   n3 = Enum to String!(Enumerator:EFortUIFeature Enum=@n2) -> (ReturnValue:String)
   n4 = Get UIFeature() -> (UIFeature:String)
   > FortBlueprintLog!(
    execute
    InString:String=Append!(A="SetNavigation: Comparing UIFeature: ", B=@n4, C:String=" to enum feature: ", D:String=@n3) -> (ReturnValue)
   ) -> (then)
   > Branch(Condition="Equal Exactly (String)"(A=@n4, B=@n3))
   > Set NavigateToFeatureTab(NavigateToFeatureTab=@n2) -> (Output_Get=ShowHome)
  ]}
 SetPromptText {nodes [
   n1 = SetPromptText!() -> (then, InText:Text) {is=title K2Node_FunctionEntry}
   > SetText(self=Get UnlockSetLabel(), InText=@n1)
  ]}
 SetShadowColor {nodes [
   n1 = SetShadowColor!() -> (then, color:Linear Color Structure=(R=0,G=0,B=0,A=0.5652)) {is=title K2Node_FunctionEntry}
   > SetBrushTintColor!(
    execute
    self:Image Object Reference=Get Shadow_img()
    TintColor_SpecifiedColor:Linear Color Structure=@n1
    TintColor_ColorUseRule:ESlateColorStylingMode Enum
   ) -> (then)
  ]}
 SetUIFeature {nodes [
   n1 = SetUIFeature!() -> (then, InUIFeature:String) {is=title K2Node_FunctionEntry}
   n2 = NOT Boolean(A=IsEmpty(InString=@n1))
   > n3 = Branch(Condition=@n2)
   > FortBlueprintLog(InString="To String (Text)"(InText=Format Text!(Format="LOCTEXT(\"Try Lookup returned: UIfeature: {UIFeature}, and IsValid: {isValid}\")", UIFeature:String=@n1, isValid:bool=@n2) -> (Result)))
   > n4 = SetNavigation!(execute, self, UIFeature:String=Get InUIFeature() -> (InUIFeature:String), IsValidUIFeature:bool=@n2) -> (then) {is=title K2Node_CallFunction}
   SetBillboard!(execute=[@n4.then, @n3.else], self, IsValidUIFeature:bool=@n2) -> (then) {is=title K2Node_CallFunction}
   UpdateVisibility(IsValidUIFeature=false)
  ]}
 SetUpPromptLayout {nodes [
   SetUpPromptLayout() {is=title K2Node_FunctionEntry}
   n1 = Get Button_Hitbox()
   > SetRow!(execute, self:Grid Slot Object Reference=SlotAsGridSlot(Widget=Get Prompt(self=@n1)), InRow:int=2) -> (then)
  ]}
 UpdateVisibility {nodes [
   n1 = UpdateVisibility!() -> (then, IsValidUIFeature:bool=~) {is=title K2Node_FunctionEntry}
   n2 = IsUsingCompactUI()
   n3 = Select(Option 0=Hidden, Option 1=SelfHitTestInvisible, Index=AND Boolean(A=@n1, B=NOT Boolean(A=@n2)))
   > SetVisibility(InVisibility=@n3)
   n4 = IsUsingCompactUI()
   n5 = Enum to String(Enumerator:ESlateVisibility Enum=@n3)
   n6 = Get IsValidFeature() -> (IsValidFeature:bool)
   n7 = Get IsValidUIFeature() -> (IsValidUIFeature:bool)
   > FortBlueprintLog(InString="To String (Text)"(InText=Format Text!(
      Format="LOCTEXT(\"Update Visibility: IsValidUIFeature {isValidUIfeature}, IsValidFeature{isValidFeature}, IsUsingCompactUI{isUsingCompactUI}. So visibility is {vis}\")"
      isValidUIfeature:bool=@n7
      isValidFeature:bool=@n6
      isUsingCompactUI:bool=@n4
      vis:String=@n5
     ) -> (Result)))
  ]}
]
variables [
 IsValidFeature {type=boolean}
 MIDBillboard {type=MaterialInstanceDynamic}
 NavigateToFeatureTab {type=EFortUIFeature, default=Max_None}
 OnBillboardClicked {}
 ShadowVisibility {type=ESlateVisibility, default=Hidden}
 WidgetDirection {type=EHorizontalAlignment, default=HAlign_Right}
]
widgets {
 info {
  inheritedWidgetCount=0
  namedSlotCount=2
  parentClass=&UserWidget
  rootWidgetClass=/FortUILibrary/UIKit/Buttons/WBP_UIKit_Button_Generic.WBP_UIKit_Button_Generic_C
  widgetCount=11
 }
 widgets [
  Button_Hitbox {
   widgetClassPath=&WBP_UIKit_Button_Generic
   bIsVariable=true
   widget=Button_Hitbox
   properties {
    bIsFocusable=false
    bRequiresHold=true
    bSimulateHoverOnTouchInput=false
    contentHorizontalAlignment=HAlign_Fill
    contentVerticalAlignment=VAlign_Fill
    enableFocusedSelectedState=false
    holdCompletedSound=/SeasonPassHub/Widgets/Reboot/Audio/MSS_UI_BattlePass_View_Customizations
    holdPressedSound=/Game/Sounds/Fort_UI_Sounds/FrontEnd/DefaultSounds/HoldableButtons/Holdable_Loud/MSS_UI_Button_Holdable_Loud_Press
    holdReleasedSound=/Game/Sounds/Fort_UI_Sounds/FrontEnd/DefaultSounds/HoldableButtons/Holdable_Loud/MSS_UI_Button_Holdable_Loud_Release
    holdSound=/Game/Sounds/Fort_UI_Sounds/FrontEnd/BattlePass/BP_Default/MSS_UI_BP_Button_Holdable_Hold_lp
    iconHorizontalAlignment=HAlign_Left
    minHeight=32
    navigation=&self:WidgetTree.Button_Hitbox.WidgetNavigation_0
    promptHorizontalAlignment=HAlign_Right
    promptVerticalAlignment=VAlign_Bottom
    secondText=""
    text=Unlock Set
    textBlockAlignment=Top
    textJustification=Left
    touchMethod=PreciseTap
    useBackgroundBlock=false
    useIconBlock=false
    useTextBlock=false
    visibility=Visible
    buttonPadding=<ltrb 24 4 24 0>
    font {
     bForceMonospaced=false
     bMaterialIsStencil=false
     fontMaterial=/FortUILibrary/UIKit/Buttons/MaterialInstances/CTAButtons/MI_UIKit_TextQuiet
     fontObject=&HeadingNow
     letterSpacing=48
     monospacedWidth=1
     size=18
     skewAmount=0
     typefaceFontName=76 Bold
     outlineSettings {
      bApplyOutlineToDropShadows=false
      bMiteredCorners=false
      bSeparateFillAlpha=false
      outlineMaterial=null
      outlineSize=0
      outlineColor=<rgba 0 0 0 1>
     }
    }
    iconSize=<32 32>
    promptNudge=<0 4>
    promptPadding=<ltrb 0 4 20 2>
    textPadding=<ltrb 0 4 0 0>
    triggeringInputAction {dataTable=&InputActionDataTable, rowName=GenericFaceButtonTop}
   }
  }
  UnlockSetLabel {
   widgetClassPath=&CommonRichTextBlock
   bIsVariable=true
   namedSlotHost=Button_Hitbox
   widget=UnlockSetLabel
   properties {
    accessibleBehavior=Auto
    bAutoCollapseWithEmptyText=true
    bCanChildrenBeAccessible=false
    bOverride_Cursor=true
    justification=Center
    scrollStyle=&TextScrollStyle-Base_C
    text="Unlock outfit style!"
    textStyleSet=&/DT_SeasonPass_Billboard
    wrapTextAt=300
    defaultTextStyleOverride {
     overflowPolicy=Clip
     transformPolicy=None
     colorAndOpacity {colorUseRule=UseColor_Specified, specifiedColor=<rgba 1 1 1 1>}
     font {
      bForceMonospaced=false
      bMaterialIsStencil=false
      fontMaterial=null
      fontObject=&HeadingNow
      letterSpacing=0
      monospacedWidth=1
      size=15.75
      skewAmount=0
      typefaceFontName=74 Regular
      outlineSettings {
       bApplyOutlineToDropShadows=false
       bMiteredCorners=false
       bSeparateFillAlpha=false
       outlineMaterial=null
       outlineSize=0
       outlineColor=<rgba 0 0 0 1>
      }
     }
     highlightColor {colorUseRule=UseColor_Specified, specifiedColor=<rgba 0 0 0 1>}
     highlightShape {
      bIsDynamicallyLoaded=false
      drawAs=Image
      imageType=NoImage
      mirroring=NoMirror
      resourceName=None
      resourceObject=null
      tiling=NoTile
      imageSize=<32 32>
      margin=<ltrb 0 0 0 0>
      outlineSettings {
       bUseBrushTransparency=false
       roundingType=HalfHeightRadius
       width=0
       color {colorUseRule=UseColor_Specified, specifiedColor=<rgba 0 0 0 0>}
       cornerRadii=<0 0 0 0>
      }
      tintColor {colorUseRule=UseColor_Specified, specifiedColor=<rgba 1 1 1 1>}
      uVRegion {bIsValid=false, max=<0 0>, min=<0 0>}
     }
     selectedBackgroundColor {colorUseRule=UseColor_Foreground, specifiedColor=<rgba 1 0 1 1>}
     shadowColorAndOpacity=<rgba 0 0 0 1>
     shadowOffset=<0 0>
     strikeBrush {
      bIsDynamicallyLoaded=false
      drawAs=Image
      imageType=NoImage
      mirroring=NoMirror
      resourceName=None
      resourceObject=null
      tiling=NoTile
      imageSize=<32 32>
      margin=<ltrb 0 0 0 0>
      outlineSettings {
       bUseBrushTransparency=false
       roundingType=HalfHeightRadius
       width=0
       color {colorUseRule=UseColor_Specified, specifiedColor=<rgba 0 0 0 0>}
       cornerRadii=<0 0 0 0>
      }
      tintColor {colorUseRule=UseColor_Specified, specifiedColor=<rgba 1 1 1 1>}
      uVRegion {bIsValid=false, max=<0 0>, min=<0 0>}
     }
     underlineBrush {
      bIsDynamicallyLoaded=false
      drawAs=Image
      imageType=NoImage
      mirroring=NoMirror
      resourceName=None
      resourceObject=null
      tiling=NoTile
      imageSize=<32 32>
      margin=<ltrb 0 0 0 0>
      outlineSettings {
       bUseBrushTransparency=false
       roundingType=HalfHeightRadius
       width=0
       color {colorUseRule=UseColor_Specified, specifiedColor=<rgba 0 0 0 0>}
       cornerRadii=<0 0 0 0>
      }
      tintColor {colorUseRule=UseColor_Specified, specifiedColor=<rgba 1 1 1 1>}
      uVRegion {bIsValid=false, max=<0 0>, min=<0 0>}
     }
    }
   }
  }
  SizeBox_0 {
   widgetClassPath=&SizeBox
   namedSlotHost=Button_Hitbox
   widget=SizeBox_0
   properties {bOverride_WidthOverride=true, flowDirectionPreference=LeftToRight, minDesiredWidth=336, widthOverride=380}
  }
  Overlay_4 {widgetClassPath=&Overlay, parent=SizeBox_0, widget=Overlay_4, slot {class=SizeBoxSlot, properties {padding=<ltrb 0 0 0 -30>}}}
  Shadow {
   widgetClassPath=&CommonVisualAttachment
   parent=Overlay_4
   widget=Shadow
   properties {contentAnchor=<0.5 0.5>}
   slot {class=OverlaySlot, properties {horizontalAlignment=HAlign_Center, verticalAlignment=VAlign_Center}}
  }
  Shadow_img {
   widgetClassPath=&Image
   bIsVariable=true
   parent=Shadow
   widget=Shadow_img
   properties {
    visibility=Hidden
    brush {
     bIsDynamicallyLoaded=false
     drawAs=Image
     imageType=FullColor
     mirroring=NoMirror
     resourceName=None
     resourceObject=/SeasonPassHub/Widgets/Reboot/Materials/M_UI_Upsell_WidgetBacking
     tiling=NoTile
     imageSize=<600 230>
     margin=<ltrb 0 0 0 0>
     outlineSettings {
      bUseBrushTransparency=false
      roundingType=HalfHeightRadius
      width=0
      color {colorUseRule=UseColor_Specified, specifiedColor=<rgba 0 0 0 0>}
      cornerRadii=<0 0 0 0>
     }
     tintColor {colorUseRule=UseColor_Specified, specifiedColor=<rgba 1 1 1 1>}
     uVRegion {bIsValid=false, max=<0 0>, min=<0 0>}
    }
    renderTransform {angle=0, scale=<1.7 2>, shear=<0 0>, translation=<0 25>}
   }
   slot {class=SizeBoxSlot, properties {verticalAlignment=VAlign_Center}}
  }
  Billboard {
   widgetClassPath=&FortLazyImage
   bIsVariable=true
   parent=Overlay_4
   widget=Billboard
   properties {brush {
     bIsDynamicallyLoaded=false
     drawAs=Image
     imageType=FullColor
     mirroring=NoMirror
     resourceName=None
     resourceObject=/SeasonPassHub/Widgets/Reboot/Materials/M_UI_CH7S2_RivalryBillboard_Revamp
     tiling=NoTile
     imageSize=<32 144>
     margin=<ltrb 0 0 0 0>
     outlineSettings {
      bUseBrushTransparency=false
      roundingType=HalfHeightRadius
      width=0
      color {colorUseRule=UseColor_Specified, specifiedColor=<rgba 0 0 0 0>}
      cornerRadii=<0 0 0 0>
     }
     tintColor {colorUseRule=UseColor_Specified, specifiedColor=<rgba 1 1 1 1>}
     uVRegion {bIsValid=false, max=<0 0>, min=<0 0>}
    }}
   slot {class=OverlaySlot, properties {horizontalAlignment=HAlign_Fill, padding=<ltrb 0 0 0 15>}}
  }
  LogoVisibilitySwitcher {
   widgetClassPath=&CommonVisibilitySwitcher
   bIsVariable=true
   parent=Overlay_4
   widget=LogoVisibilitySwitcher
   slot {class=OverlaySlot, properties {horizontalAlignment=HAlign_Fill, verticalAlignment=VAlign_Center, padding=<ltrb 0 0 70 0>}}
  }
  Logo_Image {
   widgetClassPath=&Image
   bIsVariable=true
   parent=LogoVisibilitySwitcher
   widget=Logo_Image
   properties {
    visibility=SelfHitTestInvisible
    brush {
     bIsDynamicallyLoaded=false
     drawAs=Image
     imageType=FullColor
     mirroring=NoMirror
     resourceName=None
     resourceObject=&M_UI_DebugLogo
     tiling=NoTile
     imageSize=<196 104>
     margin=<ltrb 0 0 0 0>
     outlineSettings {
      bUseBrushTransparency=false
      roundingType=HalfHeightRadius
      width=0
      color {colorUseRule=UseColor_Specified, specifiedColor=<rgba 0 0 0 0>}
      cornerRadii=<0 0 0 0>
     }
     tintColor {colorUseRule=UseColor_Specified, specifiedColor=<rgba 1 1 1 1>}
     uVRegion {bIsValid=false, max=<0 0>, min=<0 0>}
    }
   }
   slot {class=CommonVisibilitySwitcherSlot, properties {horizontalAlignment=HAlign_Right, verticalAlignment=VAlign_Center, padding=<ltrb 0 0 0 10>}}
  }
  SizeBox_1 {
   widgetClassPath=&SizeBox
   parent=LogoVisibilitySwitcher
   widget=SizeBox_1
   properties {bOverride_HeightOverride=true, heightOverride=64}
   slot {class=CommonVisibilitySwitcherSlot, properties {verticalAlignment=VAlign_Center, padding=<ltrb 0 0 0 10>}}
  }
  Logo_Text {
   widgetClassPath=&CommonTextBlock
   bIsVariable=true
   parent=SizeBox_1
   widget=Logo_Text
   properties {
    autoWrapText=true
    bAutoCollapseWithEmptyText=true
    flowDirectionPreference=Culture
    justification=Center
    scrollOrientation=Orient_Vertical
    scrollStyle=&TextScrollStyle-Base_C
    text="[PH] Logo text Logo text text"
    textTransformPolicy=ToUpper
    font {
     bForceMonospaced=false
     bMaterialIsStencil=false
     fontMaterial=null
     fontObject=&HeadingNow
     letterSpacing=0
     monospacedWidth=1
     size=18
     skewAmount=0
     typefaceFontName=76 Bold
     outlineSettings {
      bApplyOutlineToDropShadows=false
      bMiteredCorners=false
      bSeparateFillAlpha=false
      outlineMaterial=null
      outlineSize=0
      outlineColor=<rgba 0 0 0 1>
     }
    }
    shadowOffset=<0 0>
   }
   slot {class=SizeBoxSlot, properties {verticalAlignment=VAlign_Center}}
  }
 ]
}
